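# Level Sequence that animates a spawned FortPlayerMannequin over roughly 433 frames, filed under the main-player menu animations for the TopicProud_M custom skin.
# A skeletal animation section spans the shot alongside a Fortnite_Biped_CtrlRig parameter track, with the actor transform left at identity.
# The rig track carries dense, per-frame-style keys on body controls, including both FK and IK foot controls, while its switches, spaces and global/root controls stay constant.
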LevelSequence /BRCosmetics/Animation/Game/MainPlayer/Menu/BR/CustomSkins/_s40/TopicProud_M/TP_anim/TP_LS
bindings [FortPlayerMannequin {
  id=#1
  possessed_class=FortPlayerMannequin
  tracks [
   {type=MovieSceneSpawnTrack, sections [{type=MovieSceneSpawnSection, channels [None {type=MovieSceneScriptingBoolChannel, default=true}]}]}
   {
    type=MovieScene3DTransformTrack
    sections [{
      type=MovieScene3DTransformSection
      channels [
       Location.X {type=MovieSceneScriptingDoubleChannel, default=0}
       Location.Y {type=MovieSceneScriptingDoubleChannel, default=0}
       Location.Z {type=MovieSceneScriptingDoubleChannel, default=0}
       Rotation.X {type=MovieSceneScriptingDoubleChannel, default=0}
       Rotation.Y {type=MovieSceneScriptingDoubleChannel, default=0}
       Rotation.Z {type=MovieSceneScriptingDoubleChannel, default=0}
       Scale.X {type=MovieSceneScriptingDoubleChannel, default=1}
       Scale.Y {type=MovieSceneScriptingDoubleChannel, default=1}
       Scale.Z {type=MovieSceneScriptingDoubleChannel, default=1}
      ]
     }]
   }
   {type=MovieSceneSkeletalAnimationTrack, sections [{type=MovieSceneSkeletalAnimationSection, end_frame=433, start_frame=0}]}
   {
    type=MovieSceneControlRigParameterTrack
    display_name=Fortnite_Biped_CtrlRig
    sections [{
      type=MovieSceneControlRigParameterSection
      channels [
       leg_l_stretch_switch {type=MovieSceneScriptingBoolChannel, default=false, const [false,434]}
       leg_r_stretch_switch {type=MovieSceneScriptingBoolChannel, default=false, const [false,434]}
       arm_l_fk_local {type=MovieSceneScriptingBoolChannel, default=true, const [true,434]}
       Auto Interpolate_3 {type=MovieSceneScriptingBoolChannel, default=false, const [false,434]}
       Auto Interpolate {type=MovieSceneScriptingBoolChannel, default=false, const [false,434]}
       arm_l_stretch_switch {type=MovieSceneScriptingBoolChannel, default=false, const [false,434]}
       Keep Aligned {type=MovieSceneScriptingBoolChannel, default=false, const [false,434]}
       arm_r_fk_local {type=MovieSceneScriptingBoolChannel, default=true, const [true,434]}
       Auto Interpolate_4 {type=MovieSceneScriptingBoolChannel, default=false, const [false,434]}
       Auto Interpolate_2 {type=MovieSceneScriptingBoolChannel, default=false, const [false,434]}
       arm_r_stretch_switch {type=MovieSceneScriptingBoolChannel, default=false, const [false,434]}
       Keep Aligned_2 {type=MovieSceneScriptingBoolChannel, default=false, const [false,434]}
       head_local {type=MovieSceneScriptingBoolChannel, default=true, const [true,434]}
       Stretch {type=MovieSceneScriptingBoolChannel, default=false, const [false,434]}
       head_stretch {type=MovieSceneScriptingBoolChannel, default=false, const [false,434]}
       spine_stretch_switch {type=MovieSceneScriptingBoolChannel, default=false, const [false,434]}
       arm_l_fk_ik_switch {type=MovieSceneScriptingBoolChannel, default=false, const [false,434]}
       leg_l_fk_ik_switch {type=MovieSceneScriptingBoolChannel, default=true, const [true,434]}
       arm_r_fk_ik_switch {type=MovieSceneScriptingBoolChannel, default=false, const [false,434]}
       leg_r_fk_ik_switch {type=MovieSceneScriptingBoolChannel, default=true, const [true,434]}
       ShowBodyControls {type=MovieSceneScriptingBoolChannel, default=true, const [true,434]}
       spine_fk_ik_switch {type=MovieSceneScriptingBoolChannel, default=false, const [false,434]}
       neck_fk_ik_switch {type=MovieSceneScriptingBoolChannel, default=false, const [false,434]}
       space {type=MovieSceneScriptingByteChannel, default=1, const [1,434]}
       global_ctrl.Location.X {type=MovieSceneScriptingFloatChannel, default=0, const [0,434]}
       global_ctrl.Location.Y {type=MovieSceneScriptingFloatChannel, default=0, const [0,434]}
       global_ctrl.Location.Z {type=MovieSceneScriptingFloatChannel, default=0, const [0,434]}
       global_ctrl.Rotation.X {type=MovieSceneScriptingFloatChannel, default=0, const [0,434]}
       global_ctrl.Rotation.Y {type=MovieSceneScriptingFloatChannel, default=0, const [0,434]}
       global_ctrl.Rotation.Z {type=MovieSceneScriptingFloatChannel, default=0, const [0,434]}
       root_ctrl.Location.X {type=MovieSceneScriptingFloatChannel, default=0, const [0,434]}
       root_ctrl.Location.Y {type=MovieSceneScriptingFloatChannel, default=0, const [0,434]}
       root_ctrl.Location.Z {type=MovieSceneScriptingFloatChannel, default=0, const [0,434]}
       root_ctrl.Rotation.X {type=MovieSceneScriptingFloatChannel, default=0, const [0,434]}
       root_ctrl.Rotation.Y {type=MovieSceneScriptingFloatChannel, default=0, const [0,434]}
       root_ctrl.Rotation.Z {type=MovieSceneScriptingFloatChannel, default=0, const [0,434]}
       root_ctrl.Scale.X {type=MovieSceneScriptingFloatChannel, default=1, const [1,434]}
       root_ctrl.Scale.Y {type=MovieSceneScriptingFloatChannel, default=1, const [1,434]}
       root_ctrl.Scale.Z {type=MovieSceneScriptingFloatChannel, default=1, const [1,434]}
       body_offset_ctrl.Location.X {type=MovieSceneScriptingFloatChannel, default=0, const [0,434]}
       body_offset_ctrl.Location.Y {type=MovieSceneScriptingFloatChannel, default=0, const [0,434]}
       body_offset_ctrl.Location.Z {type=MovieSceneScriptingFloatChannel, default=0, const [0,434]}
       body_offset_ctrl.Rotation.X {type=MovieSceneScriptingFloatChannel, default=0, const [0,434]}
       body_offset_ctrl.Rotation.Y {type=MovieSceneScriptingFloatChannel, default=0, const [0,434]}
       body_offset_ctrl.Rotation.Z {type=MovieSceneScriptingFloatChannel, default=0, const [0,434]}
       body_offset_ctrl.Scale.X {type=MovieSceneScriptingFloatChannel, default=1, const [1,434]}
       body_offset_ctrl.Scale.Y {type=MovieSceneScriptingFloatChannel, default=1, const [1,434]}
       body_offset_ctrl.Scale.Z {type=MovieSceneScriptingFloatChannel, default=1, const [1,434]}
       thigh_l_fk_ctrl.Location.X {type=MovieSceneScriptingFloatChannel, default=0, thinned_from=434, keys [[0,0],[433,0]]}
       thigh_l_fk_ctrl.Location.Y {type=MovieSceneScriptingFloatChannel, default=0, thinned_from=434, keys [[0,0],[433,0]]}
       thigh_l_fk_ctrl.Location.Z {type=MovieSceneScriptingFloatChannel, default=0, thinned_from=434, keys [[0,0],[433,0]]}
       thigh_l_fk_ctrl.Rotation.X {
        type=MovieSceneScriptingFloatChannel
        default=1.216
        thinned_from=434
        keys [[0,15.0059],[10,15.0058],[11,14.9796],[12,14.9084],[13,14.7697],[14,14.541],[15,14.3474],[16,14.1076],[17,13.489],[18,12.16],[19,9.0811],[20,4.8089],[21,1.42],[22,-0.1755],[23,-0.8732],[24,-1.2935],[25,-1.6578],[26,-1.8352],[28,-1.9962],[29,-2.0852],[30,-2.1395],[31,-2.167],[32,-2.1752],[33,-2.1517],[34,-2.0885],[35,-1.9999],[36,-1.9005],[37,-1.7838],[38,-1.6451],[39,-1.5016],[40,-1.3706],[41,-1.2693],[42,-1.2028],[43,-1.1597],[44,-1.133],[46,-1.1002],[47,-1.087],[48,-1.0792],[52,-1.0672],[53,-1.06],[54,-1.0475],[55,-1.0319],[56,-1.0132],[57,-0.9866],[58,-0.9472],[59,-0.8902],[60,-0.8108],[61,-0.7041],[62,-0.6166],[63,-0.4243],[64,0.1338],[65,1.8071],[66,3.6928],[67,4.6355],[68,5.1794],[69,5.5393],[70,5.7597],[72,6.1036],[73,6.2298],[74,6.2802],[75,6.2652],[80,6.1058],[82,6.0728],[94,6.0543],[96,6.1142],[97,6.1708],[98,6.2547],[99,6.3715],[100,6.5513],[101,6.8193],[102,7.1737],[103,7.6126],[104,8.1902],[106,9.6033],[107,10.2213],[108,10.7314],[109,11.1861],[110,11.5729],[111,11.8794],[112,12.0811],[113,12.191],[117,12.4239],[124,12.6567],[130,12.6498],[135,12.5137],[136,12.4648],[137,12.3638],[138,12.1829],[139,11.8763],[140,11.0836],[141,9.8655],[142,8.8433],[143,8.3447],[144,8.1557],[145,8.1552],[146,8.2158],[148,8.2999],[149,8.3854],[150,7.9672],[151,7.5739],[152,9.1758],[153,13.1525],[154,17.7417],[156,28.6275],[157,36.3929],[158,35.8164],[159,34.7543],[160,34.1288],[161,34.0833],[162,34.5091],[163,36.3145],[164,38.4906],[165,39.3397],[166,39.6701],[167,39.7256],[168,39.484],[169,38.4833],[170,37.1051],[171,36.3064],[172,35.9129],[174,35.6705],[177,35.6306],[183,36.1405],[210,36.3479],[212,36.5241],[214,36.8706],[215,37.1359],[216,37.4865],[217,37.9312],[218,38.4788],[219,39.1384],[220,39.9193],[221,40.9728],[222,42.3387],[224,45.3614],[226,48.6662],[227,49.9914],[228,50.8595],[229,51.5015],[230,51.9703],[232,52.5934],[241,53.5702],[254,53.6329],[257,53.3671],[258,52.9622],[259,52.1145],[260,50.109],[261,47.2658],[262,44.8465],[263,43.1711],[264,41.8276],[265,40.8579],[266,40.2877],[267,40.0147],[271,39.5905],[286,39.3864],[287,39.3114],[288,39.0996],[289,38.6999],[291,37.6439],[292,37.0455],[295,34.9616],[297,33.8304],[298,33.3624],[299,33.0061],[302,32.2399],[306,31.5154],[311,31.0171],[318,30.8425],[321,30.9626],[324,31.3098],[326,31.7216],[330,32.8217],[333,33.4263],[335,33.6729],[338,33.7423],[344,33.4919],[345,33.381],[348,32.7076],[349,32.6037],[351,32.7391],[352,32.8945],[353,33.4465],[354,34.6842],[355,39.1212],[356,39.0312],[357,34.7702],[359,25.3399],[360,21.3055],[361,17.7719],[362,15.8736],[363,15.8799],[364,15.634],[366,14.9793],[368,14.3645],[369,14.1024],[370,13.9187],[371,13.7779],[373,13.5783],[374,13.5098],[376,13.4529],[380,13.4495],[381,13.4728],[382,13.5299],[384,13.7425],[386,14.0284],[389,14.5835],[390,14.7222],[392,14.881],[398,15.061],[433,15.0056]]
       }
       thigh_l_fk_ctrl.Rotation.Y {
        type=MovieSceneScriptingFloatChannel
        default=-1.2128
        thinned_from=434
        keys [[0,2.5104],[2,2.5104],[3,2.5229],[4,2.5229],[5,2.5383],[6,2.5383],[7,2.5516],[9,2.5516],[11,2.5593],[12,2.5694],[13,2.5904],[14,2.6256],[15,2.6425],[16,2.6196],[17,2.5683],[18,2.5001],[19,2.5637],[20,2.6819],[21,2.611],[22,2.3461],[23,2.0437],[24,1.7816],[25,1.5483],[26,1.3223],[27,1.1402],[28,1.0138],[29,0.9276],[30,0.8585],[31,0.807],[32,0.7737],[33,0.7542],[34,0.743],[35,0.7382],[36,0.7383],[37,0.7484],[38,0.7729],[39,0.8092],[40,0.8548],[41,0.9072],[44,1.0803],[45,1.1331],[46,1.178],[49,1.2963],[51,1.3677],[52,1.3991],[53,1.4268],[54,1.4499],[57,1.5085],[58,1.5341],[59,1.5672],[60,1.6108],[61,1.6675],[62,1.7502],[63,1.7986],[64,1.7059],[65,1.182],[66,0.6532],[67,0.652],[68,0.8819],[69,1.097],[70,1.2411],[71,1.3264],[72,1.3787],[73,1.4196],[75,1.4894],[76,1.5179],[77,1.5406],[78,1.5575],[79,1.5693],[80,1.5764],[81,1.5789],[83,1.5735],[85,1.5735],[86,1.5569],[87,1.5569],[88,1.5445],[89,1.5445],[90,1.5299],[92,1.5299],[93,1.5184],[95,1.5157],[96,1.5175],[97,1.523],[98,1.5334],[99,1.5496],[100,1.5755],[101,1.6156],[102,1.672],[103,1.7464],[104,1.8475],[106,2.0951],[107,2.2071],[109,2.4028],[110,2.489],[111,2.5606],[112,2.62],[113,2.6695],[114,2.7031],[115,2.7145],[116,2.7094],[127,2.5425],[135,2.4444],[136,2.4292],[137,2.4031],[138,2.36],[139,2.4844],[140,2.8361],[141,3.4877],[142,3.8885],[143,4.4177],[144,4.9019],[145,5.086],[146,5.1546],[147,5.1725],[148,5.1228],[149,4.9424],[150,4.4826],[151,3.947],[152,3.6058],[153,3.1735],[154,2.9145],[155,2.6417],[156,2.499],[157,14.6582],[158,23.3371],[159,26.8763],[160,28.4814],[161,28.9209],[162,28.5843],[163,26.0724],[164,22.0972],[165,20.0776],[166,19.0013],[167,18.4851],[168,18.6789],[169,20.6632],[170,23.3171],[171,24.8639],[172,25.723],[173,26.1515],[174,26.4187],[175,26.5695],[177,26.6624],[178,26.6203],[182,26.0082],[187,25.6805],[210,25.5279],[214,25.7208],[216,25.9442],[219,26.4951],[220,26.7274],[221,27.0175],[225,28.3917],[226,28.6961],[227,28.9122],[229,29.126],[248,29.3466],[251,29.177],[254,28.7727],[257,28.1917],[258,27.7754],[259,26.9906],[260,25.2926],[261,22.9214],[262,20.9406],[263,19.7734],[264,19.034],[265,18.6374],[266,18.4416],[267,18.3365],[270,18.2031],[283,18.1112],[284,18.1389],[285,18.2135],[286,18.3529],[287,18.5744],[288,19.1293],[289,20.1167],[291,22.5539],[292,23.8414],[294,26.6759],[295,27.9116],[296,28.9656],[297,29.9129],[298,30.7058],[299,31.2966],[301,32.1607],[303,32.8671],[306,33.6625],[309,34.1908],[313,34.5795],[319,34.731],[322,34.6035],[324,34.3662],[331,32.8182],[334,32.0558],[335,31.9078],[337,31.7996],[344,32.0315],[345,32.1371],[348,32.8154],[349,32.8741],[350,32.7267],[351,32.4293],[352,31.9361],[353,31.0053],[354,29.3354],[355,20.4553],[356,7.4243],[357,3.9957],[358,3.7969],[359,3.7137],[360,3.7799],[362,4.2196],[363,4.7815],[364,5.2446],[365,5.2343],[366,5.1601],[368,4.6848],[370,4.2369],[371,3.9761],[373,3.4081],[374,3.0892],[375,2.759],[376,2.495],[377,2.3259],[378,2.2035],[379,2.1174],[380,2.0616],[381,2.0256],[382,1.9554],[383,1.9482],[384,1.9515],[385,1.9651],[386,1.9929],[387,2.0351],[391,2.2437],[393,2.3564],[394,2.4081],[395,2.4509],[397,2.5185],[398,2.5464],[399,2.569],[401,2.5972],[402,2.5972],[403,2.6092],[406,2.6082],[407,2.601],[408,2.601],[409,2.5902],[410,2.5902],[411,2.576],[412,2.576],[413,2.5614],[414,2.5614],[415,2.5492],[416,2.5492],[417,2.5396],[418,2.5396],[419,2.531],[420,2.531],[421,2.5237],[422,2.5237],[423,2.5181],[433,2.5115]]
       }
       thigh_l_fk_ctrl.Rotation.Z {
        type=MovieSceneScriptingFloatChannel
        default=-0.1664
        thinned_from=434
        keys [[0,5.0296],[6,5.0386],[9,5.0473],[10,5.0662],[12,5.0643],[13,5.0427],[14,4.9993],[15,4.9673],[16,4.9242],[17,4.7801],[18,4.4612],[19,3.6968],[20,2.6378],[21,1.8211],[22,1.458],[23,1.3103],[24,1.2161],[25,1.1285],[26,1.0884],[28,1.0539],[29,1.0325],[30,1.0193],[31,1.0126],[32,1.0106],[33,1.0169],[34,1.0341],[35,1.0583],[36,1.0862],[37,1.1196],[39,1.2004],[40,1.2386],[41,1.2693],[42,1.2917],[43,1.3089],[44,1.3218],[46,1.3405],[48,1.3532],[53,1.3721],[56,1.3905],[58,1.4104],[59,1.4254],[60,1.4452],[61,1.471],[62,1.4871],[63,1.507],[64,1.5767],[65,1.8196],[66,2.0902],[67,2.1733],[69,2.2155],[70,2.2263],[71,2.2434],[73,2.3034],[74,2.3177],[76,2.3178],[77,2.3215],[83,2.3859],[85,2.3859],[86,2.4131],[87,2.4131],[88,2.4206],[94,2.4209],[95,2.4176],[96,2.4213],[97,2.4324],[98,2.4535],[99,2.4871],[100,2.5479],[101,2.6481],[102,2.787],[103,2.9644],[104,3.2068],[106,3.8077],[107,4.0594],[108,4.2523],[109,4.4126],[110,4.5405],[111,4.6346],[112,4.6895],[113,4.7097],[116,4.7016],[120,4.6459],[126,4.6276],[131,4.6815],[132,4.6832],[133,4.6748],[134,4.6218],[135,4.4603],[136,4.1402],[137,3.6112],[138,2.8232],[139,0.8325],[140,-3.7489],[141,-12.1844],[142,-19.6827],[143,-26.7269],[144,-32.1038],[145,-34.3505],[146,-35.2943],[147,-35.7885],[148,-35.7579],[149,-34.9084],[150,-25.3557],[151,-18.0309],[152,-13.6037],[153,-10.8131],[154,-9.8058],[155,-9.0205],[156,-8.5259],[157,-40.804],[158,-69.4927],[159,-81.0607],[160,-86.9325],[161,-89.7559],[162,-90.6573],[163,-87.6552],[164,-82.6009],[165,-79.6659],[166,-77.244],[167,-75.0322],[168,-73.4707],[169,-74.1564],[170,-76.4603],[171,-78.9274],[172,-81.2268],[173,-82.9768],[174,-84.1568],[175,-84.9545],[176,-85.4219],[178,-85.7956],[191,-83.4756],[214,-82.7608],[218,-81.9432],[220,-81.2158],[222,-79.9986],[226,-76.76],[227,-76.0674],[229,-75.2542],[234,-74.4497],[257,-74.1174],[259,-74.4329],[260,-75.0789],[261,-76.0175],[262,-76.8027],[264,-77.8118],[266,-78.3587],[284,-78.8007],[287,-79.3296],[288,-79.7076],[294,-83.1113],[298,-84.522],[318,-85.7298],[336,-83.9004],[346,-84.3278],[347,-84.2231],[348,-83.7384],[349,-82.707],[350,-80.9778],[351,-78.2351],[352,-74.3642],[353,-69.5287],[354,-63.9109],[355,-44.0968],[356,-18.1401],[357,-5.5838],[358,-0.2756],[359,2.892],[360,5.2478],[361,6.8307],[362,7.7816],[363,6.9005],[364,6.1518],[365,6.757],[366,7.1089],[367,7.5559],[368,7.7614],[369,7.8009],[370,7.7653],[371,7.691],[372,7.5587],[373,7.3504],[374,7.0453],[376,6.304],[377,5.9485],[378,5.6072],[379,5.253],[380,4.9168],[381,4.6489],[382,4.6463],[383,4.4844],[384,4.3916],[385,4.3801],[386,4.43],[387,4.5061],[389,4.6825],[390,4.7569],[392,4.8743],[394,4.9717],[397,5.0792],[400,5.1489],[401,5.1639],[402,5.1639],[403,5.1824],[408,5.1888],[410,5.1825],[411,5.1699],[412,5.1699],[413,5.154],[414,5.154],[415,5.1379],[416,5.1379],[417,5.1211],[418,5.1211],[419,5.1023],[420,5.1023],[421,5.0836],[422,5.0836],[423,5.0671],[424,5.0671],[425,5.0546],[427,5.0546],[428,5.0402],[433,5.0402]]
       }
       thigh_l_fk_ctrl.Scale.X {type=MovieSceneScriptingFloatChannel, default=1, const [1,434]}
       thigh_l_fk_ctrl.Scale.Y {type=MovieSceneScriptingFloatChannel, default=1, const [1,434]}
       thigh_l_fk_ctrl.Scale.Z {type=MovieSceneScriptingFloatChannel, default=1, const [1,434]}
       calf_l_fk_ctrl.Location.X {
        type=MovieSceneScriptingFloatChannel
        default=0
        thinned_from=434
        keys [[0,-0.9869],[11,-0.9857],[13,-0.9813],[16,-0.9675],[17,-0.9552],[18,-0.9325],[19,-0.883],[20,-0.8201],[21,-0.7782],[22,-0.7629],[23,-0.7582],[24,-0.76],[27,-0.7762],[36,-0.7852],[39,-0.7835],[53,-0.7582],[62,-0.7512],[63,-0.7564],[64,-0.7754],[65,-0.834],[66,-0.8967],[67,-0.9172],[68,-0.9172],[77,-0.8973],[106,-0.8801],[127,-0.8974],[129,-0.8959],[131,-0.8902],[132,-0.8831],[133,-0.8676],[134,-0.8393],[135,-0.7939],[136,-0.7278],[137,-0.6375],[138,-0.5195],[139,-0.364],[140,-0.0495],[141,0.0455],[142,0.0765],[143,0.1035],[144,0.1209],[145,0.1253],[146,0.1258],[147,0.1254],[148,0.1232],[149,0.117],[150,0.019],[151,0.021],[152,-0.006],[153,-0.0193],[156,-0.0193],[157,0.0995],[158,0.1566],[159,0.1751],[160,0.1823],[161,0.1834],[162,0.1796],[163,0.1588],[164,0.1227],[165,0.1034],[166,0.094],[167,0.0909],[168,0.0954],[169,0.118],[170,0.1443],[171,0.1572],[172,0.1632],[173,0.1655],[174,0.1668],[175,0.1674],[177,0.1675],[178,0.1669],[179,0.1658],[181,0.1627],[182,0.1616],[185,0.1601],[214,0.1593],[220,0.1607],[227,0.1642],[253,0.1648],[257,0.1638],[258,0.1631],[259,0.162],[260,0.1597],[262,0.1534],[263,0.1512],[264,0.1495],[265,0.1483],[267,0.1467],[271,0.1452],[282,0.1451],[286,0.1465],[287,0.1471],[288,0.1481],[292,0.1553],[294,0.1592],[296,0.1622],[298,0.1643],[304,0.1672],[316,0.1684],[336,0.164],[340,0.1647],[343,0.1664],[344,0.1672],[345,0.1693],[346,0.1733],[348,0.1823],[349,0.1857],[350,0.1876],[351,0.1882],[352,0.1877],[353,0.185],[354,0.1789],[355,0.1303],[356,0.024],[357,-0.0193],[361,-0.0193],[362,-0.2808],[363,-0.3206],[364,-0.3351],[365,-0.5132],[366,-0.8127],[367,-1.1148],[368,-1.3381],[369,-1.4506],[370,-1.4811],[371,-1.4821],[372,-1.4564],[373,-1.4068],[374,-1.2694],[375,-1.0613],[376,-0.9079],[377,-0.8369],[378,-0.7895],[379,-0.7603],[380,-0.7447],[381,-0.7382],[382,-0.7414],[383,-0.7515],[384,-0.7673],[385,-0.786],[386,-0.8087],[387,-0.8361],[389,-0.8933],[390,-0.9169],[391,-0.9363],[392,-0.9536],[393,-0.9686],[394,-0.9808],[395,-0.9901],[397,-1.0029],[399,-1.0108],[402,-1.0151],[421,-0.9879],[433,-0.9869]]
       }
       calf_l_fk_ctrl.Location.Y {
        type=MovieSceneScriptingFloatChannel
        default=0
        thinned_from=434
        keys [[0,0.0795],[11,0.0794],[13,0.079],[16,0.0779],[17,0.0769],[18,0.0751],[19,0.0711],[20,0.066],[21,0.0627],[22,0.0614],[23,0.061],[24,0.0612],[27,0.0625],[37,0.0632],[62,0.0605],[63,0.0609],[64,0.0624],[65,0.0671],[66,0.0722],[67,0.0739],[68,0.0739],[80,0.072],[83,0.0719],[84,0.0639],[85,0.0639],[86,0.0718],[115,0.0712],[127,0.0723],[130,0.072],[131,0.0717],[132,0.0711],[133,0.0699],[134,0.0676],[135,0.0639],[136,0.0586],[137,0.0513],[138,0.0418],[139,0.0291],[140,0.0043],[141,0.003],[142,0.0107],[143,0.0214],[144,0.0306],[145,0.0332],[146,0.0335],[147,0.0333],[148,0.0319],[149,0.0283],[150,0.0102],[151,0.0002],[152,0.0003],[153,0.0015],[156,0.0015],[157,0.0195],[158,0.0566],[159,0.0751],[160,0.0838],[161,0.0851],[162,0.0804],[163,0.0586],[164,0.0316],[165,0.0213],[166,0.0171],[167,0.0159],[168,0.0177],[169,0.0289],[170,0.0464],[171,0.0572],[172,0.0627],[173,0.0649],[174,0.0663],[175,0.0669],[176,0.067],[177,0.067],[178,0.0664],[179,0.0652],[181,0.0623],[182,0.0612],[184,0.0601],[187,0.0594],[212,0.0588],[218,0.0598],[221,0.0608],[226,0.0633],[230,0.0642],[251,0.0645],[256,0.0637],[257,0.0633],[258,0.0627],[259,0.0616],[260,0.0594],[261,0.0564],[262,0.0537],[263,0.0519],[264,0.0505],[265,0.0495],[267,0.0483],[270,0.0473],[278,0.0468],[279,0.0505],[280,0.0468],[284,0.0474],[287,0.0485],[288,0.0494],[289,0.0507],[292,0.0554],[294,0.059],[295,0.0605],[297,0.0628],[298,0.0638],[301,0.0655],[306,0.0672],[312,0.0679],[322,0.0673],[334,0.0638],[337,0.0635],[340,0.0642],[342,0.0652],[344,0.0667],[345,0.0689],[346,0.0731],[348,0.0837],[349,0.0881],[350,0.0906],[351,0.0915],[352,0.0907],[353,0.0872],[354,0.0796],[355,0.0364],[356,0.0004],[357,0.0015],[361,0.0015],[362,0.0226],[363,0.0256],[364,0.027],[365,0.0412],[366,0.0653],[367,0.0896],[368,0.1076],[369,0.1166],[370,0.1191],[371,0.1192],[372,0.1171],[373,0.1131],[374,0.1021],[375,0.0853],[376,0.073],[377,0.0673],[378,0.0635],[379,0.0611],[380,0.0599],[381,0.0594],[382,0.0597],[383,0.0605],[384,0.0618],[385,0.0633],[386,0.0651],[389,0.0719],[390,0.0738],[392,0.0768],[393,0.078],[394,0.079],[396,0.0803],[399,0.0814],[402,0.0817],[433,0.0795]]
       }
       calf_l_fk_ctrl.Location.Z {
        type=MovieSceneScriptingFloatChannel
        default=0
        thinned_from=434
        keys [[0,0.0006],[83,0.0006],[84,0.0036],[85,0.0036],[86,0.0006],[139,0.0002],[149,-0.0001],[156,0],[278,-0.0001],[279,0.0117],[280,-0.0001],[355,-0.0001],[361,0],[372,0.0009],[433,0.0006]]
       }
       calf_l_fk_ctrl.Rotation.X {
        type=MovieSceneScriptingFloatChannel
        default=0.0543
        thinned_from=434
        keys [[0,0.0033],[83,0.0033],[84,-0.0003],[85,-0.0003],[86,0.0033],[138,0.0033],[139,0.0022],[140,-0.0011],[141,-0.0092],[143,-0.0242],[144,-0.0295],[145,-0.031],[146,-0.0311],[147,-0.031],[148,-0.0303],[149,-0.0283],[150,-0.0132],[151,-0.004],[152,0.0011],[153,0.0033],[156,0.0033],[157,-0.023],[158,-0.0421],[159,-0.0498],[160,-0.0531],[161,-0.0536],[162,-0.0518],[163,-0.043],[164,-0.0301],[165,-0.0242],[166,-0.0215],[167,-0.0206],[168,-0.0219],[169,-0.0286],[170,-0.0375],[171,-0.0424],[172,-0.0448],[173,-0.0457],[174,-0.0462],[176,-0.0466],[178,-0.0463],[181,-0.0446],[183,-0.0439],[216,-0.0433],[231,-0.0454],[256,-0.0452],[258,-0.0447],[259,-0.0443],[260,-0.0434],[262,-0.0409],[264,-0.0394],[267,-0.0384],[278,-0.0377],[279,-0.0484],[280,-0.0377],[287,-0.0385],[288,-0.0389],[295,-0.0438],[298,-0.0452],[311,-0.0469],[337,-0.0451],[342,-0.0458],[344,-0.0464],[345,-0.0473],[346,-0.049],[348,-0.0531],[349,-0.0547],[350,-0.0557],[351,-0.056],[352,-0.0557],[353,-0.0544],[354,-0.0515],[355,-0.0326],[356,-0.0046],[357,0.0033],[362,0.0033],[363,0.0013],[364,-0.0002],[367,0.0016],[368,0.0019],[381,0.003],[382,0.0032],[433,0.0033]]
       }
       calf_l_fk_ctrl.Rotation.Y {
        type=MovieSceneScriptingFloatChannel
        default=1.2238
        thinned_from=434
        keys [[0,0.0047],[83,0.0047],[84,0.0002],[85,0.0002],[86,0.0047],[138,0.0047],[139,0.003],[140,-0.0013],[141,-0.0098],[142,-0.0154],[143,-0.0195],[144,-0.0217],[145,-0.0222],[147,-0.0222],[148,-0.022],[149,-0.0213],[151,-0.0047],[152,0.0014],[153,0.0047],[156,0.0047],[157,-0.0189],[158,-0.025],[159,-0.0257],[162,-0.0257],[163,-0.0251],[164,-0.0219],[165,-0.0195],[166,-0.0181],[167,-0.0177],[168,-0.0183],[169,-0.0214],[170,-0.0241],[171,-0.025],[172,-0.0253],[259,-0.0253],[278,-0.0241],[279,-0.0381],[280,-0.0241],[354,-0.0257],[355,-0.0228],[356,-0.0054],[357,0.0047],[362,0.0047],[363,0.0018],[364,-0.0002],[365,0.0006],[366,0.0012],[367,0.0021],[368,0.0027],[375,0.0034],[381,0.0042],[382,0.0046],[433,0.0047]]
       }
       calf_l_fk_ctrl.Rotation.Z {
        type=MovieSceneScriptingFloatChannel
        default=0.3732
        thinned_from=434
        keys [[0,-4.9603],[83,-4.9605],[84,-4.9491],[85,-4.9491],[86,-4.9607],[138,-4.9606],[139,-3.217],[140,1.4872],[141,11.7072],[142,19.9107],[143,27.373],[144,32.4152],[145,33.7278],[146,33.8841],[147,33.7722],[148,33.107],[149,31.2736],[150,16.1797],[151,5.3904],[152,-1.5367],[153,-4.9607],[156,-4.9607],[157,26.232],[158,43.772],[159,50.4991],[160,53.3853],[161,53.8227],[162,52.278],[163,44.5477],[164,32.9616],[165,27.3534],[166,24.7094],[167,23.8403],[168,25.1034],[169,31.5718],[170,39.6679],[171,43.9933],[172,46.0742],[173,46.8987],[174,47.3782],[175,47.5963],[177,47.6332],[178,47.4402],[179,47.006],[181,45.9209],[182,45.5337],[184,45.117],[211,44.6042],[219,45.0999],[227,46.4261],[252,46.6946],[256,46.43],[257,46.299],[258,46.0633],[259,45.6596],[260,44.8535],[261,43.7075],[262,42.6661],[263,41.9348],[264,41.3663],[265,40.9649],[267,40.4608],[270,40.056],[279,39.8405],[283,39.9982],[286,40.3867],[287,40.5746],[288,40.9074],[289,41.4336],[292,43.3231],[294,44.6858],[296,45.7156],[298,46.4759],[302,47.262],[310,47.9379],[322,47.7603],[335,46.4029],[338,46.4296],[341,46.7953],[344,47.5332],[345,48.3204],[346,49.8024],[348,53.3697],[349,54.7982],[350,55.5993],[351,55.8711],[352,55.6449],[353,54.5157],[354,52.004],[355,35.2678],[356,6.1624],[357,-4.9606],[362,-4.9607],[363,-1.958],[364,0.264],[365,-0.6494],[366,-1.3121],[367,-2.2968],[368,-2.8883],[369,-3.1443],[370,-3.2197],[372,-3.3046],[373,-3.3779],[374,-3.5072],[375,-3.683],[376,-3.9523],[377,-4.1634],[378,-4.3075],[379,-4.3991],[381,-4.5001],[382,-4.9405],[433,-4.9607]]
       }
       calf_l_fk_ctrl.Scale.X {type=MovieSceneScriptingFloatChannel, default=1, const [1,434]}
       calf_l_fk_ctrl.Scale.Y {type=MovieSceneScriptingFloatChannel, default=1, const [1,434]}
       calf_l_fk_ctrl.Scale.Z {type=MovieSceneScriptingFloatChannel, default=1, const [1,434]}
       foot_l_fk_ctrl.Location.X {
        type=MovieSceneScriptingFloatChannel
        default=0
        thinned_from=434
        keys [[0,-1.0581],[11,-1.0568],[13,-1.052],[16,-1.0369],[17,-1.0235],[18,-0.9987],[19,-0.9446],[20,-0.8757],[21,-0.8299],[22,-0.8131],[23,-0.808],[24,-0.81],[27,-0.8277],[36,-0.8375],[39,-0.8357],[53,-0.808],[62,-0.8003],[63,-0.8061],[64,-0.8268],[65,-0.8909],[66,-0.9595],[67,-0.9819],[68,-0.982],[70,-0.9743],[79,-0.958],[106,-0.9414],[127,-0.9603],[129,-0.9586],[131,-0.9524],[132,-0.9446],[133,-0.9277],[134,-0.8968],[135,-0.8471],[136,-0.7748],[137,-0.676],[138,-0.547],[139,-0.3844],[140,-0.0605],[141,0],[149,0],[150,-0.0476],[151,0],[361,0],[362,-0.286],[363,-0.3422],[364,-0.3676],[365,-0.5585],[367,-1.2093],[368,-1.4509],[369,-1.5729],[370,-1.606],[371,-1.6068],[372,-1.5786],[373,-1.524],[374,-1.3732],[375,-1.1449],[376,-0.976],[377,-0.8975],[378,-0.845],[379,-0.8127],[380,-0.7955],[381,-0.7881],[382,-0.7898],[383,-0.8008],[384,-0.818],[385,-0.8385],[386,-0.8633],[387,-0.8933],[389,-0.9558],[390,-0.9816],[391,-1.0028],[392,-1.0218],[393,-1.0381],[394,-1.0515],[395,-1.0616],[397,-1.0756],[399,-1.0843],[402,-1.089],[421,-1.0592],[433,-1.0582]]
       }
       foot_l_fk_ctrl.Location.Y {
        type=MovieSceneScriptingFloatChannel
        default=0.0398
        thinned_from=434
        keys [[0,-0.0066],[17,-0.0063],[19,-0.0059],[21,-0.0051],[63,-0.005],[64,-0.0051],[66,-0.0059],[132,-0.0059],[134,-0.0056],[136,-0.0048],[138,-0.0034],[139,-0.0024],[140,-0.0004],[141,0],[149,0],[150,-0.0003],[151,0],[361,0],[362,-0.0018],[364,-0.0023],[365,-0.0035],[367,-0.0075],[368,-0.009],[369,-0.0097],[371,-0.01],[373,-0.0094],[374,-0.0085],[375,-0.0071],[376,-0.006],[378,-0.0052],[382,-0.0049],[433,-0.0066]]
       }
       foot_l_fk_ctrl.Location.Z {
        type=MovieSceneScriptingFloatChannel
        default=-0.8885
        thinned_from=434
        keys [[0,0.0007],[138,0.0004],[361,0],[365,0.0004],[369,0.0011],[374,0.001],[433,0.0007]]
       }
       foot_l_fk_ctrl.Rotation.X {
        type=MovieSceneScriptingFloatChannel
        default=-1.1647
        thinned_from=434
        keys [[0,-0.206],[1,-0.206],[4,-0.2075],[5,-0.2075],[6,-0.2087],[7,-0.2087],[8,-0.2096],[11,-0.2097],[12,-0.2093],[13,-0.2085],[14,-0.2071],[15,-0.2052],[16,-0.2015],[17,-0.1933],[18,-0.1787],[19,-0.1534],[20,-0.1229],[21,-0.0985],[22,-0.0818],[23,-0.0694],[24,-0.0615],[25,-0.0565],[26,-0.0522],[27,-0.0487],[28,-0.046],[29,-0.044],[30,-0.0426],[31,-0.0417],[32,-0.0414],[34,-0.0416],[36,-0.0423],[37,-0.0423],[38,-0.0433],[45,-0.0479],[54,-0.0511],[57,-0.0518],[58,-0.0522],[59,-0.0528],[60,-0.0538],[61,-0.0553],[62,-0.0577],[63,-0.062],[64,-0.0692],[65,-0.0834],[66,-0.0998],[67,-0.1113],[68,-0.1196],[69,-0.1249],[70,-0.1282],[71,-0.1301],[72,-0.131],[73,-0.1315],[76,-0.1317],[81,-0.1305],[92,-0.1256],[96,-0.1246],[98,-0.1245],[99,-0.1246],[100,-0.1251],[101,-0.126],[102,-0.1276],[103,-0.1298],[104,-0.133],[105,-0.1371],[106,-0.1415],[107,-0.1455],[109,-0.1526],[110,-0.1558],[111,-0.1584],[112,-0.1605],[113,-0.162],[114,-0.1631],[118,-0.1649],[129,-0.1656],[133,-0.1643],[135,-0.163],[136,-0.1619],[137,-0.1603],[138,-0.1581],[139,-0.1579],[140,-0.1944],[141,-0.1397],[142,-0.1336],[143,-0.1317],[144,-0.1314],[145,-0.1309],[146,-0.1309],[148,-0.1277],[149,-0.1244],[150,-0.6385],[151,-1.4549],[152,-1.6989],[153,-1.3501],[154,-0.873],[155,-0.483],[156,-0.2324],[157,0.1243],[158,0.2542],[159,0.279],[161,0.2955],[162,0.3014],[163,0.3101],[164,0.3264],[165,0.3328],[167,0.3378],[168,0.3424],[169,0.3501],[171,0.3216],[172,0.3166],[184,0.3208],[218,0.3188],[219,0.3193],[220,0.3206],[221,0.3232],[224,0.3357],[226,0.3466],[227,0.351],[229,0.3556],[236,0.3633],[252,0.3644],[255,0.3627],[256,0.3606],[257,0.3554],[258,0.3468],[259,0.3342],[260,0.3135],[261,0.2856],[262,0.2591],[263,0.2373],[264,0.2179],[265,0.2025],[266,0.191],[267,0.1823],[268,0.1757],[269,0.1707],[270,0.1667],[271,0.1633],[272,0.1608],[273,0.1589],[274,0.1579],[277,0.1577],[278,0.1586],[279,0.1586],[281,0.163],[282,0.1664],[283,0.1715],[284,0.1777],[285,0.1848],[287,0.1996],[288,0.2074],[290,0.2246],[291,0.2324],[292,0.2396],[293,0.246],[294,0.2501],[295,0.2523],[296,0.2539],[298,0.2555],[300,0.256],[303,0.2546],[308,0.2496],[312,0.2432],[323,0.2185],[326,0.2134],[328,0.2109],[330,0.2094],[332,0.2093],[338,0.21],[341,0.2087],[343,0.2066],[344,0.2049],[345,0.2005],[346,0.1917],[347,0.1789],[348,0.1628],[349,0.1442],[350,0.1239],[351,0.1026],[352,0.0776],[353,0.0452],[354,0.0022],[355,-0.1611],[356,-0.8345],[357,-1.5944],[358,-1.2391],[359,-0.8706],[360,-0.5742],[361,-0.452],[362,-0.3935],[363,-0.2937],[364,-0.2621],[365,-0.2486],[366,-0.2363],[367,-0.2228],[368,-0.2146],[369,-0.2104],[372,-0.2031],[373,-0.2004],[374,-0.1983],[375,-0.1967],[376,-0.1943],[379,-0.1856],[381,-0.181],[383,-0.1779],[384,-0.1769],[385,-0.1765],[386,-0.1769],[387,-0.1781],[390,-0.1839],[392,-0.1888],[394,-0.1942],[397,-0.2009],[399,-0.2045],[401,-0.2066],[404,-0.208],[433,-0.206]]
       }
       foot_l_fk_ctrl.Rotation.Y {
        type=MovieSceneScriptingFloatChannel
        default=0.0792
        thinned_from=434
        keys [[0,-4.5904],[5,-4.611],[6,-4.6273],[7,-4.6273],[8,-4.6393],[12,-4.6358],[14,-4.6057],[15,-4.5792],[16,-4.528],[17,-4.4166],[18,-4.2186],[19,-3.8767],[20,-3.4689],[21,-3.1413],[22,-2.9178],[23,-2.75],[24,-2.6444],[25,-2.5773],[26,-2.5197],[27,-2.4727],[28,-2.4364],[29,-2.4096],[30,-2.3906],[31,-2.3791],[33,-2.3748],[36,-2.3865],[37,-2.3865],[38,-2.4004],[47,-2.4727],[58,-2.5176],[60,-2.5397],[61,-2.5587],[62,-2.591],[63,-2.65],[64,-2.7465],[65,-2.9393],[66,-3.1616],[67,-3.318],[68,-3.4311],[69,-3.5035],[70,-3.5492],[71,-3.5745],[73,-3.5935],[97,-3.4963],[100,-3.5042],[101,-3.5166],[102,-3.5368],[103,-3.5659],[104,-3.6083],[107,-3.7742],[110,-3.911],[112,-3.9735],[114,-4.0093],[129,-4.0428],[135,-4.0087],[137,-3.9752],[139,-3.9144],[140,-3.8981],[141,-3.7773],[142,-3.7244],[143,-3.719],[144,-3.7268],[146,-3.721],[148,-3.6737],[149,-3.6247],[150,-3.7602],[151,-4.2167],[152,-4.331],[153,-3.8567],[154,-3.2726],[155,-2.7525],[156,-2.4432],[157,0.0783],[158,1.8225],[159,2.1138],[160,2.1961],[161,2.2874],[162,2.3507],[163,2.4508],[164,2.6731],[165,2.7678],[166,2.8112],[167,2.8433],[168,2.9047],[169,3.0044],[171,2.6057],[172,2.5369],[176,2.5605],[177,2.5605],[182,2.593],[208,2.5854],[214,2.5684],[218,2.5354],[219,2.5334],[220,2.5395],[221,2.5589],[222,2.5903],[224,2.6662],[225,2.7148],[226,2.77],[227,2.8117],[228,2.8364],[235,2.9254],[238,2.9449],[252,2.9444],[255,2.9269],[256,2.9017],[257,2.8416],[258,2.7415],[259,2.5959],[260,2.3617],[261,2.0513],[262,1.7562],[263,1.5129],[264,1.2968],[265,1.1244],[266,0.9953],[267,0.8955],[268,0.8199],[269,0.763],[270,0.7172],[271,0.6789],[272,0.6488],[273,0.6275],[274,0.6156],[275,0.6133],[277,0.6133],[278,0.6255],[279,0.6255],[280,0.6543],[281,0.6802],[282,0.7225],[283,0.7847],[284,0.8616],[285,0.948],[287,1.1293],[288,1.2249],[290,1.4327],[291,1.5272],[292,1.6146],[293,1.6908],[294,1.7386],[295,1.7636],[296,1.7817],[297,1.7923],[299,1.8009],[300,1.8011],[302,1.7896],[307,1.7285],[309,1.6952],[311,1.6545],[314,1.5815],[321,1.3786],[323,1.3255],[325,1.2804],[327,1.244],[328,1.2294],[329,1.2179],[330,1.2101],[331,1.2077],[332,1.2103],[334,1.2206],[336,1.2243],[339,1.2239],[340,1.2201],[341,1.2131],[342,1.2023],[343,1.1867],[344,1.1656],[345,1.1059],[346,0.9808],[347,0.7992],[348,0.5706],[349,0.3057],[350,0.0178],[351,-0.2819],[352,-0.6335],[353,-1.0942],[354,-1.7117],[355,-2.8517],[356,-4.5244],[357,-5.714],[358,-5.4599],[359,-5.0847],[360,-4.7687],[361,-4.6817],[362,-4.8527],[363,-4.9543],[364,-4.9418],[365,-4.8869],[367,-4.7086],[368,-4.6382],[369,-4.5971],[373,-4.4757],[375,-4.4405],[376,-4.415],[380,-4.2774],[383,-4.2071],[385,-4.1884],[386,-4.194],[388,-4.2337],[391,-4.3201],[397,-4.5211],[399,-4.5701],[403,-4.6142],[433,-4.5909]]
       }
       foot_l_fk_ctrl.Rotation.Z {
        type=MovieSceneScriptingFloatChannel
        default=-0.2074
        thinned_from=434
        keys [[0,1.3478],[5,1.3486],[7,1.3466],[12,1.3146],[13,1.3103],[14,1.3099],[15,1.3008],[16,1.2878],[17,1.2971],[18,1.3438],[19,1.5184],[20,1.7962],[21,2.0155],[22,2.1022],[23,2.1289],[24,2.1507],[25,2.1785],[26,2.1908],[28,2.2021],[31,2.2252],[33,2.2304],[35,2.2221],[36,2.214],[37,2.214],[38,2.1894],[41,2.1489],[46,2.1149],[59,2.0694],[61,2.0544],[62,2.0531],[63,2.0658],[64,2.0898],[65,2.1315],[66,2.1926],[67,2.2834],[68,2.3667],[69,2.4142],[70,2.4478],[71,2.4654],[72,2.4678],[76,2.4529],[78,2.4319],[85,2.3283],[90,2.2865],[96,2.264],[98,2.2474],[99,2.2323],[100,2.2083],[101,2.1731],[102,2.127],[103,2.0711],[104,1.9983],[106,1.8225],[107,1.7474],[108,1.6873],[109,1.6351],[110,1.5921],[111,1.5585],[112,1.5369],[113,1.5259],[125,1.4926],[130,1.4939],[132,1.5017],[133,1.5084],[134,1.5268],[135,1.5714],[136,1.6491],[137,1.7664],[138,1.9308],[139,1.3439],[140,-0.1293],[141,-4.7894],[142,-8.0949],[143,-11.0982],[144,-13.0452],[145,-13.4593],[146,-13.4593],[147,-13.3326],[148,-12.9847],[149,-12.1045],[150,-0.2314],[151,11.7602],[152,17.2516],[153,20.0341],[154,20.309],[155,20.1049],[156,17.2521],[157,-1.9441],[158,-8.9429],[159,-12.246],[160,-13.8258],[161,-14.1648],[162,-13.6267],[163,-10.3466],[164,-5.0936],[165,-2.4965],[166,-1.2602],[167,-0.8321],[168,-1.3371],[169,-4.1277],[170,-7.9105],[171,-9.9998],[172,-10.9658],[173,-11.3091],[174,-11.5066],[175,-11.5932],[177,-11.6124],[178,-11.514],[179,-11.3227],[181,-10.8464],[182,-10.6765],[183,-10.5733],[184,-10.4958],[186,-10.4014],[189,-10.3466],[190,-10.3466],[191,-10.3248],[208,-10.2667],[211,-10.3238],[213,-10.4031],[215,-10.5313],[217,-10.7322],[218,-10.8667],[219,-11.0272],[220,-11.2158],[221,-11.4672],[222,-11.7874],[226,-13.1863],[227,-13.4667],[228,-13.6596],[229,-13.8097],[230,-13.9238],[232,-14.0741],[236,-14.2318],[249,-14.3117],[253,-14.2229],[255,-14.1178],[256,-14.0332],[257,-13.8789],[258,-13.6053],[259,-13.1597],[260,-12.3024],[261,-11.1023],[262,-10.0097],[263,-9.1988],[264,-8.5304],[265,-8.0356],[266,-7.692],[267,-7.4405],[268,-7.2558],[269,-7.1116],[270,-6.9949],[271,-6.9076],[272,-6.8453],[273,-6.8034],[275,-6.7639],[279,-6.7572],[281,-6.8057],[282,-6.8496],[283,-6.919],[284,-7.0113],[285,-7.1234],[287,-7.3946],[288,-7.6071],[289,-7.9136],[292,-8.9705],[293,-9.3503],[294,-9.6953],[295,-9.9797],[296,-10.2219],[297,-10.4344],[298,-10.6111],[299,-10.7463],[301,-10.9463],[304,-11.1619],[307,-11.2927],[311,-11.3566],[318,-11.2586],[324,-11.0429],[334,-10.4694],[336,-10.4003],[338,-10.4003],[340,-10.467],[342,-10.6084],[344,-10.8256],[345,-11.211],[346,-11.9846],[347,-12.9609],[348,-13.97],[349,-14.8581],[350,-15.4815],[351,-15.8499],[352,-16.0296],[353,-15.9471],[354,-15.4799],[355,-8.6357],[356,7.5044],[357,16.0408],[358,16.2089],[359,15.1394],[360,13.3938],[361,10.2075],[362,4.2466],[363,-0.2319],[364,-1.5757],[365,-1.2001],[366,-0.8857],[367,-0.4213],[368,-0.1169],[369,0.033],[370,0.0968],[371,0.1481],[372,0.2013],[373,0.2769],[374,0.3794],[375,0.5013],[376,0.6721],[377,0.8204],[378,0.9365],[379,1.028],[380,1.0943],[381,1.1526],[382,1.3815],[383,1.3935],[384,1.3943],[385,1.3884],[389,1.3423],[390,1.3345],[392,1.3304],[433,1.3478]]
       }
       foot_l_fk_ctrl.Scale.X {type=MovieSceneScriptingFloatChannel, default=1, const [1,434]}
       foot_l_fk_ctrl.Scale.Y {type=MovieSceneScriptingFloatChannel, default=1, const [1,434]}
       foot_l_fk_ctrl.Scale.Z {type=MovieSceneScriptingFloatChannel, default=1, const [1,434]}
       ball_l_fk_ctrl.Location.X {type=MovieSceneScriptingFloatChannel, default=0, thinned_from=434, keys [[0,0],[433,0]]}
       ball_l_fk_ctrl.Location.Y {type=MovieSceneScriptingFloatChannel, default=0, thinned_from=434, keys [[0,0],[433,0]]}
       ball_l_fk_ctrl.Location.Z {type=MovieSceneScriptingFloatChannel, default=0, thinned_from=434, keys [[0,0],[433,0]]}
       ball_l_fk_ctrl.Rotation.X {
        type=MovieSceneScriptingFloatChannel
        default=0
        thinned_from=434
        keys [[0,0],[152,0],[153,-1.0047],[154,-1.8458],[155,-2.268],[156,-1.8862],[157,0],[356,0],[357,-0.1069],[358,-1.4174],[359,-2.0297],[360,-2.2385],[361,-1.8386],[362,-0.5769],[363,0],[433,0]]
       }
       ball_l_fk_ctrl.Rotation.Y {
        type=MovieSceneScriptingFloatChannel
        default=0
        thinned_from=434
        keys [[0,0],[152,0],[153,-0.6167],[154,-1.2416],[155,-1.6296],[156,-1.346],[157,0],[356,0],[357,-0.0607],[358,-0.8981],[359,-1.3828],[360,-1.5951],[361,-1.2917],[362,-0.3694],[363,0],[433,0]]
       }
       ball_l_fk_ctrl.Rotation.Z {
        type=MovieSceneScriptingFloatChannel
        default=0
        thinned_from=434
        keys [[0,0],[152,0],[153,-2.5918],[154,-5.1275],[155,-6.7796],[156,-6.091],[157,0],[356,0],[357,-0.2648],[358,-3.726],[359,-5.7846],[360,-6.9064],[361,-6.0202],[362,-1.9894],[363,0],[433,0]]
       }
       ball_l_fk_ctrl.Scale.X {type=MovieSceneScriptingFloatChannel, default=1, const [1,434]}
       ball_l_fk_ctrl.Scale.Y {type=MovieSceneScriptingFloatChannel, default=1, const [1,434]}
       ball_l_fk_ctrl.Scale.Z {type=MovieSceneScriptingFloatChannel, default=1, const [1,434]}
       foot_l_ik_ctrl.Location.X {
        type=MovieSceneScriptingFloatChannel
        default=0
        thinned_from=434
        keys [[0,2.2931],[2,2.2842],[3,2.2931],[4,2.2819],[5,2.2931],[6,2.2818],[7,2.2931],[9,2.2837],[10,2.2932],[84,2.2931],[85,2.3018],[86,2.2931],[87,2.301],[88,2.2932],[89,2.304],[90,2.2931],[92,2.3064],[93,2.2931],[94,2.296],[139,2.294],[140,2.3054],[141,2.2931],[149,2.2931],[150,2.4492],[151,2.7097],[152,2.8599],[153,2.7392],[154,2.4746],[155,2.2165],[156,2.1279],[157,2.2932],[197,2.2931],[198,2.2891],[199,2.2931],[201,2.2887],[202,2.2937],[208,2.2907],[235,2.2931],[236,2.2971],[237,2.2931],[239,2.2969],[240,2.2931],[248,2.2952],[249,2.2917],[274,2.2931],[275,2.2887],[276,2.2931],[279,2.2957],[312,2.2931],[314,2.2964],[316,2.2884],[317,2.2931],[336,2.2931],[337,2.2963],[354,2.2931],[355,2.3706],[356,2.7965],[357,3.1717],[358,2.6405],[359,2.2467],[360,2.0162],[361,2.0229],[362,2.2353],[363,2.3091],[364,2.3011],[401,2.2932],[402,2.2794],[403,2.2932],[405,2.2932],[406,2.2977],[407,2.2931],[408,2.3005],[409,2.2931],[410,2.3042],[411,2.2932],[412,2.3054],[413,2.2931],[414,2.3042],[415,2.2932],[416,2.3018],[417,2.2931],[418,2.3015],[419,2.2931],[420,2.3006],[421,2.2931],[422,2.2993],[423,2.2931],[424,2.2974],[425,2.2932],[427,2.2985],[428,2.2932],[433,2.2969]]
       }
       foot_l_ik_ctrl.Location.Y {
        type=MovieSceneScriptingFloatChannel
        default=0
        thinned_from=434
        keys [[0,-2.4632],[5,-2.4632],[6,-2.4592],[7,-2.4632],[9,-2.4473],[10,-2.4633],[84,-2.4632],[85,-2.4496],[86,-2.4632],[87,-2.4568],[88,-2.4633],[93,-2.4632],[94,-2.4675],[138,-2.4633],[139,-2.459],[140,-2.4045],[141,-2.4632],[149,-2.4632],[150,-1.7152],[151,-0.4554],[152,-0.0806],[153,-0.5627],[154,-1.2376],[155,-1.7838],[156,-2.1582],[157,-2.4631],[197,-2.4632],[198,-2.4696],[199,-2.4632],[200,-2.4475],[201,-2.4703],[202,-2.452],[203,-2.4453],[204,-2.4655],[206,-2.4619],[208,-2.4493],[209,-2.4632],[235,-2.4632],[236,-2.457],[237,-2.4633],[239,-2.4572],[240,-2.4632],[247,-2.4562],[249,-2.4655],[274,-2.4633],[275,-2.4701],[276,-2.4632],[278,-2.4633],[279,-2.4591],[312,-2.4632],[314,-2.4581],[316,-2.4705],[317,-2.4632],[336,-2.4632],[337,-2.4582],[338,-2.4633],[354,-2.4633],[355,-2.3616],[356,-1.7441],[357,-0.9703],[358,-1.2063],[359,-1.5064],[360,-1.7936],[361,-1.9802],[362,-2.2033],[363,-2.3873],[364,-2.4249],[367,-2.4531],[401,-2.4633],[402,-2.4492],[403,-2.4633],[404,-2.4566],[405,-2.4633],[409,-2.4632],[410,-2.4696],[411,-2.4633],[412,-2.4725],[413,-2.4632],[414,-2.4737],[415,-2.4632],[416,-2.4736],[417,-2.4632],[418,-2.4757],[419,-2.4632],[420,-2.4762],[421,-2.4632],[422,-2.475],[423,-2.4632],[424,-2.4724],[425,-2.4633],[427,-2.4773],[428,-2.4634],[432,-2.4773],[433,-2.4773]]
       }
       foot_l_ik_ctrl.Location.Z {
        type=MovieSceneScriptingFloatChannel
        default=0
        thinned_from=434
        keys [[0,0],[1,-0.0004],[2,-0.0016],[3,0.0001],[4,-0.002],[5,0.0002],[6,-0.0022],[7,-0.0001],[8,-0.0017],[9,-0.0029],[10,0],[36,0],[37,0.0001],[84,0],[85,0.0005],[86,0],[87,0.0008],[88,0],[89,0.0015],[90,0],[91,0.0012],[92,0.002],[93,0],[94,0.0007],[95,0],[138,0],[139,0.0095],[140,0.1294],[141,0],[142,0],[143,0.0002],[149,0.0001],[150,1.5104],[151,3.5713],[152,4.123],[153,4.4606],[154,4.5171],[155,4.298],[156,3.0361],[157,0.0007],[158,0.0004],[167,0.0006],[168,0.0004],[170,0],[197,0],[198,0.002],[199,0],[200,0.0034],[202,0.0012],[203,0.0025],[204,0.0007],[205,0.0012],[207,0.0012],[208,0.0001],[235,0],[236,-0.0026],[237,0],[238,-0.0014],[239,-0.0025],[240,0],[242,-0.0012],[244,-0.0018],[247,-0.0019],[248,-0.0013],[249,0.0009],[250,0],[274,0],[275,0.0017],[276,0],[277,0.0003],[278,0],[279,-0.001],[280,0],[312,0],[313,-0.0017],[314,-0.0018],[315,-0.0003],[316,0.0025],[317,0],[336,0],[337,-0.0016],[338,0],[354,0],[355,0.2729],[356,1.7136],[357,3.1987],[358,3.7851],[359,3.8279],[360,3.5287],[361,2.7164],[362,1.0398],[363,0.1668],[364,0.0848],[365,0.058],[366,0.0418],[367,0.0225],[368,0.0137],[369,0.0106],[370,0.0097],[372,0.0087],[373,0.0079],[374,0.0067],[375,0.0052],[376,0.0032],[377,0.002],[378,0.0013],[381,0.0007],[382,0],[401,0],[402,-0.0034],[403,0],[404,-0.0009],[405,0],[406,0.0006],[407,0],[408,0.0014],[409,0],[410,0.0024],[411,0],[412,0.0028],[413,0],[414,0.0027],[415,0],[416,0.0023],[417,0],[418,0.0024],[419,0],[420,0.0023],[421,0],[422,0.002],[423,0],[424,0.0015],[425,-0.0002],[426,0.0011],[427,0.0019],[428,-0.0002],[429,0.0006],[431,0.0015],[433,0.0017]]
       }
       foot_l_ik_ctrl.Rotation.X {
        type=MovieSceneScriptingFloatChannel
        default=0
        thinned_from=434
        keys [[0,0],[3,0],[4,-0.0004],[5,0.0006],[6,-0.0019],[7,0.0025],[8,-0.0041],[9,-0.0101],[10,0],[19,0],[36,0],[37,0.0109],[38,0],[84,0],[85,-0.0101],[86,0],[87,-0.0052],[88,0],[89,-0.0025],[90,0],[91,-0.0007],[93,0],[94,0.0038],[95,0],[111,0.0001],[113,-0.0001],[120,0.0001],[123,0],[129,0.0001],[131,-0.0001],[132,0],[138,0],[139,0.026],[140,0.3541],[141,0],[142,0],[143,0.0008],[144,0.0003],[145,0.0003],[146,-0.011],[147,0.0002],[149,0.0002],[150,4.2021],[151,10.235],[152,11.9248],[153,12.8118],[154,12.8977],[155,12.3281],[156,9.01],[157,0.0022],[158,0.0013],[165,0.0013],[166,0.0017],[167,0.0017],[168,0.0014],[169,0.0005],[170,0.0001],[176,0.0001],[177,-0.0117],[178,0],[197,0],[198,0.0061],[199,-0.0079],[200,-0.0177],[201,0.0116],[202,-0.0092],[203,-0.0134],[204,0.0007],[205,-0.0015],[206,-0.0015],[207,-0.0047],[208,-0.0055],[209,0],[235,0],[236,-0.0105],[237,0],[238,-0.0058],[239,-0.0102],[240,0.01],[241,-0.0027],[242,0.0013],[243,0.0044],[244,-0.0073],[245,-0.0059],[247,-0.0059],[248,-0.0033],[249,0.0011],[250,0],[274,0],[275,0.0112],[276,-0.006],[277,-0.0066],[278,0.0002],[279,0.0016],[280,0],[312,0],[313,-0.0083],[314,-0.0086],[315,-0.0015],[316,0.012],[317,0],[336,0],[337,-0.0176],[338,-0.0015],[354,-0.0015],[355,0.748],[356,4.7801],[357,9.1207],[358,10.9925],[359,11.3582],[360,10.8019],[361,8.5804],[362,3.2818],[363,0.4565],[364,0.232],[365,0.1586],[366,0.1144],[367,0.0614],[368,0.0374],[369,0.0288],[370,0.0266],[371,0.0252],[372,0.0241],[373,0.0216],[374,0.0184],[375,0.0144],[376,0.0088],[377,0.0054],[378,0.0037],[379,0.0028],[381,0.002],[382,-0.0003],[393,-0.0001],[401,0.0001],[402,-0.009],[403,0.0001],[404,-0.0056],[405,0.0001],[406,-0.0028],[407,0],[408,-0.0003],[409,0],[410,0.0027],[411,0.0001],[412,0.0047],[413,0],[414,0.0058],[415,0],[416,0.0062],[417,0],[418,0.0078],[419,0],[420,0.0083],[421,0],[422,0.0077],[423,0],[424,0.006],[425,0],[426,0.0048],[427,0.0091],[428,-0.0001],[429,0.0033],[430,0.006],[431,0.008],[432,0.0092],[433,0.0092]]
       }
       foot_l_ik_ctrl.Rotation.Y {
        type=MovieSceneScriptingFloatChannel
        default=0
        thinned_from=434
        keys [[0,0],[1,0],[2,-0.0063],[3,0],[4,-0.0079],[5,0.0084],[6,-0.0082],[7,0.0069],[8,-0.0052],[9,-0.0078],[10,0],[36,0],[37,0.0063],[38,0],[84,0],[85,0.005],[86,0],[87,0.0051],[88,0],[89,0.0074],[90,0],[91,0.0057],[92,0.0092],[93,0],[94,0.002],[95,0],[138,0],[139,-0.0023],[140,-0.0322],[141,0],[145,0],[146,-0.0094],[147,0],[149,0],[150,-0.3831],[151,-0.9316],[152,-1.0841],[153,-0.6948],[154,-0.2517],[155,0.0494],[156,0.105],[157,-0.0001],[176,0],[177,-0.0035],[178,0],[199,0.0002],[200,-0.0071],[201,0],[202,-0.0042],[203,-0.0073],[204,0.0001],[205,-0.0016],[206,-0.0016],[207,-0.004],[208,-0.0045],[209,0],[239,0],[240,-0.0014],[241,0],[242,-0.0002],[244,0],[245,0.0005],[247,0.0004],[248,0.0009],[249,0.0008],[250,0],[275,0],[276,-0.0023],[277,-0.0058],[278,0],[279,-0.0111],[280,0],[336,0],[337,-0.0009],[354,-0.001],[355,-0.0681],[356,-0.4358],[357,-0.7824],[358,-0.2756],[359,0.0796],[360,0.2879],[361,0.2369],[362,-0.0097],[363,-0.0416],[364,-0.0211],[365,-0.0145],[366,-0.0104],[367,-0.0055],[368,-0.0033],[369,-0.0025],[371,-0.002],[372,-0.0022],[374,-0.0017],[376,-0.0008],[379,-0.0002],[381,0],[401,0],[402,-0.0104],[403,0],[404,-0.0014],[405,0],[406,0.0035],[407,0],[408,0.0056],[409,0],[410,0.0085],[411,0],[412,0.0095],[413,0],[414,0.0087],[415,0],[416,0.0069],[417,0],[418,0.0068],[419,0],[420,0.0062],[421,0],[422,0.0051],[423,0],[424,0.0037],[425,-0.001],[426,0.0026],[427,0.0041],[428,-0.0011],[429,0.0014],[430,0.0022],[431,0.0029],[432,0.0032],[433,0.0032]]
       }
       foot_l_ik_ctrl.Rotation.Z {
        type=MovieSceneScriptingFloatChannel
        default=0
        thinned_from=434
        keys [[0,-6.0414],[93,-6.0415],[94,-6.0347],[149,-6.0415],[150,-6.0354],[151,-6.0759],[152,-7.7748],[153,-12.1315],[154,-17.0973],[155,-22.7587],[156,-27.9527],[157,-31.5497],[355,-31.5465],[356,-29.2503],[358,-21.4389],[359,-17.0802],[360,-13.1621],[361,-9.2806],[362,-6.4],[363,-6.0395],[403,-6.0415],[404,-6.0482],[405,-6.0415],[406,-6.0486],[407,-6.0414],[433,-6.0422]]
       }
       foot_l_ik_ctrl.Scale.X {type=MovieSceneScriptingFloatChannel, default=1, const [1,434]}
       foot_l_ik_ctrl.Scale.Y {type=MovieSceneScriptingFloatChannel, default=1, const [1,434]}
       foot_l_ik_ctrl.Scale.Z {type=MovieSceneScriptingFloatChannel, default=1, const [1,434]}
       ball_l_ik_ctrl.Location.X {type=MovieSceneScriptingFloatChannel, default=0, thinned_from=434, keys [[0,0],[338,0],[433,0]]}
       ball_l_ik_ctrl.Location.Y {
        type=MovieSceneScriptingFloatChannel
        default=0
        thinned_from=434
        keys [[0,0],[18,-0.0001],[23,0],[30,-0.0001],[31,0.0001],[38,-0.0001],[39,0.0001],[40,-0.0001],[46,0.0001],[54,-0.0001],[55,0.0001],[57,-0.0001],[68,0],[76,-0.0001],[77,0.0001],[95,0],[98,0.0001],[108,0],[117,0.0001],[134,0],[141,0.0001],[164,0],[165,-0.0001],[189,0],[190,0.0013],[191,0],[192,0.0011],[193,0],[194,0.0009],[195,0],[196,0.0007],[197,0],[338,0],[354,-0.0002],[388,0],[393,-0.0001],[433,0]]
       }
       ball_l_ik_ctrl.Location.Z {
        type=MovieSceneScriptingFloatChannel
        default=0
        thinned_from=434
        keys [[0,0],[189,0],[190,0.0028],[191,0],[192,0.0023],[193,0],[194,0.002],[195,0],[196,0.0015],[197,0],[338,0],[339,-0.0004],[354,-0.0004],[355,0],[433,0]]
       }
       ball_l_ik_ctrl.Rotation.X {
        type=MovieSceneScriptingFloatChannel
        default=0
        thinned_from=434
        keys [[0,0],[54,0],[55,-0.0001],[108,0],[152,0],[153,-2.6426],[154,-5.24],[155,-6.9328],[156,-6.2096],[157,0],[164,0],[165,0.0003],[166,0],[189,0],[190,-0.0115],[191,0],[192,-0.0094],[193,0],[194,-0.0082],[195,0],[196,-0.0063],[197,0],[338,0],[339,0.0015],[354,0.0015],[355,0],[356,0],[357,-0.2692],[358,-3.8037],[359,-5.9116],[360,-7.0539],[361,-6.1323],[362,-2.0162],[363,0],[388,0],[393,0.0001],[433,0]]
       }
       ball_l_ik_ctrl.Rotation.Y {
        type=MovieSceneScriptingFloatChannel
        default=0
        thinned_from=434
        keys [[0,0],[152,0],[153,-0.3533],[154,-0.6397],[155,-0.7778],[156,-0.6239],[157,0],[189,0],[190,0.0009],[191,0],[192,0.0006],[193,0],[194,0.0003],[195,0],[196,0.0002],[338,0],[339,0.001],[354,0.001],[355,0],[356,0],[357,-0.0381],[358,-0.4911],[359,-0.6835],[360,-0.731],[361,-0.5834],[362,-0.1814],[363,0],[433,0]]
       }
       ball_l_ik_ctrl.Rotation.Z {
        type=MovieSceneScriptingFloatChannel
        default=0
        thinned_from=434
        keys [[0,0],[23,0.0001],[30,0],[31,0.0001],[40,-0.0001],[54,0.0001],[152,0],[153,1.0218],[154,1.9209],[155,2.399],[156,1.9811],[157,0],[189,0],[190,0.0006],[191,0],[192,0.0004],[193,0],[194,0.0004],[195,0],[196,0.0003],[197,0],[356,0],[357,0.1066],[358,1.4557],[359,2.124],[360,2.3674],[361,1.9276],[362,0.5835],[363,0],[388,0],[393,-0.0001],[433,0]]
       }
       ball_l_ik_ctrl.Scale.X {type=MovieSceneScriptingFloatChannel, default=1, const [1,434]}
       ball_l_ik_ctrl.Scale.Y {type=MovieSceneScriptingFloatChannel, default=1, const [1,434]}
       ball_l_ik_ctrl.Scale.Z {type=MovieSceneScriptingFloatChannel, default=1, const [1,434]}
       foot_bk1_l_ctrl.Location.X {type=MovieSceneScriptingFloatChannel, default=0, const [0,434]}
       foot_bk1_l_ctrl.Location.Y {type=MovieSceneScriptingFloatChannel, default=0, const [0,434]}
       foot_bk1_l_ctrl.Location.Z {type=MovieSceneScriptingFloatChannel, default=0, const [0,434]}
       foot_bk1_l_ctrl.Rotation.X {type=MovieSceneScriptingFloatChannel, default=0, const [0,434]}
       foot_bk1_l_ctrl.Rotation.Y {type=MovieSceneScriptingFloatChannel, default=0, const [0,434]}
       foot_bk1_l_ctrl.Rotation.Z {type=MovieSceneScriptingFloatChannel, default=0, const [0,434]}
       foot_bk1_l_ctrl.Scale.X {type=MovieSceneScriptingFloatChannel, default=1, const [1,434]}
       foot_bk1_l_ctrl.Scale.Y {type=MovieSceneScriptingFloatChannel, default=1, const [1,434]}
       foot_bk1_l_ctrl.Scale.Z {type=MovieSceneScriptingFloatChannel, default=1, const [1,434]}
       foot_roll_l_ctrl.Location.X {type=MovieSceneScriptingFloatChannel, default=0, const [0,434]}
       foot_roll_l_ctrl.Location.Y {type=MovieSceneScriptingFloatChannel, default=0, const [0,434]}
       foot_roll_l_ctrl.Location.Z {type=MovieSceneScriptingFloatChannel, default=0, const [0,434]}
       foot_roll_l_ctrl.Rotation.X {type=MovieSceneScriptingFloatChannel, default=0, const [0,434]}
       foot_roll_l_ctrl.Rotation.Y {type=MovieSceneScriptingFloatChannel, default=0, const [0,434]}
       foot_roll_l_ctrl.Rotation.Z {type=MovieSceneScriptingFloatChannel, default=0, const [0,434]}
       foot_roll_l_ctrl.Scale.X {type=MovieSceneScriptingFloatChannel, default=1, const [1,434]}
       foot_roll_l_ctrl.Scale.Y {type=MovieSceneScriptingFloatChannel, default=1, const [1,434]}
       foot_roll_l_ctrl.Scale.Z {type=MovieSceneScriptingFloatChannel, default=1, const [1,434]}
       roll_blend_l_ctrl {type=MovieSceneScriptingFloatChannel, default=1, const [1,434]}
       tip_l_ctrl.Location.X {type=MovieSceneScriptingFloatChannel, default=0, const [0,434]}
       tip_l_ctrl.Location.Y {type=MovieSceneScriptingFloatChannel, default=0, const [0,434]}
       tip_l_ctrl.Location.Z {type=MovieSceneScriptingFloatChannel, default=0, const [0,434]}
       tip_l_ctrl.Rotation.X {type=MovieSceneScriptingFloatChannel, default=0, const [0,434]}
       tip_l_ctrl.Rotation.Y {type=MovieSceneScriptingFloatChannel, default=0, const [0,434]}
       tip_l_ctrl.Rotation.Z {type=MovieSceneScriptingFloatChannel, default=0, const [0,434]}
       tip_l_ctrl.Scale.X {type=MovieSceneScriptingFloatChannel, default=1, const [1,434]}
       tip_l_ctrl.Scale.Y {type=MovieSceneScriptingFloatChannel, default=1, const [1,434]}
       tip_l_ctrl.Scale.Z {type=MovieSceneScriptingFloatChannel, default=1, const [1,434]}
       heel_l_ctrl.Location.X {type=MovieSceneScriptingFloatChannel, default=0, const [0,434]}
       heel_l_ctrl.Location.Y {type=MovieSceneScriptingFloatChannel, default=0, const [0,434]}
       heel_l_ctrl.Location.Z {type=MovieSceneScriptingFloatChannel, default=0, const [0,434]}
       heel_l_ctrl.Rotation.X {type=MovieSceneScriptingFloatChannel, default=0, const [0,434]}
       heel_l_ctrl.Rotation.Y {type=MovieSceneScriptingFloatChannel, default=0, const [0,434]}
       heel_l_ctrl.Rotation.Z {type=MovieSceneScriptingFloatChannel, default=0, const [0,434]}
       heel_l_ctrl.Scale.X {type=MovieSceneScriptingFloatChannel, default=1, const [1,434]}
       heel_l_ctrl.Scale.Y {type=MovieSceneScriptingFloatChannel, default=1, const [1,434]}
       heel_l_ctrl.Scale.Z {type=MovieSceneScriptingFloatChannel, default=1, const [1,434]}
       leg_l_softness {type=MovieSceneScriptingFloatChannel, default=0, const [0,434]}
       leg_l_pv_ik_ctrl.Location.X {
        type=MovieSceneScriptingFloatChannel
        default=0
        thinned_from=434
        keys [[0,0.3195],[5,0.3203],[6,0.3212],[7,0.3212],[12,0.3255],[13,0.3256],[14,0.3248],[16,0.3243],[17,0.3186],[18,0.3048],[19,0.2707],[20,0.2246],[21,0.1899],[22,0.1738],[23,0.1658],[24,0.1604],[25,0.1555],[26,0.1526],[29,0.1472],[30,0.1459],[31,0.1449],[32,0.1445],[34,0.1447],[35,0.1454],[36,0.1463],[37,0.1463],[38,0.1491],[40,0.1524],[43,0.156],[46,0.1583],[58,0.1638],[59,0.1646],[62,0.1683],[64,0.1696],[65,0.1715],[66,0.1719],[67,0.1669],[68,0.1612],[69,0.1579],[70,0.1555],[71,0.1543],[72,0.1543],[73,0.1551],[76,0.1562],[77,0.1571],[78,0.1584],[81,0.1634],[85,0.1689],[88,0.1714],[93,0.1735],[96,0.1741],[97,0.1748],[98,0.176],[99,0.1777],[100,0.1805],[101,0.1847],[102,0.1904],[103,0.1975],[104,0.2069],[106,0.2304],[107,0.2408],[108,0.2494],[109,0.257],[110,0.2635],[111,0.2688],[112,0.2724],[113,0.2745],[117,0.2778],[125,0.2804],[131,0.2795],[133,0.2777],[134,0.2753],[135,0.2697],[136,0.2602],[137,0.2458],[138,0.2259],[139,0.1437],[140,-0.0747],[141,-0.5049],[142,-0.907],[143,-1.294],[144,-1.5777],[145,-1.6663],[146,-1.684],[147,-1.6894],[148,-1.664],[149,-1.5833],[150,-1.1505],[151,-1.0475],[152,-0.8919],[153,-1.0119],[154,-1.2188],[155,-1.3315],[156,-1.1569],[157,-1.9311],[158,-2.594],[159,-2.8682],[160,-2.9865],[161,-2.9877],[162,-2.8939],[163,-2.4895],[164,-1.9077],[165,-1.64],[166,-1.517],[167,-1.4754],[168,-1.5262],[169,-1.8108],[170,-2.2136],[171,-2.4446],[172,-2.5546],[173,-2.595],[174,-2.6184],[175,-2.6288],[177,-2.6297],[178,-2.6196],[179,-2.5972],[181,-2.5417],[182,-2.522],[184,-2.5011],[215,-2.4806],[220,-2.487],[225,-2.4621],[229,-2.4405],[256,-2.4179],[260,-2.4381],[261,-2.4502],[262,-2.4683],[265,-2.5428],[267,-2.5762],[272,-2.6122],[275,-2.6187],[279,-2.6158],[282,-2.6004],[287,-2.5514],[288,-2.549],[290,-2.5651],[292,-2.5923],[297,-2.6935],[299,-2.72],[321,-2.843],[330,-2.8277],[335,-2.8026],[337,-2.802],[340,-2.8187],[343,-2.8569],[344,-2.8755],[345,-2.9228],[346,-3.0109],[348,-3.2312],[349,-3.327],[350,-3.3906],[351,-3.4276],[352,-3.4402],[353,-3.402],[354,-3.2766],[355,-2.3754],[356,-1.0906],[357,-0.7422],[358,-0.9595],[359,-1.0008],[360,-0.9298],[361,-0.6066],[362,0.084],[363,0.3071],[364,0.2533],[365,0.2796],[366,0.2916],[367,0.3128],[368,0.3221],[369,0.3236],[370,0.3209],[373,0.3078],[374,0.3053],[375,0.3049],[376,0.3064],[377,0.305],[378,0.302],[380,0.2924],[381,0.2888],[382,0.2989],[383,0.2973],[384,0.2969],[385,0.2977],[387,0.3016],[390,0.3087],[397,0.318],[400,0.3203],[433,0.3196]]
       }
       leg_l_pv_ik_ctrl.Location.Y {
        type=MovieSceneScriptingFloatChannel
        default=0
        thinned_from=434
        keys [[0,-1.9533],[138,-1.951],[139,-1.256],[140,0.5691],[141,4.2705],[142,6.9657],[143,9.1901],[144,10.5653],[145,10.906],[146,10.9462],[147,10.9176],[148,10.7462],[149,10.264],[150,5.7766],[151,2.0345],[152,-0.5855],[153,-1.9404],[155,-1.9439],[156,-1.9417],[157,8.8688],[158,13.2688],[159,14.6001],[160,15.1081],[161,15.1817],[162,14.9176],[163,13.4324],[164,10.7104],[165,9.1878],[166,8.4253],[167,8.1685],[168,8.5407],[169,10.3451],[170,12.3569],[171,13.3154],[172,13.747],[173,13.9126],[174,14.0075],[175,14.0504],[177,14.0576],[178,14.0197],[181,13.7159],[182,13.6368],[185,13.5234],[215,13.4767],[221,13.603],[227,13.8174],[254,13.8497],[257,13.7917],[258,13.7442],[259,13.6621],[260,13.496],[262,13.0304],[263,12.87],[264,12.7436],[265,12.6535],[268,12.5006],[274,12.4021],[282,12.4173],[286,12.5224],[287,12.5652],[288,12.6405],[292,13.1728],[294,13.4616],[296,13.6744],[298,13.8284],[306,14.0734],[323,14.0695],[335,13.8138],[339,13.8357],[343,13.9815],[344,14.0384],[345,14.1914],[346,14.4718],[348,15.1055],[349,15.3427],[350,15.4715],[351,15.5145],[352,15.4786],[353,15.2962],[354,14.8694],[355,11.2989],[356,2.3163],[357,-1.9395],[361,-1.9397],[362,-1.9479],[363,-0.7659],[364,0.0951],[365,-0.257],[366,-0.514],[367,-0.8985],[368,-1.131],[369,-1.2319],[370,-1.2615],[372,-1.2948],[373,-1.3237],[374,-1.3747],[375,-1.4443],[376,-1.5511],[377,-1.635],[378,-1.6922],[379,-1.7286],[381,-1.7687],[382,-1.9448],[433,-1.9534]]
       }
       leg_l_pv_ik_ctrl.Location.Z {
        type=MovieSceneScriptingFloatChannel
        default=0
        thinned_from=434
        keys [[0,-1.7606],[18,-1.7622],[22,-1.7744],[137,-1.7676],[138,-1.7695],[139,-1.1579],[140,0.5469],[141,4.5169],[142,7.9406],[143,11.2251],[144,13.5283],[145,14.1372],[146,14.2093],[147,14.156],[148,13.845],[149,12.9945],[150,6.3182],[151,1.9465],[152,-0.6471],[153,-1.8846],[154,-1.9037],[155,-1.9141],[156,-1.898],[157,10.6482],[158,18.9102],[159,22.2628],[160,23.7286],[161,23.9537],[162,23.1688],[163,19.3046],[164,13.754],[165,11.184],[166,10.0012],[167,9.6168],[168,10.1771],[169,13.1124],[170,16.9318],[171,19.0337],[172,20.0594],[173,20.4687],[174,20.7073],[175,20.816],[177,20.8345],[178,20.7384],[179,20.5223],[181,19.984],[182,19.7925],[184,19.5867],[187,19.4443],[212,19.3474],[219,19.5794],[227,20.2454],[251,20.4001],[256,20.2502],[257,20.1843],[258,20.066],[259,19.8638],[260,19.4615],[261,18.8916],[262,18.3756],[263,18.0137],[264,17.7324],[265,17.5336],[267,17.2843],[270,17.0847],[274,16.9888],[282,17.0224],[285,17.1713],[287,17.3421],[288,17.5049],[289,17.7619],[292,18.6883],[294,19.3598],[295,19.6356],[297,20.0752],[298,20.2461],[302,20.6371],[309,20.9539],[321,20.9071],[335,20.2009],[337,20.1878],[340,20.315],[342,20.4959],[344,20.7615],[345,21.1537],[346,21.8956],[348,23.6979],[349,24.4256],[350,24.8344],[351,24.9719],[352,24.8534],[353,24.2727],[354,22.9931],[355,14.801],[356,2.2434],[357,-1.8596],[358,-1.8797],[359,-1.8836],[360,-1.877],[361,-1.8471],[362,-1.7828],[363,-0.6886],[364,0.1215],[365,-0.2132],[366,-0.4549],[367,-0.8108],[368,-1.0232],[369,-1.115],[370,-1.1423],[372,-1.1735],[373,-1.2002],[374,-1.2466],[375,-1.3094],[376,-1.4052],[377,-1.4803],[378,-1.5317],[379,-1.5646],[381,-1.6011],[382,-1.7556],[433,-1.7607]]
       }
       leg_l_pv_ik_ctrl.Rotation.X {
        type=MovieSceneScriptingFloatChannel
        default=0
        thinned_from=434
        keys [[0,2.5551],[18,2.5573],[137,2.5648],[138,2.5675],[139,1.6657],[140,-0.7693],[141,-6.0713],[142,-10.3278],[143,-14.2046],[144,-16.8263],[145,-17.508],[146,-17.5885],[147,-17.5291],[148,-17.1817],[149,-16.2254],[150,-8.3442],[151,-2.697],[152,0.9172],[153,2.7292],[154,2.756],[155,2.7705],[156,2.748],[157,-13.5398],[158,-22.7116],[159,-26.2418],[160,-27.7596],[161,-27.9914],[162,-27.1814],[163,-23.131],[164,-17.081],[165,-14.1589],[166,-12.7825],[167,-12.3303],[168,-12.9885],[169,-16.3595],[170,-20.5828],[171,-22.8425],[172,-23.9311],[173,-24.363],[174,-24.6142],[175,-24.7285],[177,-24.748],[178,-24.6469],[179,-24.4195],[181,-23.8514],[182,-23.6487],[184,-23.4306],[211,-23.1621],[219,-23.4227],[227,-24.127],[251,-24.29],[256,-24.1319],[257,-24.0624],[258,-23.9375],[259,-23.7237],[260,-23.2974],[261,-22.6912],[262,-22.1398],[263,-21.7517],[264,-21.4493],[265,-21.2352],[267,-20.9662],[270,-20.7503],[279,-20.6382],[283,-20.7208],[286,-20.9286],[287,-21.0285],[288,-21.2043],[289,-21.4813],[292,-22.4748],[294,-23.1902],[296,-23.7303],[298,-24.1289],[302,-24.541],[310,-24.8936],[322,-24.7951],[335,-24.0816],[338,-24.0954],[341,-24.2865],[344,-24.6723],[345,-25.0841],[346,-25.86],[348,-27.729],[349,-28.4778],[350,-28.8969],[351,-29.0378],[352,-28.9167],[353,-28.3211],[354,-27.0012],[355,-18.2489],[356,-3.0968],[357,2.6941],[358,2.7223],[359,2.7277],[360,2.7185],[361,2.6765],[362,2.5863],[363,0.9858],[364,-0.17],[365,0.3047],[366,0.6499],[367,1.1624],[368,1.4706],[369,1.6044],[370,1.6442],[372,1.6897],[373,1.7287],[374,1.7966],[375,1.8886],[376,2.0294],[377,2.14],[378,2.2158],[379,2.2643],[381,2.3183],[382,2.5475],[433,2.5552]]
       }
       leg_l_pv_ik_ctrl.Rotation.Y {
        type=MovieSceneScriptingFloatChannel
        default=0
        thinned_from=434
        keys [[0,0.0101],[16,0.0104],[17,0.01],[18,0.0092],[19,0.0071],[20,0.0042],[21,0.002],[22,0.001],[25,-0.0001],[34,-0.0008],[66,0.0009],[68,0.0002],[71,-0.0002],[77,-0.0001],[99,0.0013],[101,0.0017],[103,0.0025],[107,0.0052],[110,0.0066],[113,0.0073],[133,0.0075],[134,0.0073],[135,0.007],[136,0.0064],[137,0.0055],[138,0.0042],[139,0.0013],[140,0.0005],[141,0.0368],[142,0.0972],[143,0.1772],[144,0.244],[145,0.2625],[146,0.2644],[147,0.2621],[148,0.2516],[149,0.2243],[150,0.0422],[151,-0.0392],[152,-0.0598],[153,-0.0753],[154,-0.089],[155,-0.0966],[156,-0.0849],[157,0.1259],[158,0.4158],[159,0.5604],[160,0.6281],[161,0.6395],[162,0.6041],[163,0.4385],[164,0.2363],[165,0.1584],[166,0.1264],[167,0.1166],[168,0.1314],[169,0.2173],[170,0.348],[171,0.4284],[172,0.47],[173,0.4871],[174,0.4972],[175,0.5019],[176,0.5031],[177,0.5027],[178,0.4987],[179,0.4896],[181,0.4671],[182,0.4592],[183,0.4545],[184,0.4508],[186,0.4463],[197,0.4402],[211,0.4405],[216,0.4451],[219,0.4511],[221,0.4574],[226,0.4803],[227,0.4836],[230,0.4889],[249,0.4929],[254,0.4888],[256,0.4852],[257,0.482],[258,0.4763],[259,0.4668],[260,0.448],[261,0.4218],[262,0.3981],[263,0.3811],[264,0.3676],[265,0.3578],[266,0.3509],[267,0.3455],[268,0.3414],[270,0.3358],[272,0.3325],[274,0.331],[279,0.3307],[281,0.332],[282,0.3332],[283,0.3352],[285,0.341],[287,0.3492],[288,0.3563],[289,0.3671],[291,0.3926],[292,0.406],[294,0.4342],[295,0.4458],[296,0.4556],[297,0.4643],[298,0.4715],[299,0.4769],[302,0.488],[306,0.4973],[310,0.5015],[316,0.5011],[322,0.4948],[334,0.4664],[336,0.4644],[337,0.4644],[339,0.467],[341,0.4728],[343,0.4819],[344,0.4878],[345,0.5042],[346,0.5358],[348,0.6153],[349,0.6482],[350,0.6665],[351,0.672],[352,0.6652],[353,0.6368],[354,0.5776],[355,0.2535],[356,-0.0372],[357,-0.0575],[358,-0.0718],[359,-0.0745],[360,-0.0698],[361,-0.0487],[362,-0.0046],[363,0.0135],[364,0.0159],[366,0.014],[367,0.0132],[381,0.0086],[382,0.0088],[387,0.0089],[433,0.0101]]
       }
       leg_l_pv_ik_ctrl.Rotation.Z {
        type=MovieSceneScriptingFloatChannel
        default=0
        thinned_from=434
        keys [[0,-0.4638],[5,-0.4649],[6,-0.4662],[7,-0.4662],[12,-0.4723],[13,-0.4724],[14,-0.4714],[16,-0.4706],[17,-0.4625],[18,-0.4429],[19,-0.3947],[20,-0.3293],[21,-0.2803],[22,-0.2574],[23,-0.2461],[24,-0.2384],[25,-0.2316],[26,-0.2274],[29,-0.2198],[30,-0.218],[31,-0.2166],[32,-0.2158],[34,-0.2162],[36,-0.2185],[37,-0.2185],[38,-0.2224],[40,-0.2272],[43,-0.2322],[46,-0.2355],[57,-0.2423],[59,-0.2444],[62,-0.2496],[64,-0.2515],[65,-0.2541],[66,-0.2547],[67,-0.2477],[68,-0.2396],[69,-0.235],[70,-0.2316],[71,-0.2298],[72,-0.2298],[73,-0.2309],[76,-0.2325],[77,-0.2338],[78,-0.2356],[81,-0.2426],[85,-0.2504],[88,-0.2541],[93,-0.2571],[96,-0.2579],[97,-0.2589],[98,-0.2605],[99,-0.263],[100,-0.2669],[101,-0.2729],[102,-0.281],[103,-0.291],[104,-0.3043],[106,-0.3376],[107,-0.3523],[108,-0.3646],[109,-0.3753],[110,-0.3846],[111,-0.3921],[112,-0.3971],[113,-0.4],[117,-0.4047],[125,-0.4085],[131,-0.4072],[133,-0.4046],[134,-0.4012],[135,-0.3933],[136,-0.3798],[137,-0.3594],[138,-0.3312],[139,-0.2086],[140,0.1049],[141,0.6781],[142,1.1669],[143,1.6053],[144,1.9106],[145,2.0041],[146,2.023],[147,2.0292],[148,2.0033],[149,1.9196],[150,1.4602],[151,1.3774],[152,1.2193],[153,1.4216],[154,1.7147],[155,1.8742],[156,1.6269],[157,2.3269],[158,2.9501],[159,3.1913],[160,3.2918],[161,3.2899],[162,3.2054],[163,2.8364],[164,2.2703],[165,1.9941],[166,1.8635],[167,1.8187],[168,1.8727],[169,2.1693],[170,2.5709],[171,2.7925],[172,2.8955],[173,2.9327],[174,2.954],[175,2.9634],[177,2.9642],[178,2.9549],[181,2.883],[182,2.8647],[185,2.8392],[220,2.8301],[225,2.7991],[228,2.7776],[256,2.7516],[259,2.7695],[260,2.7811],[261,2.7999],[263,2.8533],[265,2.913],[267,2.9515],[271,2.9877],[275,3.0003],[279,2.9974],[282,2.9805],[287,2.9244],[288,2.92],[291,2.9393],[293,2.9673],[297,3.0396],[321,3.1835],[331,3.1691],[335,3.1511],[339,3.1594],[342,3.1873],[344,3.219],[345,3.2625],[346,3.3427],[348,3.541],[349,3.6268],[350,3.6846],[351,3.7197],[352,3.7344],[353,3.7052],[354,3.5981],[355,2.7676],[356,1.4282],[357,1.0398],[358,1.3475],[359,1.4059],[360,1.3055],[361,0.8479],[362,-0.1302],[363,-0.4287],[364,-0.3413],[365,-0.3828],[366,-0.4033],[367,-0.4386],[368,-0.4553],[369,-0.4589],[370,-0.4556],[373,-0.4382],[374,-0.4356],[375,-0.436],[376,-0.4396],[377,-0.439],[378,-0.4355],[379,-0.4296],[380,-0.4227],[381,-0.4179],[382,-0.4346],[383,-0.4323],[384,-0.4318],[385,-0.4329],[387,-0.4385],[390,-0.4485],[397,-0.4617],[401,-0.4654],[433,-0.464]]
       }
       leg_l_pv_ik_ctrl.Scale.X {type=MovieSceneScriptingFloatChannel, default=1, const [1,434]}
       leg_l_pv_ik_ctrl.Scale.Y {type=MovieSceneScriptingFloatChannel, default=1, const [1,434]}
       leg_l_pv_ik_ctrl.Scale.Z {type=MovieSceneScriptingFloatChannel, default=1, const [1,434]}
       thigh_r_fk_ctrl.Location.X {type=MovieSceneScriptingFloatChannel, default=0, thinned_from=434, keys [[0,-0.0002],[433,-0.0002]]}
       thigh_r_fk_ctrl.Location.Y {type=MovieSceneScriptingFloatChannel, default=0, thinned_from=434, keys [[0,0],[433,0]]}
       thigh_r_fk_ctrl.Location.Z {type=MovieSceneScriptingFloatChannel, default=0, thinned_from=434, keys [[0,0],[433,0]]}
       thigh_r_fk_ctrl.Rotation.X {
        type=MovieSceneScriptingFloatChannel
        default=1.2139
        thinned_from=434
        keys [[0,24.5993],[11,24.6253],[12,24.6963],[13,24.8348],[16,25.4973],[17,26.1155],[18,27.4425],[19,30.5076],[20,34.7442],[21,38.1028],[22,39.7009],[23,40.4166],[25,41.2431],[26,41.4426],[30,41.7909],[33,41.812],[42,40.8343],[59,40.4569],[62,40.1683],[63,39.9756],[64,39.4359],[65,37.8265],[66,35.9869],[67,35.0412],[68,34.4789],[69,34.1028],[72,33.5166],[73,33.3867],[75,33.346],[95,33.54],[98,33.366],[100,33.065],[101,32.7901],[102,32.4252],[103,31.9723],[104,31.3747],[106,29.9144],[107,29.2795],[108,28.7591],[109,28.2981],[110,27.9082],[111,27.6011],[112,27.4007],[125,26.8616],[134,26.9861],[136,27.1371],[137,27.3263],[138,27.6645],[139,28.1756],[140,29.2342],[141,30.5985],[142,31.6667],[143,32.0629],[148,32.4905],[149,32.6494],[150,33.8915],[151,36.8998],[152,41.2894],[153,44.6482],[154,46.7391],[155,48.1896],[156,48.1963],[157,46.4149],[158,46.156],[159,45.7244],[160,45.4976],[161,45.4699],[162,45.7315],[163,46.8124],[164,48.033],[165,48.4525],[166,48.5802],[167,48.5633],[168,48.4096],[169,47.8934],[170,47.1172],[171,46.6574],[172,46.4462],[178,46.3681],[209,46.6731],[212,46.4499],[214,46.1042],[216,45.4958],[217,45.0605],[218,44.5289],[219,43.8945],[220,43.1508],[221,42.1588],[222,40.888],[224,38.1266],[226,35.1873],[227,34.0283],[228,33.2723],[229,32.7145],[230,32.3077],[231,32.0053],[233,31.5848],[240,30.9774],[254,30.774],[256,30.8206],[257,30.907],[258,31.1952],[259,31.8251],[260,33.4045],[261,35.8113],[262,38.0215],[263,39.6561],[264,41.0257],[265,42.0443],[266,42.6557],[267,42.9552],[271,43.4348],[286,43.6911],[287,43.7829],[288,44.0215],[292,45.9633],[294,47.0447],[296,47.8794],[298,48.4884],[306,49.4663],[320,49.7698],[325,49.434],[333,48.2379],[344,47.8622],[345,47.6924],[346,47.3697],[350,45.5246],[352,44.9198],[353,44.5332],[354,43.9533],[355,42.7749],[356,41.3734],[357,36.8134],[358,31.4501],[359,26.9571],[360,24.0636],[361,23.4216],[362,23.19],[363,23.1452],[364,23.3137],[368,24.4929],[369,24.7481],[376,25.7421],[380,26.0616],[382,26.0802],[384,25.9287],[386,25.6479],[389,25.0666],[391,24.8176],[399,24.512],[433,24.599]]
       }
       thigh_r_fk_ctrl.Rotation.Y {
        type=MovieSceneScriptingFloatChannel
        default=-1.213
        thinned_from=434
        keys [[0,-1.2802],[3,-1.2802],[4,-1.3001],[5,-1.3001],[6,-1.315],[7,-1.315],[8,-1.3246],[10,-1.3247],[11,-1.3274],[12,-1.3371],[13,-1.3573],[14,-1.392],[15,-1.4086],[16,-1.3856],[17,-1.3366],[18,-1.2839],[19,-1.4359],[20,-1.7953],[21,-2.0129],[22,-1.9143],[23,-1.6916],[24,-1.4803],[25,-1.2923],[26,-1.0895],[27,-0.9185],[28,-0.8025],[29,-0.7276],[30,-0.6654],[31,-0.6175],[32,-0.5852],[33,-0.5626],[34,-0.5433],[35,-0.5271],[36,-0.5147],[37,-0.5101],[38,-0.5172],[39,-0.5356],[40,-0.5649],[41,-0.6046],[42,-0.6531],[44,-0.7602],[45,-0.8105],[46,-0.8532],[49,-0.9676],[50,-1.0041],[51,-1.0381],[52,-1.0688],[53,-1.0953],[54,-1.1168],[57,-1.1678],[58,-1.1886],[59,-1.2149],[60,-1.249],[61,-1.2931],[62,-1.3655],[63,-1.3927],[64,-1.2409],[65,-0.5526],[66,0.1397],[67,0.2133],[68,0.022],[69,-0.1684],[70,-0.2977],[71,-0.3717],[72,-0.4126],[74,-0.4781],[75,-0.5125],[76,-0.5431],[77,-0.5677],[78,-0.5865],[79,-0.6002],[80,-0.6089],[81,-0.6128],[82,-0.6121],[83,-0.6083],[84,-0.6083],[85,-0.597],[86,-0.597],[87,-0.5866],[88,-0.5866],[89,-0.5723],[90,-0.5723],[91,-0.5594],[94,-0.5594],[95,-0.5493],[96,-0.549],[97,-0.5508],[98,-0.5557],[99,-0.5645],[100,-0.5789],[101,-0.6023],[102,-0.6371],[103,-0.6863],[104,-0.7562],[105,-0.844],[106,-0.9383],[107,-1.0267],[109,-1.1924],[110,-1.269],[111,-1.3339],[112,-1.3895],[113,-1.4374],[114,-1.4703],[115,-1.4813],[116,-1.4758],[117,-1.4652],[121,-1.4052],[125,-1.3364],[131,-1.2562],[133,-1.2224],[135,-1.183],[136,-1.1542],[137,-1.1111],[138,-0.7216],[139,0.029],[140,3.1373],[141,5.9687],[142,8.3911],[143,10.6775],[144,12.27],[145,12.7398],[146,12.8328],[147,12.839],[148,12.6803],[149,12.1963],[150,5.2809],[151,2.1196],[152,-0.6513],[153,-0.9428],[154,-1.3203],[155,-1.638],[156,-1.7126],[157,11.1282],[158,17.8787],[159,20.7184],[160,22.0294],[161,22.4146],[162,21.9189],[163,18.7028],[164,13.6698],[165,11.0488],[166,9.6465],[167,9.0225],[168,9.418],[169,12.2173],[170,15.6351],[171,17.5332],[172,18.5658],[173,19.0737],[174,19.3878],[175,19.5608],[176,19.6356],[177,19.6593],[178,19.6],[179,19.4297],[181,18.9782],[182,18.8127],[185,18.5317],[199,18.2184],[212,18.1667],[215,18.0455],[217,17.8714],[219,17.6077],[220,17.4394],[221,17.2129],[222,16.9137],[224,16.2227],[226,15.4206],[227,15.0865],[228,14.8686],[229,14.7098],[231,14.511],[234,14.3593],[245,14.1672],[248,14.1729],[250,14.2612],[252,14.4498],[253,14.5857],[254,14.7514],[256,15.1253],[257,15.3492],[258,15.8013],[259,16.6613],[260,18.5325],[261,21.0976],[262,23.1785],[263,24.3717],[264,25.1006],[265,25.467],[266,25.6223],[271,25.7121],[284,25.7959],[285,25.7497],[286,25.6439],[287,25.461],[288,24.9539],[289,24.0168],[291,21.6385],[292,20.3517],[294,17.4343],[295,16.1234],[296,14.9917],[297,13.9644],[298,13.0975],[299,12.4469],[300,11.9455],[301,11.4876],[302,11.072],[303,10.6972],[304,10.3616],[305,10.0638],[306,9.8023],[307,9.5743],[308,9.3769],[309,9.208],[310,9.0655],[311,8.9471],[312,8.8509],[313,8.7746],[315,8.6734],[317,8.6242],[319,8.6206],[320,8.6476],[321,8.7008],[322,8.7855],[323,8.907],[324,9.0705],[325,9.2814],[327,9.7952],[331,10.8952],[333,11.488],[334,11.7319],[335,11.8985],[336,11.9955],[340,12.1996],[341,12.273],[342,12.3823],[343,12.5407],[344,12.761],[345,13.2735],[346,14.153],[347,15.1858],[348,16.181],[349,16.9732],[350,17.4147],[351,17.409],[352,17.0441],[353,16.4591],[354,15.7937],[355,11.5804],[356,3.9778],[357,0.8659],[358,-1.5859],[359,-1.8884],[360,-2.2549],[361,-2.5883],[362,-2.6902],[363,-2.3851],[364,-2.1646],[365,-2.2443],[366,-2.4987],[367,-2.7073],[368,-2.7437],[369,-2.6461],[370,-2.4562],[371,-2.2123],[373,-1.6705],[374,-1.3645],[375,-1.0439],[376,-0.839],[377,-0.7343],[378,-0.6583],[379,-0.6044],[380,-0.5698],[381,-0.5564],[382,-0.6478],[383,-0.6496],[384,-0.6609],[385,-0.6818],[386,-0.7168],[387,-0.767],[388,-0.826],[389,-0.8878],[391,-1.0037],[393,-1.1247],[394,-1.1793],[395,-1.2241],[396,-1.2611],[397,-1.2942],[398,-1.3229],[399,-1.3458],[400,-1.3626],[401,-1.3742],[402,-1.382],[403,-1.382],[404,-1.3858],[405,-1.3858],[406,-1.3796],[407,-1.3796],[408,-1.3696],[409,-1.3696],[410,-1.3555],[411,-1.3555],[412,-1.3395],[413,-1.3395],[414,-1.3247],[415,-1.3247],[416,-1.3136],[417,-1.3136],[418,-1.3042],[419,-1.3042],[420,-1.2961],[421,-1.2961],[422,-1.2897],[423,-1.2897],[424,-1.2853],[426,-1.2853],[427,-1.2818],[433,-1.2803]]
       }
       thigh_r_fk_ctrl.Rotation.Z {
        type=MovieSceneScriptingFloatChannel
        default=-0.1592
        thinned_from=434
        keys [[0,4.9054],[5,4.9114],[9,4.9294],[11,4.9602],[12,4.9859],[13,5.0256],[16,5.1964],[17,5.3195],[18,5.5761],[19,6.1493],[20,6.935],[21,7.5631],[22,7.8707],[23,8.0123],[25,8.1558],[26,8.1886],[31,8.2489],[37,8.194],[41,8.1265],[58,8.133],[62,8.0723],[63,8.0128],[64,7.8485],[65,7.3813],[66,6.8435],[67,6.5213],[68,6.3088],[69,6.177],[70,6.0933],[71,6.0358],[72,5.9926],[73,5.9674],[74,5.96],[78,6.0228],[83,6.1224],[84,6.1224],[85,6.1451],[86,6.1451],[87,6.1585],[94,6.1657],[104,6.0178],[106,6.0008],[108,5.9542],[112,5.8106],[120,5.5797],[124,5.5088],[126,5.497],[128,5.5147],[132,5.5808],[133,5.5872],[134,5.5476],[135,5.4013],[136,5.1113],[137,4.6424],[138,3.3417],[139,1.1948],[140,-5.6419],[141,-12.0375],[142,-17.5765],[143,-23.2529],[144,-27.8475],[145,-29.8849],[146,-30.8042],[147,-31.2978],[148,-31.3386],[149,-30.7075],[150,-19.9312],[151,-12.5586],[152,-7.7608],[153,-6.9151],[154,-6.2214],[155,-5.7211],[156,-5.4677],[157,-38.6619],[158,-65.4205],[159,-76.0449],[160,-81.404],[161,-84.1055],[162,-85.102],[163,-82.7478],[164,-78.4227],[165,-75.6879],[166,-73.302],[167,-71.084],[168,-69.4997],[169,-69.9889],[170,-71.8668],[172,-76.193],[173,-77.8741],[174,-79.0124],[175,-79.7899],[176,-80.2512],[178,-80.6407],[184,-79.7657],[192,-78.4569],[213,-78.3238],[217,-78.9106],[220,-79.8046],[226,-83.3659],[227,-83.8522],[230,-84.5184],[257,-85.0857],[258,-84.952],[259,-84.6282],[260,-83.7805],[262,-81.276],[263,-80.4355],[264,-79.7684],[265,-79.3056],[267,-78.9429],[291,-78.9035],[305,-77.0247],[326,-77.0588],[346,-78.2442],[347,-78.1748],[348,-77.7772],[349,-76.9067],[350,-75.425],[351,-72.9134],[352,-69.3248],[353,-65.0744],[354,-60.5555],[355,-44.0758],[356,-20.3465],[357,-9.0551],[358,-0.9344],[359,2.1396],[360,4.5495],[361,6.3105],[362,7.2346],[363,6.8458],[364,6.4146],[365,6.4938],[366,6.9936],[367,7.5612],[368,7.8954],[369,8.0405],[370,8.076],[371,8.0547],[372,7.9633],[373,7.7896],[374,7.5142],[377,6.4593],[378,6.1289],[379,5.783],[380,5.4536],[381,5.1823],[382,5.1482],[383,4.9514],[384,4.8106],[385,4.7436],[386,4.7255],[388,4.7236],[390,4.7482],[392,4.7977],[395,4.89],[401,5.0042],[402,5.0162],[403,5.0162],[404,5.0337],[409,5.0454],[411,5.0401],[412,5.0297],[413,5.0297],[414,5.0168],[415,5.0168],[416,5.003],[417,5.003],[418,4.9861],[419,4.9861],[420,4.9673],[421,4.9673],[422,4.9495],[423,4.9495],[424,4.9353],[426,4.9353],[427,4.9198],[430,4.9198],[431,4.9071],[433,4.9071]]
       }
       thigh_r_fk_ctrl.Scale.X {type=MovieSceneScriptingFloatChannel, default=1, const [1,434]}
       thigh_r_fk_ctrl.Scale.Y {type=MovieSceneScriptingFloatChannel, default=1, const [1,434]}
       thigh_r_fk_ctrl.Scale.Z {type=MovieSceneScriptingFloatChannel, default=1, const [1,434]}
       calf_r_fk_ctrl.Location.X {
        type=MovieSceneScriptingFloatChannel
        default=0
        thinned_from=434
        keys [[0,-0.4999],[9,-0.4931],[11,-0.4933],[12,-0.4947],[13,-0.4977],[16,-0.5143],[17,-0.5242],[18,-0.5428],[19,-0.588],[20,-0.6548],[21,-0.7076],[22,-0.7301],[23,-0.7382],[24,-0.7383],[25,-0.7344],[27,-0.7228],[36,-0.7016],[38,-0.7],[40,-0.7013],[53,-0.7303],[62,-0.7377],[63,-0.7301],[64,-0.7027],[65,-0.6202],[66,-0.5327],[67,-0.4971],[68,-0.4868],[69,-0.4854],[72,-0.485],[74,-0.4883],[81,-0.5066],[88,-0.5166],[99,-0.5232],[102,-0.5286],[104,-0.5348],[108,-0.5507],[111,-0.5591],[113,-0.5627],[115,-0.5636],[128,-0.5477],[130,-0.5421],[131,-0.5379],[132,-0.5303],[133,-0.5145],[134,-0.4861],[135,-0.44],[136,-0.3721],[137,-0.2781],[138,-0.1488],[139,-0.0036],[140,0.0401],[141,0.0732],[142,0.0977],[143,0.1205],[144,0.1355],[145,0.1395],[146,0.1399],[147,0.1396],[148,0.1378],[149,0.1324],[150,0.0621],[151,0.0164],[152,-0.019],[156,-0.019],[157,0.1111],[158,0.1565],[159,0.1721],[160,0.1781],[161,0.1786],[162,0.174],[163,0.1509],[164,0.1101],[165,0.0874],[166,0.0759],[167,0.0719],[168,0.0776],[169,0.1046],[170,0.1346],[171,0.149],[172,0.1556],[173,0.1581],[174,0.1596],[175,0.1602],[177,0.1603],[178,0.1597],[179,0.1584],[181,0.1551],[182,0.1539],[184,0.1526],[221,0.1507],[225,0.1494],[228,0.1481],[249,0.1469],[255,0.1482],[257,0.1492],[258,0.1503],[259,0.1521],[260,0.1553],[261,0.1592],[262,0.1623],[263,0.1641],[264,0.1652],[266,0.1662],[283,0.1666],[287,0.1654],[288,0.1645],[291,0.1597],[295,0.1521],[297,0.1491],[298,0.1478],[302,0.1449],[308,0.1426],[320,0.1424],[335,0.1478],[339,0.1488],[341,0.1498],[343,0.1516],[344,0.1529],[345,0.1559],[346,0.1612],[348,0.1731],[349,0.1778],[350,0.1806],[351,0.1819],[352,0.1819],[353,0.18],[354,0.1752],[355,0.134],[356,0.0448],[357,0.011],[358,-0.019],[362,-0.019],[363,-0.0361],[364,-0.0608],[365,-0.2756],[366,-0.5899],[367,-0.8921],[368,-1.1047],[369,-1.1988],[370,-1.2059],[371,-1.18],[372,-1.1258],[373,-1.0477],[374,-0.8777],[375,-0.6361],[376,-0.4607],[377,-0.3797],[378,-0.3263],[379,-0.2945],[380,-0.2789],[381,-0.2741],[382,-0.2803],[383,-0.2941],[384,-0.3139],[385,-0.336],[386,-0.3612],[387,-0.3903],[388,-0.4203],[389,-0.4482],[390,-0.4711],[391,-0.4891],[392,-0.504],[393,-0.5158],[394,-0.5246],[395,-0.5304],[396,-0.5338],[397,-0.5358],[401,-0.5363],[403,-0.5339],[414,-0.5071],[420,-0.5008],[433,-0.4999]]
       }
       calf_r_fk_ctrl.Location.Y {
        type=MovieSceneScriptingFloatChannel
        default=0
        thinned_from=434
        keys [[0,0.0411],[11,0.0406],[13,0.0409],[16,0.0423],[17,0.0431],[18,0.0446],[19,0.0483],[20,0.0538],[21,0.0582],[22,0.06],[23,0.0607],[24,0.0607],[27,0.0594],[37,0.0576],[41,0.0578],[62,0.0607],[63,0.06],[64,0.0578],[65,0.051],[66,0.0438],[67,0.0409],[68,0.04],[72,0.0399],[93,0.0427],[101,0.0433],[112,0.0461],[115,0.0463],[129,0.0448],[131,0.0442],[132,0.0436],[133,0.0423],[134,0.04],[135,0.0362],[136,0.0306],[137,0.0228],[138,0.0121],[139,0.0002],[140,0.0021],[141,0.0094],[142,0.0184],[143,0.0299],[144,0.0395],[145,0.0423],[146,0.0426],[147,0.0424],[148,0.0411],[149,0.0374],[150,0.0064],[151,0],[152,0.0015],[156,0.0015],[157,0.0248],[158,0.0559],[159,0.0711],[160,0.0779],[161,0.0785],[162,0.0733],[163,0.0512],[164,0.0243],[165,0.0142],[166,0.0102],[167,0.009],[168,0.0108],[169,0.0216],[170,0.0388],[171,0.0496],[172,0.0551],[173,0.0574],[174,0.0587],[175,0.0593],[176,0.0595],[177,0.0594],[178,0.0589],[179,0.0577],[181,0.0547],[182,0.0537],[184,0.0525],[187,0.0518],[220,0.0511],[224,0.0502],[227,0.0491],[249,0.0479],[254,0.0487],[256,0.0493],[257,0.0498],[258,0.0507],[259,0.0521],[260,0.0549],[261,0.0584],[262,0.0612],[263,0.0629],[264,0.0641],[265,0.0647],[267,0.0653],[281,0.0657],[285,0.065],[287,0.0642],[288,0.0634],[289,0.0621],[292,0.0572],[294,0.0537],[295,0.0522],[297,0.0497],[298,0.0487],[301,0.0468],[306,0.045],[313,0.0442],[322,0.0447],[335,0.0486],[339,0.0494],[341,0.0502],[343,0.0517],[344,0.0528],[345,0.0554],[346,0.0602],[348,0.0723],[349,0.0775],[350,0.0809],[351,0.0824],[352,0.0824],[353,0.0801],[354,0.0746],[355,0.0384],[356,0.0028],[357,-0.0002],[358,0.0015],[362,0.0015],[363,0.0027],[364,0.0049],[365,0.0226],[366,0.0484],[367,0.0732],[368,0.0907],[369,0.0984],[370,0.099],[371,0.0969],[372,0.0924],[373,0.086],[374,0.072],[375,0.0521],[376,0.0377],[377,0.0311],[378,0.0267],[379,0.0241],[380,0.0228],[381,0.0225],[382,0.023],[383,0.0242],[384,0.0258],[385,0.0276],[386,0.0297],[389,0.0368],[390,0.0387],[391,0.0402],[392,0.0414],[393,0.0424],[394,0.0431],[396,0.0439],[402,0.044],[418,0.0413],[433,0.0411]]
       }
       calf_r_fk_ctrl.Location.Z {
        type=MovieSceneScriptingFloatChannel
        default=0
        thinned_from=434
        keys [[0,0.0005],[137,0.0003],[149,-0.0002],[156,0],[355,-0.0002],[364,0.0001],[368,0.0012],[373,0.0011],[379,0.0003],[433,0.0005]]
       }
       calf_r_fk_ctrl.Rotation.X {
        type=MovieSceneScriptingFloatChannel
        default=0.0561
        thinned_from=434
        keys [[0,0.0055],[137,0.0055],[138,0.004],[139,0.0011],[140,-0.0126],[141,-0.0243],[143,-0.0434],[144,-0.0502],[145,-0.052],[146,-0.0522],[147,-0.0521],[148,-0.0512],[149,-0.0487],[150,-0.0202],[151,-0.0049],[152,0.0055],[156,0.0055],[157,-0.0394],[158,-0.0603],[159,-0.0686],[160,-0.072],[161,-0.0722],[162,-0.0696],[163,-0.0575],[164,-0.0389],[165,-0.0297],[166,-0.0253],[167,-0.0238],[168,-0.0259],[169,-0.0366],[170,-0.0497],[171,-0.0566],[172,-0.0598],[173,-0.0611],[174,-0.0619],[175,-0.0622],[177,-0.0623],[178,-0.062],[181,-0.0596],[182,-0.059],[185,-0.0581],[221,-0.0574],[238,-0.0556],[253,-0.0559],[256,-0.0564],[257,-0.0567],[258,-0.0572],[259,-0.0581],[260,-0.0597],[261,-0.0617],[262,-0.0633],[263,-0.0642],[264,-0.0648],[266,-0.0654],[283,-0.0656],[287,-0.0649],[288,-0.0645],[292,-0.061],[295,-0.0581],[298,-0.056],[304,-0.0541],[318,-0.0533],[340,-0.0567],[342,-0.0573],[344,-0.0585],[345,-0.06],[346,-0.0627],[348,-0.0691],[349,-0.0718],[350,-0.0734],[351,-0.0742],[352,-0.0742],[353,-0.073],[354,-0.0703],[355,-0.0494],[356,-0.0141],[357,-0.0032],[358,0.0055],[362,0.0055],[364,0.0012],[365,0.0007],[366,0.0015],[367,0.0026],[368,0.0032],[369,0.0035],[375,0.0041],[379,0.0049],[381,0.005],[382,0.0055],[433,0.0055]]
       }
       calf_r_fk_ctrl.Rotation.Y {
        type=MovieSceneScriptingFloatChannel
        default=1.1982
        thinned_from=434
        keys [[0,0.0049],[137,0.0049],[138,0.0036],[139,0.0009],[140,-0.0086],[141,-0.0141],[142,-0.017],[143,-0.0188],[144,-0.0194],[148,-0.0195],[149,-0.0194],[150,-0.0124],[152,0.0049],[156,0.0049],[157,-0.0182],[158,-0.0193],[159,-0.0183],[160,-0.0177],[161,-0.0176],[162,-0.0181],[163,-0.0195],[164,-0.0181],[165,-0.0159],[166,-0.0145],[167,-0.0139],[168,-0.0147],[169,-0.0177],[170,-0.0194],[171,-0.0195],[260,-0.0194],[288,-0.0189],[345,-0.0193],[347,-0.0187],[349,-0.0177],[351,-0.0172],[352,-0.0172],[353,-0.0174],[354,-0.018],[355,-0.0194],[356,-0.0095],[357,-0.0025],[358,0.0049],[362,0.0049],[363,0.0029],[364,0.001],[365,0.0006],[366,0.0012],[367,0.0022],[368,0.0028],[375,0.0036],[379,0.0044],[381,0.0045],[382,0.0049],[433,0.0049]]
       }
       calf_r_fk_ctrl.Rotation.Z {
        type=MovieSceneScriptingFloatChannel
        default=0.2616
        thinned_from=434
        keys [[0,-4.9605],[137,-4.9606],[138,-3.6276],[139,-0.9836],[140,10.2547],[141,18.9556],[142,25.6606],[143,32.2114],[144,36.7967],[145,38.024],[146,38.1815],[147,38.0878],[148,37.487],[149,35.8316],[150,16.0055],[151,4.1593],[152,-4.9607],[156,-4.9607],[157,29.4833],[158,43.6219],[159,49.2176],[160,51.549],[161,51.7303],[162,49.9594],[163,41.7523],[164,29.1848],[165,22.8098],[166,19.6774],[167,18.6219],[168,20.1463],[169,27.6126],[170,36.4877],[171,41.1114],[172,43.3162],[173,44.1863],[174,44.6913],[175,44.9207],[177,44.9591],[178,44.7555],[179,44.2976],[181,43.1514],[182,42.7417],[184,42.3006],[187,41.9949],[220,41.7431],[224,41.3814],[227,40.9027],[249,40.4317],[254,40.7457],[256,40.9879],[257,41.1913],[258,41.548],[259,42.1301],[260,43.2103],[261,44.5699],[262,45.6361],[263,46.2754],[264,46.6957],[265,46.9355],[271,47.2753],[283,47.1896],[286,46.9074],[287,46.7462],[288,46.4422],[289,45.9497],[292,44.1231],[294,42.7499],[296,41.6192],[298,40.7322],[301,39.972],[306,39.2107],[312,38.857],[321,39.0246],[335,40.7206],[339,41.035],[341,41.377],[343,41.9729],[344,42.3944],[345,43.4196],[346,45.2523],[348,49.6164],[349,51.4166],[350,52.5455],[351,53.0619],[352,53.0619],[353,52.2798],[354,50.4163],[355,36.3135],[356,11.4718],[357,2.7633],[358,-4.9607],[362,-4.9607],[363,-2.917],[364,-1.0841],[365,-0.6079],[366,-1.278],[367,-2.2711],[368,-2.8685],[369,-3.1274],[370,-3.2016],[372,-3.2876],[373,-3.3609],[374,-3.4921],[375,-3.6702],[376,-3.9395],[377,-4.1547],[378,-4.3015],[379,-4.3926],[381,-4.4958],[382,-4.94],[433,-4.9607]]
       }
       calf_r_fk_ctrl.Scale.X {type=MovieSceneScriptingFloatChannel, default=1, const [1,434]}
       calf_r_fk_ctrl.Scale.Y {type=MovieSceneScriptingFloatChannel, default=1, const [1,434]}
       calf_r_fk_ctrl.Scale.Z {type=MovieSceneScriptingFloatChannel, default=1, const [1,434]}
       foot_r_fk_ctrl.Location.X {
        type=MovieSceneScriptingFloatChannel
        default=0
        thinned_from=434
        keys [[0,-0.5257],[9,-0.5183],[11,-0.5186],[12,-0.5201],[13,-0.5234],[16,-0.5415],[17,-0.5524],[18,-0.5727],[19,-0.6222],[20,-0.6952],[21,-0.7529],[22,-0.7776],[23,-0.7864],[24,-0.7865],[25,-0.7822],[27,-0.7696],[36,-0.7463],[38,-0.7446],[40,-0.746],[53,-0.7777],[62,-0.7858],[63,-0.7775],[64,-0.7476],[65,-0.6574],[66,-0.5617],[67,-0.5227],[68,-0.5115],[69,-0.5099],[72,-0.5095],[74,-0.513],[81,-0.5331],[88,-0.544],[99,-0.5512],[102,-0.5572],[104,-0.564],[107,-0.5776],[110,-0.588],[113,-0.5944],[115,-0.5954],[127,-0.58],[129,-0.5755],[131,-0.5673],[132,-0.559],[133,-0.5418],[134,-0.5107],[135,-0.4603],[136,-0.386],[137,-0.2832],[138,-0.1475],[139,0.0001],[362,0.0001],[363,-0.0272],[364,-0.062],[365,-0.2989],[366,-0.6399],[367,-0.9661],[368,-1.1961],[369,-1.2979],[370,-1.3053],[371,-1.2768],[372,-1.2174],[373,-1.1317],[374,-0.9452],[375,-0.6801],[376,-0.4872],[377,-0.3978],[378,-0.3386],[379,-0.3035],[380,-0.2862],[381,-0.2808],[382,-0.2857],[383,-0.3008],[384,-0.3223],[385,-0.3466],[386,-0.3741],[387,-0.4059],[388,-0.4387],[389,-0.4693],[390,-0.4943],[391,-0.5139],[392,-0.5302],[393,-0.5432],[394,-0.5528],[395,-0.5591],[396,-0.5629],[397,-0.565],[401,-0.5656],[403,-0.5629],[414,-0.5336],[419,-0.5273],[433,-0.5258]]
       }
       foot_r_fk_ctrl.Location.Y {
        type=MovieSceneScriptingFloatChannel
        default=-0.0329
        thinned_from=434
        keys [[0,-0.0024],[18,-0.0026],[21,-0.0034],[63,-0.0035],[64,-0.0033],[66,-0.0025],[133,-0.0024],[135,-0.0021],[137,-0.0013],[139,0],[362,0],[364,-0.0003],[365,-0.0013],[367,-0.0043],[368,-0.0054],[369,-0.0058],[371,-0.0057],[373,-0.0051],[374,-0.0042],[375,-0.003],[376,-0.0022],[378,-0.0015],[382,-0.0013],[433,-0.0024]]
       }
       foot_r_fk_ctrl.Location.Z {
        type=MovieSceneScriptingFloatChannel
        default=-0.9075
        thinned_from=434
        keys [[0,0.0006],[64,0.0008],[136,0.0004],[139,0],[364,0.0001],[368,0.0014],[373,0.0013],[376,0.0005],[433,0.0006]]
       }
       foot_r_fk_ctrl.Rotation.X {
        type=MovieSceneScriptingFloatChannel
        default=-1.1648
        thinned_from=434
        keys [[0,0.0616],[2,0.062],[6,0.0636],[8,0.064],[9,0.0638],[10,0.0635],[11,0.0629],[12,0.0618],[13,0.06],[14,0.0574],[15,0.0541],[16,0.0488],[17,0.0387],[18,0.0205],[19,-0.0129],[20,-0.0567],[21,-0.0941],[22,-0.1171],[23,-0.1318],[24,-0.1406],[25,-0.1465],[26,-0.1509],[27,-0.1542],[28,-0.1567],[29,-0.1586],[30,-0.1599],[31,-0.1605],[32,-0.1606],[33,-0.1604],[37,-0.1572],[40,-0.1539],[42,-0.1523],[57,-0.1487],[58,-0.1483],[59,-0.1475],[60,-0.1463],[61,-0.1447],[62,-0.1421],[63,-0.1363],[64,-0.1245],[65,-0.0963],[66,-0.064],[67,-0.0424],[68,-0.027],[69,-0.0174],[70,-0.0112],[71,-0.0074],[72,-0.0052],[73,-0.004],[74,-0.0036],[75,-0.0038],[76,-0.0043],[77,-0.0051],[79,-0.0074],[84,-0.0133],[88,-0.017],[91,-0.0191],[93,-0.02],[95,-0.0205],[97,-0.0205],[99,-0.02],[100,-0.0194],[101,-0.0183],[102,-0.0165],[103,-0.0142],[104,-0.0108],[106,-0.0022],[107,0.0017],[109,0.0085],[110,0.0115],[111,0.014],[112,0.0158],[113,0.0172],[114,0.0182],[115,0.0182],[116,0.0195],[121,0.0212],[125,0.0217],[130,0.0213],[132,0.0206],[133,0.02],[134,0.0197],[135,0.0203],[136,0.0215],[138,0.0254],[139,0.0312],[141,0.0518],[142,0.0614],[144,0.0811],[145,0.0872],[146,0.0905],[147,0.0929],[149,0.0968],[150,-0.0616],[151,-0.5501],[152,-0.7327],[153,-0.5608],[154,-0.327],[155,-0.1279],[156,-0.0579],[157,0.3953],[158,0.434],[159,0.3968],[160,0.3671],[161,0.3535],[162,0.3441],[163,0.341],[164,0.3474],[165,0.3498],[167,0.3524],[168,0.3571],[169,0.3671],[170,0.3577],[171,0.3463],[172,0.3425],[210,0.345],[211,0.3438],[212,0.3438],[217,0.3382],[218,0.3364],[219,0.3338],[220,0.3304],[221,0.3251],[222,0.3177],[224,0.3003],[225,0.2903],[226,0.2793],[227,0.27],[228,0.2632],[229,0.2575],[230,0.253],[231,0.2495],[233,0.2443],[235,0.2406],[237,0.2384],[241,0.2369],[250,0.2367],[252,0.2375],[254,0.2395],[255,0.2411],[256,0.2446],[257,0.2522],[258,0.2643],[259,0.2816],[260,0.3092],[261,0.3435],[262,0.3738],[263,0.3989],[264,0.4215],[265,0.4398],[266,0.4538],[267,0.4651],[268,0.4741],[269,0.481],[270,0.4863],[271,0.4906],[272,0.4938],[274,0.4974],[278,0.4974],[280,0.4955],[281,0.4937],[282,0.4908],[283,0.4865],[287,0.4623],[289,0.4478],[293,0.4161],[294,0.4096],[296,0.4],[298,0.3923],[300,0.3868],[303,0.3814],[306,0.3786],[311,0.3782],[312,0.3801],[319,0.3905],[334,0.4204],[344,0.4306],[345,0.4286],[346,0.4211],[347,0.4094],[348,0.3949],[350,0.3636],[352,0.337],[353,0.3174],[354,0.2865],[355,0.1617],[356,-0.1358],[357,-0.5145],[358,-0.7041],[359,-0.5418],[360,-0.3151],[361,-0.0966],[362,-0.0566],[363,0.0044],[364,0.0219],[365,0.0285],[366,0.034],[367,0.04],[368,0.0428],[369,0.0435],[370,0.0431],[371,0.042],[373,0.0392],[374,0.0368],[375,0.0333],[376,0.0307],[377,0.0295],[378,0.0288],[379,0.0285],[381,0.0284],[382,0.0282],[384,0.0286],[385,0.0292],[386,0.0305],[387,0.0326],[391,0.0432],[394,0.0514],[395,0.0537],[397,0.0578],[398,0.0596],[399,0.0611],[400,0.0622],[401,0.0629],[403,0.0639],[406,0.0642],[423,0.0619],[433,0.0619]]
       }
       foot_r_fk_ctrl.Rotation.Y {
        type=MovieSceneScriptingFloatChannel
        default=0.1067
        thinned_from=434
        keys [[0,-0.3499],[1,-0.3485],[2,-0.3449],[3,-0.3399],[5,-0.3283],[6,-0.3232],[7,-0.3197],[8,-0.3185],[9,-0.3202],[10,-0.3247],[11,-0.3326],[12,-0.3466],[13,-0.3691],[14,-0.4027],[15,-0.4448],[16,-0.5121],[17,-0.6427],[18,-0.8779],[19,-1.3101],[20,-1.8783],[21,-2.3655],[22,-2.6675],[23,-2.8619],[24,-2.9793],[25,-3.0577],[26,-3.1161],[27,-3.1592],[28,-3.193],[29,-3.2188],[30,-3.2354],[31,-3.2439],[33,-3.2423],[37,-3.2008],[41,-3.1449],[58,-3.0817],[60,-3.056],[61,-3.0338],[62,-2.9995],[63,-2.9235],[64,-2.7686],[65,-2.4006],[66,-1.9801],[67,-1.6991],[68,-1.4998],[69,-1.375],[70,-1.2948],[71,-1.2457],[72,-1.2165],[73,-1.2017],[74,-1.1969],[75,-1.1987],[76,-1.2057],[77,-1.216],[78,-1.2292],[84,-1.3211],[88,-1.3689],[91,-1.3962],[93,-1.4083],[95,-1.4146],[97,-1.4148],[99,-1.4088],[100,-1.4011],[101,-1.3864],[102,-1.364],[103,-1.333],[104,-1.2892],[106,-1.1783],[107,-1.1275],[109,-1.0392],[110,-1.0003],[111,-0.9682],[112,-0.9438],[113,-0.9256],[114,-0.9123],[115,-0.9123],[116,-0.8956],[118,-0.8848],[122,-0.8716],[125,-0.8674],[129,-0.8709],[131,-0.8769],[132,-0.882],[133,-0.8897],[134,-0.8936],[135,-0.8875],[136,-0.8723],[137,-0.8488],[138,-0.8189],[139,-0.7782],[140,-0.7179],[141,-0.6384],[142,-0.5524],[143,-0.4585],[144,-0.3548],[145,-0.2808],[146,-0.2392],[147,-0.2075],[148,-0.181],[149,-0.1475],[150,-0.1366],[151,-0.7034],[152,-0.9223],[153,-0.6706],[154,-0.3573],[155,-0.105],[156,-0.0965],[157,3.7852],[158,4.3289],[159,3.8785],[160,3.5103],[161,3.3468],[162,3.2348],[163,3.2184],[164,3.3552],[165,3.4213],[167,3.4795],[168,3.535],[169,3.6334],[171,3.2977],[172,3.2398],[183,3.2872],[210,3.2785],[211,3.2639],[212,3.2639],[217,3.2046],[218,3.186],[219,3.1601],[220,3.1245],[221,3.07],[222,2.9941],[224,2.8129],[225,2.7093],[226,2.5956],[227,2.4988],[228,2.4255],[229,2.3634],[230,2.3131],[231,2.2747],[233,2.2176],[235,2.1763],[237,2.1515],[241,2.1373],[250,2.1366],[252,2.1455],[254,2.1668],[255,2.1838],[256,2.2249],[257,2.312],[258,2.4494],[259,2.6402],[260,2.9343],[261,3.2963],[262,3.6209],[263,3.8965],[264,4.1503],[265,4.3571],[266,4.5169],[267,4.6488],[268,4.7542],[269,4.8348],[270,4.897],[271,4.9468],[272,4.9847],[273,5.0108],[274,5.0257],[278,5.0257],[280,5.0041],[281,4.9838],[282,4.9506],[283,4.9015],[285,4.771],[288,4.5469],[293,4.1],[294,4.0261],[296,3.9155],[298,3.8267],[300,3.7627],[303,3.6981],[306,3.6641],[311,3.6584],[312,3.6785],[317,3.7549],[334,4.1479],[344,4.2658],[345,4.235],[346,4.13],[347,3.9688],[348,3.77],[350,3.3426],[351,3.1619],[352,2.9731],[353,2.7054],[354,2.2932],[355,1.2856],[356,-0.0086],[357,-0.9553],[358,-1.3231],[359,-1.07],[360,-0.7537],[361,-0.5583],[362,-0.6419],[363,-0.6515],[364,-0.6545],[365,-0.6422],[366,-0.6112],[367,-0.5784],[368,-0.5617],[369,-0.5589],[370,-0.5663],[371,-0.581],[373,-0.6211],[374,-0.6563],[375,-0.7055],[376,-0.7431],[377,-0.7622],[378,-0.7734],[379,-0.779],[382,-0.7828],[383,-0.7819],[384,-0.7783],[385,-0.7709],[386,-0.7545],[387,-0.7273],[388,-0.6937],[389,-0.6579],[391,-0.5902],[393,-0.5175],[394,-0.4829],[395,-0.4525],[396,-0.4252],[397,-0.3994],[398,-0.3761],[399,-0.3567],[400,-0.3424],[401,-0.3324],[402,-0.3251],[403,-0.3202],[404,-0.3172],[405,-0.3158],[406,-0.3155],[408,-0.317],[410,-0.3209],[414,-0.3313],[421,-0.3436],[423,-0.3461],[433,-0.3461]]
       }
       foot_r_fk_ctrl.Rotation.Z {
        type=MovieSceneScriptingFloatChannel
        default=-0.1036
        thinned_from=434
        keys [[0,1.2346],[2,1.231],[10,1.1849],[11,1.1758],[12,1.1622],[13,1.1429],[14,1.1164],[15,1.0942],[16,1.0784],[17,1.0566],[18,1.0046],[19,0.8673],[20,0.6869],[21,0.57],[22,0.5553],[23,0.5875],[24,0.6141],[25,0.6273],[27,0.6717],[28,0.6891],[30,0.7119],[31,0.7203],[38,0.7516],[40,0.7572],[41,0.7569],[42,0.7538],[46,0.7298],[52,0.7027],[55,0.6942],[59,0.6881],[61,0.682],[62,0.6724],[63,0.6696],[64,0.7012],[65,0.842],[66,1.0179],[67,1.1259],[68,1.1985],[69,1.2431],[70,1.2727],[71,1.297],[72,1.317],[73,1.3291],[74,1.3328],[75,1.3289],[77,1.3094],[80,1.2678],[82,1.2466],[84,1.2309],[87,1.217],[93,1.2136],[96,1.2225],[100,1.2441],[104,1.2766],[106,1.2973],[107,1.3056],[109,1.3152],[111,1.319],[112,1.3173],[114,1.3089],[115,1.3089],[126,1.3294],[130,1.3311],[132,1.3373],[133,1.3426],[134,1.3596],[135,1.4002],[136,1.465],[137,1.5562],[138,1.0536],[139,-0.0798],[140,-5.2083],[141,-9.0126],[142,-11.8365],[143,-14.5497],[144,-16.3637],[145,-16.7748],[146,-16.7755],[147,-16.6774],[148,-16.3624],[149,-15.5622],[150,-3.7194],[151,8.555],[152,15.3653],[153,16.6554],[154,16.861],[155,16.3314],[156,12.7096],[157,-7.3832],[158,-11.6202],[159,-13.7702],[160,-14.7268],[161,-14.7336],[162,-13.9379],[163,-10.3067],[164,-4.4853],[165,-1.4884],[166,-0.0049],[167,0.5176],[168,-0.111],[169,-3.392],[170,-7.5937],[171,-9.8569],[172,-10.8965],[173,-11.2647],[174,-11.4759],[175,-11.5683],[176,-11.5886],[177,-11.5757],[178,-11.4826],[179,-11.2771],[181,-10.765],[182,-10.5821],[183,-10.4711],[184,-10.3877],[186,-10.2862],[190,-10.2152],[191,-10.2152],[192,-10.1939],[212,-10.1422],[216,-10.0597],[218,-9.9609],[219,-9.8792],[220,-9.7685],[221,-9.5998],[222,-9.3663],[224,-8.8209],[225,-8.5045],[226,-8.1567],[227,-7.88],[228,-7.7022],[229,-7.5711],[230,-7.4743],[232,-7.336],[234,-7.24],[238,-7.1307],[241,-7.0858],[242,-7.0858],[243,-7.0698],[249,-7.0698],[251,-7.1149],[253,-7.1985],[255,-7.3327],[256,-7.4253],[257,-7.5903],[258,-7.9014],[259,-8.4288],[260,-9.4521],[261,-10.8033],[262,-11.9281],[263,-12.6766],[264,-13.2381],[265,-13.6192],[266,-13.8574],[267,-14.0088],[269,-14.1801],[271,-14.2919],[274,-14.3612],[280,-14.3268],[282,-14.2492],[284,-14.077],[286,-13.8233],[287,-13.6721],[288,-13.4532],[289,-13.1423],[292,-12.0621],[293,-11.6743],[294,-11.3109],[295,-11.001],[296,-10.7318],[297,-10.4899],[298,-10.2864],[299,-10.1333],[301,-9.9154],[303,-9.7518],[306,-9.5868],[308,-9.5229],[312,-9.5025],[318,-9.661],[322,-9.8451],[333,-10.5953],[339,-10.8103],[340,-10.8757],[341,-10.9668],[342,-11.0898],[343,-11.2505],[344,-11.4547],[345,-11.9668],[346,-12.9037],[347,-14.0576],[348,-15.246],[349,-16.3117],[350,-17.1135],[351,-17.6633],[352,-18.0486],[353,-18.2321],[354,-18.1423],[355,-12.326],[356,0.5395],[357,7.1619],[358,11.671],[359,11.4087],[360,10.3106],[361,7.9026],[363,0.3461],[364,-0.8361],[365,-1.1813],[366,-0.9205],[367,-0.5034],[368,-0.2478],[369,-0.1375],[370,-0.0984],[371,-0.0579],[372,-0.0068],[373,0.0695],[374,0.1784],[375,0.3144],[376,0.5008],[377,0.6646],[378,0.7922],[379,0.8929],[380,0.9706],[381,1.0389],[382,1.2854],[383,1.3194],[384,1.3452],[385,1.3637],[386,1.3757],[387,1.3836],[388,1.3874],[389,1.3873],[390,1.3836],[391,1.3743],[392,1.3596],[395,1.3066],[398,1.2632],[399,1.2508],[401,1.2348],[404,1.2213],[409,1.2165],[433,1.2312]]
       }
       foot_r_fk_ctrl.Scale.X {type=MovieSceneScriptingFloatChannel, default=1, const [1,434]}
       foot_r_fk_ctrl.Scale.Y {type=MovieSceneScriptingFloatChannel, default=1, const [1,434]}
       foot_r_fk_ctrl.Scale.Z {type=MovieSceneScriptingFloatChannel, default=1, const [1,434]}
       ball_r_fk_ctrl.Location.X {type=MovieSceneScriptingFloatChannel, default=0, thinned_from=434, keys [[0,0],[433,0]]}
       ball_r_fk_ctrl.Location.Y {type=MovieSceneScriptingFloatChannel, default=0, thinned_from=434, keys [[0,0],[433,0]]}
       ball_r_fk_ctrl.Location.Z {type=MovieSceneScriptingFloatChannel, default=0, thinned_from=434, keys [[0,0],[433,0]]}
       ball_r_fk_ctrl.Rotation.X {
        type=MovieSceneScriptingFloatChannel
        default=0
        thinned_from=434
        keys [[0,0],[151,0],[152,-0.0107],[153,-0.8042],[154,-1.4098],[155,-1.7164],[156,-1.1835],[157,0],[358,0],[359,-0.5322],[360,-0.9287],[361,-0.9945],[362,-0.3871],[363,0],[433,0]]
       }
       ball_r_fk_ctrl.Rotation.Y {
        type=MovieSceneScriptingFloatChannel
        default=0
        thinned_from=434
        keys [[0,0],[151,0],[152,-0.0068],[153,-0.5576],[154,-1.0679],[155,-1.3884],[156,-0.9212],[157,0],[358,0],[359,-0.3701],[360,-0.6956],[361,-0.7795],[362,-0.2874],[363,0],[433,0]]
       }
       ball_r_fk_ctrl.Rotation.Z {
        type=MovieSceneScriptingFloatChannel
        default=0
        thinned_from=434
        keys [[0,0],[151,0],[152,-0.0355],[153,-2.7612],[154,-5.1786],[155,-6.7711],[156,-4.9295],[157,0],[358,0],[359,-1.9985],[360,-3.8058],[361,-4.4737],[362,-1.8062],[363,0],[433,0]]
       }
       ball_r_fk_ctrl.Scale.X {type=MovieSceneScriptingFloatChannel, default=1, const [1,434]}
       ball_r_fk_ctrl.Scale.Y {type=MovieSceneScriptingFloatChannel, default=1, const [1,434]}
       ball_r_fk_ctrl.Scale.Z {type=MovieSceneScriptingFloatChannel, default=1, const [1,434]}
       foot_r_ik_ctrl.Location.X {
        type=MovieSceneScriptingFloatChannel
        default=0
        thinned_from=434
        keys [[0,0],[1,0.0025],[2,0.009],[3,0.0183],[4,0],[5,0.0116],[6,0],[7,0.0094],[8,0],[9,0.0026],[10,0],[83,0],[84,-0.0078],[85,0],[86,-0.0072],[87,0],[88,-0.0098],[89,-0.0001],[90,-0.0105],[91,0],[92,-0.0052],[93,-0.0084],[94,-0.012],[95,0],[137,0],[138,0.0015],[139,0],[149,0],[150,0.2436],[151,1.021],[152,1.3913],[153,1.1018],[154,0.646],[155,0.2438],[156,0.1575],[157,0.0001],[197,0],[198,-0.006],[199,0],[200,0.0014],[201,-0.0067],[202,0.0029],[203,0.0035],[204,-0.0022],[205,-0.0018],[206,-0.0018],[207,-0.0014],[208,-0.0019],[209,0.004],[210,0],[211,0.0044],[212,0],[213,0.0043],[214,0],[215,0.0045],[216,0],[217,0.0041],[218,0.0046],[219,-0.0007],[220,0],[238,0],[239,-0.0062],[240,0],[241,-0.004],[242,0],[243,-0.0023],[244,-0.0039],[245,-0.0048],[246,-0.0051],[247,-0.0051],[248,-0.0021],[249,0],[271,0],[272,0.0052],[273,0],[274,0.0007],[275,0.0002],[277,-0.0021],[278,-0.0112],[279,-0.0066],[280,0],[281,-0.0067],[282,0],[313,0],[314,-0.0019],[315,0.0008],[316,0],[317,0],[318,0.0117],[319,0],[351,0],[352,-0.0013],[353,0],[354,0],[355,0.1201],[356,0.5032],[357,0.9975],[358,1.1736],[359,0.8406],[360,0.4594],[361,0.1339],[362,0.1092],[363,0.0494],[364,0.0249],[365,0.0171],[366,0.0122],[367,0.0065],[368,0.0038],[369,0.0029],[370,0.0027],[371,0.0023],[372,0.0026],[374,0.002],[376,0.0009],[379,0.0002],[402,0],[403,0.0074],[404,0],[405,-0.0021],[406,0],[407,-0.0062],[408,0],[409,-0.0101],[410,0],[411,-0.0128],[412,0],[413,-0.0126],[414,0],[415,-0.0103],[416,0],[417,-0.0089],[418,0],[419,-0.0082],[420,0],[421,-0.007],[422,0],[423,-0.0052],[424,0.0001],[425,-0.0033],[426,-0.0063],[427,0.0001],[428,-0.0019],[429,-0.0033],[430,-0.0044],[431,0],[432,-0.0003],[433,-0.0003]]
       }
       foot_r_ik_ctrl.Location.Y {
        type=MovieSceneScriptingFloatChannel
        default=0
        thinned_from=434
        keys [[0,-7.0886],[8,-7.0885],[9,-7.0795],[83,-7.0885],[84,-7.0709],[85,-7.0886],[86,-7.0772],[87,-7.0886],[94,-7.1006],[149,-7.0885],[150,-6.793],[151,-5.9809],[152,-5.8021],[153,-6.0709],[154,-6.3778],[155,-6.6259],[156,-6.7841],[157,-7.0884],[199,-7.0886],[200,-7.0724],[201,-7.0943],[203,-7.0705],[204,-7.0904],[354,-7.0885],[355,-7.0022],[356,-6.7262],[357,-6.2783],[358,-5.9741],[359,-6.0686],[360,-6.3333],[361,-6.6854],[362,-6.8646],[363,-7.0287],[364,-7.0583],[402,-7.0887],[403,-7.0753],[404,-7.0885],[405,-7.0801],[413,-7.0959],[414,-7.0886],[415,-7.0969],[416,-7.0885],[417,-7.099],[418,-7.0886],[419,-7.1007],[420,-7.0886],[421,-7.1006],[422,-7.0886],[423,-7.0987],[424,-7.0884],[426,-7.1032],[427,-7.0884],[430,-7.1031],[431,-7.0886],[433,-7.0903]]
       }
       foot_r_ik_ctrl.Location.Z {
        type=MovieSceneScriptingFloatChannel
        default=0
        thinned_from=434
        keys [[0,0.0002],[3,0.0008],[4,0.0001],[5,0.0004],[7,0],[8,0.0002],[9,-0.0007],[10,0.0002],[83,0.0001],[84,-0.0024],[85,0.0002],[86,-0.0016],[87,0.0002],[88,-0.0011],[89,0.0002],[90,-0.0006],[91,0.0002],[92,-0.0001],[93,0.0001],[94,0.0009],[95,0.0001],[114,0.0002],[115,0],[137,0.0002],[138,0.0056],[139,0.0002],[140,0.0005],[141,0.0001],[142,0.0005],[143,0.0002],[149,0.0002],[150,0.7919],[151,2.7579],[152,3.3582],[153,3.7538],[154,3.8412],[155,3.6478],[156,2.4036],[157,0.0008],[163,0.0006],[164,0.0008],[165,0.0004],[168,0.0003],[169,0.0004],[170,0.0002],[197,0.0002],[198,0.0023],[199,0.0002],[200,0.0033],[201,0.0025],[202,0.0011],[203,0.0023],[204,0.0009],[205,0.0013],[207,0.0012],[208,0.0001],[209,-0.0013],[210,0.0002],[211,-0.0014],[212,0.0002],[213,-0.0014],[214,0.0002],[215,-0.0015],[216,0.0002],[217,-0.0013],[218,-0.0015],[219,0.0004],[220,0.0002],[238,0.0002],[239,0.0019],[240,0.0002],[241,0.0013],[242,0.0002],[243,0.0008],[245,0.0015],[247,0.0015],[249,0.0002],[271,0.0002],[272,-0.0023],[273,0.0002],[274,-0.0002],[275,0],[277,0.0011],[278,0.0012],[279,0.0032],[280,0.0002],[281,0.0033],[282,0.0002],[313,0.0002],[314,0.0008],[315,-0.0001],[316,0.0002],[317,0.0002],[318,0.001],[319,0.0002],[351,0.0002],[352,0.0009],[353,0.0002],[354,0.0002],[355,0.3153],[356,1.246],[357,2.4045],[358,2.9345],[359,3.0711],[360,2.8552],[361,2.1009],[362,0.9826],[363,0.167],[364,0.0848],[365,0.0582],[366,0.042],[367,0.0226],[368,0.0138],[369,0.0107],[370,0.0098],[372,0.0089],[373,0.0081],[374,0.0069],[375,0.0053],[376,0.0034],[377,0.0022],[378,0.0015],[381,0.0008],[382,0.0001],[402,0.0002],[403,-0.0009],[404,0.0001],[405,-0.0009],[406,0.0002],[407,-0.0005],[408,0.0001],[409,-0.0002],[410,0.0002],[412,0.0002],[413,0.0004],[414,0.0002],[415,0.0006],[416,0.0002],[417,0.0009],[418,0.0002],[419,0.0011],[420,0.0002],[421,0.0011],[422,0.0001],[423,0.001],[424,0],[426,0.0013],[427,0],[430,0.0014],[431,0.0001],[433,0.0002]]
       }
       foot_r_ik_ctrl.Rotation.X {
        type=MovieSceneScriptingFloatChannel
        default=0
        thinned_from=434
        keys [[0,0],[1,-0.0011],[2,-0.004],[3,-0.0081],[4,0],[5,-0.006],[6,0],[7,-0.0068],[8,-0.0001],[9,-0.0062],[10,0.0001],[19,0],[38,0],[44,-0.0001],[45,0.0001],[83,0],[84,-0.0067],[85,0],[86,-0.0034],[87,0.0001],[88,0.0009],[89,0.0001],[90,0.0036],[91,-0.0001],[92,0.0025],[93,0.0056],[94,0.0108],[95,0],[111,0.0001],[112,-0.0001],[115,0.0001],[129,0.0001],[131,0],[137,0],[138,0.0152],[139,0.0001],[140,0.0012],[141,-0.0001],[142,0.0011],[143,0.0003],[149,0.0003],[150,2.183],[151,7.8053],[152,9.5923],[153,10.8343],[154,11.2129],[155,10.8664],[156,7.3729],[157,0.0018],[158,0.0012],[163,0.0012],[164,0.0018],[165,0.0007],[166,0.0007],[167,0.0004],[168,0.0004],[169,0.001],[170,0.0002],[197,0.0002],[198,0.0007],[199,0],[200,-0.0089],[201,0.0071],[202,-0.0142],[203,-0.018],[204,-0.0045],[205,-0.0066],[206,-0.0066],[207,-0.0095],[209,-0.0108],[210,-0.0028],[211,-0.0081],[212,-0.0056],[213,-0.0079],[214,0],[215,-0.0084],[216,0],[217,-0.0075],[218,-0.0085],[219,0.0013],[220,0],[238,0],[239,0.0113],[240,0],[241,0.0073],[242,-0.0092],[243,0.0041],[244,0.0023],[245,0.0011],[247,0.001],[248,-0.0031],[249,0.0023],[271,0.0023],[272,-0.01],[273,0],[274,-0.0014],[275,0.0041],[276,0.0056],[277,0.0056],[279,0.0124],[280,0],[281,0.0126],[282,0],[310,0],[311,-0.0095],[312,0],[313,0],[314,0.0036],[315,-0.0015],[316,0],[317,0],[318,-0.004],[319,0],[351,0],[352,0.0025],[353,0],[354,0],[355,0.864],[356,3.4544],[357,6.7704],[358,8.3275],[359,8.9217],[360,8.5702],[361,6.6559],[362,3.0836],[363,0.4567],[364,0.2316],[365,0.1585],[366,0.1144],[367,0.0615],[368,0.0375],[369,0.0289],[370,0.0266],[372,0.024],[373,0.0216],[374,0.0184],[375,0.0143],[376,0.0088],[377,0.0054],[378,0.0037],[379,0.0029],[381,0.0021],[382,-0.0003],[400,-0.0001],[401,0.0001],[402,0.0001],[403,-0.01],[404,0],[405,-0.0034],[406,0],[407,0.0008],[408,0],[409,0.0051],[410,0.0001],[411,0.0085],[412,0],[413,0.0096],[414,0],[415,0.0091],[416,0],[417,0.0096],[418,0.0001],[419,0.0103],[420,0],[421,0.0097],[422,0],[423,0.0079],[424,-0.0009],[425,0.0043],[426,0.009],[427,-0.0026],[428,0.0013],[429,0.0044],[430,0.007],[431,-0.0034],[432,-0.0023],[433,-0.0023]]
       }
       foot_r_ik_ctrl.Rotation.Y {
        type=MovieSceneScriptingFloatChannel
        default=0
        thinned_from=434
        keys [[0,0],[1,0.0013],[2,0.0049],[3,0.0099],[4,0],[5,0.0056],[6,0],[7,0.0033],[8,0],[9,-0.002],[10,0],[83,0],[84,-0.0117],[85,0],[86,-0.0088],[87,0],[88,-0.008],[89,0],[90,-0.0066],[91,0],[92,-0.0027],[93,-0.0033],[94,-0.0018],[95,0],[114,0],[115,-0.0095],[116,0],[137,0],[138,-0.0013],[139,0],[149,0],[150,-0.199],[151,-0.7113],[152,-0.8691],[153,-0.6331],[154,-0.3541],[155,-0.1338],[156,-0.1076],[157,-0.0001],[197,0],[198,0.0003],[199,0],[200,-0.0083],[201,0.0001],[202,-0.0047],[203,-0.0082],[204,0.0002],[205,-0.0017],[206,-0.0017],[207,-0.0044],[208,-0.005],[209,0.0053],[210,0.0093],[211,0],[212,0.0067],[213,0],[241,0],[242,0.0013],[243,0],[244,0.0004],[248,0.0005],[249,0.0001],[274,0],[275,-0.0051],[276,-0.0055],[277,-0.0043],[278,-0.0011],[279,0],[310,0],[311,-0.011],[312,0],[317,0],[318,0.0084],[319,0],[354,0],[355,-0.0787],[356,-0.3149],[357,-0.6172],[358,-0.7576],[359,-0.5629],[360,-0.3163],[361,-0.0907],[362,-0.0889],[363,-0.0416],[364,-0.0211],[365,-0.0144],[366,-0.0104],[367,-0.0056],[368,-0.0035],[369,-0.0027],[370,-0.0024],[371,-0.0024],[372,-0.002],[373,-0.0022],[376,-0.0008],[379,-0.0002],[402,0],[403,-0.0011],[404,0],[405,-0.005],[406,0],[407,-0.0054],[408,0],[409,-0.0057],[410,0],[411,-0.0057],[412,0],[413,-0.0048],[414,0],[415,-0.0029],[416,0],[417,-0.0012],[418,0],[420,0],[421,0.0006],[422,0],[423,0.0009],[424,0.001],[426,0.0044],[427,0.0028],[428,0.0043],[430,0.0066],[431,0.0037],[432,0.0042],[433,0.0042]]
       }
       foot_r_ik_ctrl.Rotation.Z {
        type=MovieSceneScriptingFloatChannel
        default=0
        thinned_from=434
        keys [[0,-33.7308],[150,-33.7242],[151,-36.9782],[152,-41.6211],[153,-45.3061],[154,-47.6468],[155,-49.2388],[156,-49.0008],[157,-48.4908],[356,-48.4838],[357,-45.2054],[358,-40.8083],[359,-36.9794],[360,-34.5742],[362,-33.9182],[363,-33.7288],[433,-33.7303]]
       }
       foot_r_ik_ctrl.Scale.X {type=MovieSceneScriptingFloatChannel, default=1, const [1,434]}
       foot_r_ik_ctrl.Scale.Y {type=MovieSceneScriptingFloatChannel, default=1, const [1,434]}
       foot_r_ik_ctrl.Scale.Z {type=MovieSceneScriptingFloatChannel, default=1, const [1,434]}
       ball_r_ik_ctrl.Location.X {type=MovieSceneScriptingFloatChannel, default=0, thinned_from=434, keys [[0,0],[309,0],[310,-0.0005],[311,0],[433,0]]}
       ball_r_ik_ctrl.Location.Y {
        type=MovieSceneScriptingFloatChannel
        default=0
        thinned_from=434
        keys [[0,0],[17,-0.0001],[41,0],[44,-0.0001],[46,0],[50,-0.0001],[52,0],[54,-0.0001],[55,0.0001],[57,-0.0001],[61,0],[66,-0.0001],[68,0],[71,-0.0001],[77,0],[79,-0.0001],[80,0.0001],[106,-0.0001],[133,0],[145,-0.0001],[190,0],[191,0.0013],[192,0],[193,0.0011],[194,0],[195,0.0009],[196,0],[197,0.0006],[198,0],[249,0],[250,0.0003],[271,0.0003],[272,0],[309,0],[310,0.001],[311,0],[389,0],[396,-0.0001],[399,0],[400,-0.0001],[433,0]]
       }
       ball_r_ik_ctrl.Location.Z {
        type=MovieSceneScriptingFloatChannel
        default=0
        thinned_from=434
        keys [[0,0],[190,0],[191,0.0027],[192,0],[193,0.0022],[194,0],[195,0.0019],[196,0],[197,0.0014],[198,0],[249,0],[250,0.0006],[271,0.0006],[272,0],[309,0],[310,0.0021],[311,0],[433,0]]
       }
       ball_r_ik_ctrl.Rotation.X {
        type=MovieSceneScriptingFloatChannel
        default=0
        thinned_from=434
        keys [[0,0],[66,0.0001],[68,0],[79,0.0001],[80,-0.0001],[145,0],[146,-0.0002],[151,-0.0002],[152,-0.036],[153,-2.8063],[154,-5.2706],[155,-6.8932],[156,-5.004],[157,0],[190,0],[191,-0.0113],[192,-0.0002],[193,-0.0095],[194,-0.0002],[195,-0.0081],[196,-0.0002],[197,-0.0059],[198,0],[249,0],[250,-0.0023],[271,-0.0023],[272,0],[309,0],[310,-0.0088],[311,0],[358,0],[359,-2.0267],[360,-3.8604],[361,-4.5339],[362,-1.8264],[363,0],[389,0],[400,0.0002],[433,0]]
       }
       ball_r_ik_ctrl.Rotation.Y {
        type=MovieSceneScriptingFloatChannel
        default=0
        thinned_from=434
        keys [[0,0],[151,0],[152,-0.0035],[153,-0.2585],[154,-0.4482],[155,-0.5374],[156,-0.351],[157,0],[190,0],[191,0.0007],[192,0],[193,0.0004],[194,0],[195,0.0002],[196,0],[197,0.0001],[309,0],[310,-0.0041],[311,0],[358,0],[359,-0.1632],[360,-0.2735],[361,-0.2772],[362,-0.1046],[363,0],[433,0]]
       }
       ball_r_ik_ctrl.Rotation.Z {
        type=MovieSceneScriptingFloatChannel
        default=0
        thinned_from=434
        keys [[0,0],[46,0],[50,-0.0001],[52,0.0001],[61,-0.0001],[66,0.0001],[71,-0.0001],[77,0],[79,-0.0001],[151,0],[152,0.0106],[153,0.8196],[154,1.4705],[155,1.8205],[156,1.2314],[157,0],[190,0],[191,0.0004],[192,0],[193,0.0004],[194,0],[195,0.0003],[196,0],[197,0.0002],[198,0],[249,0],[271,0.0002],[309,0],[310,0.0003],[311,0],[358,0],[359,0.5386],[360,0.9556],[361,1.0299],[362,0.391],[363,0],[389,0],[400,-0.0001],[433,0]]
       }
       ball_r_ik_ctrl.Scale.X {type=MovieSceneScriptingFloatChannel, default=1, const [1,434]}
       ball_r_ik_ctrl.Scale.Y {type=MovieSceneScriptingFloatChannel, default=1, const [1,434]}
       ball_r_ik_ctrl.Scale.Z {type=MovieSceneScriptingFloatChannel, default=1, const [1,434]}
       foot_bk1_r_ctrl.Location.X {type=MovieSceneScriptingFloatChannel, default=0, const [0,434]}
       foot_bk1_r_ctrl.Location.Y {type=MovieSceneScriptingFloatChannel, default=0, const [0,434]}
       foot_bk1_r_ctrl.Location.Z {type=MovieSceneScriptingFloatChannel, default=0, const [0,434]}
       foot_bk1_r_ctrl.Rotation.X {type=MovieSceneScriptingFloatChannel, default=0, const [0,434]}
       foot_bk1_r_ctrl.Rotation.Y {type=MovieSceneScriptingFloatChannel, default=0, const [0,434]}
       foot_bk1_r_ctrl.Rotation.Z {type=MovieSceneScriptingFloatChannel, default=0, const [0,434]}
       foot_bk1_r_ctrl.Scale.X {type=MovieSceneScriptingFloatChannel, default=1, const [1,434]}
       foot_bk1_r_ctrl.Scale.Y {type=MovieSceneScriptingFloatChannel, default=1, const [1,434]}
       foot_bk1_r_ctrl.Scale.Z {type=MovieSceneScriptingFloatChannel, default=1, const [1,434]}
       foot_roll_r_ctrl.Location.X {type=MovieSceneScriptingFloatChannel, default=0, const [0,434]}
       foot_roll_r_ctrl.Location.Y {type=MovieSceneScriptingFloatChannel, default=0, const [0,434]}
       foot_roll_r_ctrl.Location.Z {type=MovieSceneScriptingFloatChannel, default=0, const [0,434]}
       foot_roll_r_ctrl.Rotation.X {type=MovieSceneScriptingFloatChannel, default=0, const [0,434]}
       foot_roll_r_ctrl.Rotation.Y {type=MovieSceneScriptingFloatChannel, default=0, const [0,434]}
       foot_roll_r_ctrl.Rotation.Z {type=MovieSceneScriptingFloatChannel, default=0, const [0,434]}
       foot_roll_r_ctrl.Scale.X {type=MovieSceneScriptingFloatChannel, default=1, const [1,434]}
       foot_roll_r_ctrl.Scale.Y {type=MovieSceneScriptingFloatChannel, default=1, const [1,434]}
       foot_roll_r_ctrl.Scale.Z {type=MovieSceneScriptingFloatChannel, default=1, const [1,434]}
       roll_blend_r_ctrl {type=MovieSceneScriptingFloatChannel, default=1, const [1,434]}
       tip_r_ctrl.Location.X {type=MovieSceneScriptingFloatChannel, default=0, const [0,434]}
       tip_r_ctrl.Location.Y {type=MovieSceneScriptingFloatChannel, default=0, const [0,434]}
       tip_r_ctrl.Location.Z {type=MovieSceneScriptingFloatChannel, default=0, const [0,434]}
       tip_r_ctrl.Rotation.X {type=MovieSceneScriptingFloatChannel, default=0, const [0,434]}
       tip_r_ctrl.Rotation.Y {type=MovieSceneScriptingFloatChannel, default=0, const [0,434]}
       tip_r_ctrl.Rotation.Z {type=MovieSceneScriptingFloatChannel, default=0, const [0,434]}
       tip_r_ctrl.Scale.X {type=MovieSceneScriptingFloatChannel, default=1, const [1,434]}
       tip_r_ctrl.Scale.Y {type=MovieSceneScriptingFloatChannel, default=1, const [1,434]}
       tip_r_ctrl.Scale.Z {type=MovieSceneScriptingFloatChannel, default=1, const [1,434]}
       heel_r_ctrl.Location.X {type=MovieSceneScriptingFloatChannel, default=0, const [0,434]}
       heel_r_ctrl.Location.Y {type=MovieSceneScriptingFloatChannel, default=0, const [0,434]}
       heel_r_ctrl.Location.Z {type=MovieSceneScriptingFloatChannel, default=0, const [0,434]}
       heel_r_ctrl.Rotation.X {type=MovieSceneScriptingFloatChannel, default=0, const [0,434]}
       heel_r_ctrl.Rotation.Y {type=MovieSceneScriptingFloatChannel, default=0, const [0,434]}
       heel_r_ctrl.Rotation.Z {type=MovieSceneScriptingFloatChannel, default=0, const [0,434]}
       heel_r_ctrl.Scale.X {type=MovieSceneScriptingFloatChannel, default=1, const [1,434]}
       heel_r_ctrl.Scale.Y {type=MovieSceneScriptingFloatChannel, default=1, const [1,434]}
       heel_r_ctrl.Scale.Z {type=MovieSceneScriptingFloatChannel, default=1, const [1,434]}
       leg_r_softness {type=MovieSceneScriptingFloatChannel, default=0, const [0,434]}
       leg_r_pv_ik_ctrl.Location.X {
        type=MovieSceneScriptingFloatChannel
        default=0
        thinned_from=434
        keys [[0,0.1863],[7,0.1872],[10,0.1887],[11,0.1897],[12,0.1913],[13,0.1936],[15,0.2004],[16,0.2047],[17,0.212],[18,0.2266],[19,0.2577],[20,0.3007],[21,0.3359],[22,0.3527],[23,0.3593],[25,0.365],[26,0.3657],[33,0.3634],[40,0.3562],[43,0.3558],[57,0.3599],[62,0.3574],[63,0.354],[64,0.3431],[65,0.3115],[66,0.2756],[67,0.2534],[68,0.2384],[69,0.2294],[70,0.2235],[71,0.2194],[72,0.2165],[73,0.215],[74,0.2145],[75,0.2148],[77,0.2172],[81,0.2241],[85,0.2291],[91,0.2327],[94,0.2332],[98,0.2319],[100,0.2304],[102,0.2275],[103,0.2256],[104,0.2229],[106,0.2165],[108,0.2113],[110,0.2075],[111,0.206],[113,0.2048],[115,0.2045],[116,0.2036],[125,0.201],[133,0.201],[134,0.1997],[135,0.1964],[136,0.1907],[137,0.1829],[138,0.13],[139,0.0318],[140,-0.3817],[141,-0.7462],[142,-1.0535],[143,-1.3703],[144,-1.6071],[145,-1.6788],[146,-1.6933],[147,-1.6943],[148,-1.6705],[149,-1.5986],[150,-0.8326],[152,-0.3659],[153,-0.5327],[154,-0.6422],[155,-0.6902],[156,-0.4954],[157,-1.4904],[158,-2.122],[159,-2.4454],[160,-2.6057],[161,-2.6371],[162,-2.5681],[163,-2.1706],[164,-1.5792],[165,-1.294],[166,-1.1578],[167,-1.1114],[168,-1.1717],[169,-1.4872],[170,-1.9037],[171,-2.1355],[172,-2.2458],[173,-2.287],[174,-2.311],[175,-2.3217],[177,-2.323],[178,-2.3128],[179,-2.2903],[181,-2.2343],[182,-2.2144],[184,-2.1932],[210,-2.1661],[218,-2.1858],[221,-2.2066],[225,-2.2642],[227,-2.2975],[230,-2.3273],[239,-2.3552],[256,-2.365],[262,-2.3103],[266,-2.257],[269,-2.2315],[274,-2.2155],[287,-2.2312],[288,-2.2255],[292,-2.1578],[295,-2.0919],[297,-2.0568],[299,-2.0324],[308,-1.9819],[321,-1.9563],[339,-2.0089],[341,-2.0229],[343,-2.0484],[344,-2.0667],[345,-2.1171],[346,-2.2125],[348,-2.4502],[349,-2.5524],[350,-2.6193],[351,-2.651],[352,-2.6537],[353,-2.6151],[354,-2.5172],[355,-1.8224],[356,-0.7445],[357,-0.4416],[358,-0.1519],[359,-0.2076],[360,-0.2381],[361,-0.1902],[362,0.0529],[363,0.1531],[364,0.1057],[365,0.0947],[366,0.1169],[367,0.1484],[368,0.1675],[369,0.176],[370,0.1783],[371,0.1789],[373,0.1789],[374,0.181],[375,0.1848],[376,0.1904],[377,0.1931],[378,0.1937],[379,0.1924],[380,0.19],[381,0.1883],[382,0.1995],[384,0.1954],[389,0.1878],[391,0.1862],[393,0.1856],[433,0.1863]]
       }
       leg_r_pv_ik_ctrl.Location.Y {
        type=MovieSceneScriptingFloatChannel
        default=0
        thinned_from=434
        keys [[0,-1.9532],[137,-1.9532],[138,-1.4204],[139,-0.3807],[140,3.7725],[141,6.6753],[142,8.7134],[143,10.5296],[144,11.6943],[145,11.9909],[146,12.0285],[147,12.0063],[148,11.8622],[149,11.4571],[150,5.7264],[151,1.5812],[152,-1.9444],[156,-1.9434],[157,9.798],[158,13.2616],[159,14.3912],[160,14.8201],[161,14.8525],[162,14.5309],[163,12.8544],[164,9.7168],[165,7.8735],[166,6.9083],[167,6.5745],[168,7.0552],[169,9.2773],[170,11.6221],[171,12.7108],[172,13.1971],[173,13.3831],[174,13.4894],[175,13.5374],[177,13.5454],[178,13.5029],[179,13.4066],[181,13.1615],[182,13.0725],[185,12.9449],[221,12.8408],[230,12.6126],[251,12.576],[256,12.6842],[257,12.73],[258,12.8099],[259,12.9391],[260,13.1748],[261,13.4641],[262,13.6851],[263,13.8151],[264,13.8994],[267,13.9886],[284,13.9842],[287,13.909],[288,13.848],[292,13.3686],[295,12.9409],[297,12.7165],[299,12.5544],[305,12.3008],[313,12.188],[323,12.2654],[335,12.6214],[340,12.7253],[342,12.8277],[344,12.9953],[345,13.2183],[346,13.6053],[348,14.4661],[349,14.796],[350,14.9953],[351,15.0846],[352,15.0846],[353,14.949],[354,14.6147],[355,11.5786],[356,4.2005],[357,1.0582],[358,-1.9469],[361,-1.9464],[362,-1.9504],[363,-1.1398],[364,-0.4214],[365,-0.2366],[366,-0.497],[367,-0.8857],[368,-1.1211],[369,-1.2234],[370,-1.2528],[372,-1.2868],[373,-1.3158],[374,-1.3678],[375,-1.4385],[376,-1.5456],[377,-1.6314],[378,-1.6899],[379,-1.7262],[381,-1.7673],[382,-1.9453],[433,-1.9533]]
       }
       leg_r_pv_ik_ctrl.Location.Z {
        type=MovieSceneScriptingFloatChannel
        default=0
        thinned_from=434
        keys [[0,-1.7729],[18,-1.7691],[22,-1.7573],[64,-1.7581],[137,-1.7732],[138,-1.3058],[139,-0.359],[140,3.9372],[141,7.5423],[142,10.4724],[143,13.4536],[144,15.6044],[145,16.188],[146,16.2627],[147,16.2173],[148,15.9301],[149,15.1436],[150,6.2741],[151,1.5124],[152,-1.8244],[153,-1.8399],[154,-1.8501],[155,-1.8545],[156,-1.8365],[157,12.1748],[158,18.8831],[159,21.6561],[160,22.8277],[161,22.9175],[162,22.0211],[163,17.9568],[164,12.029],[165,9.1763],[166,7.8161],[167,7.3644],[168,8.0186],[169,11.3172],[170,15.4288],[171,17.6459],[172,18.7203],[173,19.1475],[174,19.3962],[175,19.5093],[176,19.5384],[177,19.5283],[178,19.4279],[179,19.2024],[181,18.64],[182,18.4396],[184,18.2243],[187,18.0752],[220,17.9498],[224,17.7677],[227,17.5287],[238,17.3025],[250,17.3089],[255,17.5004],[256,17.5642],[257,17.6644],[258,17.8402],[259,18.1271],[260,18.6607],[261,19.3352],[262,19.8671],[263,20.1878],[264,20.3998],[265,20.5215],[271,20.6965],[283,20.6534],[286,20.511],[287,20.43],[288,20.2782],[289,20.0329],[292,19.1282],[294,18.4531],[295,18.1577],[297,17.6669],[298,17.4691],[300,17.2045],[304,16.852],[309,16.6168],[318,16.5851],[323,16.7186],[334,17.4266],[339,17.6203],[341,17.7865],[342,17.9132],[343,18.0766],[344,18.2822],[345,18.7834],[346,19.6847],[348,21.8579],[349,22.765],[350,23.3368],[351,23.599],[352,23.5988],[353,23.2009],[354,22.2582],[355,15.353],[356,4.3981],[357,0.994],[358,-1.8045],[360,-1.8125],[361,-1.8081],[362,-1.7854],[363,-1.0492],[364,-0.3889],[365,-0.2153],[366,-0.4588],[367,-0.8167],[368,-1.0304],[369,-1.1226],[370,-1.1489],[372,-1.1797],[373,-1.206],[374,-1.2526],[375,-1.3159],[376,-1.4113],[377,-1.4875],[378,-1.5395],[379,-1.5719],[381,-1.6087],[382,-1.7644],[433,-1.7729]]
       }
       leg_r_pv_ik_ctrl.Rotation.X {
        type=MovieSceneScriptingFloatChannel
        default=0
        thinned_from=434
        keys [[0,2.5727],[18,2.5675],[22,2.5509],[64,2.552],[137,2.5733],[138,1.8824],[139,0.5109],[140,-5.3243],[141,-9.8442],[142,-13.3301],[143,-16.7399],[144,-19.1291],[145,-19.7684],[146,-19.85],[147,-19.8005],[148,-19.4864],[149,-18.6219],[150,-8.2877],[151,-2.1051],[152,2.6452],[153,2.6669],[154,2.6812],[155,2.6874],[156,2.6621],[157,-15.2936],[158,-22.6771],[159,-25.603],[160,-26.8219],[161,-26.915],[162,-25.984],[163,-21.6861],[164,-15.1277],[165,-11.8082],[166,-10.1785],[167,-9.6297],[168,-10.4232],[169,-14.3108],[170,-18.9374],[171,-21.3514],[172,-22.504],[173,-22.9594],[174,-23.2237],[175,-23.3438],[177,-23.364],[178,-23.2574],[179,-23.0178],[181,-22.4182],[182,-22.2039],[184,-21.9732],[187,-21.8133],[220,-21.6787],[224,-21.4831],[227,-21.2259],[249,-20.9719],[254,-21.1367],[256,-21.2644],[257,-21.3724],[258,-21.5615],[259,-21.8696],[260,-22.4407],[261,-23.159],[262,-23.7227],[263,-24.0615],[264,-24.2849],[265,-24.413],[271,-24.5969],[283,-24.5516],[286,-24.4018],[287,-24.3165],[288,-24.1565],[289,-23.8976],[292,-22.9383],[294,-22.2179],[296,-21.6252],[298,-21.1604],[301,-20.7621],[306,-20.3635],[312,-20.179],[321,-20.2687],[335,-21.1584],[339,-21.3233],[341,-21.5023],[343,-21.8141],[344,-22.0347],[345,-22.5708],[346,-23.5292],[348,-25.8135],[349,-26.7567],[350,-27.3484],[351,-27.619],[352,-27.6187],[353,-27.208],[354,-26.2305],[355,-18.8537],[356,-5.9207],[357,-1.3926],[358,2.6172],[360,2.6285],[361,2.6222],[362,2.5903],[363,1.5075],[364,0.5542],[365,0.3063],[366,0.6542],[367,1.17],[368,1.4802],[369,1.6146],[370,1.6531],[372,1.6981],[373,1.7365],[374,1.8048],[375,1.8976],[376,2.0378],[377,2.1501],[378,2.2269],[379,2.2747],[381,2.3293],[382,2.5602],[433,2.5728]]
       }
       leg_r_pv_ik_ctrl.Rotation.Y {
        type=MovieSceneScriptingFloatChannel
        default=0
        thinned_from=434
        keys [[0,0.0018],[13,0.0022],[16,0.0029],[17,0.0034],[18,0.0043],[19,0.0062],[20,0.0089],[21,0.0111],[22,0.0121],[23,0.0125],[26,0.0129],[62,0.0124],[63,0.0122],[64,0.0115],[65,0.0096],[66,0.0073],[67,0.006],[68,0.005],[70,0.0041],[73,0.0036],[100,0.0045],[113,0.0029],[134,0.0026],[136,0.0021],[137,0.0016],[138,-0.0002],[139,-0.0009],[140,0.0323],[141,0.0948],[142,0.1647],[143,0.2513],[144,0.3223],[145,0.3425],[146,0.3448],[147,0.3429],[148,0.3324],[149,0.3045],[150,0.0581],[151,-0.019],[152,-0.0332],[153,-0.0439],[154,-0.051],[155,-0.0541],[156,-0.0415],[157,0.1997],[158,0.4367],[159,0.5502],[160,0.6001],[161,0.6031],[162,0.5622],[163,0.3937],[164,0.1902],[165,0.1125],[166,0.0807],[167,0.071],[168,0.0855],[169,0.1706],[170,0.3012],[171,0.3819],[172,0.4236],[173,0.4408],[174,0.4509],[175,0.4555],[176,0.4567],[177,0.4563],[178,0.4522],[179,0.4431],[181,0.4206],[182,0.4127],[183,0.4079],[184,0.4043],[186,0.3997],[197,0.3936],[219,0.393],[221,0.3905],[224,0.3819],[227,0.3692],[229,0.3645],[234,0.3588],[248,0.3561],[251,0.3578],[254,0.3625],[256,0.3675],[257,0.3722],[258,0.3803],[259,0.3933],[260,0.4173],[261,0.4478],[262,0.4723],[263,0.4877],[264,0.4982],[265,0.5045],[266,0.5082],[268,0.5122],[279,0.5155],[283,0.5128],[285,0.5088],[287,0.502],[288,0.4952],[289,0.4845],[292,0.4459],[294,0.418],[295,0.4061],[296,0.3958],[297,0.3865],[298,0.3787],[299,0.3729],[301,0.3643],[303,0.3575],[306,0.3502],[309,0.3459],[313,0.344],[319,0.3462],[322,0.3495],[325,0.3554],[333,0.3769],[335,0.3807],[338,0.3848],[340,0.3895],[341,0.3933],[342,0.3982],[343,0.4046],[344,0.4127],[345,0.4325],[346,0.4686],[347,0.5135],[348,0.5597],[349,0.5994],[350,0.6248],[351,0.6366],[352,0.6365],[353,0.6182],[354,0.5761],[355,0.3021],[356,0.0213],[357,-0.0162],[358,-0.0195],[359,-0.023],[360,-0.025],[361,-0.0219],[362,-0.0066],[363,0.0023],[364,0.0033],[365,0.0039],[367,0.0033],[378,0.0029],[381,0.0023],[382,0.0026],[433,0.0018]]
       }
       leg_r_pv_ik_ctrl.Rotation.Z {
        type=MovieSceneScriptingFloatChannel
        default=0
        thinned_from=434
        keys [[0,-0.2751],[7,-0.2765],[10,-0.2786],[11,-0.28],[12,-0.2822],[13,-0.2856],[15,-0.2952],[16,-0.3012],[17,-0.3117],[18,-0.3323],[19,-0.3763],[20,-0.4373],[21,-0.4872],[22,-0.511],[23,-0.5204],[25,-0.5284],[26,-0.5294],[33,-0.5261],[40,-0.516],[43,-0.5153],[57,-0.5212],[62,-0.5177],[63,-0.5129],[64,-0.4974],[65,-0.4526],[66,-0.4016],[67,-0.3702],[68,-0.3491],[69,-0.3362],[70,-0.3278],[71,-0.3221],[72,-0.318],[73,-0.3158],[74,-0.3151],[75,-0.3156],[77,-0.3189],[81,-0.3288],[85,-0.3358],[91,-0.3409],[95,-0.3415],[99,-0.3389],[101,-0.336],[103,-0.3308],[104,-0.3271],[106,-0.318],[108,-0.3107],[110,-0.3052],[111,-0.3031],[113,-0.3014],[115,-0.301],[117,-0.2991],[125,-0.2961],[133,-0.296],[134,-0.2942],[135,-0.2895],[136,-0.2814],[137,-0.2703],[138,-0.191],[139,-0.0463],[140,0.5226],[141,0.9775],[142,1.3351],[143,1.6834],[144,1.9322],[145,2.0062],[146,2.0212],[147,2.0226],[148,1.9986],[149,1.9252],[150,1.0793],[151,0.7992],[152,0.5071],[153,0.7432],[154,0.8983],[155,0.9663],[156,0.6905],[157,1.8208],[158,2.456],[159,2.7643],[160,2.9151],[161,2.9461],[162,2.8861],[163,2.5153],[164,1.9203],[165,1.6129],[166,1.4609],[167,1.4082],[168,1.4762],[169,1.8215],[170,2.2522],[171,2.481],[172,2.5871],[173,2.6261],[174,2.6487],[175,2.6587],[177,2.6598],[178,2.6503],[179,2.629],[181,2.5759],[182,2.5569],[184,2.5366],[210,2.5107],[218,2.5312],[221,2.5535],[225,2.6165],[227,2.6533],[230,2.6861],[239,2.7167],[255,2.7267],[256,2.7242],[258,2.7025],[266,2.5809],[269,2.5538],[274,2.5369],[287,2.5557],[288,2.5512],[292,2.4913],[296,2.4135],[298,2.3845],[310,2.3201],[323,2.3001],[339,2.3475],[341,2.3609],[343,2.3855],[344,2.4031],[345,2.4518],[346,2.5433],[348,2.767],[349,2.8614],[350,2.9228],[351,2.9519],[352,2.9545],[353,2.9202],[354,2.8312],[355,2.1653],[356,0.9743],[357,0.5939],[358,0.2038],[359,0.2828],[360,0.326],[361,0.2581],[362,-0.0862],[363,-0.2206],[364,-0.1477],[365,-0.1308],[366,-0.1638],[367,-0.2114],[368,-0.2403],[369,-0.2534],[370,-0.2569],[371,-0.2579],[373,-0.2583],[374,-0.2618],[375,-0.2679],[376,-0.2768],[377,-0.2816],[378,-0.2831],[379,-0.2815],[380,-0.2785],[381,-0.2762],[382,-0.2939],[384,-0.2881],[389,-0.2774],[391,-0.2751],[433,-0.2753]]
       }
       leg_r_pv_ik_ctrl.Scale.X {type=MovieSceneScriptingFloatChannel, default=1, const [1,434]}
       leg_r_pv_ik_ctrl.Scale.Y {type=MovieSceneScriptingFloatChannel, default=1, const [1,434]}
       leg_r_pv_ik_ctrl.Scale.Z {type=MovieSceneScriptingFloatChannel, default=1, const [1,434]}
       upperarm_l_fk_ctrl.Location.X {type=MovieSceneScriptingFloatChannel, default=0, thinned_from=434, keys [[0,0],[433,0]]}
       upperarm_l_fk_ctrl.Location.Y {type=MovieSceneScriptingFloatChannel, default=0, thinned_from=434, keys [[0,0],[433,0]]}
       upperarm_l_fk_ctrl.Location.Z {type=MovieSceneScriptingFloatChannel, default=0, thinned_from=434, keys [[0,0],[433,0]]}
       upperarm_l_fk_ctrl.Rotation.X {
        type=MovieSceneScriptingFloatChannel
        default=0
        thinned_from=434
        keys [[0,21.4055],[11,21.4536],[12,21.5085],[13,21.6078],[14,21.767],[15,22.0112],[16,22.367],[17,22.8587],[18,23.5052],[19,25.0196],[20,26.9015],[21,27.0886],[22,25.1411],[23,22.6053],[24,20.2003],[25,18.2533],[26,16.6296],[27,15.354],[28,14.4538],[29,13.8555],[30,13.4299],[31,13.1206],[32,12.8739],[33,12.6546],[34,12.5528],[35,12.5166],[36,12.5552],[37,12.677],[38,12.8739],[44,14.3422],[45,14.5414],[47,14.7847],[53,15.0865],[55,15.2148],[56,15.224],[57,15.1914],[58,15.0695],[59,14.8472],[60,14.5044],[61,14.0206],[62,13.352],[63,12.0408],[64,10.1377],[65,8.2806],[66,7.4543],[67,8.6056],[68,11.3566],[69,14.5756],[70,16.609],[71,17.3561],[72,17.846],[73,18.1658],[74,18.357],[75,18.4779],[76,18.5323],[77,18.4815],[78,18.3332],[81,17.7311],[86,16.8844],[92,16.1281],[96,15.7602],[98,15.6989],[99,15.7107],[100,15.8142],[101,16.0456],[102,16.3789],[103,16.7887],[104,17.3058],[105,17.9293],[107,19.2157],[109,20.4295],[110,21.0051],[111,21.5191],[113,22.3595],[114,22.7125],[115,23.0111],[117,23.4757],[120,23.9527],[123,24.2624],[129,24.4265],[130,24.4251],[133,24.2371],[136,23.8576],[139,23.3282],[140,23.1096],[141,22.7755],[142,22.198],[143,21.4435],[144,20.5546],[145,19.3434],[146,18.2738],[147,18.05],[148,17.6518],[149,15.4181],[150,6.3389],[151,-0.4875],[152,-3.1538],[153,-4.2021],[154,-5.5959],[155,-6.3445],[156,-8.0345],[157,-37.5626],[158,-49.0044],[159,-54.0371],[160,-51.9532],[161,-49.2712],[162,-47.7856],[163,-47.0645],[164,-47.7257],[165,-49.3298],[166,-51.166],[167,-52.8676],[168,-53.6813],[169,-54.1975],[170,-52.6376],[171,-50.3282],[172,-48.4121],[173,-47.1902],[174,-46.718],[181,-48.447],[183,-48.6448],[218,-47.6975],[220,-47.4464],[221,-47.1822],[222,-46.721],[223,-46.0751],[224,-45.2644],[225,-43.9375],[226,-42.3682],[227,-41.507],[243,-39.9037],[258,-39.2923],[259,-39.4495],[260,-39.9011],[262,-41.3565],[263,-42.2658],[264,-43.321],[265,-44.2178],[266,-44.8814],[267,-45.4315],[269,-46.2772],[272,-47.1647],[281,-48.5239],[286,-48.9643],[288,-49.3597],[292,-50.5784],[294,-50.9641],[295,-50.9817],[297,-50.6289],[304,-49.2101],[325,-47.3461],[330,-46.4075],[334,-46.1473],[344,-46.6361],[345,-46.4511],[346,-45.7706],[347,-44.6237],[348,-43.0629],[349,-41.1388],[350,-38.9052],[351,-36.0972],[352,-32.4142],[353,-27.7206],[354,-21.8321],[355,-7.6769],[356,7.7808],[357,15.0555],[358,18.8428],[359,20.5242],[360,21.5849],[361,22.2805],[362,22.8348],[363,23.2996],[364,23.2622],[365,22.608],[366,21.7157],[367,20.6465],[368,19.3281],[371,15.2236],[372,13.9798],[373,12.9171],[374,12.0378],[375,11.3212],[376,10.7662],[377,10.3276],[378,9.9582],[379,9.6671],[380,9.484],[381,9.4676],[382,9.6879],[383,10.1254],[384,10.7356],[385,11.4671],[386,12.3415],[387,13.3714],[388,14.4618],[389,15.5209],[390,16.4396],[392,18.0309],[393,18.7641],[394,19.4268],[395,19.9971],[396,20.492],[397,20.9388],[398,21.3355],[399,21.6632],[400,21.9219],[401,22.1085],[403,22.3234],[409,22.4531],[424,21.5771],[433,21.432]]
       }
       upperarm_l_fk_ctrl.Rotation.Y {
        type=MovieSceneScriptingFloatChannel
        default=0
        thinned_from=434
        keys [[0,3.0733],[2,3.0862],[6,3.1382],[7,3.1445],[9,3.1438],[10,3.1168],[11,3.11],[12,3.12],[13,3.1602],[14,3.2414],[15,3.3419],[16,3.3791],[17,3.2931],[18,3.1425],[19,3.2325],[20,3.3292],[21,3.232],[22,3.2845],[23,3.5338],[24,3.8143],[25,4.0379],[26,4.2224],[27,4.3537],[28,4.4213],[29,4.4187],[30,4.3701],[31,4.3103],[32,4.2717],[33,4.273],[36,4.4231],[38,4.4871],[41,4.5506],[42,4.5782],[43,4.6281],[48,4.9575],[52,5.1947],[54,5.2877],[55,5.3174],[57,5.3459],[58,5.3438],[59,5.33],[60,5.3039],[61,5.2645],[62,5.1881],[63,4.9993],[64,4.6556],[65,3.8866],[66,2.9785],[67,2.3602],[68,2.1292],[69,2.2009],[70,2.1185],[71,1.918],[72,1.7769],[73,1.6695],[74,1.5984],[75,1.5473],[76,1.5014],[77,1.4498],[78,1.395],[79,1.3473],[80,1.3083],[81,1.2816],[82,1.2477],[83,1.2364],[84,1.2346],[85,1.245],[86,1.2703],[88,1.3482],[89,1.3962],[90,1.456],[91,1.5218],[92,1.6047],[93,1.6992],[95,1.8982],[96,2.0071],[97,2.122],[98,2.2457],[99,2.3793],[100,2.5542],[101,2.7831],[103,3.2999],[104,3.574],[106,4.1889],[107,4.4726],[108,4.6994],[109,4.8751],[110,5.0173],[111,5.1231],[112,5.1849],[113,5.2072],[115,5.2041],[116,5.1909],[120,5.0375],[125,4.9108],[128,4.7946],[129,4.763],[130,4.7482],[131,4.7071],[132,4.681],[133,4.6292],[134,4.6028],[135,4.546],[136,4.4477],[137,4.2832],[138,4.0287],[139,3.6614],[140,3.1921],[141,2.5649],[142,1.6616],[143,0.0831],[144,-1.5622],[145,-2.6821],[146,-3.2109],[147,-2.7341],[148,-1.3127],[149,1.2559],[150,7.5754],[151,14.5725],[152,17.5808],[153,19.1916],[154,20.0785],[155,21.2446],[156,23.0041],[157,22.1265],[158,20.4873],[159,17.5016],[160,13.9648],[161,10.7166],[162,7.9368],[163,6.1294],[164,5.8348],[165,6.6313],[166,7.9389],[167,9.3172],[168,10.0402],[169,10.3754],[170,10.3363],[171,9.8561],[172,9.1645],[173,8.5478],[174,8.2456],[175,8.2453],[176,8.2753],[177,8.3323],[180,8.5655],[181,8.6294],[182,8.6749],[185,8.6898],[194,8.5792],[207,8.5163],[208,8.4941],[209,8.454],[210,8.3847],[211,8.2794],[212,8.1298],[213,7.9276],[214,7.6683],[215,7.3027],[216,6.8175],[217,6.2419],[218,5.5978],[219,4.9192],[220,4.2307],[221,3.4513],[222,2.5772],[223,1.7538],[224,1.1236],[225,0.88],[226,1.1552],[227,1.9355],[228,3.1339],[229,4.7123],[230,6.4637],[231,8.1793],[232,9.7128],[233,11.0573],[234,12.2126],[235,13.1787],[236,13.9565],[237,14.5425],[238,14.9997],[239,15.3837],[240,15.7059],[241,15.9719],[242,16.1927],[244,16.5333],[246,16.7744],[250,17.1241],[251,17.2577],[252,17.4268],[255,18.0626],[256,18.1974],[257,18.2071],[259,18.0649],[260,17.9446],[262,17.5523],[263,17.0246],[265,15.1654],[266,14.2934],[267,13.3735],[268,12.4833],[269,11.701],[270,11.0222],[271,10.3959],[272,9.8297],[273,9.3312],[274,8.9079],[275,8.5505],[276,8.2402],[277,7.9694],[278,7.7303],[279,7.5149],[281,7.1234],[282,6.9442],[284,6.628],[285,6.4565],[286,6.2585],[287,6.0132],[288,5.6571],[290,4.7369],[291,4.3721],[292,4.1586],[293,4.065],[294,4.067],[295,4.214],[296,4.5245],[297,4.9222],[298,5.3386],[299,5.7049],[304,7.3012],[305,7.5997],[307,8.1195],[311,9.1154],[313,9.5823],[315,9.9962],[318,10.4978],[320,10.7748],[322,10.9778],[323,11.0512],[324,11.1016],[325,11.1269],[326,11.1197],[327,11.0803],[331,10.7835],[334,10.4818],[338,10.1533],[341,9.9709],[343,9.9076],[344,9.9131],[345,9.8082],[346,9.5149],[348,8.7477],[349,8.4707],[350,8.3781],[351,8.5564],[352,9.1222],[353,9.8433],[354,10.3194],[355,9.3084],[356,8.0157],[357,6.9922],[358,5.8502],[359,4.5146],[360,3.0131],[361,1.3638],[362,-0.3932],[363,-2.2593],[364,-3.113],[365,-3.3524],[366,-3.1872],[367,-2.7047],[368,-1.9809],[369,-1.06],[370,0.0156],[371,0.8099],[372,1.0289],[373,0.9688],[374,0.7555],[375,0.392],[376,-0.0838],[377,-0.508],[378,-0.8173],[379,-1.0419],[380,-1.1618],[382,-1.2738],[383,-1.2833],[384,-1.1877],[385,-0.919],[386,-0.443],[387,0.1369],[388,0.7707],[389,1.3906],[390,1.9807],[391,2.5127],[392,2.9956],[393,3.3988],[394,3.7052],[395,3.9181],[396,4.0592],[397,4.1559],[398,4.1913],[399,4.1948],[400,4.1506],[401,4.0503],[404,3.7093],[406,3.5124],[407,3.4279],[408,3.3591],[410,3.261],[412,3.1842],[414,3.1248],[416,3.0823],[417,3.0714],[418,3.053],[422,3.0174],[425,3.0331],[426,3.0337],[431,3.071],[433,3.0769]]
       }
       upperarm_l_fk_ctrl.Rotation.Z {
        type=MovieSceneScriptingFloatChannel
        default=0
        thinned_from=434
        keys [[0,-10.3858],[1,-10.3995],[3,-10.4889],[7,-10.7319],[8,-10.7675],[10,-10.7725],[11,-10.7383],[12,-10.6457],[13,-10.4676],[14,-10.1765],[15,-9.7435],[16,-8.839],[17,-7.1478],[18,-4.6758],[19,-1.2162],[20,2.1914],[21,4.0187],[22,4.2811],[23,3.7796],[24,2.6252],[25,1.1706],[26,-0.2519],[27,-1.5431],[28,-2.6037],[29,-3.4929],[30,-4.3161],[31,-5.0464],[32,-5.6537],[33,-6.1714],[34,-6.5949],[35,-6.9334],[36,-7.1662],[37,-7.2738],[38,-7.2683],[39,-7.1822],[40,-7.0484],[41,-6.8886],[43,-6.5166],[45,-6.1253],[47,-5.7986],[49,-5.5336],[50,-5.41],[52,-5.2008],[55,-4.912],[59,-4.5905],[60,-4.5013],[61,-4.3994],[62,-4.3723],[63,-4.482],[64,-4.6548],[66,-5.3976],[67,-5.6311],[68,-5.5624],[70,-4.6965],[71,-4.2886],[72,-3.9965],[73,-3.8136],[74,-3.7044],[75,-3.6277],[76,-3.5875],[77,-3.576],[78,-3.5899],[79,-3.6259],[81,-3.7303],[82,-3.7746],[83,-3.8328],[84,-3.905],[85,-3.9982],[86,-4.1194],[87,-4.2585],[89,-4.604],[90,-4.7978],[91,-5.0119],[94,-5.6983],[95,-5.936],[97,-6.4415],[98,-6.7121],[101,-7.5805],[102,-7.8963],[103,-8.2526],[104,-8.6937],[105,-9.1944],[106,-9.6708],[107,-10.0408],[108,-10.2411],[109,-10.2968],[110,-10.2582],[113,-10.0276],[115,-9.8219],[116,-9.6862],[119,-9.1744],[120,-9.0267],[123,-8.6741],[127,-8.3092],[130,-8.0639],[132,-7.919],[133,-7.8337],[134,-7.7694],[135,-7.721],[136,-7.6425],[137,-7.4868],[138,-7.2068],[139,-6.7546],[140,-5.8189],[141,-4.3839],[142,-2.8368],[143,-0.8199],[144,1.1527],[145,2.2496],[146,2.5687],[147,2.0887],[148,1.0123],[149,-2.255],[150,-12.1188],[151,-23.4275],[152,-26.348],[153,-27.256],[154,-29.5624],[156,-32.3604],[157,-26.7004],[158,-21.295],[159,-18.3466],[160,-11.4363],[161,-6.2168],[162,-2.9922],[163,-1.7888],[164,-2.0729],[165,-5],[166,-8.7136],[167,-12.1248],[168,-13.5289],[169,-13.8377],[170,-12.6375],[171,-10.092],[172,-7.2353],[173,-4.9545],[174,-3.9737],[175,-4.1208],[176,-4.4041],[177,-4.7619],[179,-5.5519],[180,-5.9247],[181,-6.2353],[182,-6.4379],[183,-6.5368],[184,-6.554],[185,-6.554],[187,-6.4872],[191,-6.2788],[192,-6.2107],[194,-6.1083],[195,-6.0766],[196,-6.0222],[199,-5.9305],[200,-5.8902],[209,-5.5976],[210,-5.5811],[211,-5.5864],[212,-5.6216],[213,-5.6947],[214,-5.8051],[215,-5.9965],[216,-6.2646],[217,-6.5921],[218,-6.9749],[219,-7.3771],[220,-7.795],[221,-8.2489],[223,-9.2168],[224,-9.614],[225,-9.7957],[226,-9.6621],[227,-9.2198],[228,-8.5232],[229,-7.5169],[230,-6.3785],[231,-5.2856],[232,-4.31],[233,-3.4125],[234,-2.6056],[235,-1.9029],[236,-1.3179],[237,-0.8638],[238,-0.4877],[239,-0.1369],[240,0.1921],[241,0.4939],[242,0.7721],[243,1.0072],[244,1.2009],[245,1.3644],[246,1.503],[247,1.6222],[248,1.7273],[249,1.8236],[250,1.9427],[251,2.1108],[252,2.3139],[253,2.5513],[254,2.799],[255,3.0213],[256,3.3735],[257,3.9118],[258,4.4933],[259,4.988],[260,5.4072],[261,5.7425],[262,5.8296],[263,5.2401],[264,4.0584],[265,2.8564],[266,1.796],[267,0.7272],[268,-0.2819],[269,-1.1645],[270,-1.9404],[271,-2.6669],[272,-3.3334],[273,-3.9287],[274,-4.4416],[275,-4.8842],[276,-5.281],[277,-5.6389],[278,-5.9645],[279,-6.2644],[280,-6.545],[281,-6.8124],[283,-7.2644],[285,-7.6799],[286,-7.9521],[287,-8.2593],[288,-8.6831],[290,-9.7481],[291,-10.2119],[292,-10.5665],[293,-10.8475],[294,-11.0742],[295,-11.2258],[296,-11.2705],[297,-11.1921],[298,-11.0122],[299,-10.7564],[300,-10.3991],[301,-9.9619],[302,-9.4799],[304,-8.6017],[306,-7.8059],[308,-7.0746],[309,-6.7301],[310,-6.4],[312,-5.8083],[314,-5.2884],[315,-5.0528],[316,-4.8492],[317,-4.6274],[318,-4.4413],[320,-4.101],[321,-3.9505],[322,-3.8095],[323,-3.6957],[324,-3.5997],[325,-3.5225],[327,-3.3903],[328,-3.3382],[329,-3.3049],[330,-3.2986],[331,-3.3476],[332,-3.4078],[334,-3.621],[335,-3.7748],[336,-3.9209],[338,-4.3134],[339,-4.5272],[340,-4.7249],[341,-4.9449],[344,-5.4815],[345,-5.7766],[347,-6.6006],[348,-6.9719],[349,-7.2156],[350,-7.2404],[351,-6.8339],[352,-6.1166],[353,-4.836],[354,-2.6908],[355,-3.6566],[356,-5.8069],[357,-4.1968],[358,-1.6779],[359,0.7427],[360,3.3986],[361,6.2212],[362,8.9064],[363,11.3731],[364,11.9427],[365,11.0972],[366,9.8922],[367,8.4939],[368,6.7543],[369,4.5504],[370,2.0842],[371,-0.4046],[372,-2.7672],[373,-5.0608],[374,-7.1484],[375,-8.9717],[376,-10.5866],[377,-12.057],[378,-13.4329],[379,-14.6837],[380,-15.7579],[381,-16.6394],[382,-17.3325],[383,-17.8533],[384,-18.2299],[385,-18.4952],[386,-18.6316],[387,-18.6032],[388,-18.4352],[389,-18.1549],[391,-17.4891],[392,-17.1206],[393,-16.6703],[394,-16.1006],[396,-14.86],[397,-14.2804],[399,-13.1783],[400,-12.6482],[401,-12.1622],[402,-11.7338],[403,-11.3707],[405,-10.7901],[406,-10.5507],[407,-10.3447],[408,-10.1776],[409,-10.055],[411,-9.9325],[413,-9.8828],[416,-9.8875],[422,-10.0405],[429,-10.3134],[431,-10.3566],[433,-10.3717]]
       }
       upperarm_l_fk_ctrl.Scale.X {type=MovieSceneScriptingFloatChannel, default=1, const [1,434]}
       upperarm_l_fk_ctrl.Scale.Y {type=MovieSceneScriptingFloatChannel, default=1, const [1,434]}
       upperarm_l_fk_ctrl.Scale.Z {type=MovieSceneScriptingFloatChannel, default=1, const [1,434]}
       lowerarm_l_fk_ctrl.Location.X {type=MovieSceneScriptingFloatChannel, default=0, thinned_from=434, keys [[0,0],[433,0]]}
       lowerarm_l_fk_ctrl.Location.Y {type=MovieSceneScriptingFloatChannel, default=0, thinned_from=434, keys [[0,0],[433,0]]}
       lowerarm_l_fk_ctrl.Location.Z {type=MovieSceneScriptingFloatChannel, default=0, thinned_from=434, keys [[0,0],[184,0],[185,-0.0029],[186,0],[433,0]]}
       lowerarm_l_fk_ctrl.Rotation.X {
        type=MovieSceneScriptingFloatChannel
        default=0
        thinned_from=434
        keys [[0,-0.0512],[9,-0.0512],[10,-0.0507],[11,-0.0489],[12,-0.0449],[13,-0.0382],[14,-0.0281],[15,-0.0089],[16,0.0165],[17,0.0373],[18,0.0458],[19,0.0401],[20,0.0276],[21,0.0155],[22,0.0063],[23,-0.0019],[24,-0.0079],[25,-0.0109],[26,-0.012],[27,-0.012],[28,-0.0081],[29,0.0016],[30,0.0136],[31,0.0249],[32,0.036],[33,0.0477],[34,0.0582],[35,0.0659],[36,0.0711],[37,0.0751],[38,0.078],[39,0.0802],[40,0.0817],[41,0.0827],[42,0.0832],[44,0.0834],[45,0.0831],[46,0.0821],[47,0.0806],[51,0.073],[55,0.0638],[57,0.0597],[59,0.0562],[61,0.0533],[62,0.0521],[63,0.0511],[64,0.0505],[65,0.0502],[66,0.0508],[67,0.0518],[68,0.0523],[69,0.0516],[70,0.0488],[71,0.0432],[72,0.0331],[74,0.0059],[75,-0.0067],[76,-0.019],[77,-0.029],[78,-0.0363],[79,-0.0421],[80,-0.0464],[81,-0.0492],[82,-0.0511],[83,-0.0528],[84,-0.0542],[85,-0.0554],[87,-0.0573],[89,-0.0584],[90,-0.0587],[94,-0.0587],[96,-0.0576],[98,-0.0558],[99,-0.0547],[100,-0.0531],[101,-0.0509],[102,-0.0482],[104,-0.0424],[105,-0.0399],[106,-0.0378],[107,-0.0361],[108,-0.0349],[109,-0.0341],[110,-0.0336],[112,-0.0336],[113,-0.0344],[114,-0.0355],[115,-0.0369],[116,-0.0385],[117,-0.0404],[119,-0.0448],[121,-0.0496],[125,-0.0604],[127,-0.0652],[130,-0.0714],[133,-0.0769],[134,-0.0791],[135,-0.0828],[136,-0.0886],[137,-0.0972],[138,-0.1093],[139,-0.1258],[140,-0.1537],[141,-0.1984],[142,-0.2586],[143,-0.3546],[144,-0.4706],[145,-0.5614],[146,-0.6612],[147,-0.83],[148,-1.0242],[149,-1.1196],[150,-0.8355],[151,-0.0521],[152,0.142],[153,0.0594],[154,0.0259],[155,0.0121],[156,0.01],[184,0.01],[185,0.0048],[186,0.01],[214,0.01],[215,0.0106],[216,0.0121],[217,0.0143],[218,0.0171],[220,0.0232],[221,0.027],[223,0.0363],[224,0.0405],[225,0.0441],[226,0.0472],[227,0.0498],[228,0.052],[229,0.0539],[230,0.0554],[231,0.0564],[233,0.0574],[234,0.0576],[235,0.0576],[236,0.0579],[239,0.0579],[240,0.0583],[242,0.0583],[243,0.0585],[252,0.0585],[256,0.0574],[257,0.0569],[258,0.0557],[259,0.0531],[260,0.0452],[261,0.0324],[262,0.0216],[263,0.0164],[264,0.0141],[265,0.013],[266,0.0122],[267,0.0118],[268,0.0118],[272,0.0107],[273,0.0107],[274,0.0104],[276,0.0104],[277,0.0101],[344,0.0101],[345,0.0122],[346,0.0179],[347,0.0258],[349,0.0435],[350,0.0512],[351,0.0567],[352,0.0617],[353,0.0686],[354,0.0796],[355,0.101],[356,0.1226],[357,0.136],[358,0.1421],[359,0.1354],[360,0.1153],[361,0.0788],[362,0.035],[363,0.0053],[364,-0.0215],[365,-0.0918],[366,-0.1667],[367,-0.2186],[368,-0.259],[369,-0.2752],[370,-0.2704],[371,-0.2585],[372,-0.2429],[373,-0.2257],[374,-0.2073],[376,-0.169],[377,-0.1504],[378,-0.1314],[379,-0.1121],[380,-0.0944],[381,-0.0782],[382,-0.0628],[383,-0.049],[384,-0.0377],[385,-0.0288],[386,-0.0217],[387,-0.0163],[388,-0.0124],[389,-0.0102],[390,-0.0093],[391,-0.0093],[392,-0.0097],[393,-0.0108],[394,-0.0126],[395,-0.0149],[396,-0.0175],[399,-0.0263],[400,-0.029],[405,-0.0402],[407,-0.0442],[408,-0.0459],[409,-0.0472],[411,-0.0493],[413,-0.051],[415,-0.0522],[417,-0.0529],[418,-0.0529],[419,-0.0533],[423,-0.0533],[424,-0.0529],[425,-0.0529],[426,-0.0524],[427,-0.0524],[428,-0.0519],[429,-0.0519],[430,-0.0515],[432,-0.0515],[433,-0.0512]]
       }
       lowerarm_l_fk_ctrl.Rotation.Y {
        type=MovieSceneScriptingFloatChannel
        default=0
        thinned_from=434
        keys [[0,-0.1825],[9,-0.1825],[10,-0.181],[11,-0.1748],[12,-0.1616],[13,-0.1388],[14,-0.104],[15,-0.0341],[16,0.0659],[17,0.1558],[18,0.1949],[19,0.1686],[20,0.1127],[21,0.0614],[22,0.0245],[23,-0.0078],[24,-0.0303],[25,-0.0415],[26,-0.0457],[27,-0.0457],[28,-0.0313],[29,0.0059],[30,0.0539],[31,0.101],[32,0.1496],[33,0.204],[34,0.2558],[35,0.2961],[36,0.3245],[37,0.3468],[38,0.3639],[39,0.3766],[40,0.3859],[41,0.3919],[42,0.395],[43,0.3961],[44,0.3961],[45,0.394],[46,0.3879],[47,0.379],[51,0.3353],[52,0.323],[54,0.2971],[55,0.2847],[56,0.2735],[57,0.2634],[58,0.2541],[59,0.2456],[60,0.238],[61,0.2312],[62,0.2253],[63,0.2205],[64,0.2174],[65,0.2162],[66,0.2189],[67,0.2239],[68,0.2265],[69,0.2228],[70,0.2093],[71,0.1828],[72,0.1368],[73,0.0781],[74,0.0228],[75,-0.0261],[76,-0.0715],[77,-0.1071],[78,-0.1325],[79,-0.1522],[80,-0.1666],[81,-0.1758],[82,-0.1821],[83,-0.1876],[84,-0.1923],[85,-0.1963],[86,-0.1997],[87,-0.2023],[88,-0.2045],[89,-0.206],[90,-0.2071],[94,-0.2071],[96,-0.2034],[97,-0.2008],[98,-0.1977],[99,-0.1939],[100,-0.1887],[101,-0.1815],[102,-0.1727],[104,-0.1534],[105,-0.1448],[106,-0.1375],[107,-0.1317],[108,-0.1275],[109,-0.1248],[110,-0.1233],[112,-0.1233],[113,-0.1261],[114,-0.1297],[115,-0.1344],[116,-0.14],[117,-0.1465],[118,-0.1536],[120,-0.1689],[121,-0.1771],[124,-0.204],[125,-0.2126],[126,-0.2204],[128,-0.2344],[130,-0.2473],[132,-0.2589],[133,-0.2643],[134,-0.2711],[135,-0.2824],[136,-0.2998],[137,-0.3252],[138,-0.3602],[139,-0.4064],[140,-0.4813],[141,-0.5935],[142,-0.7321],[143,-0.9303],[144,-1.1412],[146,-1.439],[148,-1.8933],[149,-1.9946],[150,-1.6723],[151,-0.1855],[152,1.153],[153,1.9738],[154,2.1308],[155,2.1884],[156,2.1966],[184,2.1966],[185,2.1903],[186,2.1966],[215,2.1944],[217,2.1792],[220,2.1422],[224,2.0655],[226,2.0342],[229,2.0018],[231,1.9892],[256,1.9841],[258,1.9928],[259,2.0056],[260,2.0439],[261,2.1025],[262,2.149],[263,2.1705],[264,2.1801],[267,2.1893],[344,2.1962],[345,2.1878],[346,2.1645],[347,2.131],[350,2.0152],[352,1.9622],[353,1.9253],[354,1.8629],[355,1.723],[356,1.5349],[357,1.3521],[358,1.1164],[360,0.6289],[361,0.3686],[362,0.1454],[363,0.0206],[364,-0.0806],[365,-0.3093],[366,-0.515],[367,-0.6414],[368,-0.733],[369,-0.7683],[370,-0.758],[371,-0.7319],[372,-0.6972],[373,-0.658],[374,-0.6148],[375,-0.5686],[377,-0.4729],[378,-0.4219],[379,-0.3681],[380,-0.317],[382,-0.2201],[383,-0.1754],[384,-0.1372],[385,-0.1064],[386,-0.0813],[387,-0.0617],[388,-0.0473],[389,-0.039],[390,-0.0359],[391,-0.0359],[392,-0.037],[393,-0.0414],[394,-0.048],[395,-0.0564],[396,-0.0661],[397,-0.0765],[398,-0.0871],[399,-0.0974],[400,-0.107],[401,-0.1153],[404,-0.1383],[405,-0.1458],[406,-0.1529],[407,-0.1593],[408,-0.1649],[409,-0.1694],[410,-0.1731],[411,-0.1764],[412,-0.1793],[413,-0.1818],[415,-0.1857],[417,-0.1882],[418,-0.1882],[419,-0.1894],[423,-0.1894],[424,-0.188],[425,-0.188],[426,-0.1864],[427,-0.1864],[428,-0.1848],[429,-0.1848],[430,-0.1834],[432,-0.1834],[433,-0.1825]]
       }
       lowerarm_l_fk_ctrl.Rotation.Z {
        type=MovieSceneScriptingFloatChannel
        default=0
        thinned_from=434
        keys [[0,2.4319],[9,2.4319],[10,2.4117],[11,2.3295],[12,2.1525],[13,1.8483],[14,1.3843],[15,0.4556],[16,-0.8662],[17,-2.0492],[18,-2.5617],[19,-2.2168],[20,-1.4821],[21,-0.8076],[22,-0.3205],[23,0.1064],[24,0.405],[25,0.554],[26,0.6089],[27,0.6089],[28,0.4182],[29,-0.0749],[30,-0.7087],[31,-1.3291],[32,-1.9671],[33,-2.6807],[34,-3.3571],[35,-3.8833],[36,-4.2532],[37,-4.5435],[38,-4.7656],[39,-4.9309],[40,-5.0508],[41,-5.1287],[42,-5.1687],[43,-5.1834],[44,-5.1834],[45,-5.156],[46,-5.0765],[47,-4.961],[51,-4.393],[55,-3.7349],[56,-3.5884],[57,-3.457],[58,-3.3356],[59,-3.2248],[60,-3.1248],[61,-3.0359],[62,-2.9586],[63,-2.8965],[64,-2.8553],[65,-2.8403],[66,-2.8752],[67,-2.9401],[68,-2.975],[69,-2.9257],[70,-2.7491],[71,-2.4024],[72,-1.8001],[73,-1.0276],[74,-0.2976],[75,0.3485],[76,0.9522],[77,1.4247],[78,1.7635],[79,2.0276],[80,2.2199],[81,2.3432],[82,2.4274],[83,2.5008],[84,2.5641],[85,2.6177],[86,2.6622],[87,2.6983],[88,2.7265],[89,2.7474],[90,2.7616],[94,2.7616],[96,2.7122],[97,2.6778],[98,2.6355],[99,2.5849],[100,2.5149],[101,2.4182],[102,2.3015],[104,2.0425],[105,1.9283],[106,1.8311],[107,1.7535],[108,1.6973],[109,1.6607],[110,1.6409],[112,1.6409],[113,1.6779],[114,1.7267],[115,1.7893],[116,1.8643],[117,1.9507],[118,2.0457],[120,2.2506],[121,2.3607],[124,2.7205],[125,2.8354],[126,2.9404],[128,3.129],[130,3.3015],[132,3.4581],[133,3.5306],[134,3.623],[135,3.775],[136,4.0105],[137,4.3533],[138,4.8274],[139,5.4567],[140,6.4801],[141,8.0288],[142,9.9646],[143,12.7873],[144,15.8739],[145,18.1049],[146,20.4084],[147,24.0311],[148,27.8723],[149,29.6582],[150,24.1438],[151,2.4732],[152,-14.9196],[153,-25.5054],[154,-27.5642],[155,-28.3228],[156,-28.4312],[214,-28.4312],[216,-28.3219],[220,-27.7142],[224,-26.706],[226,-26.2962],[229,-25.8721],[231,-25.7063],[256,-25.6405],[258,-25.7535],[259,-25.922],[260,-26.4223],[261,-27.1914],[262,-27.8038],[263,-28.087],[264,-28.2133],[267,-28.3353],[344,-28.4257],[345,-28.3153],[346,-28.0073],[347,-27.567],[350,-26.0475],[352,-25.3549],[353,-24.8736],[354,-24.0614],[355,-22.2476],[356,-19.8206],[357,-17.4723],[358,-14.4517],[359,-11.3342],[360,-8.1931],[361,-4.8261],[362,-1.9117],[363,-0.269],[364,1.072],[365,4.1386],[366,6.9436],[367,8.695],[368,9.9773],[369,10.4749],[370,10.3297],[371,9.962],[372,9.4742],[373,8.9266],[374,8.3237],[375,7.6835],[377,6.3647],[378,5.6677],[379,4.9349],[380,4.2418],[381,3.5845],[382,2.9361],[383,2.3368],[384,1.8272],[385,1.4159],[386,1.0821],[387,0.8218],[388,0.6308],[389,0.5198],[390,0.479],[391,0.479],[392,0.4939],[393,0.5517],[394,0.6398],[395,0.7514],[396,0.8797],[397,1.0178],[398,1.1591],[399,1.2968],[400,1.424],[401,1.5339],[404,1.8419],[405,1.942],[406,2.0362],[407,2.1219],[408,2.1962],[409,2.2565],[410,2.306],[411,2.35],[412,2.3888],[413,2.4225],[415,2.475],[417,2.5089],[418,2.5089],[419,2.5252],[423,2.5252],[424,2.5056],[425,2.5056],[426,2.4849],[427,2.4849],[428,2.4632],[429,2.4632],[430,2.4446],[432,2.4446],[433,2.4316]]
       }
       lowerarm_l_fk_ctrl.Scale.X {type=MovieSceneScriptingFloatChannel, default=1, const [1,434]}
       lowerarm_l_fk_ctrl.Scale.Y {type=MovieSceneScriptingFloatChannel, default=1, const [1,434]}
       lowerarm_l_fk_ctrl.Scale.Z {type=MovieSceneScriptingFloatChannel, default=1, const [1,434]}
       hand_l_fk_ctrl.Location.X {type=MovieSceneScriptingFloatChannel, default=0, thinned_from=434, keys [[0,0],[433,0]]}
       hand_l_fk_ctrl.Location.Y {type=MovieSceneScriptingFloatChannel, default=0, thinned_from=434, keys [[0,0],[237,0],[238,0.0003],[239,0.0003],[240,0],[433,0]]}
       hand_l_fk_ctrl.Location.Z {type=MovieSceneScriptingFloatChannel, default=0, thinned_from=434, keys [[0,0],[237,0],[238,0.0045],[239,0.0045],[240,0],[433,0]]}
       hand_l_fk_ctrl.Rotation.X {
        type=MovieSceneScriptingFloatChannel
        default=0
        thinned_from=434
        keys [[0,0.3571],[13,0.3576],[16,0.3589],[17,0.3601],[18,0.3626],[19,0.3683],[20,0.3751],[21,0.3793],[22,0.3807],[64,0.3831],[65,0.3792],[66,0.3721],[67,0.366],[68,0.3612],[69,0.3592],[133,0.3572],[134,0.4371],[135,0.6672],[136,1.0328],[137,1.5195],[138,2.1129],[139,2.7986],[140,3.7012],[141,5.0358],[142,6.9268],[143,10.5624],[144,13.91],[145,14.7562],[146,14.9752],[147,15.2912],[148,15.5052],[149,15.5779],[150,15.5779],[151,13.745],[152,11.2723],[153,9.1372],[154,7.863],[155,6.9183],[156,6.0177],[157,5.0908],[237,5.0908],[238,5.0797],[239,5.0797],[240,5.0908],[344,5.0908],[345,5.0611],[346,4.9822],[347,4.8694],[349,4.6027],[350,4.4792],[351,4.4079],[352,4.3588],[353,4.2415],[354,3.9633],[355,2.8802],[356,1.6003],[357,1.2501],[358,1.1219],[359,0.9521],[360,0.8076],[361,0.6143],[362,0.4511],[363,0.3986],[364,0.3848],[365,0.3799],[367,0.3749],[369,0.3716],[370,0.3716],[371,0.3702],[372,0.3702],[374,0.3643],[376,0.3602],[377,0.3587],[433,0.3587]]
       }
       hand_l_fk_ctrl.Rotation.Y {
        type=MovieSceneScriptingFloatChannel
        default=0
        thinned_from=434
        keys [[0,4.5273],[16,4.5129],[17,4.5032],[18,4.4824],[19,4.4325],[20,4.3637],[21,4.3136],[22,4.2953],[64,4.2617],[65,4.3146],[66,4.3953],[67,4.4531],[68,4.4943],[69,4.5106],[133,4.5262],[134,4.6247],[135,4.9046],[136,5.3486],[137,5.9383],[138,6.6552],[139,7.4808],[140,8.5484],[141,10.123],[142,12.3818],[143,16.7629],[144,20.7319],[145,21.8019],[146,21.973],[147,21.8855],[148,21.6513],[149,21.5168],[150,21.5168],[151,14.9373],[152,5.8267],[153,-2.3803],[154,-7.5274],[156,-15.5408],[157,-19.9649],[344,-19.9649],[345,-19.7832],[346,-19.3013],[347,-18.6139],[349,-16.9994],[350,-16.2578],[351,-15.8312],[352,-15.539],[353,-14.8433],[354,-13.214],[355,-7.2062],[356,-0.8756],[357,0.7153],[358,1.2836],[359,2.0252],[360,2.6466],[361,3.4635],[362,4.1418],[363,4.3573],[364,4.4141],[365,4.4341],[369,4.468],[372,4.4736],[377,4.5205],[433,4.5205]]
       }
       hand_l_fk_ctrl.Rotation.Z {
        type=MovieSceneScriptingFloatChannel
        default=0
        thinned_from=434
        keys [[0,-8.5077],[11,-8.5077],[13,-8.4566],[16,-8.3156],[17,-8.1862],[18,-7.909],[19,-7.2438],[20,-6.3292],[21,-5.6641],[22,-5.4217],[23,-5.3454],[24,-5.29],[25,-5.2018],[26,-5.1249],[27,-5.0604],[28,-5.0095],[29,-4.9747],[30,-4.9547],[32,-4.9547],[33,-4.9362],[44,-4.9362],[45,-4.919],[56,-4.919],[57,-4.9459],[58,-4.9459],[59,-4.9765],[64,-4.9765],[65,-5.678],[66,-6.7493],[67,-7.5186],[68,-8.0681],[69,-8.2849],[70,-8.3551],[72,-8.4459],[73,-8.4665],[74,-8.4665],[75,-8.4924],[133,-8.4924],[134,-8.5593],[135,-8.7074],[136,-8.9425],[137,-9.2546],[138,-9.634],[139,-10.071],[140,-10.4823],[141,-11.1613],[142,-12.6623],[143,-16.6869],[144,-20.5417],[145,-21.5842],[146,-21.719],[147,-19.7386],[148,-16.118],[149,-14.1764],[150,-14.1764],[151,-21.7143],[152,-32.3827],[153,-42.0779],[154,-48.1827],[156,-57.743],[157,-63.0623],[344,-63.0623],[345,-62.6727],[346,-61.6376],[347,-60.1576],[349,-56.6646],[350,-55.0521],[351,-54.1223],[352,-53.4845],[353,-51.963],[354,-48.3839],[355,-35.0153],[356,-20.7303],[357,-17.1277],[358,-15.8412],[359,-14.1625],[360,-12.7566],[361,-10.9097],[362,-9.3779],[363,-8.8914],[364,-8.7631],[365,-8.7181],[369,-8.6415],[372,-8.6288],[377,-8.5232],[433,-8.5232]]
       }
       hand_l_fk_ctrl.Scale.X {type=MovieSceneScriptingFloatChannel, default=1, const [1,434]}
       hand_l_fk_ctrl.Scale.Y {type=MovieSceneScriptingFloatChannel, default=1, const [1,434]}
       hand_l_fk_ctrl.Scale.Z {type=MovieSceneScriptingFloatChannel, default=1, const [1,434]}
       thumb_01_l_ctrl.Location.X {type=MovieSceneScriptingFloatChannel, default=0, thinned_from=434, keys [[0,0],[433,0]]}
       thumb_01_l_ctrl.Location.Y {type=MovieSceneScriptingFloatChannel, default=0, thinned_from=434, keys [[0,0],[433,0]]}
       thumb_01_l_ctrl.Location.Z {type=MovieSceneScriptingFloatChannel, default=0, thinned_from=434, keys [[0,0],[433,0]]}
       thumb_01_l_ctrl.Rotation.X {
        type=MovieSceneScriptingFloatChannel
        default=0
        thinned_from=434
        keys [[0,5.0222],[150,5.0222],[151,6.0219],[152,6.2239],[154,6.3073],[155,6.3168],[156,6.3551],[157,6.555],[158,6.605],[361,6.6205],[362,6.0968],[363,5.422],[364,5.2249],[365,5.161],[367,5.076],[368,5.0548],[433,5.0227]]
       }
       thumb_01_l_ctrl.Rotation.Y {
        type=MovieSceneScriptingFloatChannel
        default=0
        thinned_from=434
        keys [[0,16.5344],[150,16.5344],[151,3.7616],[152,1.1804],[153,0.5684],[154,0.115],[155,-0.0072],[156,-0.496],[157,-3.0506],[158,-3.6892],[159,-3.8344],[160,-3.8666],[161,-3.8878],[361,-3.8878],[362,2.8036],[363,11.4253],[364,13.9444],[365,14.7604],[366,15.2571],[367,15.847],[368,16.1175],[369,16.2151],[371,16.2687],[373,16.2949],[377,16.4746],[433,16.5278]]
       }
       thumb_01_l_ctrl.Rotation.Z {
        type=MovieSceneScriptingFloatChannel
        default=0
        thinned_from=434
        keys [[0,8.3802],[134,8.3802],[135,8.3104],[136,8.1448],[137,7.8223],[138,7.2907],[139,6.4975],[140,4.6496],[141,1.8604],[142,-0.5627],[143,-1.8841],[144,-2.5239],[145,-2.7617],[147,-2.8783],[148,-2.9081],[150,-2.9081],[151,7.6585],[152,9.7954],[153,10.3021],[154,10.6775],[155,10.7786],[156,11.1833],[157,13.2982],[158,13.8269],[159,13.9472],[161,13.9914],[361,13.9914],[362,12.1581],[363,9.7859],[364,9.0928],[365,8.8683],[366,8.7316],[367,8.5693],[368,8.4949],[370,8.457],[433,8.382]]
       }
       thumb_01_l_ctrl.Scale.X {type=MovieSceneScriptingFloatChannel, default=1, const [1,434]}
       thumb_01_l_ctrl.Scale.Y {type=MovieSceneScriptingFloatChannel, default=1, const [1,434]}
       thumb_01_l_ctrl.Scale.Z {type=MovieSceneScriptingFloatChannel, default=1, const [1,434]}
       thumb_02_l_ctrl.Location.X {type=MovieSceneScriptingFloatChannel, default=0, thinned_from=434, keys [[0,0],[162,0],[163,-0.0006],[361,-0.0006],[362,0],[433,0]]}
       thumb_02_l_ctrl.Location.Y {type=MovieSceneScriptingFloatChannel, default=0, thinned_from=434, keys [[0,0],[162,0],[163,0.0008],[361,0.0008],[362,0],[433,0]]}
       thumb_02_l_ctrl.Location.Z {type=MovieSceneScriptingFloatChannel, default=0, thinned_from=434, keys [[0,0],[162,0],[163,-0.001],[361,-0.001],[362,0],[433,0]]}
       thumb_02_l_ctrl.Rotation.X {
        type=MovieSceneScriptingFloatChannel
        default=0
        thinned_from=434
        keys [[0,0],[150,0],[151,9.0997],[152,10.9386],[153,11.3746],[154,11.6976],[155,11.7847],[156,12.1329],[157,13.9529],[158,14.4078],[159,14.5113],[361,14.5584],[362,9.7822],[363,3.6398],[364,1.8452],[365,1.2638],[366,0.91],[367,0.4897],[368,0.297],[369,0.2275],[370,0.1989],[371,0.1893],[372,0.1893],[373,0.1706],[374,0.1414],[375,0.1051],[376,0.0696],[377,0.0427],[378,0.0273],[379,0.0192],[380,0.0192],[381,0.0107],[382,0.0107],[383,0],[433,0]]
       }
       thumb_02_l_ctrl.Rotation.Y {
        type=MovieSceneScriptingFloatChannel
        default=0
        thinned_from=434
        keys [[0,0],[150,0],[151,15.948],[152,19.1709],[153,19.9351],[154,20.5011],[155,20.6537],[156,21.2641],[157,24.4536],[158,25.251],[159,25.4324],[361,25.5028],[362,17.1441],[363,6.3792],[364,3.2339],[365,2.215],[366,1.5948],[367,0.8583],[368,0.5205],[369,0.3987],[370,0.3487],[371,0.3317],[372,0.3317],[373,0.299],[374,0.2478],[375,0.1842],[376,0.1219],[377,0.0748],[378,0.0479],[379,0.0336],[380,0.0336],[381,0.0187],[382,0.0187],[383,0],[433,0]]
       }
       thumb_02_l_ctrl.Rotation.Z {
        type=MovieSceneScriptingFloatChannel
        default=0
        thinned_from=434
        keys [[0,0],[150,0],[151,19.785],[152,23.7832],[153,24.7312],[154,25.4334],[155,25.6228],[156,26.38],[157,30.3369],[158,31.3262],[159,31.5511],[361,31.6607],[362,21.2688],[363,7.9139],[364,4.0119],[365,2.7479],[366,1.9785],[367,1.0648],[368,0.6457],[369,0.4946],[370,0.4325],[371,0.4115],[372,0.4115],[373,0.371],[374,0.3074],[375,0.2285],[376,0.1512],[377,0.0927],[378,0.0594],[379,0.0417],[380,0.0417],[381,0.0232],[382,0.0232],[383,0],[433,0]]
       }
       thumb_02_l_ctrl.Scale.X {type=MovieSceneScriptingFloatChannel, default=1, const [1,434]}
       thumb_02_l_ctrl.Scale.Y {type=MovieSceneScriptingFloatChannel, default=1, const [1,434]}
       thumb_02_l_ctrl.Scale.Z {type=MovieSceneScriptingFloatChannel, default=1, const [1,434]}
       thumb_03_l_ctrl.Location.X {type=MovieSceneScriptingFloatChannel, default=0, thinned_from=434, keys [[0,0],[433,0]]}
       thumb_03_l_ctrl.Location.Y {type=MovieSceneScriptingFloatChannel, default=0, thinned_from=434, keys [[0,0],[433,0]]}
       thumb_03_l_ctrl.Location.Z {type=MovieSceneScriptingFloatChannel, default=0, thinned_from=434, keys [[0,0],[433,0]]}
       thumb_03_l_ctrl.Rotation.X {type=MovieSceneScriptingFloatChannel, default=0, thinned_from=434, keys [[0,0],[433,0]]}
       thumb_03_l_ctrl.Rotation.Y {type=MovieSceneScriptingFloatChannel, default=0, thinned_from=434, keys [[0,0],[433,0]]}
       thumb_03_l_ctrl.Rotation.Z {
        type=MovieSceneScriptingFloatChannel
        default=0
        thinned_from=434
        keys [[0,0],[150,0],[151,12.3496],[152,14.8453],[153,15.437],[154,15.8754],[155,15.9935],[156,16.4662],[157,18.9361],[158,19.5535],[159,19.694],[361,19.7572],[362,13.2758],[363,4.9398],[364,2.5042],[365,1.7152],[366,1.235],[367,0.6646],[368,0.4031],[369,0.3087],[370,0.27],[371,0.2569],[372,0.2569],[373,0.2316],[374,0.1919],[375,0.1426],[376,0.0944],[377,0.0579],[378,0.0371],[379,0.0371],[380,0.02],[381,0.02],[382,0],[433,0]]
       }
       thumb_03_l_ctrl.Scale.X {type=MovieSceneScriptingFloatChannel, default=1, const [1,434]}
       thumb_03_l_ctrl.Scale.Y {type=MovieSceneScriptingFloatChannel, default=1, const [1,434]}
       thumb_03_l_ctrl.Scale.Z {type=MovieSceneScriptingFloatChannel, default=1, const [1,434]}
       index_metacarpal_l_ctrl.Location.X {type=MovieSceneScriptingFloatChannel, default=0, thinned_from=434, keys [[0,0],[433,0]]}
       index_metacarpal_l_ctrl.Location.Y {type=MovieSceneScriptingFloatChannel, default=0, thinned_from=434, keys [[0,0],[433,0]]}
       index_metacarpal_l_ctrl.Location.Z {type=MovieSceneScriptingFloatChannel, default=0, const [0,434]}
       index_metacarpal_l_ctrl.Rotation.X {
        type=MovieSceneScriptingFloatChannel
        default=0
        thinned_from=434
        keys [[0,-1.732],[150,-1.732],[151,-0.8599],[152,-0.6837],[153,-0.6419],[154,-0.611],[155,-0.6026],[156,-0.5692],[157,-0.3948],[158,-0.3512],[159,-0.3413],[160,-0.3413],[161,-0.3377],[361,-0.3377],[362,-0.7945],[363,-1.3832],[364,-1.5552],[365,-1.6109],[366,-1.6448],[367,-1.6851],[368,-1.7036],[369,-1.7102],[370,-1.7102],[371,-1.7139],[373,-1.7139],[376,-1.7254],[377,-1.7254],[378,-1.7294],[433,-1.7294]]
       }
       index_metacarpal_l_ctrl.Rotation.Y {
        type=MovieSceneScriptingFloatChannel
        default=0
        thinned_from=434
        keys [[0,0.7486],[150,0.7486],[151,-2.6346],[152,-3.3183],[153,-3.4804],[154,-3.6005],[155,-3.6328],[156,-3.7623],[157,-4.4389],[158,-4.6081],[159,-4.6466],[160,-4.6466],[161,-4.6607],[361,-4.6607],[362,-2.8883],[363,-0.6047],[364,0.0625],[365,0.2787],[366,0.4103],[367,0.5665],[368,0.6382],[369,0.664],[370,0.664],[371,0.6782],[373,0.6782],[374,0.696],[376,0.7227],[377,0.7227],[378,0.7384],[433,0.7384]]
       }
       index_metacarpal_l_ctrl.Rotation.Z {
        type=MovieSceneScriptingFloatChannel
        default=0
        thinned_from=434
        keys [[0,-0.3856],[150,-0.3856],[151,-0.8375],[152,-0.9288],[153,-0.9504],[154,-0.9665],[155,-0.9708],[156,-0.9881],[157,-1.0785],[158,-1.1011],[159,-1.1062],[361,-1.1081],[362,-0.8714],[363,-0.5663],[364,-0.4772],[365,-0.4483],[366,-0.4308],[367,-0.4099],[368,-0.4003],[369,-0.3969],[370,-0.3969],[371,-0.395],[373,-0.395],[376,-0.389],[377,-0.389],[378,-0.3869],[433,-0.3869]]
       }
       index_metacarpal_l_ctrl.Scale.X {type=MovieSceneScriptingFloatChannel, default=1, const [1,434]}
       index_metacarpal_l_ctrl.Scale.Y {type=MovieSceneScriptingFloatChannel, default=1, const [1,434]}
       index_metacarpal_l_ctrl.Scale.Z {type=MovieSceneScriptingFloatChannel, default=1, const [1,434]}
       index_01_l_ctrl.Location.X {type=MovieSceneScriptingFloatChannel, default=0, thinned_from=434, keys [[0,0],[433,0]]}
       index_01_l_ctrl.Location.Y {type=MovieSceneScriptingFloatChannel, default=0, thinned_from=434, keys [[0,0],[433,0]]}
       index_01_l_ctrl.Location.Z {type=MovieSceneScriptingFloatChannel, default=0, thinned_from=434, keys [[0,0],[433,0]]}
       index_01_l_ctrl.Rotation.X {
        type=MovieSceneScriptingFloatChannel
        default=0
        thinned_from=434
        keys [[0,-3.6768],[133,-3.6768],[134,-3.6676],[135,-3.6027],[136,-3.4267],[137,-3.0839],[138,-2.5187],[139,-1.6756],[140,0.2886],[141,3.2534],[142,5.8291],[143,7.2336],[144,7.9137],[145,8.1665],[147,8.2905],[148,8.3222],[150,8.3307],[151,2.363],[152,1.157],[153,0.8711],[154,0.6593],[155,0.6022],[156,0.3738],[157,-0.8197],[158,-1.1181],[159,-1.186],[160,-1.201],[161,-1.2109],[163,-1.2109],[164,-1.2165],[361,-1.2165],[362,-2.0245],[363,-3.062],[364,-3.3651],[365,-3.4633],[366,-3.5231],[367,-3.5941],[368,-3.6267],[370,-3.6432],[433,-3.675]]
       }
       index_01_l_ctrl.Rotation.Y {
        type=MovieSceneScriptingFloatChannel
        default=0
        thinned_from=434
        keys [[0,2.5922],[150,2.5922],[151,-5.5842],[152,-7.2366],[153,-7.6283],[154,-7.9186],[155,-7.9968],[156,-8.3097],[157,-9.945],[158,-10.3538],[159,-10.4468],[161,-10.4809],[361,-10.4887],[362,-6.1974],[363,-0.6783],[364,0.9342],[365,1.4566],[366,1.7746],[367,2.1522],[368,2.3254],[369,2.3878],[370,2.4135],[371,2.4135],[373,2.4389],[374,2.4652],[376,2.5297],[377,2.5539],[378,2.5677],[380,2.5677],[381,2.5827],[433,2.5827]]
       }
       index_01_l_ctrl.Rotation.Z {
        type=MovieSceneScriptingFloatChannel
        default=0
        thinned_from=434
        keys [[0,22.2532],[134,22.237],[135,22.1232],[136,21.8144],[137,21.2131],[138,20.2218],[139,18.7429],[140,15.2976],[141,10.0971],[142,5.5791],[143,3.1155],[144,1.9225],[145,1.4792],[146,1.3741],[147,1.2616],[148,1.2061],[149,1.1912],[150,1.1912],[151,14.8729],[152,17.6377],[153,18.2933],[154,18.7789],[155,18.9098],[156,19.4335],[157,22.1698],[158,22.8539],[159,23.0094],[361,23.0795],[362,22.81],[363,22.4604],[364,22.3582],[433,22.2538]]
       }
       index_01_l_ctrl.Scale.X {type=MovieSceneScriptingFloatChannel, default=1, const [1,434]}
       index_01_l_ctrl.Scale.Y {type=MovieSceneScriptingFloatChannel, default=1, const [1,434]}
       index_01_l_ctrl.Scale.Z {type=MovieSceneScriptingFloatChannel, default=1, const [1,434]}
       index_02_l_ctrl.Location.X {type=MovieSceneScriptingFloatChannel, default=0, thinned_from=434, keys [[0,0],[433,0]]}
       index_02_l_ctrl.Location.Y {type=MovieSceneScriptingFloatChannel, default=0, thinned_from=434, keys [[0,0],[433,0]]}
       index_02_l_ctrl.Location.Z {type=MovieSceneScriptingFloatChannel, default=0, thinned_from=434, keys [[0,0],[433,0]]}
       index_02_l_ctrl.Rotation.X {
        type=MovieSceneScriptingFloatChannel
        default=0
        thinned_from=434
        keys [[0,-3.6768],[150,-3.6768],[151,-1.3788],[152,-0.9144],[153,-0.8043],[154,-0.7227],[155,-0.7007],[156,-0.6128],[157,-0.1532],[158,-0.0383],[159,-0.0122],[160,-0.0064],[161,-0.0026],[162,-0.0026],[163,-0.0007],[361,-0.0007],[362,-1.2065],[363,-2.7576],[364,-3.2108],[365,-3.3576],[366,-3.447],[367,-3.5531],[368,-3.6018],[369,-3.6194],[374,-3.6411],[378,-3.6699],[433,-3.6741]]
       }
       index_02_l_ctrl.Rotation.Y {
        type=MovieSceneScriptingFloatChannel
        default=0
        thinned_from=434
        keys [[0,2.5922],[150,2.5922],[151,0.9721],[152,0.6447],[153,0.567],[154,0.5095],[155,0.494],[156,0.432],[157,0.108],[158,0.027],[159,0.0086],[160,0.0045],[161,0.0018],[162,0.0018],[163,0.0005],[361,0.0005],[362,0.8506],[363,1.9442],[364,2.2637],[365,2.3672],[366,2.4302],[367,2.505],[368,2.5394],[369,2.5517],[374,2.5671],[378,2.5874],[433,2.5903]]
       }
       index_02_l_ctrl.Rotation.Z {
        type=MovieSceneScriptingFloatChannel
        default=0
        thinned_from=434
        keys [[0,22.2532],[134,22.2396],[135,22.1448],[136,21.8873],[137,21.3859],[138,20.5593],[139,19.3262],[140,16.4533],[141,12.1169],[142,8.3496],[143,6.2953],[144,5.3006],[145,4.9309],[147,4.7495],[148,4.7032],[149,4.6908],[150,4.6908],[151,23.6714],[152,27.507],[153,28.4165],[154,29.0902],[155,29.2718],[156,29.9982],[157,33.7943],[158,34.7433],[159,34.9591],[361,35.054],[362,30.8575],[363,25.4548],[364,23.8762],[365,23.3648],[366,23.0536],[367,22.6839],[368,22.5144],[370,22.4282],[433,22.2626]]
       }
       index_02_l_ctrl.Scale.X {type=MovieSceneScriptingFloatChannel, default=1, const [1,434]}
       index_02_l_ctrl.Scale.Y {type=MovieSceneScriptingFloatChannel, default=1, const [1,434]}
       index_02_l_ctrl.Scale.Z {type=MovieSceneScriptingFloatChannel, default=1, const [1,434]}
       index_03_l_ctrl.Location.X {type=MovieSceneScriptingFloatChannel, default=0, thinned_from=434, keys [[0,-0.0001],[433,-0.0001]]}
       index_03_l_ctrl.Location.Y {type=MovieSceneScriptingFloatChannel, default=0, thinned_from=434, keys [[0,0.0003],[433,0.0003]]}
       index_03_l_ctrl.Location.Z {type=MovieSceneScriptingFloatChannel, default=0, thinned_from=434, keys [[0,0],[433,0]]}
       index_03_l_ctrl.Rotation.X {
        type=MovieSceneScriptingFloatChannel
        default=0
        thinned_from=434
        keys [[0,-0.1093],[136,-0.1095],[139,-0.1103],[141,-0.1115],[150,-0.111],[151,-0.1116],[361,-0.1115],[362,-0.1111],[363,-0.1101],[364,-0.1097],[433,-0.1093]]
       }
       index_03_l_ctrl.Rotation.Y {
        type=MovieSceneScriptingFloatChannel
        default=0
        thinned_from=434
        keys [[0,-0.0223],[134,-0.0223],[135,-0.0221],[136,-0.0217],[137,-0.0207],[138,-0.0192],[139,-0.0169],[140,-0.0114],[141,-0.0032],[142,0.0039],[143,0.0078],[144,0.0097],[145,0.0104],[150,0.0109],[151,0.001],[152,-0.001],[154,-0.0018],[155,-0.0019],[156,-0.0023],[157,-0.0043],[158,-0.0048],[361,-0.0049],[362,-0.0107],[363,-0.018],[364,-0.0202],[365,-0.0208],[367,-0.0218],[433,-0.0223]]
       }
       index_03_l_ctrl.Rotation.Z {
        type=MovieSceneScriptingFloatChannel
        default=0
        thinned_from=434
        keys [[0,11.8666],[133,11.8666],[134,11.8799],[135,11.9725],[136,12.2239],[137,12.7136],[138,13.5208],[139,14.725],[140,17.5305],[141,21.7652],[142,25.4441],[143,27.4502],[144,28.4216],[145,28.7826],[148,29.0049],[150,29.0171],[151,23.9292],[152,22.901],[153,22.6572],[154,22.4766],[155,22.4279],[156,22.2332],[157,21.2157],[158,20.9613],[159,20.9034],[361,20.878],[362,17.9201],[363,14.1191],[364,13.0085],[365,12.6487],[366,12.4297],[367,12.1697],[368,12.0504],[369,12.0074],[377,11.893],[433,11.8669]]
       }
       index_03_l_ctrl.Scale.X {type=MovieSceneScriptingFloatChannel, default=1, const [1,434]}
       index_03_l_ctrl.Scale.Y {type=MovieSceneScriptingFloatChannel, default=1, const [1,434]}
       index_03_l_ctrl.Scale.Z {type=MovieSceneScriptingFloatChannel, default=1, const [1,434]}
       middle_metacarpal_l_ctrl.Location.X {type=MovieSceneScriptingFloatChannel, default=0, const [0,434]}
       middle_metacarpal_l_ctrl.Location.Y {type=MovieSceneScriptingFloatChannel, default=0, const [0,434]}
       middle_metacarpal_l_ctrl.Location.Z {type=MovieSceneScriptingFloatChannel, default=0, const [0,434]}
       middle_metacarpal_l_ctrl.Rotation.X {type=MovieSceneScriptingFloatChannel, default=0, const [0,434]}
       middle_metacarpal_l_ctrl.Rotation.Y {type=MovieSceneScriptingFloatChannel, default=0, const [0,434]}
       middle_metacarpal_l_ctrl.Rotation.Z {type=MovieSceneScriptingFloatChannel, default=0, const [0,434]}
       middle_metacarpal_l_ctrl.Scale.X {type=MovieSceneScriptingFloatChannel, default=1, const [1,434]}
       middle_metacarpal_l_ctrl.Scale.Y {type=MovieSceneScriptingFloatChannel, default=1, const [1,434]}
       middle_metacarpal_l_ctrl.Scale.Z {type=MovieSceneScriptingFloatChannel, default=1, const [1,434]}
       middle_01_l_ctrl.Location.X {type=MovieSceneScriptingFloatChannel, default=0, const [0,434]}
       middle_01_l_ctrl.Location.Y {type=MovieSceneScriptingFloatChannel, default=0, thinned_from=434, keys [[0,0],[433,0]]}
       middle_01_l_ctrl.Location.Z {type=MovieSceneScriptingFloatChannel, default=0, thinned_from=434, keys [[0,0],[433,0]]}
       middle_01_l_ctrl.Rotation.X {type=MovieSceneScriptingFloatChannel, default=0, thinned_from=434, keys [[0,0],[433,0]]}
       middle_01_l_ctrl.Rotation.Y {type=MovieSceneScriptingFloatChannel, default=0, thinned_from=434, keys [[0,0],[433,0]]}
       middle_01_l_ctrl.Rotation.Z {
        type=MovieSceneScriptingFloatChannel
        default=0
        thinned_from=434
        keys [[0,11.8666],[150,11.8666],[151,17.2513],[152,18.3394],[153,18.5974],[154,18.7885],[155,18.8401],[156,19.0461],[157,20.1231],[158,20.3923],[159,20.4535],[361,20.4804],[362,17.6551],[363,14.0205],[364,12.9585],[365,12.6145],[366,12.4051],[367,12.1564],[368,12.0424],[369,12.0012],[377,11.8919],[433,11.878]]
       }
       middle_01_l_ctrl.Scale.X {type=MovieSceneScriptingFloatChannel, default=1, const [1,434]}
       middle_01_l_ctrl.Scale.Y {type=MovieSceneScriptingFloatChannel, default=1, const [1,434]}
       middle_01_l_ctrl.Scale.Z {type=MovieSceneScriptingFloatChannel, default=1, const [1,434]}
       middle_02_l_ctrl.Location.X {type=MovieSceneScriptingFloatChannel, default=0, thinned_from=434, keys [[0,-0.0452],[433,-0.0452]]}
       middle_02_l_ctrl.Location.Y {type=MovieSceneScriptingFloatChannel, default=0, thinned_from=434, keys [[0,0.0507],[433,0.0507]]}
       middle_02_l_ctrl.Location.Z {type=MovieSceneScriptingFloatChannel, default=0, const [0.1146,434]}
       middle_02_l_ctrl.Rotation.X {
        type=MovieSceneScriptingFloatChannel
        default=0
        thinned_from=434
        keys [[0,-3.4915],[150,-3.4915],[151,-2.5014],[152,-2.2406],[153,-2.1763],[154,-2.1281],[155,-2.115],[156,-2.0624],[157,-1.7788],[158,-1.7058],[159,-1.6891],[161,-1.6829],[361,-1.6817],[362,-2.4066],[363,-3.1614],[364,-3.336],[365,-3.3877],[367,-3.4527],[368,-3.4681],[433,-3.4915]]
       }
       middle_02_l_ctrl.Rotation.Y {
        type=MovieSceneScriptingFloatChannel
        default=0
        thinned_from=434
        keys [[0,1.6725],[150,1.6725],[151,2.9552],[152,3.1574],[153,3.202],[154,3.2342],[155,3.2428],[156,3.2765],[157,3.4386],[158,3.4754],[361,3.4871],[362,3.0328],[363,2.2349],[364,1.9646],[365,1.874],[366,1.8182],[367,1.7513],[368,1.7204],[369,1.7092],[373,1.7001],[377,1.6794],[433,1.6725]]
       }
       middle_02_l_ctrl.Rotation.Z {
        type=MovieSceneScriptingFloatChannel
        default=0
        thinned_from=434
        keys [[0,-3.7078],[150,-3.7078],[151,20.4506],[152,25.3381],[153,26.4973],[154,27.356],[155,27.5875],[156,28.5134],[157,33.3534],[158,34.5636],[159,34.8389],[361,34.9599],[362,22.2643],[363,5.9502],[364,1.1875],[365,-0.355],[366,-1.2939],[367,-2.4087],[368,-2.9199],[369,-3.1043],[370,-3.18],[371,-3.2057],[372,-3.2184],[373,-3.2552],[374,-3.3328],[376,-3.5232],[377,-3.5946],[378,-3.6352],[379,-3.6569],[380,-3.6569],[381,-3.6795],[382,-3.6795],[383,-3.6953],[399,-3.6953],[400,-3.7076],[433,-3.7076]]
       }
       middle_02_l_ctrl.Scale.X {type=MovieSceneScriptingFloatChannel, default=1, const [1,434]}
       middle_02_l_ctrl.Scale.Y {type=MovieSceneScriptingFloatChannel, default=1, const [1,434]}
       middle_02_l_ctrl.Scale.Z {type=MovieSceneScriptingFloatChannel, default=1, const [1,434]}
       middle_03_l_ctrl.Location.X {type=MovieSceneScriptingFloatChannel, default=0, thinned_from=434, keys [[0,0.0206],[433,0.0206]]}
       middle_03_l_ctrl.Location.Y {type=MovieSceneScriptingFloatChannel, default=0, thinned_from=434, keys [[0,-0.0583],[433,-0.0583]]}
       middle_03_l_ctrl.Location.Z {type=MovieSceneScriptingFloatChannel, default=0, const [0.1324,434]}
       middle_03_l_ctrl.Rotation.X {
        type=MovieSceneScriptingFloatChannel
        default=0
        thinned_from=434
        keys [[0,1.0173],[150,1.0173],[151,1.2205],[152,1.2602],[154,1.2764],[155,1.2783],[156,1.2857],[157,1.3241],[158,1.3337],[361,1.3367],[362,1.2353],[363,1.0999],[364,1.0594],[365,1.0462],[367,1.0285],[368,1.0241],[433,1.0173]]
       }
       middle_03_l_ctrl.Rotation.Y {
        type=MovieSceneScriptingFloatChannel
        default=0
        thinned_from=434
        keys [[0,2.1201],[150,2.1201],[151,2.0101],[152,1.9854],[154,1.975],[156,1.969],[157,1.9434],[158,1.9369],[361,1.9348],[362,2.001],[363,2.0785],[364,2.0994],[367,2.1147],[433,2.1201]]
       }
       middle_03_l_ctrl.Rotation.Z {
        type=MovieSceneScriptingFloatChannel
        default=0
        thinned_from=434
        keys [[0,11.8658],[150,11.8658],[151,17.4995],[152,18.6379],[153,18.9078],[154,19.1078],[155,19.1617],[156,19.3773],[157,20.5038],[158,20.7855],[159,20.8495],[361,20.8755],[362,17.922],[363,14.1193],[364,13.0082],[365,12.6483],[366,12.4292],[367,12.169],[368,12.0497],[369,12.0066],[377,11.8922],[433,11.8661]]
       }
       middle_03_l_ctrl.Scale.X {type=MovieSceneScriptingFloatChannel, default=1, const [1,434]}
       middle_03_l_ctrl.Scale.Y {type=MovieSceneScriptingFloatChannel, default=1, const [1,434]}
       middle_03_l_ctrl.Scale.Z {type=MovieSceneScriptingFloatChannel, default=1, const [1,434]}
       ring_metacarpal_l_ctrl.Location.X {type=MovieSceneScriptingFloatChannel, default=0, thinned_from=434, keys [[0,0],[433,0]]}
       ring_metacarpal_l_ctrl.Location.Y {type=MovieSceneScriptingFloatChannel, default=0, thinned_from=434, keys [[0,0],[433,0]]}
       ring_metacarpal_l_ctrl.Location.Z {type=MovieSceneScriptingFloatChannel, default=0, thinned_from=434, keys [[0,0],[433,0]]}
       ring_metacarpal_l_ctrl.Rotation.X {
        type=MovieSceneScriptingFloatChannel
        default=0
        thinned_from=434
        keys [[0,0],[150,0],[151,0.0953],[152,0.1145],[153,0.1191],[154,0.1225],[155,0.1234],[156,0.127],[157,0.1461],[158,0.1508],[159,0.1519],[361,0.1519],[362,0.1024],[363,0.0381],[364,0.0193],[365,0.0132],[366,0.0095],[367,0.0051],[368,0.0031],[369,0.0024],[372,0.0024],[373,0.0018],[374,0.0018],[375,0.0011],[376,0.0011],[377,0.0004],[433,0.0004]]
       }
       ring_metacarpal_l_ctrl.Rotation.Y {
        type=MovieSceneScriptingFloatChannel
        default=0
        thinned_from=434
        keys [[0,-3.4928],[150,-3.4928],[151,-1.323],[152,-0.8845],[153,-0.7806],[154,-0.7036],[155,-0.6828],[156,-0.5997],[157,-0.1658],[158,-0.0573],[159,-0.0326],[361,-0.0326],[362,-1.1603],[363,-2.6249],[364,-3.0528],[365,-3.1915],[366,-3.2758],[367,-3.376],[368,-3.422],[369,-3.4386],[372,-3.4386],[373,-3.4521],[374,-3.4521],[375,-3.4677],[376,-3.4677],[377,-3.4826],[433,-3.4826]]
       }
       ring_metacarpal_l_ctrl.Rotation.Z {
        type=MovieSceneScriptingFloatChannel
        default=0
        thinned_from=434
        keys [[0,0],[150,0],[151,0.6717],[152,0.8075],[153,0.8397],[154,0.8635],[155,0.8699],[156,0.8957],[157,1.03],[158,1.0636],[159,1.0712],[361,1.0712],[362,0.7221],[363,0.2687],[364,0.1362],[365,0.0933],[366,0.0672],[367,0.0362],[368,0.0219],[369,0.0168],[372,0.0168],[373,0.0126],[374,0.0126],[375,0.0078],[376,0.0078],[377,0.0032],[433,0.0032]]
       }
       ring_metacarpal_l_ctrl.Scale.X {type=MovieSceneScriptingFloatChannel, default=1, const [1,434]}
       ring_metacarpal_l_ctrl.Scale.Y {type=MovieSceneScriptingFloatChannel, default=1, const [1,434]}
       ring_metacarpal_l_ctrl.Scale.Z {type=MovieSceneScriptingFloatChannel, default=1, const [1,434]}
       ring_01_l_ctrl.Location.X {type=MovieSceneScriptingFloatChannel, default=0, thinned_from=434, keys [[0,0],[167,0],[361,-0.0002],[433,0]]}
       ring_01_l_ctrl.Location.Y {type=MovieSceneScriptingFloatChannel, default=0, thinned_from=434, keys [[0,0],[167,0],[168,0.0005],[361,0.0005],[362,0],[433,0]]}
       ring_01_l_ctrl.Location.Z {type=MovieSceneScriptingFloatChannel, default=0, thinned_from=434, keys [[0,0],[167,0],[168,0.0014],[361,0.0014],[362,0],[433,0]]}
       ring_01_l_ctrl.Rotation.X {
        type=MovieSceneScriptingFloatChannel
        default=0
        thinned_from=434
        keys [[0,0],[150,0],[151,0.0728],[152,0.0876],[153,0.0911],[154,0.0936],[155,0.0943],[156,0.0971],[157,0.1117],[158,0.1153],[159,0.1162],[167,0.1165],[168,0.1169],[361,0.1169],[362,0.0783],[363,0.0291],[364,0.0148],[365,0.0101],[366,0.0073],[367,0.0039],[368,0.0024],[369,0.0018],[370,0.0016],[372,0.0016],[377,0.0003],[433,0]]
       }
       ring_01_l_ctrl.Rotation.Y {
        type=MovieSceneScriptingFloatChannel
        default=0
        thinned_from=434
        keys [[0,0],[150,0],[151,5.9019],[152,7.0946],[153,7.3774],[154,7.5868],[155,7.6433],[156,7.8692],[157,9.0496],[158,9.3446],[159,9.4118],[361,9.4499],[362,6.3445],[363,2.3607],[364,1.1968],[365,0.8197],[366,0.5902],[367,0.3176],[368,0.1926],[369,0.1475],[370,0.129],[372,0.129],[373,0.1107],[374,0.0917],[375,0.0682],[376,0.0451],[377,0.0277],[378,0.0277],[379,0.0124],[384,0.0124],[385,0.0017],[433,0.0017]]
       }
       ring_01_l_ctrl.Rotation.Z {
        type=MovieSceneScriptingFloatChannel
        default=0
        thinned_from=434
        keys [[0,9.4289],[150,9.4289],[151,16.3474],[152,17.7455],[153,18.077],[154,18.3225],[155,18.3887],[156,18.6535],[157,20.0372],[158,20.3831],[159,20.4618],[361,20.4983],[362,16.8662],[363,12.1963],[364,10.8318],[365,10.3898],[366,10.1208],[367,9.8013],[368,9.6547],[369,9.6019],[370,9.5802],[372,9.5802],[377,9.4614],[433,9.431]]
       }
       ring_01_l_ctrl.Scale.X {type=MovieSceneScriptingFloatChannel, default=1, const [1,434]}
       ring_01_l_ctrl.Scale.Y {type=MovieSceneScriptingFloatChannel, default=1, const [1,434]}
       ring_01_l_ctrl.Scale.Z {type=MovieSceneScriptingFloatChannel, default=1, const [1,434]}
       ring_02_l_ctrl.Location.X {type=MovieSceneScriptingFloatChannel, default=0, thinned_from=434, keys [[0,0],[433,0]]}
       ring_02_l_ctrl.Location.Y {type=MovieSceneScriptingFloatChannel, default=0, thinned_from=434, keys [[0,0],[433,0]]}
       ring_02_l_ctrl.Location.Z {type=MovieSceneScriptingFloatChannel, default=0, thinned_from=434, keys [[0,0],[433,0]]}
       ring_02_l_ctrl.Rotation.X {type=MovieSceneScriptingFloatChannel, default=0, thinned_from=434, keys [[0,0],[433,0]]}
       ring_02_l_ctrl.Rotation.Y {type=MovieSceneScriptingFloatChannel, default=0, thinned_from=434, keys [[0,0],[433,0]]}
       ring_02_l_ctrl.Rotation.Z {
        type=MovieSceneScriptingFloatChannel
        default=0
        thinned_from=434
        keys [[0,13.9361],[150,13.9361],[151,27.1383],[152,29.8063],[153,30.4389],[154,30.9075],[155,31.0338],[156,31.5391],[157,34.1795],[158,34.8396],[159,34.9897],[361,35.0574],[362,28.1285],[363,19.217],[364,16.6132],[365,15.7697],[366,15.2563],[367,14.6467],[368,14.367],[369,14.2662],[371,14.2107],[373,14.1837],[377,13.998],[433,13.9387]]
       }
       ring_02_l_ctrl.Scale.X {type=MovieSceneScriptingFloatChannel, default=1, const [1,434]}
       ring_02_l_ctrl.Scale.Y {type=MovieSceneScriptingFloatChannel, default=1, const [1,434]}
       ring_02_l_ctrl.Scale.Z {type=MovieSceneScriptingFloatChannel, default=1, const [1,434]}
       ring_03_l_ctrl.Location.X {type=MovieSceneScriptingFloatChannel, default=0, thinned_from=434, keys [[0,0],[433,0]]}
       ring_03_l_ctrl.Location.Y {type=MovieSceneScriptingFloatChannel, default=0, thinned_from=434, keys [[0,0],[433,0]]}
       ring_03_l_ctrl.Location.Z {type=MovieSceneScriptingFloatChannel, default=0, thinned_from=434, keys [[0,0],[433,0]]}
       ring_03_l_ctrl.Rotation.X {type=MovieSceneScriptingFloatChannel, default=0, thinned_from=434, keys [[0,0],[433,0]]}
       ring_03_l_ctrl.Rotation.Y {type=MovieSceneScriptingFloatChannel, default=0, thinned_from=434, keys [[0,0],[433,0]]}
       ring_03_l_ctrl.Rotation.Z {
        type=MovieSceneScriptingFloatChannel
        default=0
        thinned_from=434
        keys [[0,0],[150,0],[151,13.0478],[152,15.6846],[153,16.3098],[154,16.7729],[155,16.8977],[156,17.3971],[157,20.0066],[158,20.659],[159,20.8074],[361,20.8742],[362,14.0264],[363,5.2191],[364,2.6458],[365,1.8122],[366,1.3048],[367,0.7022],[368,0.4258],[369,0.3262],[370,0.2852],[371,0.2714],[372,0.2714],[373,0.2446],[374,0.2027],[375,0.1507],[376,0.0997],[377,0.0611],[378,0.0392],[379,0.0275],[380,0.0275],[381,0.0153],[382,0],[433,0]]
       }
       ring_03_l_ctrl.Scale.X {type=MovieSceneScriptingFloatChannel, default=1, const [1,434]}
       ring_03_l_ctrl.Scale.Y {type=MovieSceneScriptingFloatChannel, default=1, const [1,434]}
       ring_03_l_ctrl.Scale.Z {type=MovieSceneScriptingFloatChannel, default=1, const [1,434]}
       pinky_metacarpal_l_ctrl.Location.X {type=MovieSceneScriptingFloatChannel, default=0, const [0,434]}
       pinky_metacarpal_l_ctrl.Location.Y {type=MovieSceneScriptingFloatChannel, default=0, thinned_from=434, keys [[0,0],[433,0]]}
       pinky_metacarpal_l_ctrl.Location.Z {type=MovieSceneScriptingFloatChannel, default=0, thinned_from=434, keys [[0,0],[433,0]]}
       pinky_metacarpal_l_ctrl.Rotation.X {
        type=MovieSceneScriptingFloatChannel
        default=0
        thinned_from=434
        keys [[0,0],[150,0],[151,-0.6731],[152,-0.8092],[153,-0.8414],[154,-0.8653],[155,-0.8717],[156,-0.8975],[157,-1.0321],[158,-1.0658],[159,-1.0734],[361,-1.0765],[362,-0.7236],[363,-0.2693],[364,-0.1365],[365,-0.0935],[366,-0.0673],[367,-0.0362],[368,-0.022],[369,-0.0168],[370,-0.0147],[372,-0.0147],[374,-0.0105],[376,-0.0051],[377,-0.0032],[378,-0.0032],[379,-0.0014],[385,-0.0014],[386,-0.0001],[433,-0.0001]]
       }
       pinky_metacarpal_l_ctrl.Rotation.Y {
        type=MovieSceneScriptingFloatChannel
        default=0
        thinned_from=434
        keys [[0,-7.0585],[150,-7.0585],[151,-0.7395],[152,0.5375],[153,0.8403],[154,1.0645],[155,1.125],[156,1.3668],[157,2.6306],[158,2.9466],[159,3.0184],[160,3.0343],[161,3.0343],[162,3.0475],[361,3.0475],[362,-0.2656],[363,-4.5309],[364,-5.7771],[365,-6.1809],[366,-6.4266],[367,-6.7184],[368,-6.8523],[369,-6.9005],[370,-6.9204],[372,-6.9204],[377,-7.0289],[378,-7.0289],[379,-7.0452],[385,-7.0452],[433,-7.0573]]
       }
       pinky_metacarpal_l_ctrl.Rotation.Z {
        type=MovieSceneScriptingFloatChannel
        default=0
        thinned_from=434
        keys [[0,0],[150,0],[151,0.1871],[152,0.2249],[153,0.2339],[154,0.2405],[155,0.2423],[156,0.2495],[157,0.2869],[158,0.2962],[159,0.2984],[361,0.2992],[362,0.2011],[363,0.0748],[364,0.0379],[365,0.026],[366,0.0187],[367,0.0101],[368,0.0061],[369,0.0047],[370,0.0041],[372,0.0041],[374,0.0029],[376,0.0014],[377,0.0009],[378,0.0009],[379,0.0004],[385,0.0004],[386,0],[433,0]]
       }
       pinky_metacarpal_l_ctrl.Scale.X {type=MovieSceneScriptingFloatChannel, default=1, const [1,434]}
       pinky_metacarpal_l_ctrl.Scale.Y {type=MovieSceneScriptingFloatChannel, default=1, const [1,434]}
       pinky_metacarpal_l_ctrl.Scale.Z {type=MovieSceneScriptingFloatChannel, default=1, const [1,434]}
       pinky_01_l_ctrl.Location.X {type=MovieSceneScriptingFloatChannel, default=0, thinned_from=434, keys [[0,0],[433,0]]}
       pinky_01_l_ctrl.Location.Y {type=MovieSceneScriptingFloatChannel, default=0, thinned_from=434, keys [[0,0],[433,0]]}
       pinky_01_l_ctrl.Location.Z {type=MovieSceneScriptingFloatChannel, default=0, thinned_from=434, keys [[0,0],[433,0]]}
       pinky_01_l_ctrl.Rotation.X {
        type=MovieSceneScriptingFloatChannel
        default=0
        thinned_from=434
        keys [[0,0],[134,0],[135,-0.0792],[136,-0.2674],[137,-0.6338],[138,-1.2379],[139,-2.1392],[140,-4.2387],[141,-7.4079],[142,-10.161],[143,-11.6624],[144,-12.3893],[145,-12.6595],[147,-12.7921],[148,-12.8259],[150,-12.8259],[151,-4.8131],[152,-3.192],[153,-2.8076],[154,-2.5229],[155,-2.4462],[156,-2.1392],[157,-0.5348],[158,-0.1337],[159,-0.0425],[160,-0.0223],[161,-0.009],[162,-0.009],[163,-0.0024],[361,-0.0024],[362,0],[433,0]]
       }
       pinky_01_l_ctrl.Rotation.Y {
        type=MovieSceneScriptingFloatChannel
        default=0
        thinned_from=434
        keys [[0,0],[134,0],[135,-0.0345],[136,-0.1164],[137,-0.2759],[138,-0.5389],[139,-0.9313],[140,-1.8454],[141,-3.2251],[142,-4.4237],[143,-5.0773],[144,-5.3938],[145,-5.5114],[147,-5.5691],[148,-5.5839],[150,-5.5839],[151,9.0584],[152,12.0181],[153,12.7199],[154,13.2398],[155,13.3799],[156,13.9404],[157,16.8696],[158,17.6019],[159,17.7685],[161,17.8297],[361,17.8417],[362,11.9903],[363,4.4615],[364,2.2617],[365,1.5491],[366,1.1154],[367,0.6003],[368,0.3641],[369,0.2789],[370,0.2439],[371,0.232],[372,0.232],[373,0.2091],[374,0.1733],[375,0.1288],[376,0.0853],[377,0.0523],[378,0.0335],[379,0.0335],[380,0.0181],[382,0.0181],[383,0.0058],[433,0.0058]]
       }
       pinky_01_l_ctrl.Rotation.Z {
        type=MovieSceneScriptingFloatChannel
        default=0
        thinned_from=434
        keys [[0,9.4289],[150,9.4289],[151,14.8163],[152,15.905],[153,16.1632],[154,16.3544],[155,16.4059],[156,16.6121],[157,17.6896],[158,17.959],[159,18.0202],[361,18.0472],[362,15.2204],[363,11.5839],[364,10.5214],[365,10.1772],[366,9.9677],[367,9.7189],[368,9.6048],[369,9.5636],[378,9.4451],[433,9.4317]]
       }
       pinky_01_l_ctrl.Scale.X {type=MovieSceneScriptingFloatChannel, default=1, const [1,434]}
       pinky_01_l_ctrl.Scale.Y {type=MovieSceneScriptingFloatChannel, default=1, const [1,434]}
       pinky_01_l_ctrl.Scale.Z {type=MovieSceneScriptingFloatChannel, default=1, const [1,434]}
       pinky_02_l_ctrl.Location.X {type=MovieSceneScriptingFloatChannel, default=0, thinned_from=434, keys [[0,0],[433,0]]}
       pinky_02_l_ctrl.Location.Y {type=MovieSceneScriptingFloatChannel, default=0, thinned_from=434, keys [[0,0],[433,0]]}
       pinky_02_l_ctrl.Location.Z {type=MovieSceneScriptingFloatChannel, default=0, thinned_from=434, keys [[0,0],[433,0]]}
       pinky_02_l_ctrl.Rotation.X {type=MovieSceneScriptingFloatChannel, default=0, thinned_from=434, keys [[0,0],[433,0]]}
       pinky_02_l_ctrl.Rotation.Y {type=MovieSceneScriptingFloatChannel, default=0, thinned_from=434, keys [[0,0],[433,0]]}
       pinky_02_l_ctrl.Rotation.Z {
        type=MovieSceneScriptingFloatChannel
        default=0
        thinned_from=434
        keys [[0,-8.3839],[150,-8.3839],[151,18.7684],[152,24.2554],[153,25.5564],[154,26.5202],[155,26.78],[156,27.8191],[157,33.2495],[158,34.6071],[159,34.9159],[161,35.0293],[361,35.0549],[362,20.8047],[363,2.4769],[364,-2.878],[365,-4.6127],[366,-5.6686],[367,-6.9225],[368,-7.4977],[369,-7.7051],[370,-7.7903],[372,-7.8333],[373,-7.8748],[374,-7.962],[376,-8.1763],[377,-8.2566],[378,-8.3023],[380,-8.3399],[381,-8.3399],[382,-8.3636],[433,-8.3759]]
       }
       pinky_02_l_ctrl.Scale.X {type=MovieSceneScriptingFloatChannel, default=1, const [1,434]}
       pinky_02_l_ctrl.Scale.Y {type=MovieSceneScriptingFloatChannel, default=1, const [1,434]}
       pinky_02_l_ctrl.Scale.Z {type=MovieSceneScriptingFloatChannel, default=1, const [1,434]}
       pinky_03_l_ctrl.Location.X {type=MovieSceneScriptingFloatChannel, default=0, thinned_from=434, keys [[0,0],[161,0],[163,-0.0001],[433,0]]}
       pinky_03_l_ctrl.Location.Y {type=MovieSceneScriptingFloatChannel, default=0, thinned_from=434, keys [[0,0],[161,0],[162,0.0004],[163,0.0004],[164,0],[433,0]]}
       pinky_03_l_ctrl.Location.Z {type=MovieSceneScriptingFloatChannel, default=0, thinned_from=434, keys [[0,0],[433,0]]}
       pinky_03_l_ctrl.Rotation.X {type=MovieSceneScriptingFloatChannel, default=0, thinned_from=434, keys [[0,0],[161,0],[162,-0.0002],[163,-0.0002],[164,0],[433,0]]}
       pinky_03_l_ctrl.Rotation.Y {type=MovieSceneScriptingFloatChannel, default=0, thinned_from=434, keys [[0,0],[433,0]]}
       pinky_03_l_ctrl.Rotation.Z {
        type=MovieSceneScriptingFloatChannel
        default=0
        thinned_from=434
        keys [[0,-16.4531],[150,-16.4531],[151,6.8779],[152,11.5928],[153,12.7107],[154,13.5388],[155,13.762],[156,14.655],[157,19.3211],[158,20.4876],[159,20.7529],[161,20.8504],[361,20.8696],[362,8.6277],[363,-7.1207],[364,-11.722],[365,-13.2126],[366,-14.12],[367,-15.1974],[368,-15.6916],[369,-15.8698],[370,-15.943],[373,-16.0156],[377,-16.3437],[379,-16.4039],[433,-16.4529]]
       }
       pinky_03_l_ctrl.Scale.X {type=MovieSceneScriptingFloatChannel, default=1, const [1,434]}
       pinky_03_l_ctrl.Scale.Y {type=MovieSceneScriptingFloatChannel, default=1, const [1,434]}
       pinky_03_l_ctrl.Scale.Z {type=MovieSceneScriptingFloatChannel, default=1, const [1,434]}
       hand_l_ik_ctrl.Location.X {
        type=MovieSceneScriptingFloatChannel
        default=0
        thinned_from=434
        keys [[0,-2.5171],[1,-2.5218],[2,-2.5344],[6,-2.6162],[7,-2.6326],[8,-2.6425],[11,-2.6368],[12,-2.6388],[13,-2.6566],[14,-2.685],[15,-2.7002],[16,-2.6367],[17,-2.4865],[18,-2.3514],[19,-2.6632],[20,-3.6542],[21,-4.9574],[22,-6.2355],[23,-7.4506],[24,-8.5679],[25,-9.5708],[26,-10.4531],[27,-11.2034],[28,-11.8143],[29,-12.3347],[30,-12.81],[31,-13.233],[32,-13.5938],[33,-13.9163],[34,-14.1797],[35,-14.3757],[36,-14.5038],[37,-14.5355],[38,-14.485],[40,-14.2455],[45,-13.5749],[50,-13.1336],[58,-12.598],[60,-12.3945],[62,-12.1239],[63,-11.947],[64,-11.5703],[65,-10.7043],[66,-9.5595],[68,-7.3381],[69,-6.3597],[70,-5.7446],[71,-5.3999],[72,-5.146],[73,-4.9701],[74,-4.8684],[75,-4.8046],[76,-4.77],[77,-4.7527],[78,-4.7479],[81,-4.7842],[84,-4.8156],[87,-4.8908],[90,-5.0179],[96,-5.3422],[98,-5.4088],[99,-5.4253],[100,-5.415],[101,-5.368],[102,-5.2868],[103,-5.1739],[104,-5.0169],[105,-4.8168],[107,-4.3857],[108,-4.1624],[110,-3.6654],[111,-3.4525],[112,-3.2878],[113,-3.1459],[114,-3.0246],[115,-2.9138],[117,-2.7168],[119,-2.544],[120,-2.473],[121,-2.4123],[123,-2.3152],[125,-2.2471],[129,-2.1676],[130,-2.1526],[131,-2.1494],[133,-2.1247],[134,-2.107],[135,-2.1002],[136,-2.1017],[139,-2.1314],[140,-2.115],[141,-1.9749],[142,-1.6549],[143,-0.838],[144,-0.1267],[145,-0.1095],[146,-0.5224],[147,-1.3667],[148,-2.4949],[149,-4.0648],[150,-11.1401],[151,-21.1599],[152,-28.1449],[153,-32.4982],[154,-34.9187],[155,-36.0773],[157,-36.4998],[158,-34.146],[159,-32.1834],[160,-27.9907],[161,-25.342],[162,-24.4088],[163,-26.191],[164,-29.1328],[165,-31.9283],[166,-34.3543],[167,-36.1632],[168,-36.441],[169,-35.2277],[171,-29.5655],[172,-27.3331],[173,-25.8905],[174,-25.362],[175,-25.6357],[178,-26.7065],[180,-27.5076],[181,-27.8459],[182,-28.062],[183,-28.1729],[185,-28.1879],[196,-27.6808],[210,-27.3355],[212,-27.3769],[214,-27.5279],[219,-28.4425],[220,-28.583],[221,-28.6346],[222,-28.5588],[223,-28.3409],[224,-27.959],[225,-27.0738],[226,-25.6328],[228,-22.6116],[229,-20.845],[230,-19.0028],[231,-17.2822],[232,-15.7772],[233,-14.4532],[234,-13.3107],[235,-12.3536],[236,-11.5715],[237,-10.9792],[238,-10.4955],[239,-10.0697],[240,-9.6844],[241,-9.3517],[242,-9.0551],[243,-8.8034],[244,-8.6036],[245,-8.4368],[246,-8.2973],[247,-8.1804],[248,-8.0834],[250,-7.936],[253,-7.614],[254,-7.5228],[255,-7.4942],[256,-7.4236],[257,-7.3921],[258,-7.65],[259,-8.4355],[260,-10.4232],[261,-13.3212],[262,-15.9413],[263,-18.2571],[264,-20.6755],[265,-22.7009],[266,-24.2901],[267,-25.7236],[268,-26.9943],[269,-28.1035],[270,-29.0769],[271,-29.9556],[272,-30.7379],[273,-31.4183],[274,-32.0052],[276,-32.9424],[278,-33.6889],[281,-34.5721],[283,-34.9756],[288,-35.7121],[289,-35.7618],[290,-35.7074],[291,-35.5385],[292,-35.1747],[293,-34.6121],[295,-33.2321],[296,-32.4062],[298,-30.5512],[299,-29.7442],[302,-27.4595],[304,-26.075],[306,-24.8991],[308,-23.8905],[310,-23.0038],[312,-22.2434],[314,-21.6161],[315,-21.3428],[317,-20.8865],[320,-20.3694],[322,-20.1591],[324,-20.1062],[328,-20.3274],[330,-20.551],[332,-20.9918],[335,-21.7674],[339,-22.5001],[341,-22.7806],[344,-23.0492],[345,-23.02],[346,-22.7291],[347,-22.1751],[348,-21.3848],[349,-20.3713],[350,-19.1467],[351,-17.3969],[352,-15.0833],[353,-12.2728],[354,-9.027],[355,-8.7089],[356,-8.3732],[357,-5.1559],[358,-2.5107],[359,-0.7738],[360,0.5971],[361,1.809],[362,2.7485],[363,3.0572],[364,2.6468],[365,1.5985],[366,0.4264],[367,-0.5121],[368,-1.2234],[369,-1.6152],[370,-1.8213],[371,-2.0951],[372,-2.4644],[373,-2.8872],[374,-3.2273],[375,-3.4939],[376,-3.8631],[377,-4.3722],[378,-4.9468],[379,-5.5581],[380,-6.1331],[381,-6.6109],[382,-6.9584],[383,-7.1765],[384,-7.2742],[385,-7.2678],[386,-7.1553],[387,-6.931],[388,-6.6371],[389,-6.3049],[390,-5.9879],[392,-5.4117],[393,-5.1138],[395,-4.4585],[396,-4.1473],[397,-3.8677],[398,-3.6092],[399,-3.3759],[400,-3.1601],[401,-2.9722],[402,-2.8203],[403,-2.6931],[404,-2.5932],[405,-2.5018],[406,-2.4269],[407,-2.363],[408,-2.3116],[409,-2.2738],[411,-2.2488],[413,-2.2505],[416,-2.2744],[417,-2.2933],[418,-2.3068],[422,-2.3755],[424,-2.4159],[426,-2.445],[427,-2.4636],[430,-2.5004],[431,-2.5058],[433,-2.5107]]
       }
       hand_l_ik_ctrl.Location.Y {
        type=MovieSceneScriptingFloatChannel
        default=0
        thinned_from=434
        keys [[0,4.0667],[7,4.0308],[9,4.0371],[10,4.0569],[11,4.0539],[12,4.0305],[13,3.9666],[14,3.8608],[16,3.4482],[17,3.4034],[18,3.5388],[19,3.6636],[20,3.7803],[21,3.9441],[22,3.962],[23,3.9018],[24,3.8654],[25,3.8931],[26,3.943],[27,4.0209],[28,4.123],[29,4.2564],[31,4.5729],[32,4.6902],[33,4.7534],[34,4.7601],[35,4.7454],[36,4.708],[37,4.6542],[40,4.4213],[43,4.2393],[45,4.091],[47,3.9792],[50,3.8562],[52,3.7857],[54,3.7275],[56,3.695],[58,3.6452],[59,3.6096],[60,3.5557],[61,3.4808],[62,3.4286],[63,3.3561],[64,3.1914],[65,3.1475],[66,3.2023],[67,3.23],[68,3.2201],[69,3.2302],[70,3.4502],[71,3.7365],[72,3.9919],[73,4.2316],[74,4.4199],[75,4.5798],[76,4.7205],[77,4.8275],[78,4.8993],[79,4.946],[80,4.9698],[82,4.9818],[84,4.9369],[85,4.8989],[86,4.8494],[88,4.7277],[89,4.6589],[91,4.4957],[93,4.2993],[95,4.0857],[97,3.8499],[99,3.5864],[100,3.4285],[101,3.2411],[104,2.6545],[106,2.2523],[107,2.0807],[108,1.9697],[109,1.9201],[110,1.9083],[111,1.9269],[112,1.9707],[113,2.0502],[114,2.1391],[115,2.223],[116,2.2976],[119,2.5893],[120,2.6753],[121,2.7499],[124,2.9478],[127,3.1295],[129,3.2334],[130,3.2641],[131,3.3073],[132,3.3392],[133,3.3875],[134,3.4221],[135,3.4923],[136,3.6158],[137,3.8194],[138,4.1292],[139,4.5701],[140,5.2525],[141,6.2502],[142,7.5161],[143,9.5167],[144,11.4973],[145,12.5718],[146,13.0795],[147,13.1446],[148,12.4818],[149,10.1097],[150,0.8669],[151,-10.1775],[152,-14.8183],[153,-16.8349],[155,-18.61],[156,-19.6261],[157,-20.9401],[158,-19.5887],[159,-16.3597],[160,-11.5823],[161,-7.4901],[162,-4.4613],[163,-3.4595],[164,-4.4409],[165,-6.4834],[166,-8.8438],[167,-10.9509],[168,-11.8001],[169,-11.6812],[170,-10.5187],[171,-9.0675],[172,-7.6642],[173,-6.5851],[174,-6.1105],[175,-6.1958],[176,-6.322],[177,-6.4785],[178,-6.6725],[180,-7.1161],[181,-7.2983],[182,-7.429],[183,-7.4762],[184,-7.4732],[186,-7.4234],[192,-7.1869],[194,-7.1317],[206,-7.0022],[209,-6.9275],[211,-6.8413],[213,-6.7218],[214,-6.6518],[215,-6.5528],[217,-6.3073],[218,-6.1953],[219,-6.1356],[220,-6.1354],[221,-6.2354],[222,-6.4632],[223,-6.8285],[224,-7.3359],[225,-8.2183],[226,-9.4093],[227,-10.4909],[229,-12.3238],[230,-13.1839],[231,-13.9426],[232,-14.5681],[233,-15.0647],[234,-15.4514],[235,-15.7462],[236,-15.9642],[237,-16.1136],[238,-16.2195],[240,-16.3264],[255,-16.2652],[256,-16.1163],[257,-15.7462],[258,-15.2194],[259,-14.5814],[260,-13.5957],[261,-12.1375],[262,-10.648],[264,-8.0156],[265,-6.8705],[266,-5.9702],[267,-5.1974],[268,-4.5367],[269,-3.961],[270,-3.4555],[271,-3.0185],[272,-2.6462],[273,-2.3349],[274,-2.0738],[275,-1.8632],[276,-1.6932],[277,-1.5399],[278,-1.4202],[279,-1.3157],[280,-1.2356],[281,-1.1653],[282,-1.1341],[283,-1.1542],[284,-1.2026],[286,-1.3459],[287,-1.4094],[288,-1.4938],[289,-1.6436],[290,-1.8632],[291,-2.1527],[292,-2.5546],[293,-3.0603],[294,-3.5957],[295,-4.1657],[296,-4.8145],[297,-5.4484],[298,-5.992],[299,-6.3763],[300,-6.6124],[301,-6.7872],[303,-7.0309],[304,-7.132],[305,-7.211],[307,-7.3004],[308,-7.3189],[309,-7.3544],[310,-7.3662],[311,-7.3973],[313,-7.4373],[316,-7.4397],[320,-7.3928],[321,-7.3642],[331,-7.2043],[334,-7.1158],[336,-7.1103],[341,-7.203],[344,-7.2112],[345,-7.1704],[346,-7.0418],[348,-6.6919],[349,-6.577],[350,-6.5548],[351,-6.516],[352,-6.5628],[353,-6.5796],[354,-6.3417],[355,-6.8473],[356,-6.4598],[357,-4.2725],[358,-1.8077],[359,0.6503],[360,3.2691],[361,6.1394],[362,8.9941],[363,11.4315],[364,12.3568],[365,12.618],[366,12.3097],[367,11.532],[368,10.4789],[369,9.2303],[370,7.8408],[371,6.6616],[372,5.9686],[373,5.559],[374,5.373],[375,5.3894],[376,5.5179],[377,5.5988],[378,5.594],[379,5.5287],[380,5.398],[381,5.2426],[382,5.1125],[383,4.9708],[384,4.7823],[385,4.5],[386,4.1018],[388,3.264],[389,2.9006],[390,2.5778],[391,2.3186],[392,2.1191],[393,2.0066],[394,1.9867],[395,2.0377],[396,2.1378],[397,2.2653],[399,2.6061],[400,2.798],[401,3.01],[402,3.2067],[403,3.3879],[404,3.5507],[405,3.6961],[406,3.828],[407,3.9415],[408,4.0336],[409,4.1012],[410,4.1469],[412,4.2115],[414,4.2461],[416,4.2637],[422,4.2223],[424,4.1806],[426,4.1521],[430,4.0855],[433,4.0682]]
       }
       hand_l_ik_ctrl.Location.Z {
        type=MovieSceneScriptingFloatChannel
        default=0
        thinned_from=434
        keys [[0,-5.9185],[1,-5.9281],[3,-5.9901],[6,-6.119],[7,-6.1525],[8,-6.1736],[9,-6.1667],[11,-6.1864],[12,-6.215],[13,-6.2618],[14,-6.3371],[15,-6.441],[16,-6.3209],[17,-5.7216],[18,-4.583],[19,-3.3933],[20,-2.9381],[21,-3.3924],[22,-4.3036],[23,-5.3581],[24,-6.7504],[25,-8.3178],[26,-9.7458],[27,-10.9913],[28,-12.076],[29,-13.1132],[30,-14.121],[31,-15.0244],[32,-15.817],[33,-16.5504],[34,-17.1716],[35,-17.649],[36,-17.9573],[37,-18.1147],[38,-18.1401],[39,-18.0734],[41,-17.8132],[49,-16.4887],[59,-15.1351],[61,-14.8006],[62,-14.7007],[63,-14.6834],[64,-14.5678],[65,-14.2354],[66,-13.6692],[67,-12.9957],[68,-11.941],[69,-10.6148],[70,-9.6856],[71,-9.0662],[72,-8.4961],[73,-8.0056],[74,-7.6368],[75,-7.36],[76,-7.1385],[77,-6.9838],[78,-6.8922],[79,-6.8454],[80,-6.8359],[82,-6.8552],[84,-6.9066],[85,-6.9558],[87,-7.116],[89,-7.3534],[90,-7.49],[94,-8.1599],[98,-8.8558],[99,-9.0158],[100,-9.1566],[101,-9.2686],[102,-9.3523],[103,-9.4086],[104,-9.433],[105,-9.4135],[106,-9.344],[107,-9.2238],[108,-9.0111],[109,-8.7001],[110,-8.3474],[111,-8.014],[113,-7.4635],[116,-6.6983],[118,-6.1669],[119,-5.9196],[120,-5.7008],[122,-5.3167],[123,-5.1459],[125,-4.8665],[126,-4.7622],[130,-4.4083],[132,-4.3091],[133,-4.2762],[134,-4.2848],[135,-4.3448],[136,-4.44],[138,-4.6648],[139,-4.7562],[140,-4.9217],[141,-5.1071],[142,-5.0675],[143,-4.2505],[144,-3.0377],[145,-2.0839],[146,-1.4362],[147,-1.0051],[148,-0.9236],[149,-2.8597],[150,-9.7499],[151,-21.4068],[152,-26.2405],[153,-28.0257],[154,-28.8275],[156,-29.3021],[157,-30.2857],[158,-23.9091],[159,-22.3455],[161,-24.9592],[162,-25.6779],[163,-25.1928],[164,-24.1132],[165,-23.2398],[166,-22.6059],[167,-22.3788],[168,-23.2111],[170,-26.7112],[171,-27.278],[172,-27.2535],[173,-27.033],[174,-26.7315],[175,-26.3066],[176,-25.9767],[180,-25.1819],[182,-24.9874],[183,-24.9873],[189,-25.6253],[194,-25.9096],[211,-26.0586],[215,-25.8959],[217,-25.7257],[218,-25.5987],[219,-25.4154],[220,-25.1806],[221,-24.8395],[222,-24.3725],[223,-23.8117],[224,-23.1879],[225,-22.3489],[226,-21.2837],[227,-20.2876],[228,-19.4271],[231,-16.7032],[232,-15.8846],[233,-15.1523],[234,-14.5132],[235,-13.972],[236,-13.5262],[237,-13.1891],[238,-12.9001],[239,-12.6626],[240,-12.4604],[242,-12.1488],[244,-11.938],[246,-11.7971],[250,-11.6046],[255,-11.2095],[256,-11.207],[257,-11.3528],[258,-11.7365],[259,-12.4435],[260,-14.0104],[261,-16.2097],[262,-18.0788],[263,-19.5421],[264,-20.9076],[265,-21.9545],[266,-22.6601],[267,-23.1693],[268,-23.5464],[269,-23.8565],[270,-24.1161],[272,-24.4624],[275,-24.7456],[287,-25.1266],[288,-25.1951],[292,-25.8341],[295,-26.509],[303,-27.5074],[308,-27.8083],[312,-27.8483],[321,-27.5466],[326,-27.1139],[331,-26.5646],[335,-26.3683],[344,-26.5773],[345,-26.6379],[346,-26.7548],[351,-27.7466],[352,-27.6207],[353,-26.6775],[354,-24.4487],[355,-19.7955],[356,-13.8758],[357,-8.9894],[358,-4.6343],[359,-1.2214],[360,2.1353],[361,5.4646],[362,8.417],[363,10.8701],[364,11.8811],[365,12.276],[366,12.3239],[367,11.8046],[368,10.711],[369,8.8839],[370,6.5765],[372,1.6886],[373,-0.7633],[374,-3.1408],[375,-5.3562],[376,-7.3463],[377,-9.0916],[378,-10.6786],[379,-12.0991],[380,-13.2826],[381,-14.229],[382,-14.9548],[383,-15.4523],[384,-15.7331],[385,-15.8212],[386,-15.6938],[387,-15.345],[388,-14.8337],[390,-13.6079],[392,-12.4341],[393,-11.8142],[395,-10.4584],[396,-9.8029],[397,-9.1962],[399,-8.0994],[400,-7.61],[401,-7.1887],[402,-6.8252],[403,-6.5223],[404,-6.2762],[405,-6.0446],[406,-5.8506],[407,-5.6819],[408,-5.5439],[409,-5.4418],[411,-5.3561],[413,-5.3382],[415,-5.3553],[416,-5.3783],[419,-5.4708],[423,-5.62],[427,-5.7826],[430,-5.8736],[431,-5.8857],[432,-5.8859],[433,-5.9008]]
       }
       hand_l_ik_ctrl.Rotation.X {
        type=MovieSceneScriptingFloatChannel
        default=0
        thinned_from=434
        keys [[0,27.8583],[12,27.9101],[14,28.0197],[15,28.302],[16,28.714],[17,28.8507],[18,28.3934],[19,27.2453],[20,25.4717],[21,22.8526],[22,19.8154],[23,17.0478],[24,14.6631],[25,12.7694],[26,11.3437],[27,10.2971],[28,9.5631],[29,9.0719],[30,8.75],[31,8.5475],[32,8.4242],[33,8.3785],[34,8.485],[35,8.6471],[36,8.8642],[37,9.1266],[38,9.4446],[40,10.1137],[41,10.3872],[42,10.6393],[44,11.0978],[45,11.279],[47,11.509],[50,11.6418],[53,11.7444],[55,11.8495],[56,11.8514],[57,11.8284],[58,11.7302],[59,11.5583],[60,11.2902],[61,10.9164],[62,10.3473],[63,9.2687],[64,7.9093],[65,7.3404],[66,7.7914],[67,9.2794],[68,11.9348],[69,14.9207],[70,16.6889],[71,17.2553],[72,17.5997],[73,17.7602],[74,17.7895],[75,17.7288],[76,17.6076],[77,17.4196],[80,16.6477],[82,16.2392],[85,15.7506],[86,15.6226],[89,15.3337],[92,15.1395],[94,15.0868],[96,15.1223],[97,15.2052],[98,15.339],[99,15.53],[100,15.8731],[101,16.425],[102,17.145],[103,17.993],[104,19.0386],[106,21.5526],[107,22.71],[108,23.7282],[109,24.6744],[110,25.5182],[111,26.226],[112,26.765],[113,27.1731],[115,27.7694],[117,28.2095],[122,28.8146],[125,28.9666],[130,28.8668],[133,28.5593],[134,28.5075],[135,28.5898],[139,29.4222],[141,29.5163],[142,29.9013],[143,31.983],[144,33.5017],[145,32.236],[146,30.6452],[147,30.563],[148,31.4468],[149,31.3728],[150,31.0885],[151,42.1117],[152,58.5232],[153,71.5551],[154,75.8667],[155,77.2043],[156,76.7045],[157,74.4914],[158,79.3142],[159,82.0717],[160,85.527],[161,88.5781],[162,90.9713],[163,92.6465],[164,93.2654],[165,93.0817],[166,92.0746],[167,90.6113],[168,89.395],[169,88.3793],[170,87.8899],[171,88.0863],[174,89.5564],[176,90.1299],[184,90.4592],[212,89.9981],[220,91.1609],[222,90.9445],[224,90.2965],[225,89.6385],[226,88.5214],[227,86.9597],[228,85.0548],[231,78.4665],[232,76.6236],[233,75.0266],[234,73.6659],[235,72.5402],[236,71.6194],[238,70.357],[241,69.0952],[245,68.1425],[251,67.3407],[257,65.8126],[258,65.7146],[259,65.8494],[260,66.7559],[263,72.0774],[264,74.2303],[265,76.1889],[266,77.7708],[267,79.1659],[269,81.4729],[271,83.3083],[273,84.7331],[277,86.5895],[286,89.0657],[287,89.5082],[288,90.2076],[291,93.1383],[293,94.5391],[295,95.3868],[297,95.4411],[322,89.0996],[330,88.2264],[343,89.1958],[347,89.2488],[348,89.0287],[349,88.445],[350,87.3616],[351,85.3783],[352,82.2427],[353,77.846],[354,72.0327],[355,59.0687],[356,44.9221],[357,39.3684],[358,36.0828],[359,33.3136],[360,31.3281],[361,30.0511],[362,29.7366],[363,29.8585],[364,29.3964],[365,28.0775],[366,26.5949],[367,25.0366],[368,23.274],[369,21.6022],[370,20.2018],[371,18.9411],[372,17.763],[373,16.7724],[374,16.0177],[375,15.4352],[376,14.9432],[377,14.5496],[378,14.2772],[379,14.13],[380,14.1321],[381,14.3075],[382,14.683],[383,15.268],[384,16.0401],[385,16.9897],[386,18.1738],[387,19.5411],[388,20.9725],[389,22.3338],[390,23.5227],[391,24.5442],[392,25.481],[393,26.2997],[394,26.9851],[395,27.5371],[396,27.9874],[397,28.3718],[398,28.6769],[399,28.9156],[400,29.0706],[402,29.1534],[411,28.7352],[422,28.0243],[433,27.8846]]
       }
       hand_l_ik_ctrl.Rotation.Y {
        type=MovieSceneScriptingFloatChannel
        default=0
        thinned_from=434
        keys [[0,-1.6043],[1,-1.6174],[2,-1.6526],[3,-1.7036],[6,-1.8882],[7,-1.9393],[8,-1.9746],[9,-1.9741],[10,-2.0014],[11,-2.0475],[12,-2.1444],[13,-2.2935],[14,-2.5094],[15,-2.9748],[16,-3.3372],[17,-2.9774],[18,-1.5871],[19,0.6881],[20,2.7491],[21,3.8762],[22,4.3197],[23,4.33],[24,3.581],[25,2.2912],[26,0.9659],[27,-0.2789],[28,-1.5052],[29,-2.8761],[30,-4.3138],[31,-5.6476],[32,-6.8709],[33,-8.0716],[34,-9.1233],[35,-9.9324],[36,-10.497],[37,-10.8431],[38,-11.0011],[39,-11.022],[40,-10.9568],[42,-10.6757],[44,-10.2877],[45,-10.0713],[48,-9.2982],[50,-8.7704],[55,-7.5088],[57,-7.0502],[59,-6.6143],[60,-6.385],[61,-6.1392],[62,-5.9871],[63,-5.9394],[64,-5.8369],[65,-5.3801],[66,-4.8675],[67,-4.7052],[68,-4.3615],[70,-3.0824],[71,-2.3046],[72,-1.3958],[73,-0.4616],[74,0.3246],[75,0.9964],[76,1.5805],[77,2.0221],[78,2.3228],[79,2.5348],[80,2.6663],[81,2.734],[82,2.774],[83,2.7911],[84,2.7847],[85,2.7476],[86,2.6725],[87,2.5692],[88,2.4299],[89,2.274],[90,2.0948],[91,1.8823],[92,1.6586],[93,1.4308],[94,1.1989],[95,0.9445],[96,0.6736],[97,0.4063],[98,0.1225],[99,-0.1704],[100,-0.4608],[101,-0.7432],[102,-1.0299],[103,-1.334],[105,-1.9896],[106,-2.285],[107,-2.5137],[108,-2.6213],[109,-2.61],[110,-2.5273],[111,-2.4248],[112,-2.3618],[113,-2.2866],[114,-2.2034],[115,-2.0944],[116,-1.9357],[117,-1.7478],[119,-1.3248],[120,-1.1255],[121,-0.9351],[122,-0.738],[123,-0.5464],[125,-0.1814],[126,-0.0135],[127,0.1523],[128,0.3172],[129,0.4833],[130,0.6524],[133,1.1168],[134,1.3568],[135,1.752],[136,2.3088],[137,3.0278],[138,3.9103],[139,4.9597],[140,6.1738],[141,8.0672],[142,11.3999],[144,26.4441],[145,29.5963],[146,31.4817],[147,33.2832],[148,34.3986],[149,32.4638],[150,19.6452],[151,-9.1888],[152,-21.9976],[153,-22.1354],[154,-18.9134],[155,-15.1008],[156,-10.7981],[157,-4.6808],[158,3.6002],[159,3.9297],[160,-0.3441],[161,-4.0147],[162,-6.7794],[163,-8.0799],[164,-7.2857],[165,-4.6502],[166,-1.9751],[167,-0.0785],[168,-0.2418],[169,-1.3838],[170,-3.9173],[171,-6.0186],[172,-7.4473],[173,-8.2643],[174,-8.3318],[175,-7.6587],[176,-7.0903],[177,-6.5859],[178,-6.0983],[179,-5.6505],[180,-5.2564],[181,-4.9687],[182,-4.7812],[183,-4.7496],[184,-4.8708],[185,-5.0472],[187,-5.4654],[188,-5.6426],[190,-5.938],[192,-6.1859],[193,-6.287],[194,-6.3657],[195,-6.4176],[196,-6.4855],[202,-6.6954],[206,-6.7936],[209,-6.8963],[210,-6.9175],[211,-6.923],[212,-6.907],[213,-6.8635],[214,-6.7884],[215,-6.6741],[216,-6.5131],[217,-6.3196],[218,-6.0758],[219,-5.7657],[220,-5.4002],[221,-4.912],[222,-4.2658],[223,-3.4957],[224,-2.6418],[225,-1.5334],[226,-0.2509],[227,0.7156],[228,1.2824],[229,1.7191],[230,2.0948],[231,2.4603],[232,2.8041],[233,3.092],[234,3.3281],[235,3.5199],[236,3.6696],[237,3.7785],[238,3.8785],[239,3.942],[240,3.9779],[242,4.0095],[246,3.9948],[250,3.9618],[251,3.9428],[252,3.9112],[253,3.8691],[255,3.7276],[256,3.521],[257,3.0554],[258,2.2774],[259,1.1397],[260,-0.905],[261,-3.535],[262,-5.6456],[263,-6.9685],[264,-7.8378],[265,-8.3122],[266,-8.4784],[267,-8.4383],[270,-8.0016],[271,-7.8201],[273,-7.4308],[274,-7.2549],[276,-6.9469],[278,-6.6662],[280,-6.4019],[284,-5.9333],[285,-5.8064],[286,-5.6493],[287,-5.4546],[289,-4.9974],[290,-4.7915],[291,-4.6369],[292,-4.5719],[293,-4.5937],[295,-4.7408],[296,-4.7732],[297,-4.8594],[298,-4.9987],[299,-5.1986],[300,-5.4926],[301,-5.8499],[302,-6.2403],[303,-6.5984],[304,-6.9243],[306,-7.5314],[308,-8.0752],[310,-8.5232],[312,-8.8799],[314,-9.1557],[315,-9.274],[317,-9.4734],[319,-9.6146],[321,-9.6979],[323,-9.6943],[325,-9.5772],[326,-9.4737],[332,-8.5886],[333,-8.431],[334,-8.3054],[335,-8.1475],[336,-8.0325],[339,-7.6318],[341,-7.4173],[343,-7.2704],[344,-7.1807],[345,-7.4453],[346,-8.38],[347,-9.875],[348,-11.7711],[349,-13.9],[350,-16.0687],[352,-20.461],[353,-22.8491],[354,-25.4307],[355,-29.6805],[356,-26.3413],[358,-10.6052],[359,-4.1641],[360,2.1387],[361,8.4093],[362,13.9463],[363,18.1505],[364,20.3917],[365,22.7902],[366,24.5732],[367,25.1525],[368,24.9225],[369,23.4667],[370,21.1313],[371,18.5081],[372,15.7731],[374,10.0464],[375,7.2783],[376,4.638],[377,2.1659],[378,-0.2059],[379,-2.474],[380,-4.4908],[381,-6.2325],[382,-7.7326],[383,-8.9671],[384,-9.9071],[385,-10.5586],[386,-10.9388],[387,-11.0642],[388,-10.9822],[389,-10.7281],[390,-10.3865],[391,-10.0074],[392,-9.608],[393,-9.1074],[394,-8.4769],[396,-7.0845],[397,-6.4212],[398,-5.7979],[399,-5.1589],[400,-4.5622],[401,-4.02],[402,-3.5252],[403,-3.0999],[404,-2.73],[405,-2.3847],[406,-2.0798],[407,-1.8145],[408,-1.5966],[409,-1.434],[410,-1.3273],[411,-1.2371],[412,-1.173],[413,-1.1241],[414,-1.0905],[415,-1.0697],[416,-1.0673],[417,-1.0816],[418,-1.0901],[419,-1.1034],[420,-1.1358],[422,-1.1846],[423,-1.2293],[424,-1.2842],[425,-1.3264],[426,-1.3827],[427,-1.4211],[428,-1.4723],[429,-1.5052],[430,-1.5471],[431,-1.5652],[432,-1.5664],[433,-1.5916]]
       }
       hand_l_ik_ctrl.Rotation.Z {
        type=MovieSceneScriptingFloatChannel
        default=0
        thinned_from=434
        keys [[0,5.0881],[2,5.0706],[6,4.9959],[9,4.9825],[10,5.0017],[11,4.975],[12,4.9188],[13,4.7795],[14,4.5648],[15,4.1448],[16,3.6272],[17,3.3089],[18,3.3205],[19,3.5491],[20,3.5854],[21,2.8854],[22,1.3412],[23,-0.4263],[24,-2.0825],[25,-3.4195],[26,-4.4777],[27,-5.2628],[28,-5.8231],[29,-6.2274],[30,-6.5189],[31,-6.7408],[33,-7.1486],[34,-7.3358],[35,-7.4641],[36,-7.5358],[37,-7.5301],[38,-7.4625],[39,-7.3529],[41,-7.0773],[43,-6.8454],[45,-6.6718],[47,-6.5716],[52,-6.422],[54,-6.3281],[55,-6.2702],[56,-6.2271],[58,-6.2263],[59,-6.2704],[60,-6.3296],[61,-6.4303],[62,-6.5748],[64,-7.1613],[65,-7.2532],[66,-6.9987],[67,-6.1713],[68,-4.8563],[69,-3.3001],[70,-2.0738],[71,-1.3215],[72,-0.6653],[73,-0.0593],[74,0.407],[75,0.7434],[76,1.0335],[77,1.2168],[78,1.2963],[79,1.3121],[80,1.2796],[81,1.2249],[82,1.179],[83,1.1056],[84,1.0239],[85,0.9341],[87,0.7409],[88,0.6299],[89,0.5243],[90,0.4074],[91,0.2782],[92,0.1446],[93,0.01],[94,-0.125],[95,-0.2534],[96,-0.3899],[97,-0.4849],[98,-0.5815],[99,-0.6458],[100,-0.6682],[101,-0.6301],[102,-0.525],[103,-0.3456],[104,-0.057],[105,0.3201],[106,0.7178],[108,1.4428],[109,1.8429],[110,2.2346],[111,2.597],[112,2.9052],[113,3.2002],[114,3.4634],[115,3.7045],[116,3.9079],[117,4.1246],[119,4.5132],[120,4.6865],[121,4.8454],[123,5.1278],[125,5.3669],[126,5.4674],[128,5.6365],[129,5.7077],[130,5.7561],[132,5.8173],[133,5.8354],[134,5.8359],[135,5.8873],[136,5.9898],[137,6.1584],[138,6.4041],[139,6.7329],[140,7.2112],[141,7.7415],[142,8.0049],[143,7.5771],[144,6.8337],[145,6.3496],[146,6.0429],[147,8.5212],[148,12.2954],[149,10.9987],[150,-2.6645],[151,-28.7984],[152,-57.0707],[153,-79.7773],[154,-91.244],[156,-106.5076],[157,-130.8902],[158,-134.1481],[159,-136.5177],[161,-130.7647],[162,-129.4242],[163,-129.0227],[164,-129.6882],[167,-135.4171],[168,-136.3232],[169,-136.8395],[170,-135.469],[172,-130.8645],[173,-129.3151],[174,-128.6885],[182,-130.7011],[218,-129.8379],[220,-129.4276],[221,-128.9906],[223,-127.337],[224,-126.2145],[225,-124.459],[226,-122.2612],[227,-120.6104],[231,-116.4687],[234,-114.3038],[242,-111.7567],[254,-110.9929],[257,-111.0816],[258,-111.3833],[259,-112.3152],[260,-114.7498],[261,-118.3958],[262,-121.7375],[264,-126.9999],[265,-129.0195],[267,-131.7028],[270,-134.3805],[274,-136.6202],[283,-138.9966],[288,-139.2584],[291,-137.952],[295,-134.1913],[298,-130.7855],[303,-127.0683],[310,-123.9755],[322,-121.9776],[331,-122.2701],[344,-124.3595],[345,-124.1656],[346,-123.2341],[347,-121.662],[348,-119.5536],[349,-116.989],[350,-114.0377],[351,-110.5593],[352,-105.9709],[353,-99.2918],[354,-89.2576],[355,-63.0729],[356,-33.7876],[357,-21.3236],[358,-13.8793],[359,-7.5053],[360,-1.811],[361,4.3933],[362,10.1097],[363,14.0641],[364,15.8091],[365,17.0712],[366,17.424],[367,16.6404],[368,15.065],[369,12.8636],[370,10.3097],[371,7.9961],[372,6.2528],[373,4.9282],[374,3.8603],[375,3.048],[376,2.5313],[377,2.1517],[378,1.7549],[379,1.3822],[380,1.0552],[381,0.7973],[382,0.692],[383,0.7032],[384,0.7698],[385,0.8114],[386,0.8155],[387,0.8758],[388,0.9691],[389,1.0968],[390,1.1931],[391,1.2837],[392,1.3981],[393,1.5763],[394,1.8224],[395,2.1115],[396,2.4254],[397,2.7464],[398,3.0965],[399,3.4115],[400,3.735],[401,4.0474],[402,4.3106],[403,4.5529],[404,4.7585],[405,4.9473],[406,5.1105],[407,5.2531],[408,5.3703],[409,5.457],[411,5.53],[413,5.5486],[416,5.5309],[422,5.359],[424,5.2789],[425,5.26],[430,5.1145],[432,5.1061],[433,5.0921]]
       }
       hand_l_ik_ctrl.Scale.X {type=MovieSceneScriptingFloatChannel, default=1, const [1,434]}
       hand_l_ik_ctrl.Scale.Y {type=MovieSceneScriptingFloatChannel, default=1, const [1,434]}
       hand_l_ik_ctrl.Scale.Z {type=MovieSceneScriptingFloatChannel, default=1, const [1,434]}
       Softness {type=MovieSceneScriptingFloatChannel, default=0, const [0,434]}
       hand_l_ikorient_ctrl.Rotation.Roll {type=MovieSceneScriptingFloatChannel, default=0, const [0,434]}
       hand_l_ikorient_ctrl.Rotation.Pitch {type=MovieSceneScriptingFloatChannel, default=0, const [0,434]}
       hand_l_ikorient_ctrl.Rotation.Yaw {type=MovieSceneScriptingFloatChannel, default=0, const [0,434]}
       arm_l_pv_ik_ctrl.Location.X {
        type=MovieSceneScriptingFloatChannel
        default=0
        thinned_from=434
        keys [[0,9.0988],[13,9.0135],[14,9.018],[16,9.1629],[17,9.1296],[18,8.8926],[19,8.4016],[20,7.4464],[21,5.8697],[22,4.0453],[23,2.2838],[24,0.6023],[25,-0.9191],[26,-2.1977],[27,-3.2426],[28,-4.0718],[29,-4.7524],[30,-5.3336],[31,-5.8162],[32,-6.1978],[33,-6.4913],[34,-6.6405],[35,-6.699],[36,-6.6728],[37,-6.5453],[38,-6.3229],[39,-6.0469],[40,-5.7583],[41,-5.494],[44,-4.7854],[45,-4.5793],[46,-4.42],[47,-4.2783],[48,-4.1724],[50,-3.9829],[53,-3.7349],[54,-3.641],[55,-3.5582],[57,-3.4522],[58,-3.4301],[59,-3.4272],[60,-3.447],[61,-3.4895],[62,-3.6419],[63,-3.996],[64,-4.296],[65,-3.9517],[66,-3.0372],[67,-1.7199],[68,0.111],[69,1.9695],[70,2.9951],[71,3.3518],[72,3.6117],[73,3.7812],[74,3.8675],[75,3.8962],[76,3.8848],[77,3.8297],[78,3.738],[80,3.5171],[81,3.4231],[82,3.3404],[83,3.2689],[84,3.2045],[85,3.1482],[86,3.1008],[87,3.0703],[88,3.0258],[91,2.9457],[92,2.931],[94,2.928],[95,2.9432],[96,2.9683],[97,3.0327],[98,3.1203],[99,3.2464],[100,3.4671],[101,3.8156],[102,4.2626],[103,4.7789],[104,5.4056],[105,6.1416],[106,6.8925],[107,7.5632],[108,8.141],[109,8.6653],[110,9.1212],[111,9.4897],[112,9.7509],[113,9.9301],[115,10.1695],[116,10.2636],[118,10.3816],[123,10.554],[125,10.5783],[130,10.5306],[132,10.4622],[134,10.3662],[135,10.2967],[136,10.1844],[137,10.0043],[138,9.7317],[139,9.3418],[140,8.6494],[141,7.6353],[142,6.5128],[143,5.3254],[144,4.1868],[145,3.1562],[146,2.4192],[147,2.4161],[148,2.7752],[149,2.8081],[150,1.1696],[151,0.8432],[152,1.1577],[153,1.5975],[154,1.2147],[155,1.3415],[156,1.5177],[157,-2.6275],[158,-5.3517],[159,-8.594],[160,-10.7074],[161,-12.5071],[162,-14.2345],[163,-15.4302],[164,-15.758],[165,-15.5235],[166,-14.9167],[167,-14.1648],[168,-13.5626],[170,-12.6346],[171,-12.3682],[172,-12.3452],[176,-12.8865],[177,-12.994],[179,-13.1134],[181,-13.1775],[209,-13.0069],[211,-13.0683],[213,-13.2001],[214,-13.289],[215,-13.439],[218,-14.0577],[219,-14.2181],[220,-14.3155],[221,-14.2905],[222,-14.0956],[223,-13.7067],[224,-13.1006],[225,-11.8786],[226,-10.1229],[227,-8.526],[228,-7.177],[229,-5.6816],[230,-4.1713],[231,-2.7738],[232,-1.5611],[233,-0.5175],[234,0.3638],[235,1.0902],[236,1.6684],[237,2.0989],[238,2.4379],[239,2.7227],[240,2.9627],[241,3.1626],[242,3.3334],[243,3.4721],[244,3.5832],[245,3.6754],[246,3.752],[247,3.8162],[248,3.8679],[250,3.9361],[253,4.0607],[254,4.0829],[255,4.0638],[256,3.975],[257,3.6932],[258,3.1061],[259,2.0915],[260,0.0655],[261,-2.87],[262,-5.6487],[263,-8.0699],[264,-10.5097],[265,-12.5381],[266,-14.0966],[267,-15.4494],[268,-16.6169],[269,-17.6229],[270,-18.5012],[271,-19.272],[272,-19.9403],[273,-20.5117],[274,-20.9922],[275,-21.3942],[277,-22.0394],[279,-22.5316],[281,-22.915],[282,-23.0317],[283,-23.0877],[287,-23.0122],[288,-22.9151],[289,-22.6475],[290,-22.232],[291,-21.6965],[292,-20.963],[293,-20.0213],[294,-19.0054],[295,-17.942],[297,-15.6186],[298,-14.5477],[299,-13.7096],[300,-13.0526],[301,-12.4421],[302,-11.881],[303,-11.3677],[304,-10.9022],[305,-10.4952],[306,-10.1352],[307,-9.8137],[309,-9.2557],[311,-8.804],[312,-8.612],[313,-8.4422],[315,-8.1697],[317,-7.9902],[318,-7.9191],[320,-7.8148],[322,-7.796],[323,-7.8207],[324,-7.8678],[329,-8.2442],[330,-8.3528],[331,-8.4996],[334,-9.0529],[335,-9.1973],[336,-9.3039],[338,-9.4728],[345,-9.9187],[346,-9.9276],[347,-9.8156],[348,-9.5196],[349,-8.9824],[350,-8.1555],[351,-6.8621],[352,-4.9903],[353,-2.7523],[354,-0.3841],[355,3.5347],[356,7.5972],[357,9.6758],[358,10.54],[359,10.5525],[360,10.1775],[361,9.5109],[362,8.6492],[363,7.5935],[364,7.0007],[365,6.6473],[366,6.4718],[367,6.4257],[368,6.4152],[369,6.4699],[370,6.5771],[371,6.4739],[372,6.0354],[373,5.4457],[374,4.796],[376,3.4123],[377,2.7972],[378,2.2695],[379,1.8299],[380,1.5271],[381,1.3762],[382,1.3853],[383,1.5728],[384,1.9331],[385,2.4613],[386,3.1818],[387,4.0338],[388,4.9307],[389,5.7809],[390,6.5217],[391,7.1457],[392,7.6967],[393,8.171],[394,8.5685],[395,8.8886],[396,9.1445],[397,9.3512],[398,9.5053],[399,9.6215],[400,9.697],[401,9.7213],[403,9.6974],[408,9.5479],[422,9.164],[433,9.1088]]
       }
       arm_l_pv_ik_ctrl.Location.Y {
        type=MovieSceneScriptingFloatChannel
        default=0
        thinned_from=434
        keys [[0,-0.8144],[1,-0.8217],[2,-0.8413],[3,-0.8693],[5,-0.9355],[6,-0.9662],[7,-0.9902],[8,-1.0041],[9,-0.9963],[10,-0.9926],[12,-0.9757],[13,-0.9696],[14,-0.9682],[15,-0.8461],[16,-0.4569],[17,0.151],[18,0.7979],[19,0.8133],[20,-0.2861],[21,-2.2743],[22,-4.5002],[23,-6.6048],[24,-8.677],[25,-10.5981],[26,-12.2086],[27,-13.5248],[28,-14.5546],[29,-15.3878],[30,-16.1126],[31,-16.7266],[32,-17.2263],[33,-17.6431],[34,-17.9598],[35,-18.191],[36,-18.3167],[37,-18.3263],[38,-18.2332],[45,-16.9499],[47,-16.6857],[60,-15.6572],[61,-15.5426],[62,-15.4957],[63,-15.5171],[64,-15.3817],[65,-14.8131],[66,-13.8149],[67,-12.5471],[68,-10.8003],[69,-8.8399],[70,-7.558],[71,-6.9262],[72,-6.4765],[73,-6.1938],[74,-6.0473],[75,-5.9838],[76,-5.9745],[77,-6.0034],[78,-6.0634],[79,-6.1477],[81,-6.3305],[84,-6.5394],[85,-6.621],[87,-6.8105],[90,-7.172],[94,-7.7323],[96,-7.9876],[98,-8.1596],[99,-8.2093],[100,-8.2004],[101,-8.1093],[102,-7.9391],[103,-7.6931],[104,-7.3438],[105,-6.8857],[107,-5.8319],[108,-5.2659],[110,-4.0103],[111,-3.4507],[112,-2.9911],[113,-2.5839],[114,-2.2235],[115,-1.8978],[116,-1.591],[117,-1.2976],[118,-1.0253],[119,-0.7799],[120,-0.5683],[121,-0.3835],[122,-0.2167],[123,-0.0701],[124,0.0385],[125,0.1376],[126,0.1969],[127,0.2443],[128,0.2832],[129,0.3184],[130,0.3486],[131,0.3313],[132,0.3113],[133,0.2726],[134,0.2012],[135,0.0825],[136,-0.0796],[137,-0.2807],[138,-0.516],[139,-0.7787],[140,-1.2003],[141,-1.742],[142,-2.1795],[143,-2.0463],[144,-1.7286],[145,-1.7995],[146,-2.2601],[147,-3.0995],[148,-4.3291],[149,-6.9605],[150,-13.2896],[151,-17.4646],[152,-18.264],[153,-18.453],[154,-19.0999],[155,-19.3649],[156,-19.469],[157,-19.1955],[158,-9.8688],[159,-6.2884],[160,-4.9217],[161,-4.3551],[162,-4.0701],[163,-3.9536],[164,-3.9019],[165,-3.9869],[166,-4.4122],[167,-5.1978],[168,-6.4468],[169,-7.928],[170,-8.7842],[171,-8.4107],[172,-7.5597],[173,-6.7391],[174,-6.1441],[175,-5.6796],[176,-5.3788],[177,-5.1939],[178,-5.064],[179,-4.9766],[180,-4.9195],[181,-4.9024],[182,-4.8981],[183,-4.9445],[184,-5.057],[185,-5.199],[187,-5.496],[188,-5.6078],[190,-5.7699],[191,-5.837],[193,-5.937],[196,-6.0104],[202,-6.059],[208,-6.0622],[210,-6.0419],[212,-5.9905],[213,-5.9489],[215,-5.8254],[216,-5.7363],[217,-5.6279],[218,-5.4924],[219,-5.3103],[220,-5.0895],[221,-4.7718],[222,-4.3284],[223,-3.7859],[224,-3.1681],[225,-2.2868],[226,-1.1716],[227,-0.2575],[228,0.3839],[229,1.0471],[230,1.7295],[231,2.4007],[232,3.0272],[233,3.6015],[234,4.1131],[235,4.5543],[236,4.9206],[237,5.2002],[238,5.4402],[239,5.6427],[240,5.8179],[241,5.9655],[242,6.1004],[243,6.2062],[244,6.2915],[245,6.3619],[247,6.4649],[253,6.655],[254,6.6579],[255,6.6355],[256,6.5613],[257,6.3369],[258,5.8205],[259,4.8834],[260,2.9675],[261,0.4336],[262,-1.6508],[263,-3.2134],[264,-4.5814],[265,-5.5718],[266,-6.2223],[267,-6.6987],[268,-7.0538],[269,-7.335],[270,-7.5618],[271,-7.7325],[272,-7.8595],[273,-7.9544],[275,-8.0874],[277,-8.1683],[279,-8.2137],[281,-8.2362],[283,-8.2271],[284,-8.2054],[285,-8.1649],[286,-8.1072],[287,-8.0148],[288,-7.8632],[289,-7.6486],[291,-7.1679],[293,-6.7448],[294,-6.5672],[295,-6.4187],[297,-6.0926],[298,-5.955],[299,-5.8584],[300,-5.801],[302,-5.7234],[306,-5.6142],[308,-5.5794],[309,-5.5523],[312,-5.4961],[319,-5.3474],[320,-5.3186],[321,-5.3009],[325,-5.1917],[326,-5.1522],[329,-4.9849],[330,-4.9554],[331,-4.9487],[333,-4.984],[334,-5.0169],[339,-5.1173],[341,-5.1861],[342,-5.2328],[344,-5.349],[345,-5.3828],[347,-5.3553],[348,-5.3671],[349,-5.4383],[350,-5.59],[351,-5.9087],[352,-6.1306],[353,-5.7645],[354,-4.2871],[355,-3.0685],[356,-1.0019],[357,2.1485],[358,4.9417],[359,6.9021],[360,8.9166],[362,13.2296],[363,15.2132],[364,15.5429],[365,14.4235],[366,12.9215],[367,11.2002],[368,9.1133],[369,6.7488],[370,4.3066],[371,1.8939],[372,-0.3815],[373,-2.5328],[374,-4.4475],[375,-6.0922],[376,-7.5156],[377,-8.7499],[378,-9.8385],[379,-10.7664],[380,-11.487],[381,-11.9844],[382,-12.2442],[383,-12.275],[384,-12.1123],[385,-11.7912],[386,-11.2735],[387,-10.5447],[388,-9.6825],[389,-8.7731],[390,-7.9258],[391,-7.1546],[394,-4.9458],[395,-4.2493],[396,-3.6035],[397,-3.0164],[398,-2.4857],[399,-2.0007],[400,-1.5738],[401,-1.2236],[402,-0.9442],[403,-0.7187],[404,-0.5532],[405,-0.3962],[406,-0.2759],[407,-0.1717],[408,-0.0868],[409,-0.025],[410,-0.0245],[411,-0.0214],[412,-0.0484],[413,-0.0776],[414,-0.1142],[415,-0.1547],[416,-0.1983],[417,-0.2536],[418,-0.298],[419,-0.3517],[420,-0.4029],[421,-0.4514],[422,-0.4961],[423,-0.5429],[424,-0.5881],[426,-0.6621],[427,-0.7009],[428,-0.7239],[429,-0.7554],[430,-0.7755],[431,-0.7841],[432,-0.7839],[433,-0.792]]
       }
       arm_l_pv_ik_ctrl.Location.Z {
        type=MovieSceneScriptingFloatChannel
        default=0
        thinned_from=434
        keys [[0,7.7816],[2,7.808],[7,7.9563],[8,7.9705],[11,7.9455],[13,7.8967],[15,7.8057],[16,7.608],[17,7.1557],[18,6.5519],[19,6.0808],[20,5.9389],[21,5.9601],[22,6.0585],[23,6.2849],[24,6.6284],[26,7.4774],[27,7.879],[28,8.2123],[29,8.4892],[30,8.7466],[31,8.9852],[32,9.1828],[34,9.5192],[36,9.7852],[37,9.8538],[38,9.8786],[39,9.8703],[40,9.8399],[41,9.7796],[46,9.5495],[51,9.3656],[55,9.272],[57,9.1898],[58,9.1216],[59,9.0256],[60,8.8916],[61,8.7145],[62,8.4971],[63,8.1365],[64,7.5895],[65,6.8115],[66,6.1187],[67,5.8635],[68,5.9049],[69,6.0501],[70,6.0028],[71,5.7771],[72,5.6059],[73,5.4791],[75,5.3284],[78,5.1665],[79,5.1253],[80,5.0951],[82,5.0506],[83,5.0453],[84,5.0509],[85,5.0713],[86,5.11],[88,5.231],[89,5.307],[91,5.4908],[95,5.9498],[97,6.2139],[99,6.5095],[100,6.6717],[102,7.0419],[103,7.2413],[104,7.4618],[106,7.9395],[107,8.1435],[108,8.2793],[109,8.3481],[112,8.3992],[116,8.3679],[120,8.1962],[125,8.0508],[131,7.7762],[132,7.7166],[134,7.5623],[137,7.3718],[138,7.2871],[139,7.1711],[140,7.026],[141,6.728],[142,6.1218],[144,3.2698],[145,2.721],[146,2.7994],[147,3.6038],[148,5.0553],[149,8.1013],[150,19.2303],[151,28.3955],[152,30.7238],[153,31.5885],[154,32.9099],[155,33.7076],[156,33.646],[157,28.6849],[158,28.348],[159,25.7478],[160,21.6622],[161,18.9947],[162,17.7961],[163,19.5995],[164,22.6831],[165,25.3071],[166,27.4366],[167,28.8593],[168,28.6447],[169,26.6779],[170,23.742],[171,21.5587],[172,20.041],[173,19.1277],[174,18.8819],[176,19.651],[178,20.3254],[180,21.0785],[181,21.3882],[182,21.5902],[183,21.6686],[185,21.5909],[192,21.0357],[210,20.6114],[212,20.6543],[214,20.8033],[215,20.9404],[217,21.3376],[220,22.155],[221,22.4895],[224,23.6187],[225,23.8849],[226,23.9418],[227,23.6939],[228,23.1958],[229,22.4682],[231,20.8236],[232,20.0941],[233,19.4329],[234,18.8462],[235,18.3408],[236,17.9235],[237,17.5968],[239,17.0894],[241,16.6562],[243,16.3124],[245,16.0716],[248,15.8316],[250,15.7201],[254,15.3242],[255,15.2492],[256,15.0842],[258,14.5578],[259,14.4326],[260,14.5014],[262,14.8248],[263,15.2285],[265,16.5403],[268,18.3486],[269,18.8722],[271,19.7689],[272,20.1729],[273,20.5341],[274,20.8419],[276,21.3453],[280,22.0953],[286,22.9206],[288,23.3283],[290,23.8048],[291,23.9746],[292,24.0426],[293,24.0138],[294,23.9209],[295,23.7739],[296,23.5756],[297,23.2868],[299,22.5974],[300,22.2163],[304,20.4906],[306,19.7381],[308,19.0638],[310,18.4597],[312,17.9337],[314,17.4921],[317,16.9638],[320,16.5896],[322,16.4328],[323,16.3971],[325,16.4262],[328,16.6411],[330,16.8522],[332,17.1637],[335,17.6863],[339,18.198],[341,18.3614],[344,18.4133],[345,18.3564],[346,18.1712],[347,17.8544],[348,17.4324],[349,16.9137],[350,16.299],[351,15.3048],[352,14.0347],[353,12.6517],[354,11.3201],[355,13.7859],[356,15.3157],[357,12.3615],[358,9.3215],[360,3.9831],[361,1.1166],[362,-1.6886],[363,-3.8586],[364,-4.4906],[365,-4.2375],[366,-3.4835],[367,-2.5143],[368,-1.501],[369,-0.4841],[370,0.5395],[371,1.4967],[372,2.2819],[373,3.0113],[374,3.5914],[375,4.0179],[376,4.4838],[377,5.0831],[378,5.7572],[379,6.479],[380,7.1868],[381,7.8126],[382,8.3236],[383,8.748],[384,9.1045],[386,9.7251],[387,9.9843],[388,10.202],[389,10.3606],[390,10.4874],[391,10.5781],[392,10.6388],[393,10.6275],[394,10.5175],[395,10.3433],[397,9.9363],[399,9.4904],[401,8.992],[402,8.7579],[403,8.5504],[404,8.3679],[406,8.056],[407,7.9297],[408,7.8281],[409,7.7546],[411,7.6624],[413,7.6056],[416,7.5675],[422,7.6053],[431,7.7712],[433,7.7811]]
       }
       arm_l_pv_ik_ctrl.Rotation.X {
        type=MovieSceneScriptingFloatChannel
        default=0
        thinned_from=434
        keys [[0,18.7131],[3,18.8142],[7,19.0487],[10,19.0804],[12,18.9856],[13,18.8728],[14,18.6892],[15,18.4214],[16,17.7653],[17,16.4365],[18,14.488],[19,12.1403],[20,10.0529],[21,8.5732],[22,7.5473],[23,6.9386],[24,6.9609],[25,7.4508],[26,8.0404],[27,8.6518],[28,9.2082],[31,10.7911],[32,11.2288],[33,11.6041],[34,11.9529],[35,12.2535],[36,12.5104],[37,12.6761],[38,12.7696],[39,12.8075],[40,12.8064],[42,12.7174],[45,12.504],[51,12.0108],[56,11.7264],[57,11.6354],[58,11.5066],[59,11.3373],[60,11.0969],[62,10.445],[63,9.9435],[64,9.2453],[65,8.7485],[66,8.7383],[67,9.454],[68,10.4922],[69,11.3352],[70,11.6588],[72,11.3384],[73,11.2387],[80,10.8857],[82,10.8301],[84,10.8264],[85,10.8593],[86,10.9264],[88,11.141],[90,11.4333],[92,11.7989],[95,12.4295],[97,12.9199],[99,13.4965],[100,13.8352],[101,14.2239],[102,14.6629],[103,15.1521],[104,15.7349],[106,17.0485],[107,17.6066],[108,18.0117],[109,18.2862],[110,18.4654],[112,18.6884],[114,18.8007],[116,18.7994],[128,17.9227],[133,17.4004],[134,17.3233],[135,17.3221],[136,17.389],[137,17.513],[139,17.8882],[140,18.3657],[141,19.0697],[142,19.5636],[143,19.4257],[145,18.7062],[146,18.766],[147,19.7476],[148,21.2723],[149,24.3905],[150,31.2372],[151,37.6537],[152,38.53],[153,38.487],[154,39.639],[155,40.4876],[156,41.0618],[157,39.0042],[158,51.9176],[159,53.682],[160,51.3174],[161,50.4118],[162,49.9891],[163,50.2336],[164,51.1564],[165,52.807],[166,54.0699],[167,54.5007],[168,53.1123],[169,50.7948],[170,48.0678],[171,46.9992],[172,46.9197],[173,47.2257],[174,47.7468],[175,48.5688],[176,49.1666],[178,50.0635],[180,50.7468],[182,51.1047],[183,51.0844],[188,49.9041],[193,49.1881],[210,48.7319],[213,48.9782],[215,49.4293],[217,50.2737],[220,52.0412],[221,52.8201],[222,53.7998],[224,55.987],[225,57.1973],[226,58.1797],[227,58.2698],[228,57.5512],[229,56.4274],[231,53.8489],[232,52.7178],[233,51.6988],[234,50.7995],[235,50.0287],[236,49.3953],[237,48.9017],[240,47.8184],[243,47.0407],[251,45.9449],[255,44.8639],[256,44.4597],[257,43.8536],[258,43.0782],[259,42.1495],[260,40.7707],[261,39.0047],[262,37.503],[263,36.8583],[264,36.8828],[265,37.2152],[266,37.6905],[269,39.5309],[272,41.0356],[274,41.8515],[278,42.9547],[284,44.2475],[286,44.9464],[287,45.4521],[288,46.2547],[289,47.3799],[291,49.7637],[293,51.7044],[294,52.5247],[295,53.204],[296,53.7672],[297,54.1803],[298,54.4158],[299,54.4521],[301,54.0692],[314,50.1933],[321,48.8367],[325,48.4383],[334,48.9562],[341,49.9837],[344,50.0728],[345,50.1823],[347,50.7132],[348,50.8396],[349,50.6823],[350,50.097],[351,48.4666],[352,45.9813],[354,40.3438],[355,36.1174],[356,29.4947],[357,22.6366],[358,16.793],[359,12.1404],[360,7.7743],[361,3.6452],[362,-0.0618],[363,-3.4717],[364,-5.0012],[365,-5.2066],[366,-4.8177],[367,-3.9641],[368,-2.658],[369,-0.8088],[370,1.3427],[371,3.4559],[372,5.4063],[373,7.3791],[374,9.3293],[375,11.1661],[376,12.8652],[377,14.4291],[378,15.8999],[379,17.2785],[380,18.5106],[381,19.5753],[382,20.4947],[383,21.2888],[384,21.9597],[385,22.5175],[386,22.9817],[387,23.3466],[388,23.61],[389,23.7654],[390,23.8486],[392,23.87],[393,23.7593],[394,23.4991],[398,22.004],[402,20.3442],[403,20.0016],[405,19.4285],[407,18.9743],[408,18.806],[410,18.5992],[413,18.4361],[416,18.3787],[422,18.4235],[431,18.6927],[433,18.7092]]
       }
       arm_l_pv_ik_ctrl.Rotation.Y {
        type=MovieSceneScriptingFloatChannel
        default=0
        thinned_from=434
        keys [[0,-2.4502],[5,-2.4568],[7,-2.4696],[8,-2.4797],[9,-2.4796],[10,-2.5113],[11,-2.5194],[12,-2.5145],[13,-2.4658],[14,-2.3826],[15,-2.2537],[16,-2.0426],[17,-1.6959],[18,-1.111],[19,-0.0607],[20,1.0936],[21,1.8992],[22,2.5544],[23,3.176],[24,3.6299],[25,3.9358],[26,4.1982],[27,4.3957],[28,4.5031],[29,4.5101],[30,4.4542],[31,4.382],[32,4.3322],[33,4.3235],[36,4.3812],[37,4.3688],[38,4.3364],[39,4.2895],[40,4.2336],[41,4.1667],[42,4.1241],[43,4.0944],[44,4.0905],[45,4.0991],[46,4.1203],[52,4.3387],[54,4.3965],[58,4.4625],[59,4.4684],[61,4.5108],[62,4.498],[63,4.4431],[64,4.2849],[65,3.5424],[66,2.4735],[67,1.4867],[68,0.6503],[69,0.0545],[70,-0.4401],[71,-0.7646],[72,-0.992],[73,-1.1646],[74,-1.2731],[75,-1.3602],[76,-1.4318],[77,-1.4882],[78,-1.5289],[79,-1.5545],[80,-1.5673],[81,-1.5696],[82,-1.5799],[83,-1.5682],[84,-1.5504],[85,-1.5269],[86,-1.498],[88,-1.4322],[89,-1.3968],[90,-1.3529],[91,-1.3057],[92,-1.2476],[93,-1.1817],[94,-1.1118],[95,-1.0464],[96,-0.9662],[97,-0.9057],[98,-0.8299],[99,-0.7669],[100,-0.6927],[101,-0.6074],[102,-0.5368],[103,-0.5066],[104,-0.5294],[105,-0.5864],[106,-0.6581],[107,-0.728],[108,-0.8101],[109,-0.9128],[111,-1.1121],[112,-1.2214],[113,-1.3417],[114,-1.4509],[115,-1.5401],[116,-1.5819],[117,-1.6363],[119,-1.7064],[120,-1.733],[121,-1.7507],[122,-1.7495],[124,-1.7614],[125,-1.7604],[126,-1.7664],[127,-1.7683],[128,-1.7647],[129,-1.755],[130,-1.7255],[131,-1.7092],[133,-1.672],[134,-1.6635],[135,-1.7012],[136,-1.7954],[137,-1.96],[138,-2.2082],[139,-2.5531],[140,-3.0262],[141,-3.6878],[142,-4.5673],[143,-6.0337],[144,-7.5463],[145,-8.4004],[146,-8.7349],[147,-8.4668],[148,-7.2919],[149,-4.5156],[150,4.8764],[151,15.1274],[152,19.2647],[153,21.1855],[154,22.8747],[155,24.2785],[156,26.6796],[157,44.4708],[158,48.9114],[159,50.4395],[160,45.424],[161,40.6106],[162,37.5102],[163,36.0669],[164,36.872],[165,39.7726],[166,43.0397],[167,45.8567],[168,46.8829],[169,47.0983],[170,44.8904],[171,41.7908],[172,39.0539],[173,37.1753],[174,36.4945],[180,39.1532],[181,39.4775],[182,39.685],[183,39.7827],[196,38.8888],[214,38.2258],[220,38.3728],[221,38.3113],[222,38.1249],[223,37.813],[224,37.3724],[225,36.5024],[226,35.3564],[227,34.6362],[230,33.8533],[235,33.0641],[244,32.2366],[249,32.0654],[253,32.1861],[255,32.4066],[256,32.4297],[257,32.3664],[258,32.4525],[259,32.9162],[260,34.1439],[261,35.8803],[262,37.4194],[264,39.8505],[265,40.7686],[267,42.0759],[269,43.1158],[272,44.3136],[276,45.436],[281,46.4257],[284,46.7546],[287,46.8834],[288,46.7585],[289,46.3736],[291,45.19],[292,44.4184],[298,38.4292],[299,37.62],[302,35.4251],[304,34.2051],[306,33.2083],[308,32.3779],[310,31.7209],[312,31.2128],[315,30.6877],[318,30.3536],[321,30.1969],[323,30.2167],[330,30.9402],[335,32.0891],[343,33.241],[344,33.437],[345,33.5494],[346,33.3741],[347,32.8648],[348,32.0167],[349,30.8187],[350,29.2684],[351,26.8616],[352,23.6043],[353,19.7978],[354,15.6504],[355,9.1593],[356,3.8053],[357,1.3633],[358,0.5431],[359,0.1706],[360,-0.1597],[361,-0.5447],[362,-1.054],[363,-1.6711],[364,-2.0949],[365,-2.6099],[366,-2.8982],[367,-2.9103],[368,-2.7327],[369,-2.402],[370,-1.8452],[371,-1.399],[372,-1.3598],[373,-1.5004],[374,-1.6699],[375,-1.8962],[377,-2.6342],[378,-2.9005],[379,-3.1007],[380,-3.216],[381,-3.2992],[382,-3.4358],[383,-3.5938],[384,-3.7197],[385,-3.7441],[386,-3.6495],[387,-3.5303],[388,-3.3868],[389,-3.2363],[391,-2.8585],[392,-2.6894],[393,-2.5506],[394,-2.4387],[395,-2.352],[396,-2.2877],[397,-2.2425],[398,-2.2438],[399,-2.2162],[400,-2.2261],[401,-2.2658],[402,-2.2915],[404,-2.365],[406,-2.4195],[408,-2.4606],[409,-2.4727],[412,-2.4842],[415,-2.4821],[416,-2.4935],[419,-2.4796],[420,-2.4872],[424,-2.4641],[425,-2.4698],[426,-2.4701],[427,-2.4601],[428,-2.4586],[429,-2.4519],[433,-2.4467]]
       }
       arm_l_pv_ik_ctrl.Rotation.Z {
        type=MovieSceneScriptingFloatChannel
        default=0
        thinned_from=434
        keys [[0,24.9368],[12,24.8288],[14,24.8928],[15,25.1019],[17,26.005],[18,26.1913],[19,25.7841],[20,24.4606],[21,21.891],[22,18.6088],[23,15.4186],[24,12.4124],[25,9.7937],[26,7.6779],[27,6.0229],[28,4.786],[29,3.8638],[30,3.1498],[31,2.5966],[32,2.161],[33,1.8168],[34,1.6576],[35,1.6128],[36,1.6815],[37,1.8805],[38,2.205],[39,2.6007],[40,3.0137],[41,3.3934],[42,3.739],[43,4.0681],[44,4.3816],[45,4.6503],[46,4.8483],[47,5.0219],[49,5.2376],[50,5.3333],[51,5.4103],[53,5.5859],[54,5.7001],[55,5.8017],[56,5.8557],[57,5.8797],[58,5.8315],[59,5.7088],[60,5.4988],[61,5.1893],[62,4.661],[63,3.5942],[64,2.3158],[65,2.0925],[66,3.0083],[67,4.8427],[68,7.8756],[69,11.3155],[70,13.5185],[71,14.4925],[72,15.1983],[73,15.6703],[74,15.9263],[75,16.0426],[76,16.0656],[77,15.9821],[78,15.8051],[81,15.1094],[86,14.1613],[87,13.9951],[92,13.0829],[94,12.7621],[96,12.5079],[97,12.4516],[98,12.4338],[99,12.4765],[100,12.6719],[101,13.0803],[102,13.665],[103,14.3898],[104,15.3224],[105,16.4562],[106,17.6436],[107,18.7364],[109,20.7465],[110,21.6733],[111,22.4706],[112,23.0865],[113,23.5757],[115,24.3291],[117,24.9434],[119,25.4237],[122,25.9672],[125,26.3095],[130,26.4796],[133,26.3083],[135,26.0606],[136,25.8657],[137,25.599],[138,25.2405],[139,24.7683],[140,23.8715],[141,22.5075],[142,21.0651],[143,20.0438],[144,19.2159],[145,18.1129],[146,17.0525],[147,16.6275],[148,16.0911],[149,13.5462],[150,5.6931],[151,2.3498],[152,2.0074],[153,2.1676],[154,1.6788],[155,1.963],[156,2.1696],[157,-4.6736],[158,2.6832],[159,0.7808],[160,-4.2347],[161,-7.6226],[162,-10.349],[163,-11.8491],[164,-11.6852],[165,-10.1608],[166,-8.3116],[167,-6.8717],[168,-7.275],[169,-8.701],[170,-10.5521],[171,-11.2351],[172,-11.3369],[173,-11.2703],[174,-11.0344],[175,-10.5648],[176,-10.2384],[178,-9.703],[179,-9.4588],[180,-9.2392],[181,-9.0956],[182,-8.9897],[183,-9.0079],[184,-9.1408],[185,-9.3155],[187,-9.6998],[188,-9.8551],[190,-10.105],[191,-10.2122],[193,-10.3661],[194,-10.419],[199,-10.5352],[202,-10.5693],[206,-10.5873],[210,-10.6499],[212,-10.6388],[213,-10.6102],[215,-10.4837],[216,-10.3711],[217,-10.2097],[218,-9.98],[219,-9.6155],[220,-9.1306],[221,-8.3649],[222,-7.2343],[223,-5.8081],[224,-4.1628],[225,-1.82],[226,1.0276],[227,3.3263],[228,4.9638],[229,6.5371],[230,8.0181],[231,9.3692],[232,10.5475],[233,11.5487],[234,12.3848],[235,13.0692],[236,13.6154],[237,14.0202],[238,14.3652],[239,14.6405],[240,14.8648],[242,15.2003],[244,15.4083],[246,15.5401],[252,15.7284],[253,15.7401],[255,15.659],[256,15.4843],[257,15.0212],[258,14.0839],[259,12.4808],[260,9.1716],[261,4.4689],[262,0.2567],[263,-2.9861],[264,-5.8784],[265,-8.0795],[266,-9.5722],[267,-10.6685],[268,-11.5084],[269,-12.2349],[270,-12.8829],[271,-13.4069],[272,-13.8318],[273,-14.1835],[274,-14.4882],[275,-14.755],[277,-15.1684],[279,-15.4673],[281,-15.7059],[283,-15.8425],[286,-15.8592],[287,-15.8247],[288,-15.7244],[289,-15.5221],[290,-15.2394],[291,-14.9028],[292,-14.4899],[293,-13.9964],[294,-13.4612],[295,-12.8639],[297,-11.3759],[298,-10.7327],[299,-10.291],[300,-10.0322],[301,-9.8356],[304,-9.4094],[306,-9.215],[308,-9.0773],[309,-8.9847],[310,-8.9257],[313,-8.6794],[319,-8.3196],[320,-8.242],[321,-8.1971],[322,-8.1178],[323,-8.059],[325,-7.9099],[326,-7.8026],[327,-7.6444],[328,-7.4689],[329,-7.3104],[330,-7.2024],[331,-7.1407],[333,-7.1206],[334,-7.1428],[335,-7.1246],[336,-7.1211],[339,-7.0663],[341,-7.0642],[343,-7.1019],[344,-7.1],[345,-7.0421],[346,-6.8807],[347,-6.6476],[348,-6.3468],[349,-5.9823],[350,-5.5458],[351,-5.2554],[352,-4.6165],[353,-3.0744],[354,-0.2186],[355,7.1202],[356,16.7505],[357,22.8121],[358,26.806],[359,29.0655],[360,30.9431],[362,34.4208],[363,36.0596],[364,36.2234],[365,35.0195],[366,33.3792],[367,31.4758],[368,29.1827],[370,24.332],[371,22.02],[372,19.9225],[373,18.0085],[374,16.3156],[375,14.8418],[376,13.5577],[377,12.427],[378,11.4078],[379,10.5249],[380,9.8456],[381,9.4274],[382,9.3283],[383,9.5451],[384,10.0373],[385,10.7617],[386,11.7647],[387,13.0339],[389,15.8314],[390,17.0726],[391,18.1626],[392,19.1883],[393,20.1573],[394,21.0702],[395,21.8925],[396,22.6283],[397,23.288],[398,23.8711],[399,24.3833],[400,24.8108],[401,25.1417],[402,25.3909],[403,25.5823],[405,25.8277],[409,26.073],[427,25.0556],[433,24.9664]]
       }
       arm_l_pv_ik_ctrl.Scale.X {type=MovieSceneScriptingFloatChannel, default=1, const [1,434]}
       arm_l_pv_ik_ctrl.Scale.Y {type=MovieSceneScriptingFloatChannel, default=1, const [1,434]}
       arm_l_pv_ik_ctrl.Scale.Z {type=MovieSceneScriptingFloatChannel, default=1, const [1,434]}
       upperarm_r_fk_ctrl.Location.X {type=MovieSceneScriptingFloatChannel, default=0, thinned_from=434, keys [[0,0],[433,0]]}
       upperarm_r_fk_ctrl.Location.Y {type=MovieSceneScriptingFloatChannel, default=0, thinned_from=434, keys [[0,0],[433,0]]}
       upperarm_r_fk_ctrl.Location.Z {type=MovieSceneScriptingFloatChannel, default=0, thinned_from=434, keys [[0,0],[433,0]]}
       upperarm_r_fk_ctrl.Rotation.X {
        type=MovieSceneScriptingFloatChannel
        default=0
        thinned_from=434
        keys [[0,30.84],[11,30.8037],[13,30.625],[15,30.2608],[17,29.6765],[19,28.9609],[20,29.1811],[21,30.8644],[22,33.7188],[23,36.644],[24,39.1537],[25,41.0508],[26,42.4886],[27,43.5079],[28,44.1539],[29,44.5268],[30,44.7275],[33,44.7849],[36,44.2953],[44,42.1597],[46,41.8009],[56,41.1104],[59,41.1606],[61,41.3958],[63,41.7315],[64,41.4637],[65,42.0368],[66,43.5393],[67,43.5396],[68,40.7968],[69,37.0005],[70,35.1027],[71,34.7863],[72,34.7879],[73,34.9779],[75,35.5988],[77,36.2742],[81,37.0762],[85,37.5106],[96,37.5602],[99,37.341],[100,37.1557],[101,36.8351],[103,35.9146],[104,35.3143],[107,33.1726],[110,31.2496],[111,30.7195],[112,30.2781],[114,29.5872],[116,29.1082],[119,28.6732],[123,28.3575],[130,28.2792],[133,28.4892],[135,28.7516],[138,29.3602],[139,29.6128],[140,29.9373],[141,30.4328],[142,31.1877],[143,32.1404],[144,33.2345],[145,34.5729],[146,36.2429],[147,38.372],[148,41.496],[149,46.8464],[150,67.5119],[151,87.1656],[152,99.0224],[153,105.1341],[156,105.9567],[157,107.5462],[158,107.9919],[159,105.0414],[160,95.9376],[161,91.1733],[163,86.0832],[164,84.0086],[165,82.4344],[166,81.1699],[167,80.147],[168,80.2108],[169,82.2851],[170,86.0904],[171,88.5548],[173,92.235],[174,93.731],[175,94.6576],[176,95.1752],[177,95.0163],[182,91.4408],[185,90.055],[191,88.4654],[199,88.34],[214,89.9126],[225,89.3277],[226,89.0713],[229,87.7831],[257,85.261],[259,84.5628],[260,83.3054],[261,81.2651],[262,79.6231],[263,78.911],[264,78.6393],[265,78.8434],[266,79.581],[268,81.6615],[272,85.0422],[274,86.3719],[279,88.3164],[284,89.3182],[287,89.3382],[294,87.8345],[327,89.5368],[344,90.4788],[345,90.8342],[346,91.6181],[347,92.1382],[348,91.9178],[349,90.5285],[350,87.6491],[351,82.0928],[352,74.3076],[354,56.2537],[355,47.8214],[356,39.2664],[357,31.8465],[358,27.4122],[359,25.3727],[360,24.2393],[361,23.881],[362,24.1493],[363,24.7983],[364,26.2925],[366,30.898],[367,32.8816],[368,34.6556],[369,35.9845],[370,37.0313],[372,38.9072],[373,39.6838],[377,42.0571],[379,43.0454],[380,43.3803],[381,43.5277],[382,43.441],[383,43.1294],[384,42.6163],[385,41.9252],[386,41.0154],[387,39.9113],[389,37.5626],[390,36.5269],[392,34.705],[393,33.8778],[394,33.1398],[395,32.5138],[397,31.5044],[398,31.0958],[399,30.759],[400,30.4997],[401,30.3237],[403,30.1284],[409,29.9615],[430,30.8013],[433,30.8132]]
       }
       upperarm_r_fk_ctrl.Rotation.Y {
        type=MovieSceneScriptingFloatChannel
        default=0
        thinned_from=434
        keys [[0,6.457],[2,6.457],[5,6.4306],[6,6.4306],[7,6.4175],[11,6.4012],[13,6.4591],[14,6.519],[15,6.6321],[16,6.7795],[17,6.8988],[18,6.9269],[19,6.8715],[20,6.7423],[21,6.5536],[22,6.3209],[23,6.116],[24,5.9814],[25,5.7746],[26,5.3523],[30,3.2315],[31,2.7706],[32,2.4425],[33,2.2218],[34,2.0263],[35,1.8702],[36,1.7498],[37,1.667],[38,1.6174],[39,1.5923],[40,1.5929],[41,1.5826],[42,1.6107],[43,1.6762],[44,1.764],[47,2.0455],[48,2.1476],[49,2.2579],[50,2.3538],[51,2.4661],[52,2.5695],[53,2.6494],[54,2.7363],[55,2.792],[57,2.9168],[58,2.9905],[59,3.0801],[60,3.1918],[61,3.3318],[62,3.548],[63,3.9044],[64,4.6166],[65,6.063],[66,7.2679],[67,7.7058],[68,7.9502],[69,7.8966],[70,7.7541],[71,7.574],[72,7.1892],[73,6.7081],[74,6.0292],[76,4.6082],[77,4.1466],[78,3.8725],[79,3.7648],[80,3.6678],[81,3.6041],[83,3.5474],[84,3.5304],[85,3.5299],[87,3.5408],[91,3.5868],[92,3.6159],[93,3.6069],[94,3.5756],[95,3.5316],[96,3.4698],[97,3.4003],[99,3.2443],[100,3.1544],[101,3.055],[102,2.9624],[103,2.8924],[104,2.8339],[105,2.7845],[106,2.7695],[107,2.8111],[108,2.9189],[110,3.2113],[111,3.3286],[112,3.4316],[114,3.6645],[115,3.771],[117,3.9678],[118,4.0596],[119,4.1399],[120,4.2064],[121,4.252],[124,4.3369],[127,4.395],[128,4.4068],[131,4.425],[133,4.4523],[134,4.4516],[136,4.404],[137,4.3884],[138,4.3983],[139,4.4485],[140,4.3151],[141,3.9559],[142,3.6469],[143,3.5403],[144,3.5017],[145,3.3802],[146,3.014],[147,2.3584],[148,2.1936],[149,3.0429],[150,7.1913],[151,6.1688],[152,9.7159],[153,14.0108],[154,15.9757],[155,17.2296],[156,18.662],[157,28.5446],[158,39.498],[159,40.5882],[160,36.1391],[161,32.5053],[162,29.393],[163,26.3988],[164,24.1861],[165,22.6858],[166,21.6068],[167,20.9046],[168,21.3109],[169,24.0613],[170,27.7273],[171,29.8773],[172,31.3061],[173,32.2073],[174,32.6799],[175,32.8732],[176,32.9105],[177,32.7311],[178,32.3751],[181,31.0697],[183,30.4156],[185,29.9646],[191,29.0411],[198,28.4047],[204,28.1108],[206,27.8994],[209,27.3291],[212,26.4841],[214,25.8338],[215,25.415],[218,23.918],[219,23.5889],[220,23.4613],[221,23.6049],[222,24.0238],[223,24.7076],[224,25.6466],[225,27.1247],[226,29.0784],[227,30.9477],[228,32.9035],[229,35.2155],[230,37.4469],[231,39.1608],[232,40.3827],[233,41.4427],[234,42.3595],[235,43.1519],[236,43.8387],[238,44.9628],[240,45.7951],[243,46.59],[249,47.6895],[258,48.8417],[259,49.1395],[261,50.5372],[262,50.2133],[263,48.6256],[265,44.1536],[268,37.858],[269,35.9973],[270,34.327],[271,32.7528],[272,31.3223],[273,30.0867],[274,29.0999],[275,28.3422],[276,27.737],[277,27.2542],[278,26.8634],[286,24.5091],[287,24.1787],[288,23.7766],[291,22.3771],[292,21.9727],[293,21.6156],[294,21.3657],[295,21.2742],[296,21.3438],[299,21.8757],[310,23.6083],[315,24.4356],[325,25.5636],[327,25.9378],[329,26.3845],[332,26.8216],[337,27.2196],[342,27.4936],[344,27.7451],[345,28.0072],[346,28.5312],[349,30.3645],[351,31.373],[352,32.0075],[353,31.8882],[354,30.9159],[355,26.9348],[356,21.9498],[357,18.4066],[358,14.5283],[359,12.0026],[360,9.8219],[361,8.1722],[362,6.7701],[363,5.8131],[364,5.1297],[365,4.4959],[366,3.773],[367,3.0955],[368,2.7986],[369,3.1349],[370,3.8166],[371,4.6977],[372,5.5241],[373,6.4159],[374,7.4261],[375,8.1981],[376,8.6148],[377,8.8114],[378,8.7685],[379,8.5443],[381,7.9301],[384,7.1008],[386,6.4572],[387,6.1028],[389,5.3154],[390,4.9658],[391,4.7078],[392,4.5281],[393,4.3806],[394,4.264],[395,4.1788],[396,4.1236],[397,4.0951],[398,4.092],[399,4.1127],[400,4.1565],[401,4.2224],[402,4.3062],[403,4.4077],[405,4.6646],[406,4.823],[407,4.9624],[412,5.7057],[414,5.9718],[415,6.086],[416,6.1832],[418,6.3463],[420,6.4807],[422,6.5794],[424,6.6329],[426,6.6417],[431,6.4854],[432,6.4631],[433,6.4547]]
       }
       upperarm_r_fk_ctrl.Rotation.Z {
        type=MovieSceneScriptingFloatChannel
        default=0
        thinned_from=434
        keys [[0,-5.7867],[2,-5.7867],[3,-5.7669],[4,-5.7669],[5,-5.7422],[6,-5.7422],[7,-5.72],[10,-5.7182],[11,-5.7559],[12,-5.8608],[13,-6.0589],[14,-6.3824],[15,-6.8324],[16,-7.4293],[17,-8.2353],[18,-9.3169],[19,-11.1609],[20,-13.0386],[21,-13.0578],[22,-11.2416],[23,-9.2302],[24,-7.3923],[25,-5.8461],[26,-4.5197],[27,-3.4671],[28,-2.7346],[29,-2.2891],[30,-2.0312],[31,-1.906],[32,-1.8619],[33,-1.8547],[34,-1.9001],[35,-1.9652],[36,-2.044],[37,-2.1326],[40,-2.4338],[41,-2.5291],[42,-2.6362],[43,-2.7804],[45,-3.1003],[46,-3.2329],[48,-3.4475],[50,-3.64],[53,-3.9009],[55,-4.0497],[56,-4.0994],[57,-4.1397],[58,-4.1577],[59,-4.1559],[60,-4.133],[61,-4.0877],[62,-3.995],[64,-3.785],[65,-3.07],[66,-1.9928],[67,-1.1907],[68,-1.2118],[69,-1.5919],[70,-1.8153],[71,-2.2141],[72,-2.5488],[73,-2.8182],[74,-3.0243],[76,-3.3677],[78,-3.6524],[79,-3.7887],[80,-3.8998],[81,-3.9841],[82,-4.0462],[83,-4.0951],[84,-4.1275],[85,-4.1403],[86,-4.1304],[87,-4.1069],[89,-3.9912],[90,-3.9105],[91,-3.8097],[92,-3.6987],[93,-3.5779],[94,-3.4438],[95,-3.3019],[96,-3.1498],[98,-2.8599],[99,-2.7207],[100,-2.5977],[101,-2.5022],[102,-2.4414],[103,-2.423],[104,-2.4303],[105,-2.4657],[106,-2.5676],[107,-2.7753],[108,-3.1168],[109,-3.5533],[110,-4.0229],[111,-4.4617],[113,-5.2675],[114,-5.6456],[115,-5.9831],[116,-6.288],[117,-6.5752],[118,-6.837],[119,-7.0654],[120,-7.2459],[121,-7.3939],[122,-7.5121],[123,-7.6052],[125,-7.7264],[127,-7.8072],[130,-7.8687],[132,-7.8261],[133,-7.7854],[134,-7.7222],[135,-7.6139],[136,-7.4447],[137,-7.198],[138,-6.8569],[139,-6.4038],[140,-5.7122],[141,-4.6198],[142,-3.0515],[143,-1.1211],[144,1.0562],[145,3.5521],[146,6.4233],[147,11.5331],[148,18.2658],[149,27.6435],[150,48.5168],[151,61.2059],[152,71.0285],[153,77.0017],[154,78.9202],[155,79.5487],[156,80.9521],[157,112.6155],[158,150.3015],[159,156.3589],[160,145.9205],[161,138.6784],[162,133.6845],[163,128.3859],[164,125.0291],[165,122.9238],[166,121.2173],[167,119.9702],[168,120.0374],[169,123.3899],[170,127.941],[172,135.8466],[173,138.6851],[174,139.675],[175,139.8068],[176,139.4563],[177,138.6428],[180,134.8817],[182,133.0474],[190,130.4697],[219,131.159],[222,130.5596],[224,129.6081],[225,128.6179],[227,126.3049],[230,124.5716],[242,122.3809],[259,121.4999],[260,120.6511],[261,119.1116],[262,117.9637],[263,117.7984],[264,118.126],[265,118.8736],[268,122.8431],[272,126.9204],[274,128.3424],[281,130.8053],[294,131.1875],[345,130.4683],[346,130.1164],[347,129.0372],[348,126.9323],[349,123.528],[350,118.6382],[351,110.877],[352,100.3919],[354,76.2266],[355,55.8029],[356,33.7086],[357,20.3092],[358,12.5211],[359,6.8783],[360,2.1912],[361,-0.6807],[362,-3.2377],[363,-6.1464],[364,-8.8108],[365,-10.4364],[366,-11.0215],[367,-10.8693],[368,-10.1632],[369,-9.1375],[370,-7.9134],[372,-5.382],[373,-4.1439],[374,-2.9465],[375,-1.8103],[376,-0.7667],[377,0.1804],[378,1.031],[379,1.7875],[380,2.4421],[381,2.9765],[382,3.3756],[383,3.6487],[384,3.8042],[385,3.8324],[386,3.7028],[387,3.4663],[388,3.1981],[389,2.9104],[390,2.6095],[391,2.2392],[392,1.7748],[394,0.7847],[395,0.3141],[396,-0.123],[397,-0.542],[398,-0.9507],[399,-1.3423],[400,-1.7103],[401,-2.0473],[402,-2.3585],[403,-2.6635],[404,-2.9617],[405,-3.2696],[406,-3.5677],[407,-3.8584],[408,-4.1392],[409,-4.4024],[411,-4.8583],[412,-5.066],[413,-5.2624],[414,-5.4389],[415,-5.5917],[416,-5.7153],[418,-5.9081],[419,-5.9877],[420,-6.055],[421,-6.1092],[423,-6.1797],[424,-6.19],[426,-6.1601],[428,-6.0443],[430,-5.9049],[431,-5.8518],[432,-5.8142],[433,-5.7999]]
       }
       upperarm_r_fk_ctrl.Scale.X {type=MovieSceneScriptingFloatChannel, default=1, const [1,434]}
       upperarm_r_fk_ctrl.Scale.Y {type=MovieSceneScriptingFloatChannel, default=1, const [1,434]}
       upperarm_r_fk_ctrl.Scale.Z {type=MovieSceneScriptingFloatChannel, default=1, const [1,434]}
       lowerarm_r_fk_ctrl.Location.X {type=MovieSceneScriptingFloatChannel, default=0, thinned_from=434, keys [[0,-0.0001],[433,-0.0001]]}
       lowerarm_r_fk_ctrl.Location.Y {type=MovieSceneScriptingFloatChannel, default=0, thinned_from=434, keys [[0,0],[433,0]]}
       lowerarm_r_fk_ctrl.Location.Z {type=MovieSceneScriptingFloatChannel, default=0, thinned_from=434, keys [[0,0],[433,0]]}
       lowerarm_r_fk_ctrl.Rotation.X {
        type=MovieSceneScriptingFloatChannel
        default=0
        thinned_from=434
        keys [[0,0.0323],[1,0.0323],[7,0.0303],[8,0.0301],[11,0.0301],[13,0.0309],[14,0.0315],[15,0.0323],[16,0.0334],[17,0.0347],[18,0.0363],[19,0.0402],[20,0.046],[21,0.0506],[22,0.0532],[23,0.0547],[24,0.0552],[25,0.0552],[26,0.0545],[27,0.0529],[28,0.0497],[29,0.0448],[30,0.0383],[31,0.0309],[33,0.0145],[34,0.0068],[35,0.0001],[36,-0.0048],[37,-0.0079],[38,-0.0098],[39,-0.0108],[40,-0.0112],[41,-0.0112],[42,-0.0098],[43,-0.0063],[45,0.003],[46,0.0068],[48,0.0129],[49,0.0157],[50,0.0183],[51,0.0206],[52,0.0225],[53,0.0239],[54,0.025],[55,0.0257],[58,0.0272],[59,0.0275],[61,0.0275],[62,0.0267],[63,0.0233],[64,0.0156],[65,-0.0042],[66,-0.0265],[67,-0.0353],[68,-0.0371],[69,-0.0327],[70,-0.0226],[71,-0.0111],[72,0],[73,0.014],[74,0.0256],[75,0.0356],[76,0.0444],[77,0.051],[78,0.0556],[79,0.0592],[80,0.0616],[81,0.0624],[82,0.0617],[83,0.0598],[84,0.0572],[85,0.0545],[86,0.052],[90,0.0426],[91,0.0404],[92,0.0384],[94,0.0351],[96,0.0325],[98,0.0306],[101,0.0285],[104,0.0271],[105,0.0268],[106,0.0268],[107,0.0266],[108,0.0269],[109,0.0278],[112,0.0316],[123,0.0477],[125,0.0503],[127,0.052],[129,0.0533],[131,0.0541],[132,0.0544],[134,0.0544],[135,0.0534],[136,0.0515],[137,0.0479],[138,0.042],[139,0.0331],[140,0.0197],[141,0.0015],[142,-0.0205],[143,-0.0447],[145,-0.0937],[146,-0.1074],[147,-0.1034],[148,-0.0866],[149,-0.0502],[150,0.0861],[151,0.2175],[152,0.1795],[153,0.183],[157,0.183],[158,-1.3056],[159,-3.8405],[160,-1.7724],[161,-0.6631],[162,-0.3923],[163,-0.0216],[164,0.0791],[165,0.183],[344,0.183],[345,0.1846],[346,0.1886],[348,0.1994],[349,0.2038],[350,0.2063],[351,0.2057],[352,0.2026],[353,0.1975],[354,0.1892],[355,0.1639],[356,0.1076],[357,0.023],[358,-0.0169],[359,-0.0465],[360,-0.058],[361,-0.0392],[363,0.0249],[364,0.0605],[365,0.1068],[366,0.1329],[367,0.1394],[368,0.1415],[369,0.1419],[370,0.1415],[371,0.1399],[372,0.1366],[373,0.13],[374,0.1186],[375,0.1023],[376,0.079],[377,0.0485],[378,0.0201],[379,-0.0028],[380,-0.0234],[381,-0.0383],[382,-0.0439],[383,-0.042],[384,-0.0372],[387,-0.0179],[388,-0.011],[390,0.0035],[391,0.0105],[392,0.017],[393,0.0229],[394,0.028],[395,0.0322],[396,0.0351],[397,0.0372],[398,0.0386],[399,0.0396],[400,0.0402],[401,0.0405],[404,0.0405],[406,0.0395],[412,0.0342],[414,0.0328],[415,0.0324],[433,0.0324]]
       }
       lowerarm_r_fk_ctrl.Rotation.Y {
        type=MovieSceneScriptingFloatChannel
        default=0
        thinned_from=434
        keys [[0,0.1308],[1,0.1308],[3,0.1283],[6,0.1235],[7,0.1221],[8,0.1212],[11,0.1212],[12,0.1228],[13,0.1247],[14,0.1273],[15,0.131],[16,0.1356],[17,0.1415],[18,0.1487],[19,0.1663],[20,0.1936],[21,0.2159],[22,0.2283],[23,0.2356],[24,0.238],[25,0.238],[26,0.2347],[27,0.2268],[28,0.2114],[29,0.1877],[30,0.158],[31,0.1246],[32,0.0897],[33,0.0554],[34,0.0242],[35,-0.002],[36,-0.0208],[37,-0.0325],[38,-0.0397],[39,-0.0433],[40,-0.0447],[41,-0.0447],[42,-0.0396],[43,-0.0263],[44,-0.0087],[45,0.0094],[46,0.0244],[48,0.0487],[49,0.0603],[50,0.071],[51,0.0806],[52,0.0885],[53,0.0946],[54,0.0991],[55,0.1024],[56,0.1048],[58,0.1085],[59,0.1098],[61,0.1098],[62,0.1067],[63,0.0918],[64,0.06],[65,-0.0184],[66,-0.1],[67,-0.1308],[68,-0.137],[69,-0.1219],[70,-0.0862],[71,-0.0445],[72,0],[73,0.0534],[74,0.102],[75,0.1458],[76,0.1862],[77,0.2174],[78,0.2403],[79,0.2585],[80,0.2706],[81,0.2749],[82,0.2711],[83,0.2615],[84,0.2485],[85,0.2346],[86,0.2224],[87,0.2114],[89,0.1885],[90,0.1775],[91,0.1673],[92,0.1583],[93,0.1504],[94,0.1434],[95,0.1372],[96,0.1318],[97,0.1272],[98,0.1234],[99,0.12],[100,0.117],[101,0.1143],[102,0.1119],[103,0.1099],[104,0.1083],[105,0.1072],[106,0.1072],[107,0.1062],[108,0.1076],[109,0.1111],[110,0.1161],[111,0.1218],[112,0.1278],[114,0.1407],[119,0.1748],[121,0.1881],[123,0.2019],[124,0.2084],[125,0.2141],[126,0.2188],[127,0.2227],[128,0.2261],[129,0.229],[130,0.2313],[131,0.233],[132,0.234],[134,0.234],[135,0.2295],[136,0.2199],[137,0.2024],[138,0.1746],[139,0.1344],[140,0.0769],[141,0.0035],[142,-0.0787],[143,-0.1627],[144,-0.2418],[145,-0.3164],[146,-0.3562],[147,-0.2969],[148,0.0878],[149,1.1085],[150,6.0667],[151,10.7643],[152,12.1398],[153,12.3291],[344,12.3291],[345,12.3124],[346,12.2509],[347,12.1274],[348,11.9249],[349,11.6261],[350,11.2141],[351,10.6482],[352,9.8769],[353,8.8668],[354,7.5846],[355,5.3597],[356,2.6404],[357,0.328],[358,0.0195],[359,-0.14],[360,-0.1854],[361,-0.1233],[362,-0.0083],[363,0.1197],[364,0.2856],[365,0.5741],[366,0.8442],[367,0.9681],[368,1.0358],[369,1.0563],[370,1.0363],[371,0.9822],[372,0.9029],[373,0.7953],[374,0.6629],[375,0.5241],[376,0.3705],[377,0.2096],[378,0.0823],[379,-0.0096],[380,-0.0857],[381,-0.1376],[382,-0.157],[383,-0.1508],[384,-0.1348],[385,-0.1134],[386,-0.091],[387,-0.0682],[388,-0.0429],[389,-0.0159],[391,0.0394],[392,0.066],[393,0.0906],[394,0.1124],[395,0.1304],[396,0.1436],[397,0.1528],[398,0.1593],[399,0.1637],[400,0.1664],[401,0.1678],[404,0.1678],[405,0.1659],[406,0.1633],[407,0.1599],[409,0.1517],[411,0.1432],[412,0.1393],[413,0.1359],[414,0.1332],[415,0.1315],[433,0.1315]]
       }
       lowerarm_r_fk_ctrl.Rotation.Z {
        type=MovieSceneScriptingFloatChannel
        default=0
        thinned_from=434
        keys [[0,-1.6971],[1,-1.6971],[3,-1.6632],[6,-1.6002],[7,-1.5828],[8,-1.5708],[11,-1.5708],[12,-1.5919],[13,-1.6162],[14,-1.6512],[15,-1.6986],[16,-1.76],[17,-1.8372],[18,-1.9317],[19,-2.1626],[20,-2.5199],[21,-2.8115],[22,-2.9732],[23,-3.0692],[24,-3.1009],[25,-3.1009],[26,-3.0574],[27,-2.9541],[28,-2.753],[29,-2.4428],[30,-2.0537],[31,-1.6149],[32,-1.1555],[33,-0.7049],[34,-0.2921],[35,0.0535],[36,0.3027],[37,0.4583],[38,0.5527],[39,0.6012],[40,0.619],[41,0.619],[42,0.5519],[43,0.3759],[44,0.143],[45,-0.0971],[46,-0.295],[47,-0.456],[48,-0.6154],[49,-0.7685],[50,-0.9102],[51,-1.0358],[52,-1.1403],[53,-1.2206],[54,-1.28],[55,-1.3232],[56,-1.355],[58,-1.4037],[59,-1.4208],[61,-1.4208],[62,-1.3797],[63,-1.184],[64,-0.7647],[65,0.2707],[66,1.3546],[67,1.7647],[68,1.8483],[69,1.6464],[70,1.1709],[71,0.6172],[72,0],[73,-0.6781],[74,-1.3179],[75,-1.8937],[76,-2.4225],[77,-2.8308],[78,-3.1302],[79,-3.3684],[80,-3.5256],[81,-3.5824],[82,-3.5332],[83,-3.4074],[84,-3.2375],[85,-3.0562],[86,-2.8961],[87,-2.7527],[89,-2.4537],[90,-2.309],[91,-2.175],[92,-2.0571],[93,-1.9544],[94,-1.8625],[95,-1.7811],[96,-1.7102],[97,-1.6496],[98,-1.599],[99,-1.5552],[100,-1.5153],[101,-1.4797],[102,-1.4485],[103,-1.4223],[104,-1.4013],[105,-1.3859],[106,-1.3859],[107,-1.3732],[108,-1.3911],[109,-1.438],[110,-1.5038],[111,-1.5787],[112,-1.6571],[114,-1.826],[119,-2.2739],[121,-2.4479],[123,-2.6286],[124,-2.7132],[125,-2.7884],[126,-2.85],[127,-2.9005],[128,-2.9448],[129,-2.9825],[130,-3.0128],[131,-3.0352],[132,-3.049],[134,-3.049],[135,-2.99],[136,-2.8642],[137,-2.635],[138,-2.2717],[139,-1.7437],[140,-0.987],[141,-0.0187],[142,1.0708],[143,2.191],[144,3.2513],[145,4.2578],[146,4.7971],[147,4.6485],[148,4.0231],[149,2.6521],[150,-2.4776],[151,-18.8173],[152,-24.2923],[157,-24.277],[159,-24.3565],[344,-24.277],[345,-24.1509],[346,-23.7989],[347,-23.2608],[348,-22.5762],[349,-21.785],[351,-20.0388],[352,-18.9846],[353,-17.5427],[354,-15.4917],[355,-11.5104],[356,-6.3501],[357,-1.1291],[358,0.9086],[359,2.2768],[360,2.777],[361,1.9451],[362,0.4158],[363,-1.2745],[364,-3.4479],[365,-7.1914],[366,-10.6675],[367,-12.2561],[368,-13.1237],[369,-13.387],[370,-13.133],[371,-12.4508],[372,-11.4599],[373,-10.1358],[374,-8.5031],[375,-6.7547],[376,-4.7742],[377,-2.6782],[378,-1.01],[379,0.2008],[380,1.2097],[381,1.9004],[382,2.1564],[383,2.0696],[384,1.85],[385,1.5585],[386,1.2563],[387,0.9488],[388,0.6087],[389,0.2478],[390,-0.122],[391,-0.4889],[392,-0.841],[393,-1.1664],[394,-1.4533],[395,-1.6898],[396,-1.8641],[397,-1.9843],[398,-2.0704],[399,-2.1282],[400,-2.1632],[401,-2.1812],[404,-2.1812],[405,-2.1575],[406,-2.1223],[407,-2.0778],[409,-1.9713],[411,-1.8593],[412,-1.8081],[413,-1.7636],[414,-1.7284],[415,-1.7054],[433,-1.7054]]
       }
       lowerarm_r_fk_ctrl.Scale.X {type=MovieSceneScriptingFloatChannel, default=1, const [1,434]}
       lowerarm_r_fk_ctrl.Scale.Y {type=MovieSceneScriptingFloatChannel, default=1, const [1,434]}
       lowerarm_r_fk_ctrl.Scale.Z {type=MovieSceneScriptingFloatChannel, default=1, const [1,434]}
       hand_r_fk_ctrl.Location.X {type=MovieSceneScriptingFloatChannel, default=0, thinned_from=434, keys [[0,-0.0001],[433,-0.0001]]}
       hand_r_fk_ctrl.Location.Y {type=MovieSceneScriptingFloatChannel, default=0, thinned_from=434, keys [[0,0],[433,0]]}
       hand_r_fk_ctrl.Location.Z {type=MovieSceneScriptingFloatChannel, default=0, thinned_from=434, keys [[0,0],[433,0]]}
       hand_r_fk_ctrl.Rotation.X {
        type=MovieSceneScriptingFloatChannel
        default=0
        thinned_from=434
        keys [[0,-1.844],[133,-1.844],[134,-1.8122],[135,-1.7209],[136,-1.5766],[137,-1.3858],[138,-1.1556],[139,-0.8941],[140,-0.5736],[141,-0.137],[142,0.4804],[143,1.7775],[144,2.9047],[145,3.0427],[146,3.0827],[147,3.1604],[148,3.23],[149,3.2602],[150,-0.5926],[151,-21.71],[152,-15.532],[153,-9.5481],[154,-6.9227],[155,-4.5375],[156,-3.168],[157,-2.2949],[159,-2.2949],[160,-3.5075],[214,-3.5075],[215,-3.4496],[216,-3.2779],[217,-2.9951],[218,-2.6043],[219,-2.1082],[220,-1.5098],[221,-0.5943],[222,0.6556],[223,1.9417],[224,2.9675],[225,3.5667],[226,3.8822],[227,4.0827],[228,4.2326],[229,4.3065],[230,4.3382],[232,4.3863],[233,4.4031],[236,4.4031],[237,4.4264],[239,4.4264],[240,4.4408],[251,4.4408],[253,4.3844],[254,4.3435],[256,4.2413],[257,4.1549],[258,3.9292],[259,3.4634],[260,2.0782],[261,-0.0561],[262,-1.7593],[263,-2.548],[264,-2.9001],[265,-3.0701],[266,-3.1885],[267,-3.2401],[269,-3.2887],[271,-3.3683],[273,-3.4308],[275,-3.4692],[276,-3.4822],[277,-3.4822],[278,-3.4993],[344,-3.4993],[345,-3.5041],[346,-3.4933],[347,-3.4731],[348,-3.4411],[349,-3.3963],[350,-3.3392],[351,-3.2714],[352,-3.1818],[353,-3.0479],[354,-2.8442],[355,-2.4152],[356,-2.0204],[357,-1.9358],[358,-1.9035],[359,-1.8821],[361,-1.8494],[433,-1.844]]
       }
       hand_r_fk_ctrl.Rotation.Y {
        type=MovieSceneScriptingFloatChannel
        default=0
        thinned_from=434
        keys [[0,14.0562],[133,14.0562],[134,14.1114],[135,14.275],[136,14.5437],[137,14.9142],[138,15.3833],[139,15.9475],[140,16.798],[141,18.0775],[142,19.7118],[143,22.5335],[144,24.9054],[145,25.2054],[146,25.2924],[148,25.6133],[149,25.6791],[150,2.8513],[151,-14.6173],[152,-18.691],[153,-22.9968],[154,-25.9857],[155,-28.0701],[156,-30.4699],[157,-34.168],[159,-34.168],[160,-28.9924],[215,-29.0116],[217,-29.1625],[219,-29.458],[220,-29.6581],[221,-29.9652],[223,-30.8236],[224,-31.1742],[225,-31.3799],[228,-31.6094],[257,-31.5826],[258,-31.5048],[259,-31.3444],[260,-30.8702],[261,-30.1465],[262,-29.5746],[263,-29.3114],[264,-29.1942],[344,-28.9951],[345,-28.7465],[346,-28.059],[347,-27.0059],[348,-25.6639],[349,-24.1091],[350,-22.4169],[351,-20.6538],[352,-18.5573],[353,-15.7165],[354,-11.7366],[355,-3.5865],[356,5.4628],[357,8.2571],[358,9.6073],[359,10.6887],[360,11.8037],[361,13.1885],[362,13.5989],[363,13.8195],[364,13.9249],[365,13.964],[433,14.0464]]
       }
       hand_r_fk_ctrl.Rotation.Z {
        type=MovieSceneScriptingFloatChannel
        default=0
        thinned_from=434
        keys [[0,-4.4811],[133,-4.4811],[134,-4.4922],[135,-4.5871],[136,-4.8586],[137,-5.3994],[138,-6.3025],[139,-7.6612],[140,-10.9548],[142,-21.5562],[143,-26.8909],[144,-30.7341],[145,-31.2713],[146,-31.4273],[148,-32.003],[149,-32.1212],[150,-17.6521],[151,-32.5813],[152,-48.5437],[153,-64.3174],[154,-72.8634],[155,-79.3213],[156,-82.104],[216,-82.0232],[220,-81.3241],[224,-79.5185],[228,-78.999],[257,-79.031],[259,-79.3153],[260,-79.8812],[261,-80.7434],[262,-81.4232],[263,-81.7355],[344,-82.1102],[345,-81.683],[346,-80.4786],[347,-78.6318],[348,-76.274],[349,-73.5359],[351,-67.4237],[352,-63.695],[353,-58.6176],[354,-51.4566],[355,-36.6445],[356,-20.0874],[357,-14.9871],[358,-12.5298],[359,-10.5657],[360,-8.5455],[361,-6.0438],[362,-5.3041],[363,-4.907],[364,-4.7172],[365,-4.6469],[367,-4.6143],[368,-4.6143],[372,-4.5122],[373,-4.4986],[433,-4.4986]]
       }
       hand_r_fk_ctrl.Scale.X {type=MovieSceneScriptingFloatChannel, default=1, const [1,434]}
       hand_r_fk_ctrl.Scale.Y {type=MovieSceneScriptingFloatChannel, default=1, const [1,434]}
       hand_r_fk_ctrl.Scale.Z {type=MovieSceneScriptingFloatChannel, default=1, const [1,434]}
       thumb_01_r_ctrl.Location.X {type=MovieSceneScriptingFloatChannel, default=0, const [-0.0001,434]}
       thumb_01_r_ctrl.Location.Y {type=MovieSceneScriptingFloatChannel, default=0, const [0.0001,434]}
       thumb_01_r_ctrl.Location.Z {type=MovieSceneScriptingFloatChannel, default=0, const [0.0001,434]}
       thumb_01_r_ctrl.Rotation.X {
        type=MovieSceneScriptingFloatChannel
        default=0
        thinned_from=434
        keys [[0,5.0222],[146,5.0222],[147,3.4387],[148,0.4979],[149,-1.0856],[150,-1.0856],[151,-10.2473],[152,-12.0988],[153,-12.5378],[154,-12.8375],[155,-12.8997],[156,-13.3012],[157,-16.0497],[158,-16.7368],[159,-16.8931],[161,-16.9505],[344,-16.9618],[345,-16.8323],[346,-16.4773],[347,-15.9696],[349,-14.7709],[350,-14.2174],[351,-13.8982],[352,-13.6792],[353,-13.1567],[354,-11.927],[355,-7.2626],[356,-2.3835],[357,-1.2264],[358,-0.952],[359,-0.7647],[360,-0.7258],[361,-0.4748],[362,1.4482],[363,3.6479],[364,4.3255],[365,4.545],[366,4.6786],[367,4.8373],[368,4.91],[369,4.9363],[372,4.9363],[373,4.9578],[374,4.9578],[375,4.9825],[376,4.9825],[377,5.0061],[433,5.0061]]
       }
       thumb_01_r_ctrl.Rotation.Y {
        type=MovieSceneScriptingFloatChannel
        default=0
        thinned_from=434
        keys [[0,16.5344],[146,16.5344],[147,16.4585],[148,16.3175],[149,16.2416],[150,16.2416],[151,15.8023],[152,15.7136],[156,15.6559],[157,15.5242],[345,15.4866],[353,15.6629],[354,15.7218],[356,16.1794],[357,16.2348],[361,16.2709],[363,16.4685],[433,16.5337]]
       }
       thumb_01_r_ctrl.Rotation.Z {
        type=MovieSceneScriptingFloatChannel
        default=0
        thinned_from=434
        keys [[0,8.3802],[134,8.3802],[135,8.3104],[136,8.1448],[137,7.8223],[138,7.2906],[139,6.4975],[140,4.6496],[141,1.8604],[142,-0.5627],[143,-1.8841],[144,-2.5239],[145,-2.7813],[146,-2.8181],[147,-1.9334],[148,-0.2903],[149,0.5944],[150,0.5944],[151,5.7131],[152,6.7474],[153,6.9927],[154,7.1602],[155,7.1949],[156,7.4193],[157,8.9696],[158,9.3571],[159,9.4453],[161,9.4777],[345,9.4796],[351,9.332],[353,9.2947],[354,9.2328],[356,8.7527],[357,8.6945],[361,8.6567],[363,8.4493],[364,8.4152],[433,8.381]]
       }
       thumb_01_r_ctrl.Scale.X {type=MovieSceneScriptingFloatChannel, default=1, const [1,434]}
       thumb_01_r_ctrl.Scale.Y {type=MovieSceneScriptingFloatChannel, default=1, const [1,434]}
       thumb_01_r_ctrl.Scale.Z {type=MovieSceneScriptingFloatChannel, default=1, const [1,434]}
       thumb_02_r_ctrl.Location.X {type=MovieSceneScriptingFloatChannel, default=0, thinned_from=434, keys [[0,0],[146,0],[378,0.0001],[433,0]]}
       thumb_02_r_ctrl.Location.Y {type=MovieSceneScriptingFloatChannel, default=0, thinned_from=434, keys [[0,0],[433,0]]}
       thumb_02_r_ctrl.Location.Z {type=MovieSceneScriptingFloatChannel, default=0, thinned_from=434, keys [[0,0],[433,0]]}
       thumb_02_r_ctrl.Rotation.X {
        type=MovieSceneScriptingFloatChannel
        default=0
        thinned_from=434
        keys [[0,0],[146,0],[147,-0.0359],[148,-0.1026],[149,-0.1385],[150,-0.1385],[151,-0.3462],[152,-0.3881],[153,-0.3981],[154,-0.4049],[155,-0.4063],[156,-0.4154],[157,-0.4777],[158,-0.4933],[159,-0.4968],[344,-0.4984],[345,-0.4954],[346,-0.4874],[347,-0.4759],[349,-0.4487],[350,-0.4362],[351,-0.4289],[352,-0.424],[353,-0.4121],[354,-0.3842],[355,-0.2785],[356,-0.1679],[357,-0.1417],[358,-0.1354],[359,-0.1312],[360,-0.1303],[361,-0.1246],[362,-0.081],[363,-0.0312],[364,-0.0158],[365,-0.0108],[366,-0.0078],[367,-0.0042],[368,-0.0025],[369,-0.0019],[370,-0.0017],[372,-0.0017],[377,-0.0004],[378,-0.0004],[379,0],[433,0]]
       }
       thumb_02_r_ctrl.Rotation.Y {
        type=MovieSceneScriptingFloatChannel
        default=0
        thinned_from=434
        keys [[0,0],[146,0],[147,2.0237],[148,5.7821],[149,7.8058],[150,7.8058],[151,19.5145],[152,21.8807],[153,22.4417],[154,22.8248],[155,22.9043],[156,23.4174],[157,26.93],[158,27.8081],[159,28.0079],[344,28.0957],[345,27.9301],[346,27.4764],[347,26.8275],[349,25.2957],[350,24.5883],[351,24.1803],[352,23.9004],[353,23.2326],[354,21.6611],[355,15.7],[356,9.4645],[357,7.9857],[358,7.635],[359,7.3956],[360,7.3459],[361,7.0252],[362,4.5676],[363,1.7563],[364,0.8903],[365,0.6098],[366,0.4391],[367,0.2363],[368,0.1433],[369,0.1098],[370,0.096],[372,0.096],[373,0.0823],[374,0.0682],[375,0.0507],[376,0.0336],[377,0.0206],[378,0.0206],[379,0],[433,0]]
       }
       thumb_02_r_ctrl.Rotation.Z {
        type=MovieSceneScriptingFloatChannel
        default=0
        thinned_from=434
        keys [[0,0],[146,0],[147,1.8414],[148,5.2613],[149,7.1028],[150,7.1028],[151,17.7569],[152,19.9099],[153,20.4204],[154,20.7689],[155,20.8413],[156,21.3082],[157,24.5044],[158,25.3035],[159,25.4852],[344,25.5651],[345,25.4144],[346,25.0016],[347,24.4112],[349,23.0173],[350,22.3736],[351,22.0024],[352,21.7477],[353,21.1401],[354,19.7101],[355,14.2859],[356,8.6121],[357,7.2664],[358,6.9474],[359,6.7295],[360,6.6843],[361,6.3924],[362,4.1562],[363,1.5981],[364,0.8101],[365,0.5549],[366,0.3995],[367,0.215],[368,0.1304],[369,0.0999],[370,0.0873],[372,0.0873],[373,0.0749],[374,0.0621],[375,0.0461],[376,0.0305],[377,0.0187],[378,0.0187],[379,0],[433,0]]
       }
       thumb_02_r_ctrl.Scale.X {type=MovieSceneScriptingFloatChannel, default=1, const [1,434]}
       thumb_02_r_ctrl.Scale.Y {type=MovieSceneScriptingFloatChannel, default=1, const [1,434]}
       thumb_02_r_ctrl.Scale.Z {type=MovieSceneScriptingFloatChannel, default=1, const [1,434]}
       thumb_03_r_ctrl.Location.X {type=MovieSceneScriptingFloatChannel, default=0, thinned_from=434, keys [[0,0],[433,0]]}
       thumb_03_r_ctrl.Location.Y {type=MovieSceneScriptingFloatChannel, default=0, thinned_from=434, keys [[0,0],[433,0]]}
       thumb_03_r_ctrl.Location.Z {type=MovieSceneScriptingFloatChannel, default=0, thinned_from=434, keys [[0,0],[433,0]]}
       thumb_03_r_ctrl.Rotation.X {type=MovieSceneScriptingFloatChannel, default=0, thinned_from=434, keys [[0,0],[433,0]]}
       thumb_03_r_ctrl.Rotation.Y {type=MovieSceneScriptingFloatChannel, default=0, thinned_from=434, keys [[0,0],[433,0]]}
       thumb_03_r_ctrl.Rotation.Z {
        type=MovieSceneScriptingFloatChannel
        default=0
        thinned_from=434
        keys [[0,0],[146,0],[147,1.3473],[148,3.8494],[149,5.1967],[150,5.1967],[151,12.9916],[152,14.5668],[153,14.9403],[154,15.1954],[155,15.2483],[156,15.5899],[157,17.9284],[158,18.513],[159,18.6459],[344,18.7066],[345,18.5942],[346,18.2921],[347,17.8602],[349,16.8403],[350,16.3694],[351,16.0978],[352,15.9115],[353,15.4669],[354,14.4207],[355,10.4521],[356,6.3009],[357,5.3164],[358,5.0829],[359,4.9236],[360,4.8905],[361,4.6769],[362,3.0408],[363,1.1692],[364,0.5927],[365,0.406],[366,0.2923],[367,0.1573],[368,0.0954],[369,0.0731],[370,0.0731],[371,0.0608],[373,0.0608],[374,0.0454],[375,0.0338],[376,0.0338],[377,0.0137],[378,0],[433,0]]
       }
       thumb_03_r_ctrl.Scale.X {type=MovieSceneScriptingFloatChannel, default=1, const [1,434]}
       thumb_03_r_ctrl.Scale.Y {type=MovieSceneScriptingFloatChannel, default=1, const [1,434]}
       thumb_03_r_ctrl.Scale.Z {type=MovieSceneScriptingFloatChannel, default=1, const [1,434]}
       index_metacarpal_r_ctrl.Location.X {type=MovieSceneScriptingFloatChannel, default=0, const [0.0001,434]}
       index_metacarpal_r_ctrl.Location.Y {type=MovieSceneScriptingFloatChannel, default=0, const [0.0002,434]}
       index_metacarpal_r_ctrl.Location.Z {type=MovieSceneScriptingFloatChannel, default=0, const [0,434]}
       index_metacarpal_r_ctrl.Rotation.X {
        type=MovieSceneScriptingFloatChannel
        default=0
        thinned_from=434
        keys [[0,-1.732],[146,-1.732],[147,-1.6701],[148,-1.5551],[149,-1.4932],[150,-1.4932],[151,-1.1351],[152,-1.0627],[153,-1.0455],[154,-1.0338],[155,-1.0314],[156,-1.0157],[157,-0.9082],[158,-0.8814],[159,-0.8753],[344,-0.873],[345,-0.8776],[346,-0.8915],[347,-0.9114],[349,-0.9582],[350,-0.9799],[351,-0.9923],[352,-1.0009],[353,-1.0213],[354,-1.0694],[355,-1.2518],[356,-1.4425],[357,-1.4877],[358,-1.4985],[359,-1.5058],[360,-1.5073],[361,-1.5171],[362,-1.5923],[363,-1.6783],[364,-1.7048],[367,-1.7248],[433,-1.7314]]
       }
       index_metacarpal_r_ctrl.Rotation.Y {
        type=MovieSceneScriptingFloatChannel
        default=0
        thinned_from=434
        keys [[0,0.7486],[146,0.7486],[147,1.2217],[148,2.1005],[149,2.5736],[150,2.5736],[151,5.3112],[152,5.8644],[153,5.9956],[154,6.0851],[155,6.1037],[156,6.2237],[157,7.045],[158,7.2503],[159,7.297],[344,7.3141],[345,7.2788],[346,7.1727],[347,7.021],[349,6.6628],[350,6.4975],[351,6.4021],[352,6.3366],[353,6.1805],[354,5.8131],[355,4.4193],[356,2.9614],[357,2.6157],[358,2.5337],[359,2.4777],[360,2.4661],[361,2.3911],[362,1.8165],[363,1.1592],[364,0.9567],[365,0.8912],[366,0.8512],[367,0.8038],[368,0.7821],[369,0.7742],[372,0.7742],[373,0.7678],[374,0.7678],[375,0.7604],[376,0.7604],[377,0.7534],[433,0.7534]]
       }
       index_metacarpal_r_ctrl.Rotation.Z {
        type=MovieSceneScriptingFloatChannel
        default=0
        thinned_from=434
        keys [[0,-0.3856],[146,-0.3856],[147,0.6919],[148,2.693],[149,3.7705],[150,3.7705],[151,10.0046],[152,11.2644],[153,11.5631],[154,11.7671],[155,11.8094],[156,12.0826],[157,13.9528],[158,14.4204],[159,14.5267],[344,14.5658],[345,14.4853],[346,14.2438],[347,13.8983],[349,13.0827],[350,12.706],[351,12.4888],[352,12.3398],[353,11.9843],[354,11.1475],[355,7.9736],[356,4.6536],[357,3.8663],[358,3.6796],[359,3.5521],[360,3.5256],[361,3.3549],[362,2.0464],[363,0.5495],[364,0.0885],[365,-0.0609],[366,-0.1518],[367,-0.2597],[368,-0.3093],[369,-0.3271],[372,-0.3271],[373,-0.3417],[374,-0.3417],[375,-0.3586],[376,-0.3586],[377,-0.3746],[433,-0.3746]]
       }
       index_metacarpal_r_ctrl.Scale.X {type=MovieSceneScriptingFloatChannel, default=1, const [1,434]}
       index_metacarpal_r_ctrl.Scale.Y {type=MovieSceneScriptingFloatChannel, default=1, const [1,434]}
       index_metacarpal_r_ctrl.Scale.Z {type=MovieSceneScriptingFloatChannel, default=1, const [1,434]}
       index_01_r_ctrl.Location.X {type=MovieSceneScriptingFloatChannel, default=0, thinned_from=434, keys [[0,0],[433,0]]}
       index_01_r_ctrl.Location.Y {type=MovieSceneScriptingFloatChannel, default=0, thinned_from=434, keys [[0,0],[433,0]]}
       index_01_r_ctrl.Location.Z {type=MovieSceneScriptingFloatChannel, default=0, thinned_from=434, keys [[0,0],[433,0]]}
       index_01_r_ctrl.Rotation.X {
        type=MovieSceneScriptingFloatChannel
        default=0
        thinned_from=434
        keys [[0,-3.6768],[133,-3.6768],[134,-3.6676],[135,-3.6027],[136,-3.4267],[137,-3.0839],[138,-2.5187],[139,-1.6756],[140,0.2886],[141,3.2534],[142,5.8291],[143,7.2336],[144,7.9137],[145,8.1874],[146,8.2264],[147,7.5493],[148,6.2917],[149,5.6145],[150,5.6145],[151,1.6967],[152,0.9049],[153,0.7172],[154,0.589],[155,0.5624],[156,0.3907],[157,-0.8003],[158,-1.098],[159,-1.1657],[160,-1.1807],[161,-1.1906],[163,-1.1906],[164,-1.1962],[344,-1.1962],[345,-1.2123],[346,-1.2523],[347,-1.3096],[349,-1.4448],[350,-1.5072],[351,-1.5432],[352,-1.5679],[353,-1.6268],[354,-1.7655],[355,-2.2915],[356,-2.8417],[357,-2.9722],[358,-3.0031],[359,-3.0242],[360,-3.0242],[361,-3.0569],[362,-3.2738],[363,-3.5218],[364,-3.5983],[365,-3.623],[367,-3.656],[368,-3.656],[369,-3.6671],[433,-3.6671]]
       }
       index_01_r_ctrl.Rotation.Y {
        type=MovieSceneScriptingFloatChannel
        default=0
        thinned_from=434
        keys [[0,2.5922],[146,2.5922],[147,2.3825],[148,1.9929],[149,1.7831],[150,1.7831],[151,0.5694],[152,0.3242],[153,0.266],[154,0.2263],[155,0.2181],[156,0.1649],[157,-0.1992],[158,-0.2902],[159,-0.311],[160,-0.3155],[161,-0.3186],[163,-0.3186],[164,-0.3203],[344,-0.3203],[345,-0.3029],[346,-0.2559],[347,-0.1886],[348,-0.1102],[349,-0.0298],[350,0.0435],[351,0.0858],[352,0.1148],[353,0.184],[354,0.3469],[355,0.9648],[356,1.6112],[357,1.7645],[358,1.8008],[359,1.8256],[360,1.8256],[361,1.864],[362,2.1188],[363,2.4102],[364,2.5],[365,2.529],[367,2.5678],[368,2.5678],[369,2.5809],[433,2.5809]]
       }
       index_01_r_ctrl.Rotation.Z {
        type=MovieSceneScriptingFloatChannel
        default=0
        thinned_from=434
        keys [[0,22.2532],[134,22.237],[135,22.1232],[136,21.8144],[137,21.2131],[138,20.2218],[139,18.7429],[140,15.2976],[141,10.0971],[142,5.5791],[143,3.1155],[144,1.9225],[145,1.4426],[146,1.3741],[147,2.9193],[148,5.789],[149,7.3342],[150,7.3342],[151,16.2743],[152,18.0809],[153,18.5093],[154,18.8018],[155,18.8625],[156,19.2543],[157,21.9637],[158,22.6411],[159,22.7951],[353,22.7606],[354,22.7262],[356,22.4599],[361,22.4066],[363,22.2916],[433,22.2556]]
       }
       index_01_r_ctrl.Scale.X {type=MovieSceneScriptingFloatChannel, default=1, const [1,434]}
       index_01_r_ctrl.Scale.Y {type=MovieSceneScriptingFloatChannel, default=1, const [1,434]}
       index_01_r_ctrl.Scale.Z {type=MovieSceneScriptingFloatChannel, default=1, const [1,434]}
       index_02_r_ctrl.Location.X {type=MovieSceneScriptingFloatChannel, default=0, thinned_from=434, keys [[0,-0.0001],[433,-0.0001]]}
       index_02_r_ctrl.Location.Y {type=MovieSceneScriptingFloatChannel, default=0, thinned_from=434, keys [[0,-0.0001],[433,-0.0001]]}
       index_02_r_ctrl.Location.Z {type=MovieSceneScriptingFloatChannel, default=0, thinned_from=434, keys [[0,-0.0001],[433,-0.0001]]}
       index_02_r_ctrl.Rotation.X {
        type=MovieSceneScriptingFloatChannel
        default=0
        thinned_from=434
        keys [[0,-3.6768],[146,-3.6768],[147,-3.412],[148,-2.9203],[149,-2.6555],[150,-2.6555],[151,-1.1235],[152,-0.8139],[153,-0.7405],[154,-0.6903],[155,-0.6799],[156,-0.6128],[157,-0.1532],[158,-0.0383],[159,-0.0122],[160,-0.0064],[161,-0.0026],[162,-0.0026],[163,-0.0007],[344,-0.0007],[345,-0.0223],[346,-0.0817],[347,-0.1666],[348,-0.2655],[349,-0.367],[350,-0.4596],[351,-0.513],[352,-0.5496],[353,-0.637],[354,-0.8426],[355,-1.6226],[356,-2.4384],[357,-2.6319],[358,-2.6778],[359,-2.7091],[360,-2.7156],[361,-2.7576],[362,-3.0792],[363,-3.447],[364,-3.5603],[365,-3.597],[367,-3.6459],[368,-3.6581],[433,-3.6741]]
       }
       index_02_r_ctrl.Rotation.Y {
        type=MovieSceneScriptingFloatChannel
        default=0
        thinned_from=434
        keys [[0,2.5922],[146,2.5922],[147,2.4056],[148,2.0589],[149,1.8722],[150,1.8722],[151,0.7921],[152,0.5738],[153,0.522],[154,0.4867],[155,0.4794],[156,0.432],[157,0.108],[158,0.027],[159,0.0086],[160,0.0045],[161,0.0018],[162,0.0018],[163,0.0005],[344,0.0005],[345,0.0157],[346,0.0576],[347,0.1175],[348,0.1872],[349,0.2588],[350,0.324],[351,0.3617],[352,0.3875],[353,0.4491],[354,0.594],[355,1.1439],[356,1.7192],[357,1.8556],[358,1.8879],[359,1.91],[360,1.9146],[361,1.9442],[362,2.1709],[363,2.4302],[364,2.5101],[365,2.536],[367,2.5705],[368,2.579],[433,2.5903]]
       }
       index_02_r_ctrl.Rotation.Z {
        type=MovieSceneScriptingFloatChannel
        default=0
        thinned_from=434
        keys [[0,22.2532],[134,22.2396],[135,22.1448],[136,21.8873],[137,21.3859],[138,20.5593],[139,19.3262],[140,16.4533],[141,12.1169],[142,8.3496],[143,6.2953],[144,5.3006],[145,4.9004],[146,4.8432],[147,7.0171],[148,11.0544],[149,13.2283],[150,13.2283],[151,25.8057],[152,28.3474],[153,28.9501],[154,29.3616],[155,29.447],[156,29.9982],[157,33.7943],[158,34.7433],[159,34.9591],[344,35.054],[345,34.9818],[346,34.7751],[350,33.4589],[352,33.1454],[353,32.841],[354,32.1249],[355,29.4082],[356,26.5665],[357,25.8925],[359,25.6236],[360,25.6009],[361,25.4548],[362,24.3348],[363,23.0536],[364,22.6589],[365,22.5311],[367,22.3609],[433,22.2626]]
       }
       index_02_r_ctrl.Scale.X {type=MovieSceneScriptingFloatChannel, default=1, const [1,434]}
       index_02_r_ctrl.Scale.Y {type=MovieSceneScriptingFloatChannel, default=1, const [1,434]}
       index_02_r_ctrl.Scale.Z {type=MovieSceneScriptingFloatChannel, default=1, const [1,434]}
       index_03_r_ctrl.Location.X {type=MovieSceneScriptingFloatChannel, default=0, thinned_from=434, keys [[0,-0.0001],[433,-0.0001]]}
       index_03_r_ctrl.Location.Y {type=MovieSceneScriptingFloatChannel, default=0, thinned_from=434, keys [[0,0.0001],[433,0.0001]]}
       index_03_r_ctrl.Location.Z {type=MovieSceneScriptingFloatChannel, default=0, thinned_from=434, keys [[0,-0.0001],[433,-0.0001]]}
       index_03_r_ctrl.Rotation.X {type=MovieSceneScriptingFloatChannel, default=0, thinned_from=434, keys [[0,0],[433,0]]}
       index_03_r_ctrl.Rotation.Y {type=MovieSceneScriptingFloatChannel, default=0, thinned_from=434, keys [[0,0],[433,0]]}
       index_03_r_ctrl.Rotation.Z {
        type=MovieSceneScriptingFloatChannel
        default=0
        thinned_from=434
        keys [[0,11.8666],[133,11.8666],[134,11.8798],[135,11.9725],[136,12.2239],[137,12.7135],[138,13.5208],[139,14.725],[140,17.5305],[141,21.7652],[142,25.4441],[143,27.4502],[144,28.4216],[145,28.8124],[146,28.8682],[147,28.3467],[148,27.3781],[149,26.8566],[150,26.8566],[151,23.8391],[152,23.2293],[154,22.986],[155,22.9655],[156,22.8332],[157,21.9057],[158,21.6738],[344,21.6094],[345,21.5373],[346,21.3803],[350,20.3802],[351,20.239],[352,20.1421],[353,19.9109],[354,19.3667],[355,17.3027],[356,15.1437],[357,14.6316],[358,14.5102],[359,14.4273],[360,14.4101],[361,14.2991],[362,13.4481],[363,12.4747],[364,12.1749],[365,12.0778],[367,11.9484],[369,11.9046],[433,11.8666]]
       }
       index_03_r_ctrl.Scale.X {type=MovieSceneScriptingFloatChannel, default=1, const [1,434]}
       index_03_r_ctrl.Scale.Y {type=MovieSceneScriptingFloatChannel, default=1, const [1,434]}
       index_03_r_ctrl.Scale.Z {type=MovieSceneScriptingFloatChannel, default=1, const [1,434]}
       middle_metacarpal_r_ctrl.Location.X {type=MovieSceneScriptingFloatChannel, default=0, const [0,434]}
       middle_metacarpal_r_ctrl.Location.Y {type=MovieSceneScriptingFloatChannel, default=0, const [0.0001,434]}
       middle_metacarpal_r_ctrl.Location.Z {type=MovieSceneScriptingFloatChannel, default=0, const [0,434]}
       middle_metacarpal_r_ctrl.Rotation.X {type=MovieSceneScriptingFloatChannel, default=0, thinned_from=434, keys [[0,0],[433,0]]}
       middle_metacarpal_r_ctrl.Rotation.Y {
        type=MovieSceneScriptingFloatChannel
        default=0
        thinned_from=434
        keys [[0,0],[146,0],[147,-0.2333],[148,-0.6665],[149,-0.8998],[150,-0.8998],[151,-2.2494],[152,-2.5221],[153,-2.5868],[154,-2.6309],[155,-2.6401],[156,-2.6992],[157,-3.1041],[158,-3.2053],[159,-3.2283],[344,-3.2377],[345,-3.2194],[346,-3.1671],[347,-3.0923],[349,-2.9157],[350,-2.8342],[351,-2.7872],[352,-2.7549],[353,-2.6779],[354,-2.4968],[355,-1.8097],[356,-1.0909],[357,-0.9205],[358,-0.8801],[359,-0.8525],[360,-0.8467],[361,-0.8098],[362,-0.5265],[363,-0.2024],[364,-0.1026],[365,-0.0703],[366,-0.0506],[367,-0.0272],[368,-0.0165],[369,-0.0165],[370,-0.0111],[374,-0.0111],[375,-0.0058],[377,-0.0058],[378,-0.0015],[433,-0.0015]]
       }
       middle_metacarpal_r_ctrl.Rotation.Z {
        type=MovieSceneScriptingFloatChannel
        default=0
        thinned_from=434
        keys [[0,0],[146,0],[147,0.6854],[148,1.9583],[149,2.6437],[150,2.6437],[151,6.6091],[152,7.4105],[153,7.6005],[154,7.7302],[155,7.7572],[156,7.9309],[157,9.1206],[158,9.418],[159,9.4856],[344,9.513],[345,9.4593],[346,9.3056],[347,9.0859],[349,8.5671],[350,8.3275],[351,8.1893],[352,8.0945],[353,7.8684],[354,7.3361],[355,5.3172],[356,3.2054],[357,2.7046],[358,2.5858],[359,2.5047],[360,2.4879],[361,2.3793],[362,1.5469],[363,0.5948],[364,0.3015],[365,0.2065],[366,0.1487],[367,0.08],[368,0.0485],[369,0.0485],[370,0.0325],[374,0.0325],[375,0.0172],[377,0.0172],[378,0.0045],[433,0.0045]]
       }
       middle_metacarpal_r_ctrl.Scale.X {type=MovieSceneScriptingFloatChannel, default=1, const [1,434]}
       middle_metacarpal_r_ctrl.Scale.Y {type=MovieSceneScriptingFloatChannel, default=1, const [1,434]}
       middle_metacarpal_r_ctrl.Scale.Z {type=MovieSceneScriptingFloatChannel, default=1, const [1,434]}
       middle_01_r_ctrl.Location.X {type=MovieSceneScriptingFloatChannel, default=0, thinned_from=434, keys [[0,0.0002],[433,0.0002]]}
       middle_01_r_ctrl.Location.Y {type=MovieSceneScriptingFloatChannel, default=0, thinned_from=434, keys [[0,0],[433,0]]}
       middle_01_r_ctrl.Location.Z {type=MovieSceneScriptingFloatChannel, default=0, thinned_from=434, keys [[0,0],[433,0]]}
       middle_01_r_ctrl.Rotation.X {type=MovieSceneScriptingFloatChannel, default=0, thinned_from=434, keys [[0,0],[433,0]]}
       middle_01_r_ctrl.Rotation.Y {type=MovieSceneScriptingFloatChannel, default=0, thinned_from=434, keys [[0,0],[433,0]]}
       middle_01_r_ctrl.Rotation.Z {
        type=MovieSceneScriptingFloatChannel
        default=0
        thinned_from=434
        keys [[0,11.8666],[146,11.8666],[147,12.3118],[148,13.1387],[149,13.5839],[150,13.5839],[151,16.1598],[152,16.6803],[153,16.8038],[154,16.888],[155,16.9055],[156,17.0184],[157,17.7912],[158,17.9844],[159,18.0283],[344,18.0445],[345,18.0112],[346,17.9114],[350,17.276],[352,17.1247],[353,16.9778],[354,16.632],[355,15.3206],[356,13.9488],[357,13.6235],[359,13.4937],[360,13.4937],[361,13.4122],[362,12.8715],[363,12.253],[364,12.0625],[365,12.0008],[367,11.9186],[433,11.874]]
       }
       middle_01_r_ctrl.Scale.X {type=MovieSceneScriptingFloatChannel, default=1, const [1,434]}
       middle_01_r_ctrl.Scale.Y {type=MovieSceneScriptingFloatChannel, default=1, const [1,434]}
       middle_01_r_ctrl.Scale.Z {type=MovieSceneScriptingFloatChannel, default=1, const [1,434]}
       middle_02_r_ctrl.Location.X {type=MovieSceneScriptingFloatChannel, default=0, thinned_from=434, keys [[0,0.0008],[433,0.0008]]}
       middle_02_r_ctrl.Location.Y {type=MovieSceneScriptingFloatChannel, default=0, thinned_from=434, keys [[0,0.0009],[433,0.0009]]}
       middle_02_r_ctrl.Location.Z {type=MovieSceneScriptingFloatChannel, default=0, thinned_from=434, keys [[0,0.0476],[433,0.0476]]}
       middle_02_r_ctrl.Rotation.X {
        type=MovieSceneScriptingFloatChannel
        default=0
        thinned_from=434
        keys [[0,-0.0059],[146,-0.0059],[147,-0.0065],[148,-0.0075],[149,-0.0081],[150,-0.0081],[151,-0.0108],[152,-0.0112],[156,-0.0115],[157,-0.0121],[347,-0.0121],[353,-0.0115],[354,-0.0112],[355,-0.01],[356,-0.0085],[357,-0.0081],[361,-0.0079],[363,-0.0064],[433,-0.0059]]
       }
       middle_02_r_ctrl.Rotation.Y {
        type=MovieSceneScriptingFloatChannel
        default=0
        thinned_from=434
        keys [[0,-0.0122],[146,-0.0122],[147,-0.0119],[148,-0.0113],[149,-0.0109],[150,-0.0109],[151,-0.0083],[152,-0.0076],[156,-0.0072],[157,-0.0062],[345,-0.0059],[353,-0.0073],[354,-0.0077],[356,-0.0106],[357,-0.0109],[361,-0.0111],[363,-0.012],[433,-0.0122]]
       }
       middle_02_r_ctrl.Rotation.Z {
        type=MovieSceneScriptingFloatChannel
        default=0
        thinned_from=434
        keys [[0,-3.6158],[146,-3.6158],[147,-0.8306],[148,4.342],[149,7.1273],[150,7.1273],[151,23.2419],[152,26.4984],[153,27.2705],[154,27.7977],[155,27.9071],[156,28.6134],[157,33.4477],[158,34.6563],[159,34.9312],[344,35.052],[345,34.8242],[346,34.1997],[347,33.3067],[349,31.1984],[350,30.2248],[351,29.6634],[352,29.2781],[353,28.3591],[354,26.1962],[355,17.992],[356,9.4101],[357,7.3748],[358,6.8922],[359,6.5627],[360,6.4943],[361,6.0529],[362,2.6705],[363,-1.1987],[364,-2.3905],[365,-2.7765],[366,-3.0115],[367,-3.2906],[368,-3.4186],[369,-3.4648],[370,-3.4837],[372,-3.4837],[377,-3.5875],[378,-3.5875],[379,-3.6031],[390,-3.6031],[391,-3.6155],[433,-3.6155]]
       }
       middle_02_r_ctrl.Scale.X {type=MovieSceneScriptingFloatChannel, default=1, const [1,434]}
       middle_02_r_ctrl.Scale.Y {type=MovieSceneScriptingFloatChannel, default=1, const [1,434]}
       middle_02_r_ctrl.Scale.Z {type=MovieSceneScriptingFloatChannel, default=1, const [1,434]}
       middle_03_r_ctrl.Location.X {type=MovieSceneScriptingFloatChannel, default=0, thinned_from=434, keys [[0,0.0005],[433,0.0005]]}
       middle_03_r_ctrl.Location.Y {type=MovieSceneScriptingFloatChannel, default=0, thinned_from=434, keys [[0,0],[433,0]]}
       middle_03_r_ctrl.Location.Z {type=MovieSceneScriptingFloatChannel, default=0, thinned_from=434, keys [[0,0.0061],[433,0.0061]]}
       middle_03_r_ctrl.Rotation.X {
        type=MovieSceneScriptingFloatChannel
        default=0
        thinned_from=434
        keys [[0,-0.01],[147,-0.0099],[150,-0.0097],[151,-0.0093],[354,-0.0092],[356,-0.0097],[433,-0.01]]
       }
       middle_03_r_ctrl.Rotation.Y {
        type=MovieSceneScriptingFloatChannel
        default=0
        thinned_from=434
        keys [[0,0.0048],[146,0.0048],[149,0.0053],[150,0.0053],[151,0.0059],[156,0.0062],[354,0.0061],[356,0.0054],[362,0.0051],[433,0.0048]]
       }
       middle_03_r_ctrl.Rotation.Z {
        type=MovieSceneScriptingFloatChannel
        default=0
        thinned_from=434
        keys [[0,11.8664],[146,11.8664],[147,12.5671],[148,13.8684],[149,14.5692],[150,14.5692],[151,18.6233],[152,19.4425],[153,19.6368],[154,19.7694],[155,19.7969],[156,19.9746],[157,21.1908],[158,21.4949],[159,21.5641],[344,21.592],[345,21.5371],[346,21.38],[350,20.38],[351,20.2388],[352,20.1419],[353,19.9106],[354,19.3665],[355,17.3025],[356,15.1435],[357,14.6314],[358,14.51],[359,14.4271],[360,14.4099],[361,14.2989],[362,13.4479],[363,12.4745],[364,12.1747],[365,12.0775],[367,11.9482],[369,11.9044],[433,11.878]]
       }
       middle_03_r_ctrl.Scale.X {type=MovieSceneScriptingFloatChannel, default=1, const [1,434]}
       middle_03_r_ctrl.Scale.Y {type=MovieSceneScriptingFloatChannel, default=1, const [1,434]}
       middle_03_r_ctrl.Scale.Z {type=MovieSceneScriptingFloatChannel, default=1, const [1,434]}
       ring_metacarpal_r_ctrl.Location.X {type=MovieSceneScriptingFloatChannel, default=0, const [0.0001,434]}
       ring_metacarpal_r_ctrl.Location.Y {type=MovieSceneScriptingFloatChannel, default=0, const [0.0002,434]}
       ring_metacarpal_r_ctrl.Location.Z {type=MovieSceneScriptingFloatChannel, default=0, const [0,434]}
       ring_metacarpal_r_ctrl.Rotation.X {
        type=MovieSceneScriptingFloatChannel
        default=0
        thinned_from=434
        keys [[0,0],[146,0],[147,0.011],[148,0.0314],[149,0.0424],[150,0.0424],[151,0.106],[152,0.1189],[153,0.1219],[154,0.124],[155,0.124],[156,0.1272],[157,0.1463],[158,0.1511],[345,0.1511],[346,0.1493],[347,0.1458],[349,0.1374],[350,0.1336],[351,0.1314],[352,0.1314],[353,0.1262],[354,0.1177],[355,0.0853],[356,0.0514],[357,0.0434],[358,0.0415],[360,0.0415],[361,0.0382],[362,0.0248],[363,0.0095],[364,0.0048],[365,0.0048],[366,0.0024],[368,0.0024],[369,0.0006],[433,0.0006]]
       }
       ring_metacarpal_r_ctrl.Rotation.Y {
        type=MovieSceneScriptingFloatChannel
        default=0
        thinned_from=434
        keys [[0,-3.4928],[146,-3.4928],[149,-3.4782],[150,-3.4782],[151,-3.4562],[156,-3.4489],[354,-3.4522],[356,-3.4751],[433,-3.4926]]
       }
       ring_metacarpal_r_ctrl.Rotation.Z {
        type=MovieSceneScriptingFloatChannel
        default=0
        thinned_from=434
        keys [[0,0],[146,0],[147,0.0767],[148,0.2193],[149,0.296],[150,0.296],[151,0.7401],[152,0.8298],[153,0.8511],[154,0.8656],[155,0.8656],[156,0.8881],[157,1.0213],[158,1.0546],[345,1.0546],[346,1.042],[347,1.0174],[349,0.9593],[350,0.9325],[351,0.917],[352,0.917],[353,0.8811],[354,0.8215],[355,0.5954],[356,0.3589],[357,0.3028],[358,0.2895],[360,0.2895],[361,0.2664],[362,0.1732],[363,0.0666],[364,0.0338],[365,0.0338],[366,0.0167],[368,0.0167],[369,0.0042],[433,0.0042]]
       }
       ring_metacarpal_r_ctrl.Scale.X {type=MovieSceneScriptingFloatChannel, default=1, const [1,434]}
       ring_metacarpal_r_ctrl.Scale.Y {type=MovieSceneScriptingFloatChannel, default=1, const [1,434]}
       ring_metacarpal_r_ctrl.Scale.Z {type=MovieSceneScriptingFloatChannel, default=1, const [1,434]}
       ring_01_r_ctrl.Location.X {type=MovieSceneScriptingFloatChannel, default=0, thinned_from=434, keys [[0,-0.0002],[433,-0.0002]]}
       ring_01_r_ctrl.Location.Y {type=MovieSceneScriptingFloatChannel, default=0, thinned_from=434, keys [[0,0],[433,0]]}
       ring_01_r_ctrl.Location.Z {type=MovieSceneScriptingFloatChannel, default=0, thinned_from=434, keys [[0,0],[433,0]]}
       ring_01_r_ctrl.Rotation.X {type=MovieSceneScriptingFloatChannel, default=0, thinned_from=434, keys [[0,0],[433,0]]}
       ring_01_r_ctrl.Rotation.Y {type=MovieSceneScriptingFloatChannel, default=0, thinned_from=434, keys [[0,0],[433,0]]}
       ring_01_r_ctrl.Rotation.Z {
        type=MovieSceneScriptingFloatChannel
        default=0
        thinned_from=434
        keys [[0,9.4289],[146,9.4289],[147,10.0497],[148,11.2026],[149,11.8233],[150,11.8233],[151,15.4149],[152,16.1407],[153,16.3128],[154,16.4303],[155,16.4547],[156,16.6121],[157,17.6896],[158,17.959],[159,18.0202],[344,18.0472],[345,17.9964],[346,17.8572],[350,16.9713],[351,16.8461],[352,16.7603],[353,16.5555],[354,16.0734],[355,14.2449],[356,12.3321],[357,11.8785],[358,11.7709],[359,11.6975],[360,11.6823],[361,11.5839],[362,10.83],[363,9.9677],[364,9.702],[365,9.616],[367,9.5014],[368,9.4729],[374,9.4584],[433,9.4318]]
       }
       ring_01_r_ctrl.Scale.X {type=MovieSceneScriptingFloatChannel, default=1, const [1,434]}
       ring_01_r_ctrl.Scale.Y {type=MovieSceneScriptingFloatChannel, default=1, const [1,434]}
       ring_01_r_ctrl.Scale.Z {type=MovieSceneScriptingFloatChannel, default=1, const [1,434]}
       ring_02_r_ctrl.Location.X {type=MovieSceneScriptingFloatChannel, default=0, thinned_from=434, keys [[0,0.0003],[433,0.0003]]}
       ring_02_r_ctrl.Location.Y {type=MovieSceneScriptingFloatChannel, default=0, thinned_from=434, keys [[0,0.0001],[433,0.0001]]}
       ring_02_r_ctrl.Location.Z {type=MovieSceneScriptingFloatChannel, default=0, thinned_from=434, keys [[0,0],[433,0]]}
       ring_02_r_ctrl.Rotation.X {type=MovieSceneScriptingFloatChannel, default=0, thinned_from=434, keys [[0,0],[433,0]]}
       ring_02_r_ctrl.Rotation.Y {type=MovieSceneScriptingFloatChannel, default=0, thinned_from=434, keys [[0,0],[433,0]]}
       ring_02_r_ctrl.Rotation.Z {
        type=MovieSceneScriptingFloatChannel
        default=0
        thinned_from=434
        keys [[0,13.9361],[146,13.9361],[147,15.4574],[148,18.2825],[149,19.8038],[150,19.8038],[151,28.6053],[152,30.3839],[153,30.8056],[154,31.0936],[155,31.1534],[156,31.5391],[157,34.1795],[158,34.8396],[159,34.9897],[344,35.0574],[345,34.9313],[346,34.5902],[347,34.1025],[350,32.4192],[351,32.1126],[352,31.9022],[353,31.4002],[354,30.2189],[355,25.7379],[356,21.0506],[357,19.939],[358,19.6754],[359,19.4954],[360,19.4581],[361,19.217],[362,17.3696],[363,15.2563],[364,14.6054],[365,14.3945],[367,14.1138],[368,14.0438],[433,13.946]]
       }
       ring_02_r_ctrl.Scale.X {type=MovieSceneScriptingFloatChannel, default=1, const [1,434]}
       ring_02_r_ctrl.Scale.Y {type=MovieSceneScriptingFloatChannel, default=1, const [1,434]}
       ring_02_r_ctrl.Scale.Z {type=MovieSceneScriptingFloatChannel, default=1, const [1,434]}
       ring_03_r_ctrl.Location.X {type=MovieSceneScriptingFloatChannel, default=0, thinned_from=434, keys [[0,0],[433,0]]}
       ring_03_r_ctrl.Location.Y {type=MovieSceneScriptingFloatChannel, default=0, thinned_from=434, keys [[0,0],[433,0]]}
       ring_03_r_ctrl.Location.Z {type=MovieSceneScriptingFloatChannel, default=0, thinned_from=434, keys [[0,0],[433,0]]}
       ring_03_r_ctrl.Rotation.X {type=MovieSceneScriptingFloatChannel, default=0, thinned_from=434, keys [[0,0],[433,0]]}
       ring_03_r_ctrl.Rotation.Y {type=MovieSceneScriptingFloatChannel, default=0, thinned_from=434, keys [[0,0],[433,0]]}
       ring_03_r_ctrl.Rotation.Z {
        type=MovieSceneScriptingFloatChannel
        default=0
        thinned_from=434
        keys [[0,0],[146,0],[147,2.0073],[148,5.7351],[149,7.7424],[150,7.7424],[151,19.3559],[152,21.7029],[153,22.2593],[154,22.6393],[155,22.7181],[156,23.2271],[157,26.7112],[158,27.5822],[159,27.7803],[344,27.8674],[345,27.7031],[346,27.2531],[347,26.6096],[349,25.0901],[350,24.3885],[351,23.9838],[352,23.7062],[353,23.0439],[354,21.4851],[355,15.5724],[356,9.3876],[357,7.9208],[358,7.573],[359,7.3355],[360,7.2862],[361,6.9681],[362,4.5305],[363,1.742],[364,0.8831],[365,0.6049],[366,0.4355],[367,0.2344],[368,0.1421],[369,0.1089],[370,0.0952],[372,0.0952],[373,0.0816],[374,0.0676],[375,0.0503],[376,0.0333],[377,0.0204],[378,0.0204],[379,0],[433,0]]
       }
       ring_03_r_ctrl.Scale.X {type=MovieSceneScriptingFloatChannel, default=1, const [1,434]}
       ring_03_r_ctrl.Scale.Y {type=MovieSceneScriptingFloatChannel, default=1, const [1,434]}
       ring_03_r_ctrl.Scale.Z {type=MovieSceneScriptingFloatChannel, default=1, const [1,434]}
       pinky_metacarpal_r_ctrl.Location.X {type=MovieSceneScriptingFloatChannel, default=0, const [-0.0001,434]}
       pinky_metacarpal_r_ctrl.Location.Y {type=MovieSceneScriptingFloatChannel, default=0, const [0,434]}
       pinky_metacarpal_r_ctrl.Location.Z {type=MovieSceneScriptingFloatChannel, default=0, const [0,434]}
       pinky_metacarpal_r_ctrl.Rotation.X {
        type=MovieSceneScriptingFloatChannel
        default=0
        thinned_from=434
        keys [[0,0],[146,0],[147,-0.0395],[148,-0.113],[149,-0.1525],[150,-0.1525],[151,-0.3813],[152,-0.4276],[153,-0.4385],[154,-0.446],[155,-0.446],[156,-0.4576],[157,-0.5262],[158,-0.5434],[159,-0.5434],[160,-0.5481],[345,-0.5481],[346,-0.5369],[347,-0.5242],[349,-0.4943],[350,-0.4805],[351,-0.4725],[352,-0.467],[353,-0.454],[354,-0.4233],[355,-0.3068],[356,-0.1849],[357,-0.156],[358,-0.1492],[359,-0.1445],[360,-0.1445],[361,-0.1373],[362,-0.0893],[363,-0.0343],[364,-0.0174],[365,-0.0119],[366,-0.0119],[367,-0.0046],[377,-0.0046],[378,-0.0003],[433,-0.0003]]
       }
       pinky_metacarpal_r_ctrl.Rotation.Y {
        type=MovieSceneScriptingFloatChannel
        default=0
        thinned_from=434
        keys [[0,-7.0585],[146,-7.0585],[147,-6.9705],[148,-6.8069],[149,-6.7189],[150,-6.7189],[151,-6.2095],[152,-6.1066],[154,-6.0655],[155,-6.0655],[156,-6.0397],[157,-5.8869],[158,-5.8487],[345,-5.8381],[350,-5.9887],[352,-6.0187],[353,-6.0477],[354,-6.1161],[356,-6.6467],[357,-6.7111],[359,-6.7367],[360,-6.7367],[361,-6.7529],[362,-6.8598],[363,-6.9821],[364,-7.0198],[367,-7.0482],[377,-7.0482],[433,-7.0579]]
       }
       pinky_metacarpal_r_ctrl.Rotation.Z {
        type=MovieSceneScriptingFloatChannel
        default=0
        thinned_from=434
        keys [[0,0],[146,0],[147,0.109],[148,0.3115],[149,0.4205],[150,0.4205],[151,1.0513],[152,1.1788],[153,1.209],[154,1.2297],[155,1.2297],[156,1.2616],[157,1.4508],[158,1.4981],[159,1.4981],[160,1.5113],[345,1.5113],[346,1.4803],[347,1.4453],[349,1.3628],[350,1.3247],[351,1.3027],[352,1.2876],[353,1.2516],[354,1.167],[355,0.8458],[356,0.5099],[357,0.4302],[358,0.4113],[359,0.3984],[360,0.3984],[361,0.3785],[362,0.2461],[363,0.0946],[364,0.048],[365,0.0329],[366,0.0329],[367,0.0127],[377,0.0127],[378,0.0007],[433,0.0007]]
       }
       pinky_metacarpal_r_ctrl.Scale.X {type=MovieSceneScriptingFloatChannel, default=1, const [1,434]}
       pinky_metacarpal_r_ctrl.Scale.Y {type=MovieSceneScriptingFloatChannel, default=1, const [1,434]}
       pinky_metacarpal_r_ctrl.Scale.Z {type=MovieSceneScriptingFloatChannel, default=1, const [1,434]}
       pinky_01_r_ctrl.Location.X {type=MovieSceneScriptingFloatChannel, default=0, thinned_from=434, keys [[0,0.0001],[433,0.0001]]}
       pinky_01_r_ctrl.Location.Y {type=MovieSceneScriptingFloatChannel, default=0, thinned_from=434, keys [[0,0.0002],[433,0.0002]]}
       pinky_01_r_ctrl.Location.Z {type=MovieSceneScriptingFloatChannel, default=0, thinned_from=434, keys [[0,0],[433,0]]}
       pinky_01_r_ctrl.Rotation.X {
        type=MovieSceneScriptingFloatChannel
        default=0
        thinned_from=434
        keys [[0,0],[134,0],[135,-0.0792],[136,-0.2674],[137,-0.6338],[138,-1.2379],[139,-2.1392],[140,-4.2387],[141,-7.4079],[142,-10.161],[143,-11.6624],[144,-12.3893],[145,-12.6818],[146,-12.7236],[147,-11.8089],[148,-10.1101],[149,-9.1954],[150,-9.1954],[151,-3.9032],[152,-2.8338],[153,-2.5802],[154,-2.407],[155,-2.3711],[156,-2.1392],[157,-0.5348],[158,-0.1337],[159,-0.0425],[160,-0.0223],[161,-0.009],[344,-0.009],[345,0],[433,0]]
       }
       pinky_01_r_ctrl.Rotation.Y {
        type=MovieSceneScriptingFloatChannel
        default=0
        thinned_from=434
        keys [[0,0],[134,0],[135,-0.0345],[136,-0.1164],[137,-0.2759],[138,-0.5389],[139,-0.9313],[140,-1.8454],[141,-3.2251],[142,-4.4237],[143,-5.0773],[144,-5.3938],[145,-5.5211],[146,-5.5393],[147,-5.1411],[148,-4.4015],[149,-4.0033],[150,-4.0033],[151,-1.6993],[152,-1.2337],[153,-1.1233],[154,-1.0479],[155,-1.0323],[156,-0.9313],[157,-0.2328],[158,-0.0582],[159,-0.0185],[160,-0.0097],[161,-0.0039],[344,-0.0039],[345,0],[433,0]]
       }
       pinky_01_r_ctrl.Rotation.Z {
        type=MovieSceneScriptingFloatChannel
        default=0
        thinned_from=434
        keys [[0,9.4289],[146,9.4289],[147,10.0497],[148,11.2026],[149,11.8233],[150,11.8233],[151,15.4149],[152,16.1407],[153,16.3128],[154,16.4303],[155,16.4547],[156,16.6121],[157,17.6896],[158,17.959],[159,18.0202],[344,18.0427],[345,17.9964],[346,17.8572],[350,16.9713],[351,16.8461],[352,16.7603],[353,16.5554],[354,16.0734],[355,14.2449],[356,12.3321],[357,11.8785],[358,11.7709],[359,11.6975],[360,11.6823],[361,11.5839],[362,10.83],[363,9.9677],[364,9.702],[365,9.616],[367,9.5014],[368,9.4729],[374,9.4584],[433,9.4318]]
       }
       pinky_01_r_ctrl.Scale.X {type=MovieSceneScriptingFloatChannel, default=1, const [1,434]}
       pinky_01_r_ctrl.Scale.Y {type=MovieSceneScriptingFloatChannel, default=1, const [1,434]}
       pinky_01_r_ctrl.Scale.Z {type=MovieSceneScriptingFloatChannel, default=1, const [1,434]}
       pinky_02_r_ctrl.Location.X {type=MovieSceneScriptingFloatChannel, default=0, thinned_from=434, keys [[0,-0.0001],[433,-0.0001]]}
       pinky_02_r_ctrl.Location.Y {type=MovieSceneScriptingFloatChannel, default=0, thinned_from=434, keys [[0,0.0001],[433,0.0001]]}
       pinky_02_r_ctrl.Location.Z {type=MovieSceneScriptingFloatChannel, default=0, thinned_from=434, keys [[0,0],[433,0]]}
       pinky_02_r_ctrl.Rotation.X {type=MovieSceneScriptingFloatChannel, default=0, thinned_from=434, keys [[0,0],[433,0]]}
       pinky_02_r_ctrl.Rotation.Y {type=MovieSceneScriptingFloatChannel, default=0, thinned_from=434, keys [[0,0],[433,0]]}
       pinky_02_r_ctrl.Rotation.Z {
        type=MovieSceneScriptingFloatChannel
        default=0
        thinned_from=434
        keys [[0,-8.3839],[146,-8.3839],[147,-5.2552],[148,0.5552],[149,3.6839],[150,3.6839],[151,21.7853],[152,25.4433],[153,26.3107],[154,26.9029],[155,27.0258],[156,27.8191],[157,33.2495],[158,34.6071],[159,34.9159],[161,35.0293],[344,35.0549],[345,34.7957],[346,34.0942],[347,33.0911],[349,30.7229],[350,29.6293],[351,28.9985],[352,28.5658],[353,27.5335],[354,25.1039],[355,15.8881],[356,6.2481],[357,3.9619],[358,3.4198],[359,3.0497],[360,2.9728],[361,2.4769],[362,-1.3224],[363,-5.6686],[364,-7.0074],[365,-7.4411],[366,-7.7051],[367,-8.0186],[368,-8.1623],[369,-8.2142],[370,-8.2355],[372,-8.2355],[377,-8.3521],[378,-8.3521],[379,-8.3696],[384,-8.3696],[433,-8.3819]]
       }
       pinky_02_r_ctrl.Scale.X {type=MovieSceneScriptingFloatChannel, default=1, const [1,434]}
       pinky_02_r_ctrl.Scale.Y {type=MovieSceneScriptingFloatChannel, default=1, const [1,434]}
       pinky_02_r_ctrl.Scale.Z {type=MovieSceneScriptingFloatChannel, default=1, const [1,434]}
       pinky_03_r_ctrl.Location.X {type=MovieSceneScriptingFloatChannel, default=0, thinned_from=434, keys [[0,0.0002],[161,0.0002],[163,0.0001],[433,0.0002]]}
       pinky_03_r_ctrl.Location.Y {
        type=MovieSceneScriptingFloatChannel
        default=0
        thinned_from=434
        keys [[0,-0.0002],[161,-0.0002],[162,0.0002],[163,0.0002],[164,-0.0002],[433,-0.0002]]
       }
       pinky_03_r_ctrl.Location.Z {type=MovieSceneScriptingFloatChannel, default=0, thinned_from=434, keys [[0,0],[433,0]]}
       pinky_03_r_ctrl.Rotation.X {type=MovieSceneScriptingFloatChannel, default=0, thinned_from=434, keys [[0,0],[161,0],[162,-0.0002],[163,-0.0002],[164,0],[433,0]]}
       pinky_03_r_ctrl.Rotation.Y {type=MovieSceneScriptingFloatChannel, default=0, thinned_from=434, keys [[0,0],[433,0]]}
       pinky_03_r_ctrl.Rotation.Z {
        type=MovieSceneScriptingFloatChannel
        default=0
        thinned_from=434
        keys [[0,-16.4531],[146,-16.4531],[147,-13.2609],[148,-7.3325],[149,-4.1403],[150,-4.1403],[151,14.3287],[152,18.0609],[153,18.9459],[154,19.5501],[155,19.6755],[156,20.485],[157,26.0256],[158,27.4108],[159,27.7258],[162,27.8648],[344,27.8677],[345,27.6032],[346,26.8875],[347,25.864],[349,23.4477],[350,22.3319],[351,21.6883],[352,21.2468],[353,20.1935],[354,17.7146],[355,8.3118],[356,-1.524],[357,-3.8566],[358,-4.4097],[359,-4.7874],[360,-4.8658],[361,-5.3717],[362,-9.2482],[363,-13.6827],[364,-15.0487],[365,-15.4912],[366,-15.7605],[367,-16.0803],[368,-16.227],[370,-16.3017],[433,-16.451]]
       }
       pinky_03_r_ctrl.Scale.X {type=MovieSceneScriptingFloatChannel, default=1, const [1,434]}
       pinky_03_r_ctrl.Scale.Y {type=MovieSceneScriptingFloatChannel, default=1, const [1,434]}
       pinky_03_r_ctrl.Scale.Z {type=MovieSceneScriptingFloatChannel, default=1, const [1,434]}
       hand_r_ik_ctrl.Location.X {
        type=MovieSceneScriptingFloatChannel
        default=0
        thinned_from=434
        keys [[0,-5.8991],[2,-5.8939],[11,-5.7982],[14,-5.804],[15,-5.8585],[16,-5.9412],[17,-5.9611],[18,-5.8149],[19,-5.3583],[20,-4.3758],[21,-2.6579],[22,-0.6621],[23,0.9939],[24,2.2465],[25,3.178],[26,3.8934],[27,4.398],[28,4.7077],[29,4.8661],[30,4.9249],[31,4.9183],[32,4.8753],[33,4.8163],[35,4.6532],[39,4.3681],[41,4.2421],[43,4.1607],[46,4.0728],[55,3.929],[57,3.9092],[61,3.9021],[62,3.9226],[63,3.8984],[64,3.745],[65,3.2622],[66,2.5467],[67,1.9629],[68,1.3263],[69,0.6703],[70,0.173],[71,-0.2208],[72,-0.5401],[73,-0.7992],[74,-0.9868],[75,-1.1205],[76,-1.2222],[77,-1.2925],[78,-1.347],[79,-1.3933],[80,-1.4263],[81,-1.4496],[82,-1.4652],[83,-1.4764],[84,-1.4823],[85,-1.4817],[87,-1.4647],[89,-1.4229],[90,-1.395],[94,-1.2465],[96,-1.1845],[97,-1.1761],[98,-1.1717],[99,-1.1875],[100,-1.2308],[101,-1.312],[102,-1.4348],[103,-1.6037],[104,-1.8302],[105,-2.1103],[106,-2.426],[107,-2.7594],[108,-3.1185],[110,-3.884],[111,-4.2241],[112,-4.5189],[113,-4.7868],[114,-5.0255],[115,-5.2377],[116,-5.414],[117,-5.5795],[118,-5.7188],[119,-5.8423],[120,-5.9398],[121,-6.0229],[123,-6.1429],[125,-6.214],[127,-6.2477],[130,-6.2442],[132,-6.1779],[134,-6.0411],[135,-5.9546],[136,-5.835],[137,-5.6582],[138,-5.3998],[139,-5.0335],[140,-4.2413],[141,-2.9395],[142,-1.4451],[143,0.1089],[144,1.5668],[145,2.819],[146,4.0915],[147,6.0094],[149,10.231],[150,9.9396],[151,3.9633],[152,-2.5285],[153,-7.6498],[154,-10.3193],[155,-11.839],[156,-12.7099],[157,-10.8469],[158,-2.4917],[159,6.9278],[160,19.9323],[161,28.6224],[162,34.1141],[163,36.3256],[164,36.6149],[165,36.2723],[166,35.0608],[167,33.1739],[168,30.6932],[169,27.3894],[170,23.6982],[171,22.1836],[172,21.7753],[173,22.1762],[174,23.0637],[175,24.0443],[177,25.79],[179,27.5997],[180,28.3715],[181,28.9374],[182,29.3574],[183,29.6019],[184,29.6787],[188,29.4587],[210,28.7849],[213,28.8869],[215,29.0895],[217,29.4711],[219,30.0083],[220,30.3235],[221,30.7086],[223,31.5844],[224,31.9214],[225,32.111],[226,32.0661],[227,31.8472],[228,31.47],[229,30.8634],[230,30.1691],[231,29.5747],[233,28.7064],[235,28.0045],[237,27.4558],[239,27.022],[241,26.7166],[249,26.0499],[254,26.0578],[256,26.1411],[257,26.2554],[258,26.4533],[259,26.7538],[260,27.3402],[261,28.3005],[262,29.4106],[263,30.4571],[264,31.3533],[265,31.9264],[266,32.1731],[267,32.2351],[269,32.0615],[271,31.6785],[274,30.9731],[277,30.4775],[283,29.8759],[286,29.7939],[290,30.1125],[291,30.1424],[292,30.0947],[294,29.8026],[303,28.1943],[311,27.3579],[316,27.1203],[323,27.1586],[335,27.7662],[342,27.6835],[344,27.5628],[346,27.5875],[347,27.751],[348,28.0344],[349,28.383],[350,28.6677],[351,28.607],[352,27.7751],[353,26.4511],[354,24.819],[355,15.4703],[356,4.0308],[357,-1.2959],[358,-3.8413],[359,-5.3181],[360,-6.3059],[361,-6.5218],[362,-6.5057],[363,-6.9456],[364,-7.7736],[365,-8.621],[366,-9.2714],[367,-9.3549],[368,-8.9446],[369,-8.2866],[370,-7.5068],[372,-5.8795],[373,-5.088],[374,-4.1991],[375,-3.2398],[376,-2.4685],[377,-1.9483],[378,-1.5442],[379,-1.2399],[380,-1.0293],[381,-0.9136],[382,-0.9017],[383,-0.9798],[384,-1.1212],[385,-1.3034],[386,-1.5444],[387,-1.8394],[388,-2.149],[389,-2.4474],[390,-2.7099],[391,-2.9536],[392,-3.2046],[393,-3.4461],[394,-3.6718],[395,-3.8747],[396,-4.0525],[398,-4.3619],[400,-4.6138],[401,-4.725],[409,-5.5051],[412,-5.7204],[414,-5.8412],[416,-5.9381],[420,-6.0484],[423,-6.0811],[426,-6.0684],[431,-5.9277],[433,-5.9059]]
       }
       hand_r_ik_ctrl.Location.Y {
        type=MovieSceneScriptingFloatChannel
        default=0
        thinned_from=434
        keys [[0,-1.769],[2,-1.7741],[3,-1.7719],[4,-1.7782],[5,-1.7755],[6,-1.7817],[7,-1.7784],[9,-1.7837],[11,-1.7792],[12,-1.8222],[13,-1.8775],[14,-1.9751],[15,-2.1114],[16,-2.2622],[17,-2.395],[18,-2.4944],[19,-2.7899],[20,-3.2426],[21,-3.211],[22,-2.4091],[23,-1.357],[24,-0.3716],[25,0.5874],[26,1.6909],[27,2.7988],[28,3.7826],[29,4.656],[30,5.468],[31,6.1534],[32,6.6438],[33,6.979],[34,7.2266],[35,7.3948],[36,7.4943],[37,7.5102],[38,7.4587],[39,7.3631],[41,7.1234],[42,6.9767],[43,6.7651],[45,6.2969],[46,6.0954],[49,5.576],[52,5.1203],[54,4.8796],[58,4.5536],[59,4.4565],[60,4.3464],[61,4.2081],[62,4.0281],[63,3.7247],[64,3.0258],[65,1.765],[66,0.7152],[67,0.0558],[68,-0.8],[69,-1.5589],[70,-1.9177],[71,-1.9855],[72,-1.8362],[73,-1.5858],[74,-1.1413],[75,-0.6591],[76,-0.1618],[77,0.1535],[78,0.3342],[79,0.3846],[80,0.4483],[81,0.507],[82,0.5617],[83,0.6327],[84,0.7053],[85,0.7652],[86,0.8102],[87,0.8528],[88,0.8917],[89,0.9272],[90,0.9592],[91,0.9933],[92,1.0018],[93,1.0389],[94,1.0911],[95,1.1511],[96,1.225],[97,1.2944],[98,1.3709],[99,1.4404],[100,1.5053],[101,1.5572],[102,1.5844],[103,1.5768],[104,1.5388],[105,1.4775],[106,1.3883],[107,1.2764],[108,1.1266],[109,0.9567],[110,0.7978],[111,0.675],[112,0.5727],[113,0.4669],[114,0.3673],[115,0.2801],[116,0.2128],[117,0.1394],[118,0.0748],[119,0.0131],[120,-0.04],[121,-0.0791],[122,-0.1047],[123,-0.1287],[124,-0.1532],[125,-0.1687],[126,-0.1902],[127,-0.2082],[128,-0.2184],[129,-0.2192],[130,-0.227],[131,-0.2303],[132,-0.2407],[133,-0.2362],[134,-0.2246],[135,-0.1816],[136,-0.1253],[137,-0.0626],[138,0.0012],[139,0.0662],[140,0.3631],[141,0.9993],[142,1.7713],[143,2.585],[144,3.3824],[145,4.1478],[146,5.0536],[147,6.2571],[148,7.1304],[149,7.294],[150,4.7199],[151,-0.8092],[152,-5.8136],[153,-9.6224],[154,-11.6867],[155,-12.9739],[156,-13.8746],[157,-14.2151],[158,-12.3616],[159,-8.5939],[160,-1.9773],[161,2.7709],[162,6.4932],[163,8.7579],[164,9.986],[165,10.5747],[166,10.5438],[167,10.0514],[168,8.804],[169,6.2657],[170,3.1984],[171,1.7605],[172,1.1895],[173,1.0856],[174,1.1774],[175,1.3972],[176,1.6286],[177,1.9349],[178,2.3142],[179,2.7046],[180,3.0672],[181,3.3924],[182,3.6493],[183,3.8535],[184,4.0024],[185,4.093],[188,4.2689],[191,4.4985],[194,4.68],[196,4.7808],[197,4.8428],[199,4.9307],[200,4.9553],[203,5.0738],[204,5.1304],[205,5.2184],[206,5.3246],[207,5.4622],[211,6.119],[212,6.2455],[213,6.3319],[215,6.369],[216,6.3504],[217,6.2482],[218,6.0056],[219,5.5123],[220,4.7407],[221,3.4804],[222,1.6826],[223,-0.5012],[224,-2.9224],[225,-6.0294],[226,-9.596],[227,-12.5402],[229,-17.3539],[230,-19.484],[231,-21.0753],[232,-22.2098],[233,-23.1544],[234,-23.9402],[235,-24.5966],[236,-25.1514],[238,-26.0575],[239,-26.4292],[240,-26.7378],[242,-27.2072],[248,-28.0509],[252,-28.3243],[257,-28.4185],[258,-28.2551],[259,-27.8172],[260,-26.6194],[261,-24.5086],[262,-21.9491],[263,-19.1381],[264,-16.118],[265,-13.4154],[266,-11.224],[267,-9.2807],[268,-7.5346],[269,-5.9341],[270,-4.4462],[271,-3.0828],[272,-1.8627],[273,-0.8061],[274,0.0652],[275,0.7662],[276,1.3284],[277,1.7988],[278,2.1743],[279,2.5136],[280,2.8196],[282,3.4774],[283,3.8395],[284,4.2143],[285,4.6204],[286,5.0978],[287,5.6224],[288,6.279],[290,7.8439],[291,8.5562],[292,9.1805],[293,9.7198],[294,10.0712],[295,10.2181],[296,10.1451],[297,9.9706],[298,9.7625],[299,9.5982],[303,9.0889],[304,8.9405],[308,8.4224],[309,8.2527],[310,8.1063],[313,7.5975],[315,7.2785],[316,7.1428],[317,6.983],[319,6.6941],[320,6.5159],[321,6.3619],[323,5.9755],[325,5.5241],[326,5.2193],[327,4.823],[328,4.3934],[329,3.9885],[330,3.666],[331,3.4316],[332,3.2454],[333,3.0962],[335,2.8722],[336,2.7907],[337,2.7263],[338,2.6696],[340,2.5719],[341,2.5273],[342,2.4541],[343,2.3495],[344,2.1884],[345,1.9557],[346,1.5034],[347,0.8583],[348,-0.031],[349,-1.2127],[350,-2.704],[351,-4.9667],[352,-8.1345],[353,-11.1542],[354,-13.3691],[355,-16.2405],[356,-16.5575],[357,-13.7436],[358,-10.5571],[359,-8.0202],[360,-5.693],[361,-3.8852],[362,-2.2181],[363,-0.9695],[364,-0.2812],[365,-0.2171],[366,-0.3394],[367,-0.3002],[368,-0.4703],[369,-1.0339],[370,-1.7501],[371,-2.4923],[372,-3.0607],[373,-3.5515],[374,-3.9578],[375,-4.0418],[376,-3.8031],[377,-3.4193],[378,-2.9589],[380,-1.9982],[381,-1.5967],[382,-1.3398],[383,-1.1882],[384,-1.0774],[385,-0.9757],[386,-0.8968],[387,-0.8209],[388,-0.7275],[389,-0.6266],[390,-0.5345],[391,-0.4879],[392,-0.4827],[394,-0.4799],[396,-0.4731],[397,-0.474],[398,-0.4843],[399,-0.4932],[400,-0.515],[401,-0.5469],[402,-0.5835],[403,-0.6366],[405,-0.7749],[406,-0.8646],[407,-0.9349],[408,-1.0162],[409,-1.0948],[410,-1.1756],[411,-1.2528],[412,-1.3327],[413,-1.4076],[414,-1.4783],[416,-1.6068],[417,-1.6577],[418,-1.7047],[420,-1.7895],[421,-1.8217],[422,-1.8481],[423,-1.8658],[424,-1.8786],[426,-1.8882],[428,-1.8518],[431,-1.7865],[432,-1.7722],[433,-1.7668]]
       }
       hand_r_ik_ctrl.Location.Z {
        type=MovieSceneScriptingFloatChannel
        default=0
        thinned_from=434
        keys [[0,-14.2377],[4,-14.2082],[11,-14.1199],[13,-14.1885],[14,-14.2688],[15,-14.4453],[17,-14.9669],[18,-15.0969],[19,-14.9726],[20,-14.047],[21,-11.6324],[22,-8.411],[23,-5.56],[24,-3.1878],[25,-1.1881],[26,0.5674],[27,2.0128],[28,3.0935],[29,3.8645],[30,4.4354],[31,4.8298],[32,5.0739],[33,5.217],[34,5.262],[35,5.2518],[36,5.1857],[37,5.0786],[38,4.9302],[40,4.5913],[41,4.4464],[42,4.2864],[43,4.0905],[44,3.8708],[45,3.6657],[46,3.4978],[47,3.3594],[48,3.2369],[50,3.0248],[53,2.7568],[55,2.595],[56,2.5372],[58,2.4146],[59,2.3489],[60,2.2757],[61,2.1841],[62,2.1261],[63,2.0203],[64,1.7153],[65,1.2391],[66,0.4707],[67,-0.5204],[68,-1.9503],[69,-3.5165],[70,-4.7257],[71,-5.7304],[72,-6.5616],[73,-7.2549],[74,-7.7756],[75,-8.1328],[76,-8.3963],[77,-8.5704],[78,-8.7016],[79,-8.8055],[80,-8.8655],[81,-8.8853],[82,-8.8666],[83,-8.8193],[87,-8.4995],[90,-8.1603],[96,-7.3834],[97,-7.2813],[98,-7.1961],[99,-7.1317],[100,-7.1015],[101,-7.1234],[102,-7.2106],[103,-7.3774],[104,-7.6254],[105,-7.9479],[106,-8.3517],[107,-8.8328],[108,-9.4218],[110,-10.7643],[111,-11.3797],[112,-11.9246],[113,-12.4303],[114,-12.8898],[115,-13.2952],[116,-13.66],[117,-13.9892],[118,-14.284],[119,-14.5375],[121,-14.9173],[123,-15.2012],[125,-15.3844],[129,-15.5637],[130,-15.5874],[134,-15.4748],[135,-15.4183],[136,-15.3226],[137,-15.1578],[138,-14.8958],[139,-14.5077],[140,-13.9546],[141,-13.1591],[142,-12.0252],[143,-10.8092],[144,-9.374],[145,-7.3309],[146,-4.9615],[147,-0.9991],[148,4.3401],[149,11.9663],[150,29.4825],[151,41.3465],[152,50.0376],[153,54.5554],[154,56.1566],[156,57.4588],[157,62.5506],[158,72.9222],[159,74.3665],[160,71.9168],[161,68.1224],[162,64.6937],[163,61.83],[164,59.7714],[165,58.6916],[166,58.383],[167,58.6799],[168,59.8401],[169,62.5802],[170,65.9329],[171,67.9338],[172,69.3755],[173,70.06],[174,70.0457],[176,69.4338],[178,68.314],[180,66.9534],[182,65.9888],[184,65.5143],[195,64.9148],[212,65.6222],[215,65.3299],[218,64.5727],[220,63.6461],[221,62.9257],[222,61.9188],[223,60.6436],[224,59.1307],[225,56.9858],[226,54.3449],[227,52.1636],[228,50.6258],[229,49.2313],[230,48.0281],[231,47.0816],[232,46.3507],[233,45.7379],[235,44.804],[238,43.859],[242,43.0492],[252,42.304],[256,42.2834],[257,42.3466],[258,42.6513],[259,43.3585],[260,44.8306],[264,52.7838],[265,54.4727],[266,55.9176],[267,57.2101],[268,58.3409],[270,60.1675],[272,61.6475],[274,62.686],[277,63.6104],[284,64.6889],[295,65.6824],[319,66.7188],[326,66.2908],[335,65.1975],[345,65.0048],[346,64.9009],[347,64.468],[348,63.5311],[349,61.9129],[350,59.4574],[351,55.2618],[352,49.1926],[353,41.7194],[354,33.752],[355,23.4536],[356,12.0096],[357,4.1023],[358,-0.9108],[359,-4.7367],[360,-8.2128],[361,-10.9506],[363,-16.5514],[364,-18.7557],[365,-20.2744],[366,-20.8257],[367,-20.3381],[368,-19.3582],[369,-18.1861],[370,-16.8603],[371,-15.4239],[372,-13.9518],[373,-12.5107],[374,-11.1739],[375,-9.9297],[376,-8.6217],[377,-7.2665],[378,-6.0916],[379,-5.1166],[380,-4.2843],[381,-3.655],[382,-3.2909],[383,-3.1625],[384,-3.1987],[385,-3.3879],[386,-3.7528],[387,-4.25],[388,-4.8018],[389,-5.3726],[390,-5.9309],[391,-6.5208],[393,-7.8268],[394,-8.4552],[395,-9.0307],[396,-9.5469],[397,-10.0215],[398,-10.4634],[399,-10.8689],[400,-11.2295],[401,-11.5408],[403,-12.0705],[406,-12.7824],[408,-13.2033],[409,-13.3933],[413,-13.9698],[415,-14.1897],[418,-14.4074],[422,-14.5595],[424,-14.5791],[426,-14.5456],[431,-14.2957],[433,-14.2553]]
       }
       hand_r_ik_ctrl.Rotation.X {
        type=MovieSceneScriptingFloatChannel
        default=0
        thinned_from=434
        keys [[0,19.7241],[12,19.6941],[13,19.7196],[14,19.7908],[16,19.755],[17,19.9426],[18,20.6091],[19,22.633],[20,26.1544],[21,30.227],[22,33.8833],[23,36.9372],[24,39.5057],[25,41.5023],[26,42.8557],[27,43.7106],[28,44.2067],[29,44.425],[30,44.4177],[33,43.9173],[37,42.5466],[40,41.3536],[42,40.7547],[45,40.1837],[48,39.9606],[58,39.8028],[61,40.071],[63,40.5002],[64,40.2079],[65,40.0934],[66,40.4779],[67,40.1232],[68,37.6609],[69,34.2704],[70,32.485],[71,31.9363],[72,31.5591],[74,31.184],[76,31.0608],[78,31.1749],[82,31.5938],[94,31.9393],[96,31.8794],[98,31.6617],[99,31.4855],[100,31.1702],[101,30.6561],[102,29.977],[103,29.1664],[104,28.1508],[106,25.6958],[107,24.5725],[108,23.5817],[109,22.6513],[110,21.8188],[111,21.12],[112,20.5876],[113,20.1983],[114,19.9021],[116,19.4281],[118,19.1189],[122,18.7387],[125,18.5938],[130,18.5964],[132,18.7395],[133,18.8458],[134,19.0145],[135,19.2901],[136,19.7336],[137,20.3977],[138,21.3337],[139,22.5866],[140,24.8362],[141,28.1167],[142,31.5996],[143,35.7306],[144,39.3167],[146,42.4479],[147,45.1256],[148,50.1933],[149,60.5018],[150,73.7258],[151,48.6219],[152,-8.6965],[153,-71.0415],[154,-72.6414],[155,-72.8097],[156,-71.0197],[157,-66.1492],[158,-64.0468],[159,-62.9804],[160,-63.9515],[162,-59.5356],[163,-56.1324],[164,-52.3023],[165,-49.8357],[166,-48.4761],[167,-48.1619],[168,-49.7167],[169,-54.2915],[170,-59.3458],[171,-61.6771],[172,-63.1007],[173,-64.0492],[174,-64.5488],[175,-64.7747],[176,-64.8314],[177,-64.5385],[181,-61.8148],[182,-61.2494],[183,-60.8263],[185,-60.3141],[191,-59.5078],[200,-59.9946],[206,-60.9821],[211,-62.2254],[213,-62.4824],[214,-62.484],[215,-62.3306],[216,-61.9128],[217,-61.209],[218,-60.1506],[219,-58.6551],[220,-56.7087],[221,-53.773],[222,-49.7619],[223,-45.2849],[224,-40.9477],[225,-36.4795],[226,-31.8869],[227,-28.2045],[228,-25.4788],[229,-23.1005],[230,-21.1285],[231,-19.6317],[232,-18.5019],[233,-17.5623],[234,-16.7845],[235,-16.1497],[236,-15.6164],[237,-15.1513],[238,-14.7355],[239,-14.3845],[240,-14.0639],[242,-13.5751],[244,-13.2295],[248,-12.7283],[250,-12.5693],[252,-12.5261],[253,-12.5421],[255,-12.6766],[256,-12.7646],[257,-13.0022],[258,-13.6945],[259,-15.1231],[260,-18.5125],[261,-23.4539],[262,-27.951],[263,-31.4152],[264,-34.4368],[265,-37.1195],[266,-39.6748],[267,-42.1419],[268,-44.4591],[269,-46.5603],[271,-50.4649],[272,-52.2702],[273,-53.8695],[274,-55.1799],[275,-56.2345],[276,-57.1203],[278,-58.5531],[281,-60.1012],[284,-61.2015],[288,-62.2139],[295,-65.0648],[307,-66.7387],[316,-66.7152],[323,-65.9781],[326,-65.2889],[330,-63.9898],[335,-63.0644],[345,-62.7903],[346,-62.4388],[347,-61.2126],[348,-58.5228],[349,-53.5041],[350,-44.949],[351,-30.1618],[352,-10.1602],[353,10.7304],[354,28.0628],[355,42.1499],[356,39.512],[357,30.6404],[358,23.9183],[359,20.1976],[360,17.4929],[361,15.7757],[362,14.7831],[363,14.3669],[364,15.0278],[365,16.9065],[366,18.9316],[367,20.5584],[368,22.2627],[369,23.9013],[370,25.3731],[371,26.7594],[372,27.9991],[373,29.1143],[374,30.122],[375,30.9798],[376,31.6822],[377,32.2911],[378,32.8349],[379,33.2962],[380,33.6188],[381,33.7598],[382,33.6635],[383,33.3122],[384,32.7258],[385,31.9229],[386,30.8623],[387,29.5696],[389,26.8084],[390,25.5972],[391,24.5387],[392,23.5507],[393,22.6466],[394,21.8437],[395,21.1586],[396,20.5723],[397,20.0576],[398,19.624],[399,19.2591],[400,18.9899],[401,18.811],[402,18.6924],[403,18.6215],[405,18.5783],[409,18.614],[422,19.5059],[427,19.6754],[433,19.6996]]
       }
       hand_r_ik_ctrl.Rotation.Y {
        type=MovieSceneScriptingFloatChannel
        default=0
        thinned_from=434
        keys [[0,-0.1604],[1,-0.1603],[2,-0.1452],[3,-0.1079],[4,-0.0889],[5,-0.0422],[6,-0.023],[7,0.0172],[8,0.0288],[9,0.0293],[10,0.0366],[11,0.0156],[12,-0.08],[13,-0.2473],[14,-0.524],[15,-0.9416],[16,-1.4916],[17,-2.1428],[18,-2.8767],[19,-4.0696],[20,-5.3416],[21,-5.3776],[23,-3.1406],[24,-2.1383],[25,-1.2647],[26,-0.4368],[27,0.309],[28,0.9429],[29,1.4719],[30,1.9404],[31,2.3529],[32,2.7285],[33,3.072],[34,3.3912],[35,3.6657],[36,3.8617],[37,4.0082],[38,4.1171],[39,4.1942],[41,4.2885],[42,4.2355],[43,4.0817],[45,3.6014],[46,3.3934],[47,3.2013],[50,2.6512],[51,2.4894],[52,2.3486],[53,2.2256],[54,2.1269],[55,2.0496],[56,1.9879],[57,1.9354],[58,1.8872],[59,1.8303],[61,1.756],[62,1.7482],[63,1.8615],[64,2.2207],[65,3.076],[66,3.8404],[67,4.627],[68,5.6488],[69,6.5459],[70,6.6169],[71,5.8891],[72,5.0885],[73,4.2603],[74,3.5248],[75,2.8832],[76,2.2809],[77,1.7846],[78,1.3782],[79,1.0149],[80,0.7382],[81,0.5731],[82,0.5185],[83,0.54],[84,0.615],[85,0.721],[86,0.8364],[87,0.9626],[88,1.1168],[89,1.288],[90,1.4746],[92,1.8641],[93,2.062],[97,2.8822],[99,3.2792],[100,3.4878],[102,3.9313],[104,4.3061],[105,4.4771],[106,4.5534],[107,4.4953],[108,4.2393],[109,3.8401],[110,3.371],[111,2.9096],[112,2.469],[113,2.0222],[114,1.5875],[115,1.1767],[117,0.4127],[118,0.0581],[119,-0.2673],[120,-0.5421],[121,-0.7792],[122,-0.9831],[123,-1.1662],[124,-1.3183],[125,-1.4475],[126,-1.5509],[127,-1.6281],[128,-1.6826],[129,-1.7175],[130,-1.7404],[131,-1.7238],[132,-1.7184],[133,-1.6815],[134,-1.6088],[135,-1.3852],[136,-0.992],[137,-0.378],[138,0.5057],[139,1.715],[140,3.8417],[141,7.2738],[142,11.5924],[143,17.1475],[144,21.8925],[146,26.5861],[147,30.9503],[148,36.4929],[149,43.313],[150,34.5871],[151,59.048],[152,80.4877],[153,66.3328],[154,55.4955],[155,48.2913],[156,43.5454],[157,33.966],[158,16.314],[159,14.7419],[160,27.6418],[161,35.3493],[162,40.8212],[163,45.6346],[164,48.6707],[165,49.7557],[166,49.7312],[167,48.9603],[168,47.5165],[169,44.0208],[170,39.4954],[171,36.5009],[172,34.3059],[173,33.6234],[174,34.2448],[178,38.2281],[179,39.2812],[180,40.2458],[181,40.992],[182,41.4742],[183,41.6983],[184,41.7931],[207,41.8025],[215,42.4119],[218,42.7967],[219,42.8469],[220,42.8017],[221,42.6421],[222,42.3679],[225,41.1145],[226,40.6057],[227,39.9395],[229,38.4223],[230,37.8141],[231,37.3891],[233,36.8607],[238,36.1263],[253,35.2726],[256,34.9138],[257,34.7103],[258,34.4203],[259,34.0171],[261,32.9151],[262,32.6834],[263,32.7617],[264,33.0542],[271,35.7605],[273,36.4106],[275,36.843],[283,37.5759],[285,37.8893],[287,38.4122],[288,38.8834],[289,39.5588],[293,42.5273],[295,43.7894],[297,44.795],[298,45.1839],[299,45.4439],[302,45.8226],[309,45.9338],[342,43.9387],[344,43.6175],[345,43.9999],[346,45.4118],[347,47.8074],[348,51.0445],[350,58.6835],[351,61.1054],[352,60.5801],[353,56.3103],[354,48.3421],[355,33.07],[356,18.7052],[357,14.5862],[358,13.0236],[359,11.4965],[360,9.6236],[361,7.9458],[362,5.1236],[363,1.4894],[364,-2.5183],[365,-7.1211],[366,-10.544],[367,-11.6344],[368,-11.6941],[369,-11.0776],[370,-9.9084],[371,-8.4142],[372,-6.7192],[373,-4.8741],[374,-2.9511],[375,-1.0677],[376,0.9064],[377,2.9274],[378,4.583],[379,5.832],[380,6.9018],[381,7.7199],[382,8.1912],[383,8.3928],[384,8.5048],[385,8.5717],[386,8.5935],[387,8.5751],[388,8.5182],[389,8.3957],[390,8.1828],[391,7.8399],[392,7.3823],[394,6.3781],[395,5.8951],[396,5.463],[398,4.69],[400,4.0056],[401,3.6904],[402,3.4128],[403,3.1181],[404,2.8159],[405,2.5238],[406,2.2405],[407,1.9752],[408,1.7218],[409,1.4886],[411,1.071],[412,0.8725],[413,0.6807],[414,0.5007],[415,0.3351],[416,0.1713],[417,0.0377],[418,-0.0853],[419,-0.196],[420,-0.3003],[421,-0.3823],[422,-0.4477],[423,-0.4972],[424,-0.5251],[425,-0.5275],[426,-0.5251],[427,-0.4751],[428,-0.4188],[429,-0.349],[430,-0.283],[431,-0.2276],[432,-0.1881],[433,-0.1728]]
       }
       hand_r_ik_ctrl.Rotation.Z {
        type=MovieSceneScriptingFloatChannel
        default=0
        thinned_from=434
        keys [[0,6.6911],[2,6.6979],[7,6.765],[11,6.7676],[12,6.7295],[14,6.6765],[16,6.4849],[17,6.5976],[18,7.1005],[19,8.2648],[20,10.0332],[21,12.2937],[22,14.7567],[23,16.9542],[24,18.8089],[25,20.3921],[26,21.8415],[27,23.1259],[28,24.2358],[29,25.2406],[30,26.18],[31,27.0087],[32,27.6712],[33,28.1906],[34,28.5754],[35,28.8348],[36,28.9777],[37,28.9698],[38,28.8444],[41,28.1519],[42,27.8676],[45,26.6118],[46,26.2546],[49,25.3881],[52,24.6633],[54,24.288],[62,23.2768],[63,23.0062],[64,22.1823],[65,21.0645],[66,20.5942],[67,20.0196],[68,18.0422],[69,15.5293],[70,14.0665],[71,13.5423],[72,13.3477],[73,13.3403],[74,13.6387],[75,14.0452],[76,14.5168],[77,14.8273],[78,15.009],[79,15.0674],[82,15.4086],[85,15.8958],[88,16.2166],[91,16.4543],[96,16.7682],[98,16.8187],[99,16.7975],[100,16.6932],[101,16.4632],[102,16.1089],[103,15.6328],[104,14.99],[105,14.1904],[106,13.3216],[107,12.4982],[109,10.9066],[110,10.1757],[111,9.5646],[112,9.0792],[113,8.6792],[114,8.348],[115,8.061],[116,7.833],[117,7.6275],[118,7.4632],[119,7.3151],[121,7.08],[123,6.9099],[125,6.8005],[127,6.7392],[130,6.706],[132,6.7617],[133,6.8266],[134,6.9117],[135,7.0477],[136,7.1601],[137,7.2002],[138,7.125],[139,6.9019],[140,5.9335],[141,4.4437],[142,3.5565],[143,3.3463],[144,4.1994],[145,6.415],[146,8.959],[147,12.1326],[148,16.3168],[149,23.0822],[150,16.7003],[151,-18.7258],[152,-91.7565],[153,-161.3321],[154,-164.0108],[155,-164.9983],[156,-164.0607],[157,-161.5913],[158,-158.1941],[159,-152.0033],[160,-150.3229],[161,-145.4377],[162,-141.1953],[164,-131.1217],[165,-127.7942],[166,-125.8737],[167,-125.2169],[168,-127.1569],[169,-133.5011],[170,-141.4128],[171,-145.1978],[172,-147.1803],[174,-149.8951],[175,-150.611],[176,-150.9983],[177,-150.704],[182,-145.9417],[185,-144.2829],[192,-142.4201],[212,-143.1461],[215,-142.5473],[218,-140.7385],[221,-137.9768],[224,-134.3655],[226,-132.8335],[227,-132.4808],[228,-132.6809],[231,-134.9562],[239,-137.6436],[260,-138.5878],[261,-138.2987],[262,-137.4409],[264,-134.5781],[265,-133.4034],[267,-132.3077],[270,-131.6976],[288,-134.4711],[291,-136.4118],[298,-144.0891],[304,-148.0412],[314,-151.0662],[323,-151.1632],[336,-148.8027],[344,-149.1772],[345,-149.528],[346,-150.2947],[347,-150.5584],[348,-149.5135],[349,-146.0717],[350,-138.8149],[351,-125.3169],[352,-106.8748],[353,-86.4668],[354,-67.0517],[355,-41.3821],[356,-20.1495],[357,-13.1731],[358,-10.2151],[359,-7.3578],[360,-4.3978],[361,-1.3054],[362,-0.0499],[363,0.8849],[364,1.8668],[365,2.5034],[366,3.0092],[367,3.9839],[368,4.7773],[369,5.0629],[370,5.1525],[371,5.2815],[372,5.643],[373,6.1141],[374,6.6744],[375,7.5778],[376,8.9097],[377,10.4099],[378,11.8293],[379,13.0822],[380,14.1946],[381,15.005],[382,15.3692],[383,15.3604],[384,15.1322],[385,14.749],[386,14.2013],[387,13.5292],[389,12.1018],[390,11.4621],[392,10.2522],[393,9.6989],[394,9.211],[395,8.7994],[396,8.4534],[397,8.1531],[398,7.8872],[399,7.6756],[400,7.4948],[401,7.3567],[403,7.1663],[406,6.9212],[409,6.716],[411,6.6311],[416,6.4829],[420,6.4431],[426,6.4898],[430,6.6264],[432,6.6649],[433,6.6708]]
       }
       hand_r_ik_ctrl.Scale.X {type=MovieSceneScriptingFloatChannel, default=1, const [1,434]}
       hand_r_ik_ctrl.Scale.Y {type=MovieSceneScriptingFloatChannel, default=1, const [1,434]}
       hand_r_ik_ctrl.Scale.Z {type=MovieSceneScriptingFloatChannel, default=1, const [1,434]}
       Softness_2 {type=MovieSceneScriptingFloatChannel, default=0, const [0,434]}
       hand_r_ikorient_ctrl.Rotation.Roll {type=MovieSceneScriptingFloatChannel, default=0, const [0,434]}
       hand_r_ikorient_ctrl.Rotation.Pitch {type=MovieSceneScriptingFloatChannel, default=0, const [0,434]}
       hand_r_ikorient_ctrl.Rotation.Yaw {type=MovieSceneScriptingFloatChannel, default=0, const [0,434]}
       arm_r_pv_ik_ctrl.Location.X {
        type=MovieSceneScriptingFloatChannel
        default=0
        thinned_from=434
        keys [[0,10.3496],[9,10.4182],[11,10.4089],[12,10.4256],[13,10.4634],[14,10.5353],[16,10.6199],[17,10.8043],[18,11.2859],[19,12.611],[20,14.7512],[21,16.8594],[22,18.2244],[23,18.9764],[24,19.4191],[25,19.5708],[26,19.3621],[30,17.6834],[31,17.3162],[32,17.0701],[34,16.7891],[36,16.5943],[39,16.424],[42,16.3456],[58,16.8571],[60,16.9893],[61,17.0925],[62,17.2551],[63,17.4665],[64,17.7353],[65,18.3667],[66,19.1248],[67,19.4483],[68,19.0517],[69,18.148],[70,17.5585],[73,16.7038],[75,16.095],[76,15.8193],[77,15.6545],[78,15.5639],[93,15.6359],[96,15.5526],[98,15.3961],[99,15.2844],[100,15.1099],[101,14.8488],[102,14.5179],[103,14.1328],[104,13.651],[106,12.4639],[107,11.9155],[109,10.9646],[110,10.5309],[111,10.1489],[112,9.8468],[113,9.6208],[114,9.4449],[115,9.2907],[116,9.1579],[117,9.0521],[119,8.901],[123,8.6908],[125,8.6439],[130,8.6596],[132,8.7542],[134,8.9013],[135,8.995],[136,9.1382],[137,9.3625],[138,9.6985],[139,10.1734],[140,10.9707],[142,13.0946],[143,13.9559],[144,14.6753],[145,15.3358],[146,15.9066],[147,16.5168],[148,17.4611],[149,18.9387],[150,21.0834],[151,18.0714],[152,16.3817],[153,16.1015],[154,16.5184],[155,16.8175],[156,16.9405],[157,16.4276],[159,15.9966],[160,15.4301],[161,14.9604],[162,14.2521],[163,13.7076],[164,13.3167],[165,13.0627],[166,12.9014],[167,12.8001],[168,12.9067],[169,13.3801],[170,13.9143],[171,14.1552],[172,14.2336],[176,14.2837],[177,14.3232],[180,14.5271],[183,14.5675],[203,13.8986],[205,13.7687],[207,13.5751],[210,13.2383],[211,13.1571],[212,13.111],[213,13.1115],[214,13.1769],[215,13.2773],[216,13.4384],[217,13.6767],[218,14.0368],[219,14.5938],[220,15.3521],[221,16.4783],[222,18.0044],[223,19.7876],[224,21.6905],[225,24.0166],[226,26.533],[227,28.4589],[228,29.8322],[229,31.027],[230,31.9664],[231,32.6189],[232,33.0664],[233,33.4119],[235,33.8853],[239,34.4216],[256,35.1112],[258,35.2447],[259,35.1572],[260,34.6847],[261,33.7781],[262,32.6876],[263,31.4636],[265,28.6723],[266,27.5099],[267,26.4579],[270,23.5753],[271,22.6781],[272,21.8459],[273,21.1022],[274,20.4714],[275,19.948],[276,19.5204],[277,19.1432],[279,18.5595],[282,17.7649],[284,17.1396],[285,16.7984],[287,15.9814],[288,15.428],[289,14.7198],[290,13.9606],[292,12.5508],[293,11.8706],[294,11.3103],[295,10.9047],[296,10.691],[297,10.5521],[300,10.2817],[303,10.0773],[305,10.0121],[308,9.9959],[311,10.1229],[312,10.1826],[316,10.5166],[319,10.8407],[321,11.1222],[323,11.4763],[325,11.9402],[326,12.2418],[329,13.3852],[330,13.6986],[331,13.9505],[332,14.1707],[333,14.3588],[335,14.6344],[338,14.8175],[342,14.9375],[343,14.9928],[344,15.1004],[345,15.2636],[346,15.5697],[347,16.1064],[348,16.957],[349,18.1828],[350,19.7775],[351,21.9875],[352,24.5954],[353,26.9053],[354,28.4735],[355,28.8146],[356,25.8139],[357,21.0018],[358,16.3978],[359,13.1584],[360,10.4382],[361,8.5787],[362,7.1157],[363,5.9942],[364,5.6133],[365,6.0183],[366,6.6438],[367,7.3528],[368,8.3213],[369,9.4719],[371,11.8683],[372,12.9702],[373,14.0077],[374,15.0126],[375,15.8532],[376,16.4684],[377,16.9182],[378,17.2278],[379,17.4148],[380,17.498],[381,17.4813],[382,17.3586],[383,17.125],[384,16.793],[385,16.3694],[386,15.8476],[387,15.2238],[389,13.8509],[390,13.2297],[391,12.6918],[392,12.198],[393,11.7422],[394,11.3316],[395,10.975],[396,10.6685],[397,10.4017],[398,10.1727],[399,9.9859],[400,9.8401],[401,9.7422],[402,9.6847],[403,9.6487],[405,9.6345],[409,9.6794],[422,10.2223],[427,10.3181],[433,10.335]]
       }
       arm_r_pv_ik_ctrl.Location.Y {
        type=MovieSceneScriptingFloatChannel
        default=0
        thinned_from=434
        keys [[0,-9.3671],[4,-9.3533],[14,-9.2359],[15,-9.2953],[16,-9.4079],[17,-9.3964],[18,-9.0548],[19,-7.8824],[20,-5.3297],[21,-1.276],[22,3.2541],[23,7.1976],[24,10.4514],[25,13.1168],[26,15.3113],[27,17.0105],[28,18.2079],[29,18.9982],[30,19.5093],[31,19.7946],[32,19.9013],[33,19.8948],[34,19.7486],[36,19.2956],[37,19.0066],[40,17.9757],[41,17.6768],[44,16.9275],[45,16.7185],[46,16.5523],[48,16.3224],[55,15.7805],[61,15.4454],[62,15.4398],[63,15.2779],[64,14.5711],[65,13.3455],[66,12.0076],[67,10.3881],[68,7.61],[69,4.5558],[70,2.7152],[71,1.7327],[72,1.0445],[73,0.5767],[74,0.3141],[75,0.2166],[76,0.2189],[77,0.2684],[78,0.3155],[79,0.3664],[80,0.4353],[81,0.4988],[82,0.5542],[85,0.7153],[86,0.7717],[87,0.8156],[88,0.8856],[89,0.9408],[90,1.0044],[92,1.1604],[94,1.3093],[95,1.3719],[96,1.4282],[97,1.4408],[98,1.4291],[99,1.3764],[100,1.2369],[101,0.9718],[102,0.5832],[103,0.0728],[104,-0.5911],[105,-1.3895],[107,-3.117],[108,-3.9966],[109,-4.896],[110,-5.7495],[111,-6.4921],[112,-7.1131],[113,-7.6508],[114,-8.114],[115,-8.5102],[116,-8.8521],[117,-9.1455],[118,-9.3954],[119,-9.6034],[121,-9.9043],[122,-10.0236],[123,-10.1201],[125,-10.2385],[130,-10.3346],[132,-10.249],[134,-10.0884],[135,-9.9771],[136,-9.8047],[137,-9.5331],[138,-9.1213],[139,-8.5246],[140,-7.3869],[141,-5.5661],[142,-3.3702],[143,-1.1026],[144,1.2191],[145,3.6771],[146,6.4228],[147,10.7487],[148,16.765],[149,25.7096],[150,47.1354],[151,59.6031],[152,64.7201],[153,65.4058],[155,64.1038],[156,63.6687],[157,66.6352],[158,76.5833],[159,82.2223],[160,87.5383],[161,89.386],[162,89.9781],[163,89.4801],[167,86.1107],[168,85.6668],[169,85.7624],[170,86.0842],[171,86.6069],[173,88.0708],[175,88.8716],[177,89.2],[203,88.3544],[214,89.2475],[217,89.075],[219,88.5418],[220,88.0275],[221,87.208],[222,86.0132],[223,84.4811],[224,82.6609],[225,80.1087],[226,76.96],[227,74.2687],[229,70.1278],[230,68.3132],[231,66.9073],[232,65.8562],[233,64.9683],[235,63.5941],[238,62.1862],[242,61.0339],[250,60.0342],[256,59.8606],[257,59.9289],[258,60.235],[259,60.9287],[260,62.42],[261,64.6595],[263,69.7306],[264,72.3812],[265,74.7013],[266,76.6435],[267,78.3862],[268,79.9305],[270,82.4902],[272,84.5833],[274,86.0827],[277,87.3876],[293,90.964],[295,91.0958],[326,89.4723],[345,87.6127],[346,87.403],[347,86.9003],[348,85.9244],[349,84.2701],[350,81.713],[351,77.0926],[352,70.0369],[353,61.3773],[354,52.1874],[355,36.9961],[356,18.8217],[357,6.8398],[358,-0.17],[359,-4.7406],[360,-8.2687],[361,-10.4043],[362,-12.0635],[363,-13.6204],[364,-14.4634],[365,-14.2122],[366,-13.3527],[367,-12.1742],[369,-9.1931],[370,-7.7472],[371,-6.3213],[372,-4.9459],[373,-3.693],[374,-2.5549],[375,-1.4653],[376,-0.4381],[377,0.4976],[378,1.352],[379,2.1173],[380,2.7439],[381,3.1805],[382,3.376],[383,3.3362],[384,3.0859],[385,2.6401],[386,1.9566],[387,1.0887],[388,0.1647],[389,-0.7343],[390,-1.5367],[391,-2.2887],[392,-3.0547],[393,-3.7868],[394,-4.4644],[395,-5.067],[396,-5.6015],[397,-6.0898],[398,-6.5341],[399,-6.9318],[400,-7.2734],[401,-7.5553],[402,-7.792],[403,-8.0063],[405,-8.3904],[408,-8.9069],[409,-9.062],[411,-9.2867],[413,-9.4616],[415,-9.5916],[418,-9.6795],[423,-9.6978],[426,-9.6359],[430,-9.4494],[432,-9.3954],[433,-9.387]]
       }
       arm_r_pv_ik_ctrl.Location.Z {
        type=MovieSceneScriptingFloatChannel
        default=0
        thinned_from=434
        keys [[0,9.5674],[4,9.5498],[11,9.5015],[13,9.6082],[14,9.7284],[15,9.9102],[17,10.353],[18,10.5505],[20,11.2748],[21,10.9292],[22,9.874],[23,8.7964],[24,7.914],[26,6.4567],[27,5.8041],[28,5.2813],[29,4.8674],[30,4.5167],[31,4.2425],[32,4.0619],[33,3.9497],[34,3.862],[35,3.8021],[36,3.7528],[37,3.735],[38,3.7438],[39,3.7721],[41,3.8643],[42,3.9273],[43,4.0244],[45,4.2473],[49,4.651],[52,4.9368],[54,5.0852],[56,5.1882],[58,5.2728],[60,5.3887],[61,5.4562],[62,5.5544],[63,5.7486],[64,6.1725],[65,7.1203],[66,8.0391],[67,8.2711],[68,8.191],[69,7.8627],[70,7.6544],[71,7.695],[72,7.6547],[73,7.5679],[76,6.9391],[77,6.8285],[78,6.792],[85,7.0138],[87,7.038],[89,7.0327],[92,6.9803],[94,6.886],[95,6.8269],[101,6.4088],[102,6.3561],[103,6.3298],[104,6.3206],[105,6.3247],[106,6.3629],[107,6.4565],[108,6.6218],[110,7.0612],[111,7.2613],[114,7.8136],[115,7.9809],[117,8.2592],[119,8.5002],[120,8.5937],[121,8.6626],[124,8.7706],[127,8.818],[130,8.7884],[132,8.7453],[135,8.619],[137,8.5775],[139,8.6026],[140,8.4288],[141,7.9249],[142,7.2151],[144,5.6314],[145,4.7864],[146,3.6197],[147,1.4739],[148,-0.5558],[149,-1.5912],[150,6.0732],[151,18.5762],[152,30.1802],[153,38.5491],[154,41.865],[155,43.7458],[156,44.8431],[157,43.4989],[158,33.7811],[159,24.2091],[160,11.7472],[161,2.5755],[162,-3.7906],[163,-6.9225],[164,-8.1291],[165,-8.5493],[166,-7.8538],[167,-6.2954],[168,-3.5546],[169,0.8763],[170,6.108],[171,8.3208],[172,9.0316],[173,9.0401],[174,8.6966],[176,7.6067],[177,6.8383],[179,4.8712],[180,3.9598],[181,3.2456],[182,2.6491],[183,2.2132],[184,1.9626],[185,1.8273],[186,1.7598],[187,1.7031],[188,1.6295],[189,1.5229],[190,1.4206],[191,1.3503],[192,1.3104],[193,1.2869],[194,1.272],[195,1.2743],[196,1.2879],[197,1.278],[198,1.3076],[199,1.3437],[200,1.3987],[201,1.4355],[202,1.4881],[203,1.5281],[204,1.5816],[205,1.6043],[206,1.6316],[207,1.6487],[208,1.6321],[209,1.6244],[210,1.59],[211,1.5168],[212,1.4294],[213,1.3147],[214,1.1826],[215,0.9819],[216,0.7184],[217,0.4418],[218,0.1495],[219,-0.118],[220,-0.3314],[221,-0.4743],[222,-0.5394],[223,-0.5103],[224,-0.3687],[225,0.0202],[226,0.647],[227,1.2551],[228,1.9563],[229,2.9238],[230,3.922],[231,4.7083],[232,5.2701],[233,5.7624],[234,6.1915],[235,6.5644],[236,6.8883],[237,7.1669],[238,7.4121],[239,7.6253],[240,7.7978],[241,7.9365],[243,8.1478],[244,8.2166],[245,8.3034],[246,8.3732],[247,8.4741],[249,8.5976],[250,8.6108],[252,8.5846],[254,8.4893],[256,8.3317],[257,8.1647],[258,7.9015],[259,7.5174],[260,6.7293],[261,5.4039],[262,3.8578],[263,2.3394],[264,0.8757],[265,-0.2519],[266,-0.9788],[267,-1.5176],[268,-1.9211],[269,-2.2501],[270,-2.506],[271,-2.6652],[272,-2.7548],[273,-2.8021],[274,-2.8335],[275,-2.8463],[276,-2.8407],[277,-2.7907],[278,-2.7542],[279,-2.6998],[281,-2.6555],[283,-2.6642],[284,-2.6942],[285,-2.7496],[286,-2.841],[287,-2.9952],[288,-3.1212],[289,-3.1379],[290,-3.0561],[291,-2.9167],[292,-2.6445],[293,-2.2258],[294,-1.6892],[295,-1.08],[296,-0.4655],[297,0.1756],[298,0.7755],[299,1.2712],[300,1.6856],[301,2.0841],[302,2.4631],[303,2.8278],[304,3.1492],[305,3.4468],[306,3.7104],[307,3.962],[308,4.1935],[309,4.4163],[311,4.815],[312,4.9925],[314,5.2731],[315,5.3959],[317,5.5471],[318,5.6105],[320,5.7043],[323,5.7406],[325,5.6662],[326,5.6378],[329,5.6512],[330,5.627],[331,5.564],[333,5.3805],[335,5.2462],[336,5.235],[338,5.2404],[339,5.2616],[340,5.2955],[341,5.3152],[342,5.3504],[344,5.4908],[345,5.6363],[346,5.995],[347,6.2746],[348,6.3625],[349,6.189],[350,5.7548],[351,5.06],[352,4.5945],[353,4.0031],[354,3.3611],[355,8.5909],[356,13.3841],[357,13.4686],[358,11.8573],[359,10.958],[360,10.1816],[361,9.0525],[362,8.0082],[363,7.8056],[364,8.4052],[365,9.143],[366,9.6765],[367,9.8831],[369,9.8268],[370,9.8193],[371,9.8858],[372,9.9192],[373,9.9769],[374,9.9346],[375,9.677],[376,9.4625],[377,9.3844],[378,9.2465],[381,8.7425],[382,8.6183],[384,8.3955],[386,8.0592],[387,7.8708],[389,7.4394],[390,7.2479],[391,7.1176],[392,7.0459],[393,6.9987],[394,6.9733],[396,6.9717],[397,6.9867],[399,7.0668],[400,7.1417],[402,7.3268],[403,7.448],[405,7.7359],[409,8.3772],[413,8.9569],[416,9.3054],[420,9.6096],[422,9.706],[424,9.758],[426,9.7724],[432,9.573],[433,9.5637]]
       }
       arm_r_pv_ik_ctrl.Rotation.X {
        type=MovieSceneScriptingFloatChannel
        default=0
        thinned_from=434
        keys [[0,19.2031],[10,19.1183],[11,19.1342],[12,19.2123],[13,19.3574],[14,19.6115],[15,19.9712],[16,20.4318],[17,21.0194],[18,21.7689],[19,23.1185],[20,24.6515],[21,24.9403],[22,23.9144],[23,22.7344],[24,21.642],[25,20.6185],[26,19.551],[27,18.5532],[28,17.7373],[29,17.0976],[30,16.5646],[31,16.1572],[32,15.8866],[33,15.714],[34,15.5787],[35,15.4801],[37,15.3765],[39,15.3866],[41,15.4549],[42,15.5335],[43,15.658],[52,16.9763],[54,17.2022],[58,17.5152],[61,17.8575],[62,18.0331],[63,18.3011],[64,18.7394],[65,19.5149],[66,20.2193],[67,20.1098],[68,19.4824],[69,18.4754],[70,17.9252],[71,18.0636],[72,18.1519],[73,18.1982],[74,18.1143],[76,17.8689],[77,17.8381],[78,17.8792],[81,18.1751],[84,18.3575],[86,18.411],[89,18.3446],[92,18.1541],[95,17.7961],[99,17.1301],[104,16.0302],[105,15.7846],[106,15.5986],[107,15.5503],[108,15.6705],[109,15.8974],[115,17.5451],[118,18.261],[119,18.4689],[120,18.6356],[122,18.8503],[126,19.0705],[130,19.1059],[134,19.0962],[135,19.1471],[136,19.2609],[137,19.4601],[138,19.7668],[139,20.2016],[140,21.0239],[141,22.1217],[142,22.9551],[143,23.7042],[144,24.0779],[145,23.373],[146,21.8178],[147,18.1746],[148,13.8901],[149,9.6543],[150,12.4194],[151,22.7353],[152,36.1858],[153,46.0455],[154,48.6137],[155,49.9695],[156,51.8206],[157,57.9277],[158,66.3028],[159,60.6576],[160,43.5574],[161,35.1631],[162,30.5451],[163,26.4795],[164,23.3339],[165,21.3216],[166,20.0732],[167,19.3402],[168,20.1009],[169,23.7946],[170,30.1055],[171,33.837],[172,36.0502],[173,37.8527],[174,39.2338],[175,39.9405],[176,40.2381],[177,39.7916],[178,38.7796],[180,36.5398],[181,35.6363],[182,34.8247],[183,34.1499],[184,33.6443],[185,33.2522],[190,31.786],[192,31.4079],[194,31.2021],[197,31.062],[207,31.4806],[210,31.3925],[212,31.1589],[214,30.7941],[215,30.4979],[218,29.3533],[219,29.0463],[220,28.8573],[221,28.8786],[222,29.1247],[223,29.5396],[224,30.0695],[225,30.8928],[226,31.9795],[227,32.9442],[228,34.0128],[230,36.9287],[231,38.0775],[232,38.8919],[233,39.6016],[234,40.218],[236,41.2181],[238,41.9775],[240,42.5399],[243,43.0541],[249,43.7196],[252,43.703],[256,43.2801],[257,42.9768],[258,42.4482],[259,41.61],[260,39.8469],[261,37.0847],[262,34.1335],[263,31.3618],[264,28.699],[265,26.608],[266,25.1239],[267,23.8704],[268,22.8297],[269,21.963],[270,21.2613],[271,20.739],[272,20.3719],[273,20.1288],[274,19.9732],[275,19.8919],[276,19.8739],[278,19.9917],[281,20.1993],[287,20.365],[288,20.6054],[289,21.1942],[291,22.8366],[292,23.8126],[293,24.9778],[295,27.4606],[297,29.827],[298,30.8894],[299,31.7381],[301,33.0779],[303,34.2569],[305,35.225],[309,36.7032],[312,37.5451],[315,38.1124],[320,38.4951],[323,38.4735],[330,38.0028],[334,37.3645],[335,37.258],[339,37.312],[342,37.5096],[344,37.7602],[345,38.1845],[346,39.16],[347,39.9704],[348,40.208],[349,39.5675],[350,37.8974],[351,35.1034],[352,32.1532],[353,28.8988],[354,25.7088],[355,21.1999],[356,16.1674],[357,13.883],[358,12.011],[359,12.2714],[360,13.1731],[361,13.6841],[362,14.6726],[363,16.446],[364,18.4609],[365,19.931],[366,20.5221],[367,20.4901],[370,19.4907],[372,19.0037],[373,18.8831],[375,18.9326],[376,18.8843],[377,18.782],[378,18.5813],[379,18.3236],[381,17.6946],[383,16.981],[384,16.5707],[387,15.1884],[388,14.6905],[389,14.2242],[390,13.8522],[391,13.6333],[392,13.5484],[393,13.5335],[394,13.5757],[396,13.7754],[397,13.9132],[399,14.2792],[402,14.9837],[405,15.8401],[409,17.0361],[413,18.0917],[415,18.5241],[418,18.9971],[421,19.3434],[423,19.4813],[426,19.5369],[432,19.2148],[433,19.1993]]
       }
       arm_r_pv_ik_ctrl.Rotation.Y {
        type=MovieSceneScriptingFloatChannel
        default=0
        thinned_from=434
        keys [[0,-1.2174],[2,-1.2173],[3,-1.2201],[4,-1.22],[5,-1.2235],[6,-1.2233],[7,-1.2264],[10,-1.2254],[11,-1.2348],[12,-1.2227],[15,-1.271],[16,-1.3367],[17,-1.5402],[18,-1.9724],[19,-2.7632],[20,-3.8328],[21,-4.8746],[22,-5.6405],[23,-6.1495],[24,-6.4074],[25,-6.5681],[26,-6.7628],[28,-7.2052],[29,-7.4803],[31,-8.1318],[32,-8.3725],[33,-8.538],[34,-8.6811],[36,-8.8965],[37,-8.948],[38,-8.9607],[39,-8.9436],[42,-8.79],[46,-8.4777],[52,-8.0365],[54,-7.9238],[57,-7.7878],[58,-7.7273],[59,-7.6459],[60,-7.541],[61,-7.4066],[62,-7.1942],[63,-6.8211],[64,-6.0208],[65,-4.5332],[66,-3.3914],[67,-2.5827],[68,-1.4596],[69,-0.4133],[70,0.0062],[71,-0.1894],[72,-0.6262],[73,-1.1805],[74,-1.9238],[75,-2.6781],[76,-3.4446],[77,-3.9774],[78,-4.3326],[79,-4.5381],[80,-4.7279],[81,-4.8692],[82,-4.9635],[83,-5.047],[84,-5.1102],[85,-5.1437],[86,-5.1514],[87,-5.1463],[88,-5.1251],[89,-5.0895],[91,-4.9864],[92,-4.9054],[93,-4.8491],[95,-4.7626],[96,-4.7414],[97,-4.701],[98,-4.6751],[99,-4.6352],[100,-4.5759],[101,-4.4849],[102,-4.3602],[103,-4.1997],[104,-3.9843],[105,-3.7213],[106,-3.4496],[107,-3.2074],[108,-2.9957],[109,-2.7982],[110,-2.6283],[111,-2.4995],[112,-2.4033],[113,-2.3232],[114,-2.2616],[115,-2.2122],[117,-2.1806],[121,-2.1759],[122,-2.1904],[124,-2.2061],[125,-2.2173],[128,-2.2307],[129,-2.24],[130,-2.2391],[131,-2.249],[132,-2.2444],[134,-2.2793],[135,-2.3531],[136,-2.4748],[137,-2.6483],[138,-2.879],[139,-3.1738],[140,-3.9716],[141,-5.2952],[142,-6.5353],[143,-7.5453],[144,-8.3222],[145,-8.662],[146,-8.8069],[147,-7.955],[148,-5.5462],[149,-0.2803],[150,14.8513],[151,22.1802],[152,26.7884],[153,28.9977],[154,30.9695],[155,31.7932],[156,32.979],[157,39.8235],[158,56.0727],[159,58.0272],[160,50.9134],[161,44.1585],[162,38.7964],[163,33.7005],[164,29.9224],[165,28.204],[166,27.7326],[167,28.2431],[168,30.0546],[169,34.7489],[170,40.2503],[171,43.4339],[172,45.6683],[173,46.3183],[174,45.6106],[176,43.5316],[180,39.6426],[181,38.9003],[182,38.4091],[183,38.1743],[184,38.0689],[211,38.2868],[215,38.0591],[221,37.3311],[223,36.8998],[224,36.5586],[225,35.9338],[226,35.1963],[227,34.8233],[228,34.8826],[231,35.5871],[239,36.5184],[254,37.2536],[257,37.6157],[258,37.8881],[259,38.3366],[260,39.0081],[261,39.57],[262,39.7371],[263,39.6515],[264,39.4823],[265,39.4058],[266,39.5389],[273,41.3809],[279,42.366],[282,42.6188],[284,42.6161],[286,42.4122],[287,42.2106],[288,41.8558],[289,41.325],[296,36.9425],[297,36.3628],[298,35.8697],[299,35.527],[302,34.9523],[307,34.5068],[322,34.6976],[335,35.6525],[341,35.8411],[343,36.049],[344,36.2101],[345,36.2909],[346,36.2068],[347,35.903],[348,35.3931],[349,34.6836],[350,33.7617],[351,32.5054],[352,31.0207],[353,28.9953],[354,26.2668],[355,21.8031],[356,17.6901],[357,15.0112],[358,11.9356],[359,9.3292],[360,6.8958],[361,5.2055],[362,3.5398],[363,1.9912],[364,0.4692],[365,-1.0386],[366,-2.4411],[367,-3.6095],[368,-4.2609],[369,-4.1319],[370,-3.5383],[371,-2.7457],[372,-1.9982],[373,-1.1956],[374,-0.3555],[375,0.2096],[376,0.5038],[377,0.6981],[378,0.731],[379,0.6315],[380,0.4925],[381,0.3655],[382,0.2738],[383,0.196],[385,0.0679],[386,0.0069],[387,-0.0682],[388,-0.1662],[389,-0.2846],[390,-0.4179],[391,-0.5459],[392,-0.6651],[393,-0.7874],[394,-0.9066],[395,-1.0164],[396,-1.1089],[397,-1.1845],[399,-1.3028],[400,-1.3414],[402,-1.397],[403,-1.4092],[404,-1.4118],[405,-1.4111],[406,-1.3904],[407,-1.3837],[408,-1.3616],[412,-1.2644],[414,-1.224],[415,-1.2074],[416,-1.1856],[418,-1.1698],[419,-1.1646],[420,-1.1525],[422,-1.1504],[424,-1.1543],[426,-1.1528],[427,-1.168],[428,-1.1771],[429,-1.1918],[431,-1.2108],[432,-1.2169],[433,-1.2191]]
       }
       arm_r_pv_ik_ctrl.Rotation.Z {
        type=MovieSceneScriptingFloatChannel
        default=0
        thinned_from=434
        keys [[0,17.6616],[11,17.6647],[14,17.604],[15,17.465],[16,17.2692],[17,17.3294],[18,17.9591],[19,20.0024],[20,23.6999],[22,33.1467],[23,37.1422],[24,40.4778],[25,43.12],[26,45.0847],[27,46.4856],[28,47.4359],[29,48.0439],[30,48.3919],[31,48.5516],[33,48.5515],[35,48.0994],[37,47.4153],[41,45.6802],[44,44.6855],[46,44.2006],[57,43.0606],[61,43.0786],[62,43.217],[63,43.1911],[64,42.4545],[65,41.6015],[66,41.4561],[67,40.833],[68,37.9299],[69,34.1389],[70,32.1762],[71,31.5696],[72,31.2443],[73,31.1394],[74,31.2066],[78,31.9789],[83,32.5775],[95,33.1271],[97,33.0848],[99,32.8549],[100,32.5581],[101,32.0405],[102,31.3298],[103,30.4531],[104,29.3343],[105,27.9868],[106,26.5731],[107,25.2549],[109,22.8325],[110,21.7313],[111,20.7941],[112,20.0539],[113,19.4686],[114,18.9934],[115,18.5827],[116,18.2214],[117,17.9174],[118,17.6614],[119,17.444],[121,17.1056],[123,16.8435],[125,16.7051],[130,16.6559],[132,16.8222],[134,17.1071],[135,17.3079],[136,17.5904],[137,17.9925],[138,18.5527],[139,19.3101],[140,20.6849],[141,22.7457],[142,25.0376],[143,27.1125],[144,29.1032],[145,31.3459],[146,34.0064],[147,38.1407],[148,43.8471],[149,52.2175],[150,76.5305],[151,98.9506],[152,112.2999],[153,118.2704],[154,118.7256],[155,118.6532],[156,119.035],[157,121.2836],[158,127.2694],[159,123.3981],[160,111.4312],[161,106.1308],[163,101.8204],[164,100.0218],[165,98.7716],[166,97.8483],[167,97.1415],[168,97.306],[169,99.1452],[170,102.9266],[171,105.3941],[173,108.605],[174,109.675],[175,110.2197],[176,110.4367],[177,110.078],[180,107.5579],[183,105.8173],[191,104.0592],[205,105.0489],[212,106.3723],[215,106.2873],[217,105.8425],[219,104.8183],[220,103.9908],[221,102.8153],[222,101.2557],[223,99.4035],[224,97.3504],[225,94.7534],[226,91.7393],[227,89.1312],[228,87.0974],[229,85.3012],[230,83.8162],[231,82.7068],[233,81.1955],[235,80.157],[239,78.8586],[257,76.836],[260,77.0751],[261,77.2766],[262,77.8073],[263,78.8948],[265,81.888],[268,86.9777],[272,92.9271],[273,94.2179],[274,95.2882],[277,97.5769],[283,100.264],[287,101.504],[288,101.9931],[295,107.9283],[299,109.7292],[308,111.5985],[323,111.3334],[335,108.7424],[344,108.6348],[346,108.9926],[347,108.7302],[348,107.5473],[349,105.1026],[350,101.1871],[351,94.8395],[352,86.4756],[353,76.793],[354,66.7422],[355,53.4388],[356,38.5456],[357,27.6424],[358,20.827],[359,16.9873],[360,14.239],[361,12.6946],[362,11.7683],[363,11.1271],[364,11.5082],[365,13.1206],[366,15.0816],[367,17.0744],[368,19.2035],[369,21.1047],[370,22.7861],[371,24.3578],[372,25.8306],[373,27.1398],[374,28.2716],[375,29.3154],[376,30.2881],[377,31.1633],[378,31.9646],[379,32.6717],[380,33.2186],[381,33.5436],[382,33.5799],[383,33.33],[384,32.8276],[385,32.1008],[386,31.0923],[387,29.8385],[389,27.1635],[390,25.983],[392,23.8754],[393,22.9018],[394,22.0179],[395,21.2461],[396,20.5703],[397,19.9616],[398,19.4229],[399,18.962],[400,18.5863],[401,18.3038],[403,17.9174],[409,17.2315],[411,17.1467],[415,17.1111],[426,17.4234],[431,17.6126],[433,17.633]]
       }
       arm_r_pv_ik_ctrl.Scale.X {type=MovieSceneScriptingFloatChannel, default=1, const [1,434]}
       arm_r_pv_ik_ctrl.Scale.Y {type=MovieSceneScriptingFloatChannel, default=1, const [1,434]}
       arm_r_pv_ik_ctrl.Scale.Z {type=MovieSceneScriptingFloatChannel, default=1, const [1,434]}
       body_ctrl.Location.X {
        type=MovieSceneScriptingFloatChannel
        default=0
        thinned_from=434
        keys [[0,-3.9014],[8,-3.97],[11,-3.9685],[13,-3.9427],[14,-3.9145],[15,-3.8752],[16,-3.8069],[17,-3.6683],[18,-3.4274],[19,-2.9984],[20,-2.4542],[21,-2.0027],[22,-1.706],[23,-1.4951],[24,-1.3807],[25,-1.3258],[26,-1.2826],[27,-1.2461],[28,-1.2148],[29,-1.1903],[30,-1.1751],[31,-1.1686],[32,-1.1694],[35,-1.1894],[38,-1.2127],[40,-1.2214],[48,-1.2214],[53,-1.2156],[57,-1.2157],[58,-1.2183],[59,-1.2233],[60,-1.2335],[61,-1.2498],[62,-1.2826],[63,-1.3653],[64,-1.5339],[65,-1.9295],[66,-2.368],[67,-2.622],[68,-2.7716],[69,-2.8542],[70,-2.9044],[71,-2.928],[72,-2.9344],[77,-2.9028],[92,-2.739],[100,-2.6948],[102,-2.7025],[103,-2.7153],[104,-2.7368],[110,-2.9144],[112,-2.9596],[114,-2.9923],[122,-3.0579],[129,-3.0762],[132,-3.0637],[135,-3.0255],[139,-2.9526],[140,-2.9252],[141,-2.8727],[142,-2.7965],[143,-2.6969],[144,-2.6129],[145,-2.5865],[149,-2.5583],[150,-2.4085],[151,-2.3684],[152,-2.3595],[156,-2.3522],[157,-2.2057],[158,-0.5988],[159,0.0445],[160,0.3778],[161,0.5482],[162,0.6509],[163,0.6989],[164,0.7614],[165,0.798],[166,0.8364],[167,0.8849],[168,0.956],[169,1.0549],[170,1.0328],[171,0.9641],[172,0.915],[173,0.8891],[174,0.8711],[175,0.8594],[176,0.8527],[178,0.846],[180,0.8425],[183,0.8436],[188,0.8612],[196,0.8717],[208,0.8714],[210,0.8672],[211,0.8627],[212,0.856],[213,0.8467],[214,0.8343],[215,0.814],[216,0.7834],[217,0.7478],[219,0.6739],[220,0.6438],[221,0.6191],[222,0.5976],[223,0.5802],[224,0.5675],[225,0.5609],[226,0.5622],[227,0.5719],[228,0.5867],[230,0.619],[232,0.6472],[234,0.6782],[235,0.6925],[236,0.7043],[237,0.7128],[238,0.715],[241,0.7123],[246,0.7038],[248,0.7025],[249,0.6994],[250,0.6934],[251,0.6839],[252,0.6703],[253,0.6519],[254,0.6292],[255,0.5994],[256,0.5368],[257,0.4088],[258,0.2112],[259,-0.0613],[260,-0.4679],[261,-0.9731],[262,-1.4355],[263,-1.8105],[264,-2.1375],[265,-2.3949],[266,-2.5987],[267,-2.7752],[268,-2.9196],[269,-3.0272],[270,-3.1084],[271,-3.1762],[272,-3.2293],[273,-3.2662],[274,-3.2855],[276,-3.2847],[278,-3.2678],[279,-3.2495],[280,-3.2224],[281,-3.1814],[282,-3.1151],[283,-3.0144],[284,-2.8872],[285,-2.7389],[286,-2.5747],[287,-2.4024],[288,-2.1834],[289,-1.9004],[290,-1.5906],[292,-0.9876],[293,-0.68],[294,-0.4147],[295,-0.2024],[296,-0.0236],[297,0.1314],[298,0.2593],[299,0.357],[300,0.4326],[301,0.4964],[302,0.5495],[303,0.5939],[304,0.6283],[305,0.6566],[306,0.6778],[307,0.693],[308,0.7003],[309,0.6993],[310,0.6924],[311,0.6789],[312,0.6595],[313,0.6338],[314,0.6039],[315,0.5694],[316,0.5309],[317,0.4881],[318,0.443],[319,0.3947],[320,0.3424],[321,0.2878],[322,0.2292],[323,0.1684],[324,0.1031],[325,0.0351],[326,-0.0389],[327,-0.1183],[328,-0.1983],[329,-0.2738],[330,-0.3398],[331,-0.3968],[332,-0.448],[333,-0.4928],[334,-0.5301],[335,-0.5612],[336,-0.5828],[337,-0.5972],[338,-0.6078],[340,-0.6219],[341,-0.6313],[342,-0.6453],[343,-0.6663],[344,-0.6978],[345,-0.7379],[346,-0.7877],[347,-0.851],[348,-0.9313],[349,-1.0323],[350,-1.1569],[351,-1.2717],[352,-1.3958],[353,-1.615],[354,-2.0203],[355,-3.0234],[356,-3.9185],[357,-4.0198],[359,-3.7985],[360,-3.7013],[361,-3.6235],[362,-3.5681],[364,-3.4838],[365,-3.4497],[366,-3.4347],[367,-3.4441],[368,-3.4705],[370,-3.5358],[371,-3.5566],[374,-3.5845],[375,-3.583],[376,-3.5639],[381,-3.414],[383,-3.3689],[384,-3.3536],[385,-3.3454],[386,-3.3494],[387,-3.3667],[389,-3.4266],[391,-3.5023],[395,-3.7009],[398,-3.8233],[399,-3.8565],[401,-3.8977],[404,-3.9288],[433,-3.9017]]
       }
       body_ctrl.Location.Y {
        type=MovieSceneScriptingFloatChannel
        default=0
        thinned_from=434
        keys [[0,-0.324],[2,-0.3238],[4,-0.3218],[5,-0.3197],[6,-0.3165],[7,-0.3119],[8,-0.3059],[9,-0.2981],[10,-0.2879],[11,-0.2744],[12,-0.2564],[13,-0.2332],[14,-0.2039],[15,-0.1681],[16,-0.1257],[17,-0.0757],[18,0.0017],[19,0.1121],[20,0.2286],[21,0.32],[22,0.3735],[23,0.399],[24,0.4052],[25,0.4048],[26,0.4031],[27,0.4001],[28,0.3961],[29,0.3908],[31,0.3778],[34,0.3613],[35,0.3568],[36,0.3554],[38,0.3553],[39,0.3563],[41,0.3597],[43,0.3688],[44,0.3751],[45,0.3804],[47,0.3896],[52,0.4066],[54,0.4128],[57,0.4192],[59,0.4218],[61,0.4211],[62,0.4152],[63,0.3809],[64,0.2899],[65,0.0406],[66,-0.2664],[67,-0.4975],[68,-0.6709],[69,-0.7764],[70,-0.8479],[71,-0.8941],[72,-0.92],[73,-0.9301],[74,-0.9316],[75,-0.9276],[76,-0.9164],[77,-0.8997],[78,-0.8778],[80,-0.8246],[81,-0.8003],[83,-0.7567],[84,-0.7363],[85,-0.7177],[86,-0.7018],[87,-0.6883],[88,-0.6765],[89,-0.6661],[91,-0.6492],[92,-0.6423],[93,-0.6371],[97,-0.6275],[98,-0.6236],[99,-0.6173],[100,-0.6083],[101,-0.5969],[102,-0.5837],[103,-0.569],[104,-0.5519],[107,-0.4924],[109,-0.4572],[110,-0.4412],[111,-0.4271],[112,-0.4158],[113,-0.4071],[114,-0.4009],[115,-0.3966],[116,-0.3932],[119,-0.3806],[121,-0.375],[124,-0.3702],[126,-0.3694],[127,-0.3705],[128,-0.3735],[129,-0.378],[130,-0.384],[131,-0.3913],[132,-0.3997],[133,-0.4094],[134,-0.431],[135,-0.4744],[136,-0.5371],[137,-0.6167],[138,-0.7106],[139,-0.816],[140,-0.9789],[141,-1.1949],[142,-1.391],[143,-1.5452],[144,-1.6407],[145,-1.6189],[146,-1.5621],[147,-1.5372],[148,-1.5181],[149,-1.5111],[150,-1.7912],[151,-2.2217],[152,-2.4013],[153,-2.5069],[154,-2.5851],[155,-2.6392],[156,-2.6623],[157,-6.1418],[158,-4.9561],[159,-3.6448],[160,-2.8339],[161,-2.4577],[162,-2.1792],[163,-1.9996],[164,-2.1301],[165,-2.2681],[166,-2.4434],[167,-2.6883],[168,-3.0591],[169,-3.5971],[170,-3.5025],[171,-3.1474],[172,-2.9013],[173,-2.7734],[174,-2.6865],[175,-2.6246],[176,-2.5859],[177,-2.5618],[178,-2.5441],[180,-2.5241],[182,-2.5191],[183,-2.5237],[185,-2.5522],[187,-2.5878],[191,-2.6285],[207,-2.6584],[210,-2.6439],[213,-2.6091],[215,-2.5728],[217,-2.5214],[220,-2.4251],[223,-2.3068],[224,-2.277],[225,-2.259],[227,-2.2458],[230,-2.2015],[234,-2.1753],[237,-2.1698],[251,-2.1938],[254,-2.2138],[255,-2.2245],[256,-2.2491],[257,-2.2938],[260,-2.4663],[261,-2.5359],[262,-2.6186],[264,-2.8101],[265,-2.8929],[267,-3.0227],[268,-3.0762],[269,-3.1178],[270,-3.148],[271,-3.1706],[273,-3.1971],[281,-3.2092],[284,-3.1971],[286,-3.1715],[287,-3.1487],[288,-3.1124],[293,-2.8492],[294,-2.8027],[297,-2.6915],[299,-2.6317],[303,-2.5388],[305,-2.5023],[309,-2.4555],[317,-2.4095],[325,-2.4053],[327,-2.4147],[330,-2.4491],[336,-2.5703],[344,-2.6781],[345,-2.6336],[346,-2.4938],[347,-2.2875],[348,-2.0444],[349,-1.7952],[350,-1.5709],[351,-1.479],[352,-1.4683],[353,-1.3229],[354,-0.8079],[355,-0.0957],[356,0.185],[357,0.3196],[358,0.2846],[359,0.1559],[360,0.0662],[361,0.0125],[362,-0.0054],[363,-0.0205],[364,-0.0486],[365,-0.0754],[366,-0.1002],[367,-0.1191],[368,-0.1321],[369,-0.1463],[370,-0.1722],[371,-0.2061],[372,-0.2409],[373,-0.273],[374,-0.2942],[375,-0.3028],[376,-0.3032],[377,-0.3025],[378,-0.3042],[379,-0.3049],[380,-0.3038],[381,-0.2998],[382,-0.2901],[383,-0.2764],[384,-0.264],[385,-0.2586],[386,-0.2605],[387,-0.2656],[388,-0.2725],[389,-0.28],[391,-0.2933],[394,-0.315],[395,-0.3208],[397,-0.3302],[399,-0.3375],[402,-0.345],[405,-0.3483],[408,-0.3491],[415,-0.3431],[419,-0.3371],[423,-0.3308],[429,-0.3254],[433,-0.3243]]
       }
       body_ctrl.Location.Z {
        type=MovieSceneScriptingFloatChannel
        default=0
        thinned_from=434
        keys [[0,1.0417],[13,1.0221],[18,1.0258],[24,1.0424],[65,1.0344],[66,1.0325],[126,1.0404],[128,1.0369],[130,1.0277],[131,1.0201],[132,1.0054],[133,0.9736],[134,0.917],[135,0.8278],[136,0.6984],[137,0.5205],[138,0.2859],[139,-0.0147],[140,-0.6324],[141,-1.634],[142,-2.7412],[143,-4.1305],[144,-5.2949],[145,-5.6634],[146,-5.741],[147,-5.7401],[148,-5.598],[149,-5.1826],[150,-0.6726],[151,2.4676],[152,3.3398],[153,3.7642],[154,3.8823],[155,3.7116],[156,2.4658],[157,-4.8842],[158,-11.6715],[159,-14.9999],[160,-16.6614],[161,-17.2505],[162,-17.0034],[163,-14.664],[164,-11.6726],[165,-10.3873],[166,-9.635],[167,-9.1219],[168,-8.9563],[169,-9.8961],[170,-11.6154],[171,-12.8946],[172,-13.7572],[173,-14.2704],[174,-14.6034],[175,-14.8062],[176,-14.9095],[177,-14.9572],[178,-14.9357],[179,-14.826],[181,-14.5139],[182,-14.3973],[186,-14.0849],[192,-13.8768],[221,-13.8788],[257,-13.9595],[259,-13.92],[262,-13.7504],[284,-13.6369],[288,-13.7135],[294,-13.9455],[337,-14.0006],[340,-14.0767],[343,-14.2463],[344,-14.3294],[345,-14.5483],[346,-14.9485],[347,-15.4163],[348,-15.8383],[349,-16.1017],[350,-16.0948],[351,-15.7593],[352,-15.1143],[353,-14.1566],[354,-12.8658],[355,-6.6882],[356,0.0682],[357,2.1594],[358,2.9096],[359,3.0233],[360,2.7456],[361,1.9147],[362,0.7373],[363,-0.0853],[364,-0.1707],[365,0.2142],[366,0.8544],[367,1.4858],[368,1.9441],[369,2.1636],[370,2.2075],[371,2.1881],[372,2.1132],[373,1.9904],[374,1.68],[375,1.2204],[376,0.8877],[377,0.7404],[378,0.6456],[379,0.5924],[380,0.5692],[381,0.5649],[382,0.577],[383,0.6063],[384,0.6463],[385,0.6905],[386,0.7407],[387,0.7992],[388,0.8599],[389,0.9167],[390,0.9635],[391,1.0003],[392,1.0313],[393,1.0563],[394,1.0754],[395,1.0883],[396,1.0967],[397,1.1022],[399,1.1072],[402,1.1068],[414,1.0545],[421,1.0422],[433,1.0417]]
       }
       body_ctrl.Rotation.X {
        type=MovieSceneScriptingFloatChannel
        default=0
        thinned_from=434
        keys [[0,0.9973],[10,0.9929],[12,0.9895],[14,0.992],[16,1.0007],[18,1.0113],[19,1.0147],[20,0.991],[21,0.9175],[22,0.7956],[23,0.6394],[24,0.5023],[25,0.4048],[26,0.324],[27,0.2613],[28,0.2167],[29,0.1872],[30,0.1686],[32,0.1621],[33,0.1584],[35,0.1498],[36,0.1648],[37,0.1724],[38,0.1817],[40,0.2011],[41,0.1976],[42,0.2156],[43,0.2128],[44,0.2265],[46,0.223],[47,0.2361],[53,0.232],[54,0.2429],[58,0.2407],[59,0.2545],[60,0.2606],[61,0.2693],[62,0.2837],[63,0.3146],[64,0.3736],[65,0.4843],[66,0.6274],[67,0.7742],[68,0.8792],[69,0.9268],[70,0.9425],[71,0.9452],[73,0.9403],[74,0.9436],[80,0.9362],[83,0.9378],[87,0.9487],[93,0.9759],[95,0.9875],[96,0.9978],[98,1.0078],[99,1.0104],[100,1.0061],[101,0.9895],[102,0.9601],[103,0.9173],[104,0.851],[106,0.6757],[107,0.6069],[108,0.5625],[109,0.532],[110,0.5135],[111,0.5047],[112,0.5051],[113,0.5138],[114,0.5277],[115,0.5465],[116,0.5627],[117,0.5878],[118,0.6103],[119,0.6342],[120,0.6532],[121,0.6681],[122,0.6772],[123,0.6886],[124,0.6974],[125,0.703],[126,0.7049],[127,0.7003],[128,0.6881],[129,0.6716],[130,0.6535],[131,0.637],[132,0.626],[133,0.6212],[134,0.653],[135,0.7663],[136,1.0008],[137,1.3944],[138,1.9853],[139,2.8114],[140,4.4015],[141,6.7799],[142,9.2294],[143,11.7644],[144,14.0454],[145,15.3986],[146,16.1818],[147,16.6686],[148,16.9434],[149,17.0198],[150,16.7548],[151,14.6761],[152,13.7437],[153,13.1731],[154,12.7498],[155,12.4483],[156,12.3379],[157,36.4114],[158,58.2104],[159,65.4533],[160,68.8238],[161,70.8574],[162,72.083],[163,72.6204],[164,73.5291],[165,73.4415],[166,72.4146],[169,67.5775],[170,65.609],[171,65.3902],[172,66.4301],[173,67.7648],[174,68.7006],[175,69.3771],[176,69.8043],[179,70.4046],[182,70.5461],[193,69.167],[225,69.1047],[257,69.082],[259,68.7746],[260,68.3944],[262,67.4309],[264,66.9209],[284,66.6197],[287,66.9262],[293,68.5394],[341,68.9419],[344,68.6653],[345,68.3923],[346,67.833],[347,66.9747],[348,65.8036],[349,64.3061],[350,62.4699],[351,59.7337],[352,55.9813],[353,51.8696],[354,48.0416],[355,35.935],[356,20.3373],[357,12.161],[358,7.4044],[359,4.6542],[360,2.7188],[361,1.3807],[362,0.4497],[363,-0.2975],[364,-0.9886],[365,-1.5819],[366,-1.9954],[367,-2.2105],[368,-2.2969],[369,-2.3029],[370,-2.279],[371,-2.2255],[372,-2.1151],[373,-1.9205],[374,-1.5995],[375,-1.1786],[376,-0.7306],[377,-0.3287],[378,0.0243],[379,0.3616],[380,0.6659],[381,0.9202],[382,1.1306],[383,1.3028],[384,1.4182],[385,1.459],[386,1.443],[387,1.405],[388,1.3537],[389,1.2976],[390,1.246],[393,1.0989],[394,1.0537],[395,1.0149],[396,0.9819],[397,0.9523],[398,0.9336],[399,0.9037],[400,0.8894],[401,0.8772],[402,0.8574],[403,0.8499],[404,0.8445],[406,0.8391],[408,0.8395],[410,0.8467],[415,0.879],[416,0.8918],[419,0.9174],[420,0.9313],[423,0.9556],[424,0.9619],[425,0.9717],[429,0.9887],[432,0.9948],[433,0.9948]]
       }
       body_ctrl.Rotation.Y {
        type=MovieSceneScriptingFloatChannel
        default=0
        thinned_from=434
        keys [[0,0.0308],[9,0.0308],[10,0.0303],[11,0.0272],[12,0.0188],[13,0.022],[14,0.0135],[15,0.0256],[16,0.0896],[17,0.2161],[18,0.4158],[19,0.6673],[20,0.9131],[21,1.1118],[22,1.2439],[23,1.3594],[24,1.5092],[25,1.6838],[26,1.8384],[27,1.9605],[28,2.0413],[29,2.0948],[30,2.1403],[31,2.1812],[32,2.2102],[33,2.2286],[34,2.2395],[35,2.2444],[36,2.264],[37,2.2645],[38,2.2528],[39,2.2305],[40,2.1989],[41,2.1506],[42,2.1144],[43,2.0624],[45,1.9746],[46,1.9356],[47,1.9069],[50,1.806],[52,1.7483],[53,1.7242],[57,1.6565],[58,1.6344],[59,1.5901],[60,1.5419],[61,1.4797],[62,1.3952],[63,1.2749],[64,1.1091],[66,0.5374],[67,0.4052],[68,0.3555],[69,0.3316],[70,0.2963],[72,0.1946],[73,0.151],[74,0.1267],[75,0.0948],[76,0.0678],[77,0.0449],[78,0.0262],[79,0.0117],[80,0.0008],[81,-0.0066],[82,-0.0114],[83,-0.0144],[84,-0.0164],[85,-0.0182],[86,-0.0204],[87,-0.0223],[88,-0.023],[90,-0.0231],[91,-0.0238],[92,-0.0257],[94,-0.0307],[95,-0.0349],[96,-0.0295],[97,-0.041],[98,-0.0382],[99,-0.0477],[100,-0.062],[101,-0.0833],[102,-0.1143],[103,-0.1581],[104,-0.2211],[105,-0.2999],[106,-0.3824],[107,-0.4564],[108,-0.521],[109,-0.5819],[110,-0.6371],[111,-0.6846],[112,-0.7223],[113,-0.7483],[114,-0.7612],[115,-0.7702],[116,-0.7469],[117,-0.7373],[118,-0.712],[121,-0.6698],[122,-0.6435],[125,-0.5991],[127,-0.5728],[131,-0.5264],[132,-0.5296],[133,-0.5166],[134,-0.5278],[135,-0.5304],[136,-0.5262],[137,-0.5066],[138,-0.4633],[139,-0.3881],[140,-0.2331],[141,-0.0218],[142,0.1474],[143,0.2153],[144,0.2105],[145,0.1799],[146,0.1435],[147,0.0976],[148,0.0431],[149,-0.0224],[150,-0.2699],[151,-0.2863],[152,-0.2786],[153,-0.2695],[154,-0.2655],[155,-0.2677],[156,-0.264],[157,-1.1187],[158,-4.1024],[159,-5.2098],[160,-5.6832],[161,-5.8072],[162,-5.8482],[164,-5.849],[165,-5.8706],[166,-5.9277],[167,-6.008],[169,-6.2014],[170,-6.235],[171,-6.1735],[172,-6.084],[173,-6.0094],[174,-5.9565],[175,-5.9283],[180,-5.8986],[183,-5.9076],[196,-5.9984],[218,-5.9861],[221,-5.9512],[222,-5.9203],[223,-5.874],[224,-5.8115],[225,-5.6989],[226,-5.5561],[227,-5.4649],[230,-5.3751],[234,-5.312],[247,-5.284],[248,-5.2746],[249,-5.2456],[250,-5.1965],[251,-5.1263],[252,-5.0345],[253,-4.9201],[254,-4.7907],[256,-4.4764],[257,-4.2723],[258,-3.9182],[259,-3.3108],[260,-2.116],[261,-0.5336],[262,0.7397],[263,1.444],[264,1.8406],[265,2.0165],[266,2.1024],[267,2.1713],[268,2.2228],[269,2.2567],[270,2.2754],[271,2.2837],[272,2.284],[275,2.2643],[276,2.2643],[277,2.2527],[278,2.2529],[279,2.2437],[280,2.2444],[282,2.221],[283,2.1778],[284,2.1055],[285,1.986],[286,1.797],[287,1.5408],[288,0.9776],[289,0.0104],[290,-1.1677],[291,-2.3641],[292,-3.6231],[293,-5.015],[294,-6.3824],[295,-7.5848],[296,-8.592],[297,-9.4958],[298,-10.2518],[299,-10.8201],[300,-11.2606],[301,-11.6606],[302,-12.022],[303,-12.3521],[305,-12.8938],[307,-13.3082],[308,-13.4744],[310,-13.7214],[312,-13.8831],[315,-13.9815],[318,-13.9636],[320,-13.8795],[321,-13.8068],[323,-13.5656],[324,-13.3846],[325,-13.1654],[326,-12.9051],[332,-11.177],[333,-10.9042],[334,-10.6808],[335,-10.5187],[336,-10.4325],[338,-10.3554],[341,-10.3484],[342,-10.3298],[343,-10.2856],[344,-10.1975],[345,-10.0235],[346,-9.7316],[347,-9.353],[348,-8.9182],[350,-8.0007],[351,-7.721],[352,-7.5805],[353,-7.2824],[354,-6.4956],[355,-3.2537],[356,-0.4686],[357,-0.4063],[358,-0.582],[359,-0.7379],[360,-0.9183],[362,-1.326],[363,-1.5265],[365,-1.9338],[366,-2.042],[367,-2.0239],[368,-1.9199],[369,-1.7684],[370,-1.5866],[371,-1.3763],[373,-0.927],[374,-0.6681],[375,-0.3991],[376,-0.2201],[377,-0.1396],[378,-0.1034],[379,-0.1032],[380,-0.1182],[381,-0.1274],[382,-0.1164],[383,-0.0911],[384,-0.062],[385,-0.0396],[386,-0.0292],[387,-0.0276],[388,-0.0319],[389,-0.0396],[390,-0.0478],[391,-0.0541],[392,-0.0578],[393,-0.0592],[394,-0.0581],[395,-0.0549],[396,-0.0497],[397,-0.0433],[398,-0.0471],[399,-0.0295],[400,-0.0302],[401,-0.0304],[402,-0.0151],[403,-0.0118],[404,-0.0072],[406,0.0039],[408,0.0144],[409,0.02],[412,0.0377],[413,0.0432],[414,0.048],[415,0.0521],[416,0.0426],[417,0.0457],[418,0.0484],[419,0.0507],[420,0.0407],[421,0.0423],[422,0.0437],[423,0.0448],[424,0.0457],[425,0.0347],[428,0.0358],[433,0.0361]]
       }
       body_ctrl.Rotation.Z {
        type=MovieSceneScriptingFloatChannel
        default=0
        thinned_from=434
        keys [[0,8.9307],[10,8.9268],[11,8.8999],[12,8.8271],[13,8.6859],[14,8.4528],[15,8.2565],[16,8.018],[17,7.4036],[18,6.0797],[19,2.9926],[20,-1.2934],[21,-4.6849],[22,-6.2735],[23,-6.9617],[24,-7.3726],[25,-7.7284],[26,-7.8969],[29,-8.1311],[30,-8.1826],[31,-8.2075],[32,-8.214],[33,-8.189],[34,-8.1241],[36,-7.9327],[37,-7.8146],[40,-7.4005],[41,-7.3004],[42,-7.2355],[43,-7.1948],[47,-7.1303],[53,-7.1137],[56,-7.0694],[58,-7.0053],[59,-6.95],[60,-6.8727],[61,-6.7685],[62,-6.6855],[63,-6.4966],[64,-5.9361],[65,-4.2385],[66,-2.3204],[67,-1.3708],[68,-0.8322],[69,-0.478],[70,-0.2621],[71,-0.0911],[72,0.0776],[73,0.2025],[74,0.2514],[75,0.2342],[76,0.1973],[77,0.1608],[78,0.1294],[79,0.0967],[80,0.0675],[81,0.0461],[82,0.0335],[83,0.0261],[84,0.0223],[85,0.0201],[86,0.0175],[87,0.0145],[88,0.0124],[89,0.011],[90,0.0101],[92,0.0095],[93,0.0104],[94,0.0168],[95,0.0342],[96,0.0684],[97,0.1242],[98,0.2079],[99,0.3243],[100,0.5046],[101,0.7746],[102,1.1319],[103,1.5746],[104,2.1574],[105,2.8629],[106,3.5804],[107,4.1994],[108,4.7072],[109,5.1573],[110,5.5382],[111,5.8384],[112,6.0338],[113,6.1375],[116,6.3104],[118,6.3955],[124,6.5869],[126,6.6114],[130,6.594],[132,6.5534],[135,6.4514],[136,6.3755],[137,6.2285],[138,5.9773],[139,5.5893],[140,4.6655],[141,3.2604],[142,2.031],[143,1.358],[144,1.0287],[145,0.9005],[146,0.862],[147,0.8178],[148,0.7858],[149,0.7611],[150,0.8016],[151,0.8255],[152,0.8353],[153,0.8412],[154,0.8449],[156,0.8481],[157,4.1097],[158,3.4468],[159,2.6589],[160,2.1221],[161,1.8153],[162,1.6026],[163,1.4815],[164,1.445],[165,1.4864],[166,1.591],[167,1.735],[168,1.8941],[169,2.0574],[170,2.1651],[171,2.1797],[172,2.149],[173,2.1005],[174,2.0675],[175,2.0252],[176,1.9889],[177,1.9611],[178,1.9406],[179,1.9242],[180,1.9212],[181,1.9028],[183,1.9042],[184,1.8984],[188,1.9242],[196,1.9434],[197,1.9558],[202,1.9581],[203,1.9696],[207,1.9711],[208,1.9791],[209,2.0006],[210,2.0425],[211,2.1117],[212,2.2148],[213,2.3588],[214,2.5504],[215,2.8072],[216,3.1565],[217,3.5845],[218,4.1074],[219,4.749],[220,5.4945],[221,6.499],[222,7.8051],[224,10.7277],[225,12.3629],[226,14.0969],[227,15.4162],[228,16.235],[229,16.8446],[230,17.2938],[231,17.6315],[232,17.8966],[233,18.1029],[234,18.2615],[236,18.4789],[242,18.8646],[249,18.9417],[252,18.8675],[254,18.7153],[256,18.4922],[257,18.295],[258,17.8039],[259,16.8147],[260,14.5242],[261,11.2245],[262,8.3531],[263,6.3761],[264,4.8277],[265,3.7464],[266,3.1168],[267,2.7997],[268,2.6471],[269,2.5114],[270,2.3683],[271,2.2798],[272,2.2273],[275,2.1192],[276,2.0979],[277,2.0637],[278,2.0518],[279,2.0291],[280,2.0244],[281,2.0082],[282,2.0074],[283,1.9965],[285,1.9932],[287,1.9549],[288,1.9242],[290,1.8352],[291,1.7943],[292,1.7595],[293,1.7292],[295,1.7076],[298,1.819],[299,1.845],[301,1.8707],[302,1.8792],[303,1.8753],[304,1.8897],[305,1.8866],[306,1.8947],[308,1.8902],[309,1.8995],[312,1.8962],[313,1.9065],[316,1.906],[317,1.9188],[318,1.9259],[319,1.938],[320,1.9656],[321,1.9882],[322,2.0278],[323,2.0618],[325,2.1775],[326,2.2665],[327,2.3935],[328,2.5352],[329,2.6683],[330,2.7693],[331,2.8349],[332,2.8806],[333,2.9116],[334,2.9247],[335,2.9511],[337,2.9659],[338,2.9807],[342,2.9922],[343,3.0017],[344,3.029],[345,3.1124],[346,3.3067],[347,3.5895],[348,3.9369],[349,4.3233],[350,4.7225],[351,5.0225],[352,5.2241],[353,5.4716],[354,5.9076],[355,7.4205],[356,8.5599],[357,9.007],[358,9.3086],[359,9.4955],[361,9.7451],[362,9.8966],[363,9.9713],[364,9.8124],[366,9.1399],[368,8.5192],[369,8.2732],[370,8.1124],[371,7.9956],[374,7.7432],[377,7.5813],[379,7.4993],[380,7.4726],[381,7.4637],[382,7.492],[383,7.5647],[384,7.6697],[385,7.7951],[386,7.95],[389,8.5082],[390,8.6475],[391,8.7417],[392,8.8067],[397,8.9702],[401,9.0114],[433,8.9308]]
       }
       body_ctrl.Scale.X {type=MovieSceneScriptingFloatChannel, default=1, const [1,434]}
       body_ctrl.Scale.Y {type=MovieSceneScriptingFloatChannel, default=1, const [1,434]}
       body_ctrl.Scale.Z {type=MovieSceneScriptingFloatChannel, default=1, const [1,434]}
       hips_ctrl.Location.X {type=MovieSceneScriptingFloatChannel, default=0, thinned_from=434, keys [[0,0],[433,0]]}
       hips_ctrl.Location.Y {type=MovieSceneScriptingFloatChannel, default=0, thinned_from=434, keys [[0,0],[433,0]]}
       hips_ctrl.Location.Z {type=MovieSceneScriptingFloatChannel, default=0, thinned_from=434, keys [[0,0],[433,0]]}
       hips_ctrl.Rotation.X {
        type=MovieSceneScriptingFloatChannel
        default=0
        thinned_from=434
        keys [[0,0.3484],[12,0.3484],[14,0.344],[17,0.3347],[18,0.3338],[19,0.3359],[20,0.3674],[21,0.4489],[22,0.5793],[23,0.7442],[24,0.8878],[25,0.9888],[26,1.0718],[27,1.1355],[28,1.1804],[29,1.2092],[30,1.2255],[35,1.2255],[36,1.2063],[40,1.1534],[41,1.1534],[42,1.1325],[43,1.1325],[44,1.1164],[46,1.1164],[47,1.1022],[53,1.1022],[54,1.0906],[58,1.0906],[59,1.0762],[60,1.0692],[61,1.0595],[62,1.0438],[63,1.0106],[64,0.9486],[65,0.8336],[66,0.6859],[67,0.5358],[68,0.4282],[69,0.3783],[70,0.36],[71,0.3545],[73,0.3545],[74,0.3492],[95,0.3492],[96,0.3445],[97,0.3445],[103,0.3338],[104,0.3328],[105,0.3328],[106,0.3343],[107,0.3372],[108,0.341],[111,0.3557],[113,0.3639],[114,0.3672],[115,0.3672],[116,0.3724],[117,0.3724],[118,0.3759],[121,0.3759],[122,0.38],[131,0.38],[132,0.3791],[133,0.3791],[134,0.3759],[137,0.3686],[139,0.3621],[141,0.3526],[142,0.3496],[143,0.3491],[144,0.3502],[145,0.3522],[146,0.3554],[147,0.3611],[148,0.3704],[149,0.3852],[150,-1.6626],[151,-2.1897],[152,-2.3002],[153,-2.329],[154,-2.3479],[155,-2.3479],[156,-2.3753],[157,-11.1145],[158,-13.2737],[160,-10.6029],[161,-9.9037],[162,-9.287],[163,-8.7847],[164,-9.3691],[165,-9.7775],[166,-10.0755],[167,-10.4045],[168,-10.9706],[169,-11.8759],[170,-11.1874],[171,-10.3438],[172,-10.019],[218,-9.9761],[221,-10.0099],[223,-10.0806],[224,-10.1352],[226,-10.3344],[227,-10.3973],[236,-10.5024],[257,-10.4887],[259,-10.4476],[265,-10.1166],[275,-9.9698],[288,-9.9036],[293,-9.817],[295,-9.817],[299,-9.9201],[326,-10.0201],[344,-10.1205],[345,-10.1045],[346,-10.0582],[348,-9.8973],[350,-9.6818],[351,-9.4431],[352,-9.0042],[353,-8.4438],[354,-7.8219],[355,-5.4448],[356,-2.4228],[357,-1.5775],[358,-1.396],[359,-1.3879],[360,-1.4598],[361,-1.4598],[362,-1.2469],[363,-0.9187],[364,-0.5752],[365,-0.2167],[366,0.0535],[367,0.1952],[368,0.2577],[369,0.2649],[370,0.2505],[371,0.2351],[372,0.2208],[373,0.2097],[374,0.2024],[375,0.198],[376,0.1959],[377,0.1958],[378,0.1988],[380,0.211],[381,0.214],[382,0.2115],[383,0.2058],[384,0.2008],[385,0.2],[386,0.2051],[387,0.2151],[388,0.2286],[390,0.259],[392,0.2875],[393,0.3011],[394,0.3135],[395,0.3243],[396,0.3337],[397,0.3422],[398,0.3422],[399,0.3556],[401,0.3556],[402,0.3653],[415,0.3653],[416,0.3597],[419,0.3597],[420,0.3548],[424,0.3548],[425,0.3504],[433,0.3504]]
       }
       hips_ctrl.Rotation.Y {
        type=MovieSceneScriptingFloatChannel
        default=0
        thinned_from=434
        keys [[0,0.6588],[12,0.6588],[14,0.6206],[15,0.586],[16,0.5193],[17,0.4074],[18,0.2371],[19,-0.0763],[20,-0.4548],[21,-0.6701],[22,-0.6691],[23,-0.614],[24,-0.5868],[25,-0.5783],[26,-0.5565],[27,-0.54],[28,-0.5274],[29,-0.5177],[30,-0.5108],[35,-0.5108],[36,-0.53],[37,-0.5396],[40,-0.5741],[41,-0.5741],[42,-0.5893],[43,-0.5893],[44,-0.5974],[46,-0.5974],[47,-0.6036],[53,-0.6036],[54,-0.6084],[58,-0.6084],[59,-0.5913],[60,-0.5775],[61,-0.5584],[62,-0.5288],[63,-0.4083],[64,-0.0838],[65,0.8224],[66,1.7449],[67,1.98],[68,1.8845],[69,1.7465],[70,1.6724],[71,1.6536],[73,1.6536],[74,1.6404],[95,1.6404],[96,1.6277],[97,1.6277],[98,1.6077],[99,1.5929],[100,1.5713],[101,1.541],[102,1.5037],[103,1.4609],[104,1.4088],[106,1.2868],[107,1.233],[108,1.1869],[109,1.1447],[110,1.1092],[111,1.0827],[112,1.0633],[113,1.047],[114,1.0337],[115,1.0337],[116,1.0141],[117,1.0141],[118,1.0021],[121,1.0021],[122,0.9902],[131,0.9902],[132,1.0055],[133,1.0055],[134,1.0282],[135,1.0442],[136,1.0638],[137,1.0873],[138,1.1152],[139,1.1478],[140,1.2326],[141,1.3732],[142,1.504],[143,1.5873],[144,1.6429],[146,1.7258],[147,1.7746],[148,1.8322],[149,1.9011],[150,1.0211],[151,0.7542],[152,0.6886],[153,0.6652],[154,0.6515],[155,0.6515],[156,0.6355],[157,0.3622],[158,2.2492],[159,3.1614],[160,3.6681],[161,3.9485],[162,4.1307],[163,4.2387],[164,4.2621],[165,4.2308],[166,4.1615],[169,3.8844],[170,3.8259],[171,3.7898],[172,3.7688],[174,3.7688],[177,3.8355],[179,3.8641],[180,3.8641],[181,3.8831],[183,3.8831],[184,3.8948],[196,3.8948],[197,3.8815],[202,3.8815],[203,3.8685],[215,3.8685],[216,3.8571],[218,3.8571],[220,3.8206],[221,3.7879],[222,3.7275],[223,3.6378],[224,3.5168],[225,3.3004],[226,3.0264],[227,2.8513],[228,2.7841],[229,2.7265],[230,2.6777],[231,2.637],[232,2.6037],[233,2.5769],[235,2.5399],[236,2.528],[237,2.528],[238,2.5139],[241,2.5139],[242,2.502],[253,2.502],[254,2.5175],[255,2.5175],[256,2.542],[257,2.5727],[258,2.6242],[259,2.7019],[260,2.8332],[261,3.0067],[262,3.1708],[263,3.3164],[264,3.4531],[265,3.5587],[266,3.6296],[267,3.6811],[269,3.7538],[271,3.8069],[275,3.8698],[276,3.8698],[277,3.8911],[278,3.8911],[279,3.9084],[280,3.9084],[281,3.9254],[282,3.9254],[283,3.9372],[285,3.9372],[287,3.9728],[288,4.0003],[291,4.1144],[293,4.1636],[295,4.1636],[298,4.0052],[299,3.9683],[302,3.9138],[303,3.9138],[304,3.8936],[305,3.8936],[306,3.8814],[308,3.8814],[309,3.8691],[312,3.8691],[313,3.8569],[316,3.8569],[317,3.8453],[319,3.8453],[320,3.8338],[321,3.8338],[322,3.821],[323,3.821],[326,3.7668],[329,3.669],[331,3.6237],[333,3.5955],[334,3.5955],[335,3.5781],[337,3.5781],[338,3.5644],[343,3.5644],[344,3.5501],[345,3.4994],[346,3.3648],[347,3.162],[348,2.9075],[349,2.619],[350,2.3152],[351,2.1169],[352,2.0707],[353,2.0719],[354,1.9877],[355,1.1326],[356,0.5294],[357,0.4388],[358,0.3832],[359,0.3347],[360,0.2904],[361,0.2904],[362,0.3475],[363,0.4326],[364,0.5012],[365,0.5736],[366,0.6287],[367,0.6715],[368,0.7091],[370,0.7774],[377,1.0428],[378,1.0897],[379,1.142],[380,1.1846],[381,1.2021],[382,1.1838],[383,1.1385],[385,1.0225],[387,0.9177],[388,0.8691],[389,0.8267],[390,0.7931],[391,0.7654],[392,0.7397],[393,0.7162],[394,0.6956],[395,0.6782],[396,0.6635],[397,0.6507],[398,0.6507],[399,0.6311],[401,0.6311],[402,0.617],[415,0.617],[416,0.6299],[419,0.6299],[420,0.6419],[424,0.6419],[425,0.6535],[433,0.6535]]
       }
       hips_ctrl.Rotation.Z {
        type=MovieSceneScriptingFloatChannel
        default=0
        thinned_from=434
        keys [[0,0.004],[12,0.004],[15,0.0035],[16,0.0031],[17,0.0024],[18,0.0014],[19,-0.0004],[21,-0.0053],[22,-0.0068],[24,-0.0091],[25,-0.01],[27,-0.0107],[35,-0.0109],[40,-0.0116],[58,-0.0116],[61,-0.0103],[62,-0.0096],[63,-0.0072],[64,-0.0014],[65,0.012],[66,0.0209],[67,0.0185],[68,0.0141],[69,0.0115],[70,0.0105],[71,0.0102],[73,0.0102],[74,0.01],[95,0.01],[96,0.0098],[101,0.0091],[108,0.0071],[131,0.0066],[139,0.0073],[140,0.0077],[142,0.0092],[147,0.0112],[148,0.0118],[149,0.0128],[150,-0.0296],[151,-0.0288],[152,-0.0277],[153,-0.0271],[154,-0.0267],[155,-0.0267],[156,-0.0264],[157,-0.0712],[158,-0.5304],[159,-0.6675],[160,-0.6862],[161,-0.6888],[162,-0.6748],[163,-0.6544],[164,-0.7025],[165,-0.7284],[166,-0.7388],[167,-0.7469],[168,-0.7703],[169,-0.8162],[170,-0.7561],[171,-0.6912],[172,-0.6653],[174,-0.6653],[177,-0.6748],[179,-0.6789],[180,-0.6789],[181,-0.6816],[183,-0.6816],[184,-0.6832],[196,-0.6832],[197,-0.6813],[202,-0.6813],[203,-0.6795],[215,-0.6795],[216,-0.6779],[218,-0.6779],[220,-0.6727],[221,-0.6681],[222,-0.6594],[223,-0.6463],[224,-0.6283],[225,-0.5951],[226,-0.5516],[227,-0.5229],[228,-0.5118],[229,-0.5021],[230,-0.4939],[231,-0.4871],[232,-0.4814],[233,-0.4768],[235,-0.4705],[236,-0.4685],[237,-0.4685],[238,-0.4661],[241,-0.4661],[242,-0.464],[253,-0.464],[254,-0.4667],[255,-0.4667],[256,-0.4709],[257,-0.4761],[258,-0.4849],[259,-0.498],[260,-0.5199],[261,-0.5484],[262,-0.5747],[263,-0.5976],[264,-0.6186],[265,-0.6345],[266,-0.6451],[267,-0.6526],[269,-0.6632],[271,-0.6708],[275,-0.6797],[276,-0.6797],[277,-0.6827],[278,-0.6827],[279,-0.6851],[280,-0.6851],[281,-0.6875],[282,-0.6875],[283,-0.6891],[285,-0.6891],[287,-0.6941],[288,-0.6978],[291,-0.7133],[293,-0.7198],[295,-0.7198],[298,-0.6985],[299,-0.6934],[302,-0.6859],[303,-0.6859],[304,-0.683],[305,-0.683],[306,-0.6813],[308,-0.6813],[309,-0.6796],[312,-0.6796],[313,-0.6779],[316,-0.6779],[317,-0.6762],[319,-0.6762],[320,-0.6746],[321,-0.6746],[322,-0.6728],[323,-0.6728],[326,-0.665],[329,-0.6509],[331,-0.6442],[333,-0.64],[334,-0.64],[335,-0.6374],[337,-0.6374],[338,-0.6354],[343,-0.6354],[344,-0.6332],[345,-0.6232],[346,-0.5965],[347,-0.5565],[348,-0.5071],[349,-0.4519],[350,-0.3949],[351,-0.352],[352,-0.3281],[353,-0.3075],[354,-0.273],[355,-0.1079],[356,-0.0224],[357,-0.0121],[358,-0.0093],[359,-0.0081],[360,-0.0074],[362,-0.0076],[363,-0.0069],[364,-0.005],[366,0.0006],[367,0.0023],[368,0.0032],[369,0.0034],[376,0.0034],[381,0.0045],[385,0.0036],[433,0.004]]
       }
       hips_ctrl.Scale.X {type=MovieSceneScriptingFloatChannel, default=1, const [1,434]}
       hips_ctrl.Scale.Y {type=MovieSceneScriptingFloatChannel, default=1, const [1,434]}
       hips_ctrl.Scale.Z {type=MovieSceneScriptingFloatChannel, default=1, const [1,434]}
       hips_tan_ctrl.Location.X {
        type=MovieSceneScriptingFloatChannel
        default=0
        thinned_from=434
        keys [[0,0.0329],[10,0.0329],[11,0.0327],[12,0.0323],[13,0.0315],[14,0.0302],[15,0.0282],[16,0.0253],[17,0.0206],[18,0.0137],[19,0.0001],[20,-0.0197],[21,-0.0388],[22,-0.0548],[23,-0.0684],[24,-0.0779],[25,-0.0839],[26,-0.0882],[27,-0.0913],[28,-0.0933],[29,-0.0946],[30,-0.0953],[31,-0.0956],[33,-0.0956],[37,-0.0941],[43,-0.0912],[47,-0.0902],[53,-0.0899],[56,-0.0896],[59,-0.0887],[60,-0.0882],[61,-0.0875],[62,-0.0864],[63,-0.0833],[64,-0.0765],[65,-0.0601],[66,-0.0401],[67,-0.024],[68,-0.0106],[69,-0.0008],[70,0.005],[71,0.0073],[72,0.0086],[73,0.0093],[74,0.0097],[76,0.0098],[85,0.0083],[87,0.0083],[88,0.008],[94,0.008],[96,0.0082],[99,0.0092],[100,0.0098],[101,0.0108],[103,0.0134],[104,0.015],[106,0.0191],[108,0.0226],[110,0.0256],[111,0.0267],[113,0.0284],[115,0.0296],[119,0.0308],[123,0.0314],[130,0.0315],[134,0.0303],[136,0.0293],[138,0.0278],[139,0.0268],[140,0.0249],[141,0.0219],[142,0.0186],[143,0.0157],[145,0.0105],[146,0.0075],[147,0.0038],[148,-0.0007],[149,-0.0061],[150,-0.0273],[151,-0.0286],[153,-0.0271],[155,-0.0266],[156,-0.0266],[157,-0.0363],[158,-0.0836],[159,-0.1062],[160,-0.1188],[161,-0.1257],[162,-0.1302],[163,-0.1328],[164,-0.1336],[165,-0.1329],[166,-0.1312],[169,-0.1246],[170,-0.1228],[171,-0.1216],[172,-0.1209],[174,-0.1209],[177,-0.1226],[181,-0.1239],[195,-0.1241],[215,-0.1234],[218,-0.123],[220,-0.1222],[221,-0.1214],[222,-0.1198],[223,-0.1175],[224,-0.1144],[225,-0.1088],[226,-0.1017],[227,-0.0972],[228,-0.0954],[229,-0.094],[230,-0.0927],[232,-0.0908],[234,-0.0896],[236,-0.0889],[238,-0.0885],[252,-0.0882],[255,-0.0888],[256,-0.0892],[257,-0.09],[258,-0.0913],[259,-0.0933],[260,-0.0967],[261,-0.1012],[262,-0.1054],[263,-0.1092],[264,-0.1127],[265,-0.1154],[266,-0.1173],[267,-0.1186],[270,-0.1213],[275,-0.1235],[286,-0.1258],[287,-0.1262],[288,-0.1269],[291,-0.1299],[293,-0.1312],[294,-0.1314],[295,-0.1309],[296,-0.1298],[298,-0.127],[299,-0.1261],[302,-0.1247],[307,-0.1238],[319,-0.1228],[321,-0.1226],[325,-0.1214],[326,-0.1208],[329,-0.1183],[331,-0.1171],[334,-0.1161],[343,-0.1155],[344,-0.1152],[345,-0.1142],[346,-0.1115],[347,-0.1072],[348,-0.1015],[349,-0.0946],[350,-0.0868],[351,-0.0796],[352,-0.0733],[353,-0.0655],[354,-0.0535],[355,-0.0172],[356,0.0154],[357,0.0269],[358,0.0327],[359,0.0359],[360,0.0381],[362,0.0409],[363,0.0416],[364,0.0403],[366,0.0346],[369,0.0269],[372,0.0197],[373,0.0175],[375,0.0137],[379,0.0074],[380,0.0063],[381,0.0059],[382,0.0063],[383,0.0076],[384,0.0093],[385,0.0113],[386,0.0135],[389,0.0209],[390,0.023],[392,0.0265],[394,0.0295],[396,0.0317],[398,0.0333],[400,0.0344],[403,0.035],[409,0.0352],[411,0.035],[422,0.0334],[433,0.033]]
       }
       hips_tan_ctrl.Location.Y {
        type=MovieSceneScriptingFloatChannel
        default=0.008
        thinned_from=434
        keys [[0,-0.0986],[19,-0.0987],[20,-0.0981],[21,-0.0965],[22,-0.094],[23,-0.0908],[24,-0.0879],[25,-0.0858],[26,-0.0841],[27,-0.0828],[28,-0.0819],[29,-0.0813],[31,-0.0809],[33,-0.0809],[45,-0.0834],[60,-0.0843],[62,-0.0848],[63,-0.0854],[64,-0.0865],[65,-0.0886],[66,-0.0914],[67,-0.0944],[68,-0.0967],[69,-0.0978],[70,-0.0982],[145,-0.0984],[147,-0.0982],[149,-0.0974],[150,-0.0736],[151,-0.0682],[152,-0.0672],[155,-0.0669],[156,-0.0667],[157,-0.1589],[158,-0.1593],[159,-0.1655],[160,-0.1705],[163,-0.1776],[164,-0.1727],[165,-0.1701],[167,-0.1677],[168,-0.1648],[169,-0.1593],[170,-0.1653],[171,-0.1722],[172,-0.175],[219,-0.1743],[221,-0.1748],[223,-0.1763],[224,-0.1774],[226,-0.1816],[227,-0.1829],[237,-0.1852],[256,-0.1851],[259,-0.184],[260,-0.1831],[264,-0.178],[266,-0.1764],[288,-0.1726],[293,-0.1708],[295,-0.1709],[299,-0.173],[326,-0.175],[344,-0.1771],[346,-0.1764],[349,-0.1742],[350,-0.1738],[351,-0.1723],[352,-0.1696],[353,-0.1675],[354,-0.1676],[355,-0.1522],[356,-0.1247],[357,-0.1161],[358,-0.1144],[359,-0.1153],[360,-0.1168],[361,-0.1179],[362,-0.118],[363,-0.1162],[364,-0.1118],[365,-0.1067],[366,-0.1026],[367,-0.1005],[368,-0.0996],[369,-0.0996],[373,-0.1007],[378,-0.1007],[381,-0.1004],[386,-0.1007],[400,-0.0984],[433,-0.0986]]
       }
       hips_tan_ctrl.Location.Z {
        type=MovieSceneScriptingFloatChannel
        default=-0.0073
        thinned_from=434
        keys [[0,-0.0122],[13,-0.0123],[16,-0.0128],[19,-0.014],[20,-0.0139],[23,-0.0114],[24,-0.0108],[28,-0.0099],[61,-0.0101],[62,-0.0103],[63,-0.0107],[64,-0.0116],[65,-0.013],[66,-0.0132],[67,-0.013],[149,-0.0128],[150,-0.0162],[156,-0.0163],[157,0.0005],[158,0.0103],[159,0.0175],[160,0.0269],[162,0.0407],[163,0.0473],[164,0.039],[165,0.0342],[167,0.0284],[168,0.0238],[169,0.0183],[170,0.0227],[171,0.0304],[172,0.0341],[344,0.0341],[345,0.0336],[346,0.0322],[348,0.0282],[350,0.0234],[351,0.0207],[353,0.0137],[354,0.0108],[355,-0.0019],[356,-0.0101],[357,-0.0116],[358,-0.0117],[362,-0.0108],[363,-0.0108],[365,-0.0117],[366,-0.0121],[433,-0.0122]]
       }
       hips_tan_ctrl.Rotation.X {type=MovieSceneScriptingFloatChannel, default=0, const [0,434]}
       hips_tan_ctrl.Rotation.Y {type=MovieSceneScriptingFloatChannel, default=0, const [0,434]}
       hips_tan_ctrl.Rotation.Z {type=MovieSceneScriptingFloatChannel, default=0, const [0,434]}
       hips_tan_ctrl.Scale.X {type=MovieSceneScriptingFloatChannel, default=1, const [1,434]}
       hips_tan_ctrl.Scale.Y {type=MovieSceneScriptingFloatChannel, default=1, const [1,434]}
       hips_tan_ctrl.Scale.Z {type=MovieSceneScriptingFloatChannel, default=1, const [1,434]}
       spine_01_ctrl.Location.X {
        type=MovieSceneScriptingFloatChannel
        default=0
        thinned_from=434
        keys [[0,0.0388],[10,0.0388],[11,0.0386],[12,0.0382],[13,0.0374],[14,0.0361],[15,0.0341],[16,0.031],[17,0.026],[18,0.0185],[19,0.0047],[20,-0.0137],[21,-0.0292],[22,-0.0396],[23,-0.0477],[24,-0.0528],[25,-0.0555],[26,-0.0573],[27,-0.0585],[28,-0.0594],[29,-0.0599],[31,-0.0603],[57,-0.0593],[59,-0.0588],[61,-0.0577],[62,-0.0566],[63,-0.0533],[64,-0.0454],[65,-0.0254],[66,-0.0032],[67,0.0092],[68,0.0161],[69,0.0197],[70,0.0221],[71,0.0233],[72,0.0241],[74,0.025],[94,0.0253],[99,0.026],[101,0.0267],[103,0.0278],[107,0.0309],[110,0.0326],[116,0.0339],[130,0.0343],[136,0.0332],[139,0.0321],[140,0.0309],[142,0.0272],[143,0.0261],[147,0.0235],[149,0.0218],[150,0.0038],[151,0],[152,-0.0006],[155,-0.0007],[156,-0.0008],[157,-0.0027],[158,-0.0268],[159,-0.0379],[160,-0.0442],[161,-0.0477],[162,-0.0499],[163,-0.0513],[164,-0.0514],[165,-0.051],[166,-0.0501],[169,-0.0464],[170,-0.0459],[173,-0.0454],[181,-0.0469],[218,-0.0465],[220,-0.0461],[221,-0.0457],[222,-0.045],[223,-0.044],[224,-0.0425],[225,-0.0399],[226,-0.0367],[227,-0.0346],[229,-0.0331],[232,-0.0316],[236,-0.0307],[252,-0.0304],[256,-0.0308],[257,-0.0312],[258,-0.0318],[259,-0.0328],[260,-0.0343],[262,-0.0384],[264,-0.0418],[265,-0.043],[266,-0.0439],[269,-0.0453],[282,-0.0474],[285,-0.0476],[288,-0.0482],[292,-0.0499],[294,-0.0503],[295,-0.0501],[298,-0.0483],[302,-0.0472],[325,-0.0457],[331,-0.0438],[335,-0.0432],[344,-0.0429],[345,-0.0424],[346,-0.0409],[347,-0.0386],[348,-0.0355],[349,-0.0317],[350,-0.0275],[351,-0.0242],[352,-0.022],[353,-0.0192],[354,-0.0135],[355,0.0095],[356,0.0297],[357,0.0339],[358,0.0354],[359,0.0365],[361,0.038],[362,0.0393],[363,0.0404],[364,0.0402],[366,0.0388],[377,0.0335],[380,0.0316],[381,0.0313],[382,0.0316],[388,0.0359],[393,0.038],[400,0.0392],[433,0.0388]]
       }
       spine_01_ctrl.Location.Y {
        type=MovieSceneScriptingFloatChannel
        default=0
        thinned_from=434
        keys [[0,-0.1419],[19,-0.142],[20,-0.1416],[21,-0.1404],[22,-0.1386],[23,-0.1363],[24,-0.1342],[26,-0.1316],[27,-0.1308],[29,-0.1297],[33,-0.1294],[62,-0.132],[63,-0.1325],[64,-0.1334],[65,-0.135],[67,-0.1391],[68,-0.1407],[69,-0.1414],[148,-0.1415],[149,-0.1413],[150,-0.1071],[151,-0.0987],[152,-0.0971],[154,-0.0965],[155,-0.0965],[156,-0.0961],[157,-0.1514],[158,-0.139],[159,-0.1529],[160,-0.1656],[161,-0.172],[162,-0.1778],[163,-0.1831],[164,-0.1751],[165,-0.1702],[167,-0.1642],[168,-0.158],[169,-0.1474],[170,-0.1572],[171,-0.1687],[172,-0.1733],[220,-0.173],[223,-0.1741],[224,-0.1748],[226,-0.1774],[227,-0.1782],[257,-0.1794],[259,-0.1789],[265,-0.1746],[294,-0.1706],[349,-0.1748],[350,-0.175],[352,-0.1744],[353,-0.175],[354,-0.178],[355,-0.1741],[356,-0.1583],[357,-0.1526],[358,-0.1515],[359,-0.1525],[362,-0.1562],[363,-0.1556],[364,-0.1523],[365,-0.1483],[366,-0.145],[367,-0.1433],[368,-0.1426],[375,-0.144],[387,-0.1438],[433,-0.1419]]
       }
       spine_01_ctrl.Location.Z {
        type=MovieSceneScriptingFloatChannel
        default=0
        thinned_from=434
        keys [[0,0.0027],[14,0.0025],[17,0.0018],[19,0.0011],[20,0.0012],[23,0.0027],[25,0.0032],[61,0.0034],[62,0.0033],[63,0.003],[64,0.0023],[65,0.0011],[66,0.0009],[68,0.0015],[149,0.0018],[150,-0.002],[151,-0.0023],[156,-0.0024],[157,0.0315],[158,0.0438],[159,0.0506],[160,0.0583],[162,0.0694],[163,0.0748],[164,0.068],[165,0.0641],[167,0.0595],[168,0.0558],[169,0.0512],[170,0.055],[171,0.0614],[172,0.0644],[224,0.0646],[344,0.0645],[345,0.064],[346,0.0624],[347,0.0603],[350,0.0527],[351,0.0495],[353,0.0412],[354,0.0378],[355,0.022],[356,0.0088],[357,0.0062],[358,0.0059],[361,0.007],[362,0.0071],[363,0.0066],[365,0.0043],[366,0.0034],[367,0.003],[433,0.0027]]
       }
       spine_01_ctrl.Rotation.X {
        type=MovieSceneScriptingFloatChannel
        default=0
        thinned_from=434
        keys [[0,-2.185],[19,-2.1884],[20,-2.1853],[21,-2.1745],[22,-2.1536],[24,-2.0988],[26,-2.0634],[28,-2.0415],[30,-2.0324],[64,-2.0744],[65,-2.0729],[66,-2.0804],[68,-2.1432],[69,-2.1632],[70,-2.1718],[143,-2.1829],[147,-2.1673],[149,-2.1476],[150,-1.6459],[151,-1.529],[152,-1.507],[154,-1.4998],[155,-1.4994],[156,-1.4942],[157,-2.5962],[158,-2.4775],[159,-2.6829],[160,-2.8559],[161,-2.9403],[162,-3.0153],[163,-3.0825],[164,-2.9795],[165,-2.916],[167,-2.8381],[168,-2.7562],[169,-2.612],[170,-2.7451],[171,-2.8995],[172,-2.9594],[221,-2.9578],[224,-2.9803],[226,-3.0159],[227,-3.027],[257,-3.0432],[260,-3.0281],[265,-2.9769],[294,-2.9222],[345,-2.9759],[350,-2.9442],[352,-2.9187],[353,-2.9183],[354,-2.9509],[355,-2.8212],[356,-2.5022],[357,-2.4025],[358,-2.3876],[359,-2.4065],[361,-2.453],[362,-2.4667],[363,-2.4499],[364,-2.3874],[365,-2.3091],[366,-2.2458],[367,-2.2109],[368,-2.1941],[369,-2.191],[382,-2.1837],[433,-2.185]]
       }
       spine_01_ctrl.Rotation.Y {
        type=MovieSceneScriptingFloatChannel
        default=0
        thinned_from=434
        keys [[0,-0.6282],[10,-0.6282],[11,-0.6259],[12,-0.6205],[13,-0.6099],[14,-0.5924],[15,-0.5665],[16,-0.5245],[17,-0.4554],[18,-0.3481],[19,-0.169],[20,0.0402],[21,0.1772],[22,0.2288],[23,0.2508],[24,0.2466],[26,0.2099],[27,0.1969],[28,0.1891],[29,0.1854],[30,0.1841],[33,0.1843],[34,0.1885],[35,0.1915],[36,0.1952],[37,0.1995],[44,0.2345],[45,0.2387],[46,0.2419],[47,0.244],[49,0.247],[50,0.247],[51,0.2489],[52,0.2489],[53,0.2507],[55,0.2531],[56,0.2536],[57,0.2534],[58,0.252],[59,0.2494],[60,0.2451],[61,0.239],[62,0.2255],[63,0.1805],[64,0.0759],[65,-0.2005],[66,-0.4841],[67,-0.578],[68,-0.5649],[69,-0.5121],[70,-0.488],[71,-0.4885],[72,-0.491],[77,-0.5163],[80,-0.5282],[84,-0.5381],[86,-0.541],[87,-0.541],[88,-0.5429],[90,-0.5429],[91,-0.5447],[94,-0.5447],[95,-0.547],[96,-0.547],[99,-0.5512],[100,-0.5517],[102,-0.5488],[104,-0.5434],[106,-0.5356],[108,-0.5248],[112,-0.4968],[114,-0.4863],[116,-0.4789],[121,-0.4687],[123,-0.4661],[124,-0.4661],[125,-0.4647],[130,-0.4647],[139,-0.4632],[140,-0.4514],[141,-0.4297],[142,-0.4186],[143,-0.426],[144,-0.4443],[145,-0.4718],[146,-0.5077],[147,-0.5514],[148,-0.6055],[149,-0.6724],[150,-0.5196],[151,-0.4344],[152,-0.4068],[153,-0.3933],[154,-0.3862],[155,-0.3834],[156,-0.3797],[157,-0.6196],[158,-0.4419],[159,-0.4128],[160,-0.4177],[161,-0.4203],[163,-0.4357],[164,-0.4193],[165,-0.4098],[167,-0.4004],[168,-0.3901],[169,-0.3713],[170,-0.3917],[171,-0.4145],[172,-0.4234],[218,-0.4208],[221,-0.4228],[222,-0.4247],[223,-0.4275],[224,-0.4313],[225,-0.4382],[226,-0.4471],[227,-0.4529],[231,-0.4601],[237,-0.464],[254,-0.4643],[257,-0.4622],[259,-0.4579],[260,-0.4535],[262,-0.4424],[264,-0.4333],[265,-0.43],[267,-0.4262],[282,-0.4184],[287,-0.4172],[293,-0.4115],[295,-0.4117],[299,-0.4173],[326,-0.4235],[335,-0.4294],[344,-0.4303],[345,-0.4368],[346,-0.4542],[347,-0.4798],[348,-0.5105],[349,-0.5429],[350,-0.5731],[351,-0.5684],[352,-0.5192],[353,-0.4586],[354,-0.4253],[355,-0.586],[356,-0.681],[357,-0.6022],[358,-0.5415],[359,-0.5224],[360,-0.5085],[361,-0.5046],[362,-0.5171],[363,-0.5315],[364,-0.5447],[365,-0.5692],[366,-0.5985],[367,-0.6294],[368,-0.6629],[369,-0.6943],[371,-0.7505],[372,-0.7755],[373,-0.7968],[374,-0.8144],[375,-0.8288],[376,-0.84],[377,-0.8479],[378,-0.8512],[381,-0.8465],[383,-0.8497],[384,-0.8492],[385,-0.8434],[386,-0.8304],[387,-0.8118],[390,-0.7458],[392,-0.7075],[393,-0.6901],[394,-0.6744],[395,-0.6612],[397,-0.6398],[398,-0.6312],[399,-0.6241],[400,-0.6188],[401,-0.6152],[403,-0.6115],[404,-0.6115],[405,-0.6101],[409,-0.6101],[423,-0.6251],[424,-0.6251],[433,-0.628]]
       }
       spine_01_ctrl.Rotation.Z {
        type=MovieSceneScriptingFloatChannel
        default=0
        thinned_from=434
        keys [[0,-0.0038],[10,-0.0038],[11,-0.0083],[12,-0.019],[13,-0.0398],[14,-0.0741],[15,-0.1253],[16,-0.1966],[17,-0.2916],[18,-0.4134],[19,-0.7991],[20,-1.5818],[21,-2.6085],[22,-3.7259],[23,-4.7804],[24,-5.6198],[25,-6.2235],[26,-6.6823],[27,-7.0119],[28,-7.2278],[29,-7.3544],[30,-7.4193],[31,-7.4432],[33,-7.4432],[36,-7.3082],[43,-6.8622],[45,-6.7647],[47,-6.707],[49,-6.6761],[50,-6.6761],[51,-6.6568],[52,-6.6568],[59,-6.5494],[61,-6.485],[62,-6.4431],[63,-6.3653],[64,-6.1949],[65,-5.8871],[66,-5.3041],[67,-4.3099],[68,-3.1197],[69,-2.0654],[70,-1.4783],[71,-1.291],[72,-1.1947],[73,-1.1593],[74,-1.1543],[75,-1.1674],[76,-1.1973],[78,-1.2635],[80,-1.3348],[81,-1.3649],[83,-1.4133],[84,-1.4339],[85,-1.4515],[86,-1.4656],[87,-1.4656],[88,-1.4852],[90,-1.4852],[91,-1.4984],[94,-1.4984],[95,-1.4944],[96,-1.4944],[98,-1.4608],[99,-1.4383],[100,-1.3931],[101,-1.3152],[102,-1.2124],[103,-1.0927],[104,-0.9481],[105,-0.7778],[106,-0.6001],[107,-0.4332],[108,-0.276],[109,-0.1208],[110,0.0241],[111,0.1503],[112,0.2552],[113,0.3435],[114,0.4172],[115,0.4782],[116,0.5279],[117,0.5676],[118,0.5995],[119,0.6258],[120,0.6487],[121,0.6692],[122,0.6865],[123,0.7005],[124,0.7005],[125,0.7173],[130,0.7173],[131,0.6975],[132,0.6816],[133,0.6592],[134,0.6304],[135,0.5948],[136,0.5514],[137,0.4993],[138,0.4374],[139,0.3646],[140,0.2775],[141,0.1512],[142,-0.0471],[143,-0.3033],[144,-0.594],[145,-0.9373],[146,-1.3514],[147,-1.8558],[148,-2.4752],[149,-3.2362],[150,-4.7146],[151,-4.5261],[152,-4.3524],[153,-4.2202],[154,-4.1575],[155,-4.1264],[156,-4.1182],[157,-5.6659],[158,-8.1959],[159,-9.4815],[160,-10.2453],[161,-10.6647],[162,-10.947],[163,-11.1054],[164,-11.1708],[165,-11.1328],[166,-11.0229],[168,-10.7384],[170,-10.4818],[171,-10.3813],[172,-10.3309],[173,-10.3181],[175,-10.3673],[179,-10.494],[181,-10.5266],[194,-10.5426],[216,-10.4899],[218,-10.4763],[220,-10.4195],[221,-10.3634],[222,-10.2605],[223,-10.1081],[224,-9.9036],[225,-9.5408],[226,-9.0866],[227,-8.7989],[229,-8.5951],[231,-8.4495],[233,-8.3519],[236,-8.2728],[252,-8.2307],[254,-8.245],[256,-8.2954],[257,-8.3452],[258,-8.4287],[259,-8.5551],[260,-8.7693],[262,-9.3253],[264,-9.7963],[265,-9.9742],[266,-10.0941],[267,-10.1816],[270,-10.3558],[282,-10.6115],[285,-10.6272],[287,-10.6812],[292,-10.9749],[293,-11.0125],[294,-11.0277],[295,-10.9975],[298,-10.7371],[299,-10.6733],[304,-10.5447],[321,-10.442],[325,-10.3634],[330,-10.1166],[335,-10.0069],[343,-9.976],[345,-9.928],[346,-9.8402],[347,-9.6885],[348,-9.4649],[349,-9.1614],[350,-8.7705],[351,-8.2554],[352,-7.5725],[353,-6.7034],[354,-5.63],[355,-3.7276],[356,-1.7086],[357,-0.4283],[358,0.3203],[359,0.6622],[360,0.8927],[361,1.0312],[362,1.097],[363,1.1115],[364,0.9638],[365,0.6324],[366,0.2748],[367,-0.0717],[368,-0.4445],[369,-0.7989],[370,-1.1308],[371,-1.457],[372,-1.7643],[373,-2.0393],[374,-2.2796],[375,-2.492],[376,-2.6766],[377,-2.8335],[378,-2.9687],[379,-3.0801],[380,-3.1558],[381,-3.1837],[382,-3.1533],[383,-3.0662],[384,-2.9283],[385,-2.7457],[386,-2.5099],[387,-2.2262],[388,-1.9216],[389,-1.623],[390,-1.3572],[391,-1.1205],[392,-0.8949],[393,-0.6861],[394,-0.5],[395,-0.3422],[396,-0.2078],[397,-0.0891],[398,0.0126],[399,0.0958],[400,0.1593],[401,0.2016],[402,0.2279],[403,0.2455],[404,0.2455],[405,0.2617],[409,0.2617],[410,0.2452],[411,0.2326],[412,0.2179],[413,0.2019],[416,0.1526],[418,0.1189],[419,0.1022],[420,0.0862],[421,0.071],[422,0.057],[423,0.0445],[424,0.0445],[425,0.025],[426,0.025],[427,0.0113],[429,0.0113],[430,-0.0005],[433,-0.0005]]
       }
       spine_01_ctrl.Scale.X {type=MovieSceneScriptingFloatChannel, default=1, const [1,434]}
       spine_01_ctrl.Scale.Y {type=MovieSceneScriptingFloatChannel, default=1, const [1,434]}
       spine_01_ctrl.Scale.Z {type=MovieSceneScriptingFloatChannel, default=1, const [1,434]}
       spine_02_ctrl.Location.X {
        type=MovieSceneScriptingFloatChannel
        default=0
        thinned_from=434
        keys [[0,0.03],[11,0.0298],[12,0.0296],[13,0.0291],[14,0.0282],[15,0.027],[16,0.0252],[17,0.022],[18,0.0172],[19,0.0093],[20,0.0001],[21,-0.0061],[22,-0.009],[23,-0.0106],[24,-0.0106],[26,-0.0091],[27,-0.0086],[29,-0.0082],[33,-0.0082],[48,-0.0108],[58,-0.0111],[61,-0.0106],[62,-0.01],[63,-0.0082],[64,-0.0042],[65,0.006],[66,0.0165],[67,0.0204],[68,0.0203],[69,0.0182],[70,0.0173],[83,0.0201],[96,0.0207],[107,0.0219],[123,0.0197],[138,0.0189],[139,0.0187],[140,0.0177],[141,0.0158],[142,0.0146],[143,0.0145],[144,0.0152],[145,0.0164],[146,0.0181],[147,0.0202],[148,0.0227],[149,0.0259],[151,0.02],[152,0.0189],[153,0.0183],[156,0.0177],[157,0.0369],[158,0.0199],[159,0.0132],[160,0.0102],[161,0.0085],[162,0.0075],[164,0.0065],[165,0.0065],[168,0.0075],[169,0.0076],[171,0.0092],[172,0.0095],[219,0.009],[221,0.0094],[223,0.0104],[224,0.0112],[225,0.0126],[226,0.0144],[227,0.0156],[231,0.017],[236,0.0177],[256,0.0176],[258,0.0171],[259,0.0165],[260,0.0157],[262,0.0134],[264,0.0116],[265,0.0109],[268,0.0098],[288,0.008],[293,0.0069],[295,0.007],[299,0.0082],[326,0.0095],[334,0.0107],[344,0.0109],[345,0.0116],[346,0.0133],[347,0.0158],[350,0.025],[351,0.0254],[352,0.0222],[353,0.018],[354,0.0158],[355,0.0272],[356,0.0339],[357,0.0298],[358,0.0268],[359,0.026],[360,0.0254],[361,0.0253],[364,0.0263],[365,0.0273],[367,0.03],[369,0.033],[371,0.0356],[373,0.0377],[375,0.039],[377,0.0396],[378,0.0395],[380,0.0388],[381,0.0387],[384,0.0395],[385,0.0394],[386,0.039],[388,0.0374],[393,0.0329],[396,0.0311],[399,0.0299],[403,0.0293],[433,0.0299]]
       }
       spine_02_ctrl.Location.Y {
        type=MovieSceneScriptingFloatChannel
        default=0
        thinned_from=434
        keys [[0,-0.0867],[19,-0.0868],[20,-0.0864],[21,-0.0854],[22,-0.0837],[23,-0.0813],[24,-0.0793],[25,-0.0778],[26,-0.0765],[27,-0.0755],[28,-0.0748],[30,-0.0741],[33,-0.074],[47,-0.0761],[61,-0.0767],[63,-0.0773],[64,-0.0778],[65,-0.0787],[66,-0.0802],[67,-0.0825],[68,-0.0845],[69,-0.0856],[70,-0.0861],[144,-0.0865],[147,-0.0859],[149,-0.085],[150,-0.0681],[151,-0.0643],[152,-0.0637],[156,-0.0633],[157,-0.1748],[158,-0.1853],[159,-0.1914],[160,-0.1951],[163,-0.2002],[164,-0.1971],[165,-0.1954],[167,-0.1939],[168,-0.192],[169,-0.1884],[170,-0.1923],[171,-0.1967],[172,-0.1985],[220,-0.1982],[223,-0.1993],[224,-0.2],[226,-0.2025],[227,-0.2032],[257,-0.2043],[260,-0.2033],[265,-0.1997],[294,-0.1959],[344,-0.1998],[345,-0.1994],[346,-0.1983],[350,-0.1908],[351,-0.1879],[353,-0.1795],[354,-0.1769],[355,-0.1534],[356,-0.1183],[357,-0.1083],[358,-0.1066],[359,-0.1075],[360,-0.1093],[361,-0.1105],[362,-0.1101],[363,-0.1075],[364,-0.1023],[365,-0.0963],[366,-0.0915],[367,-0.0889],[368,-0.0877],[369,-0.0875],[373,-0.0882],[381,-0.0876],[387,-0.088],[433,-0.0867]]
       }
       spine_02_ctrl.Location.Z {
        type=MovieSceneScriptingFloatChannel
        default=0
        thinned_from=434
        keys [[0,0.0136],[14,0.0134],[19,0.0122],[20,0.0123],[24,0.0133],[62,0.0134],[63,0.0132],[64,0.0127],[65,0.0118],[66,0.0119],[68,0.0126],[149,0.0129],[150,0.0065],[151,0.0055],[156,0.0052],[157,0.0305],[158,0.0379],[159,0.0447],[160,0.0523],[162,0.0628],[163,0.0677],[164,0.0613],[165,0.0576],[167,0.0532],[168,0.0495],[169,0.0447],[170,0.0487],[171,0.0551],[172,0.0581],[224,0.0582],[260,0.0587],[344,0.0582],[345,0.0578],[346,0.0567],[347,0.055],[349,0.0512],[350,0.0495],[351,0.0472],[353,0.0412],[354,0.0391],[355,0.028],[356,0.0181],[357,0.0162],[358,0.0159],[361,0.017],[362,0.0172],[363,0.017],[365,0.015],[366,0.0142],[367,0.0139],[433,0.0136]]
       }
       spine_02_ctrl.Rotation.X {
        type=MovieSceneScriptingFloatChannel
        default=0
        thinned_from=434
        keys [[0,-1.5988],[19,-1.6017],[20,-1.598],[21,-1.5858],[22,-1.5623],[24,-1.5005],[25,-1.4788],[26,-1.4601],[27,-1.4456],[28,-1.4352],[29,-1.4286],[31,-1.4232],[37,-1.4329],[46,-1.455],[61,-1.464],[64,-1.4737],[65,-1.4742],[66,-1.4848],[68,-1.554],[69,-1.5755],[70,-1.5847],[142,-1.598],[146,-1.5861],[148,-1.5713],[149,-1.559],[150,-1.2721],[151,-1.2099],[152,-1.1991],[155,-1.1964],[156,-1.1936],[157,-3.4461],[158,-3.7331],[159,-3.8586],[160,-3.9199],[161,-3.9487],[163,-3.9889],[164,-3.9539],[165,-3.9341],[167,-3.9144],[168,-3.8893],[169,-3.8403],[171,-3.9479],[172,-3.9696],[221,-3.9683],[224,-3.986],[227,-4.0212],[259,-4.0276],[295,-3.941],[344,-3.9837],[345,-3.9737],[346,-3.9459],[347,-3.9052],[350,-3.7548],[351,-3.6903],[353,-3.5189],[354,-3.4575],[355,-2.9529],[356,-2.2358],[357,-2.0412],[358,-2.0101],[359,-2.0318],[360,-2.0686],[361,-2.0926],[362,-2.0826],[363,-2.0275],[364,-1.9214],[365,-1.7963],[366,-1.6972],[367,-1.6424],[368,-1.6153],[369,-1.6089],[374,-1.6119],[381,-1.5975],[388,-1.6085],[433,-1.5987]]
       }
       spine_02_ctrl.Rotation.Y {
        type=MovieSceneScriptingFloatChannel
        default=0
        thinned_from=434
        keys [[0,-0.581],[10,-0.581],[11,-0.5789],[12,-0.5739],[13,-0.5642],[14,-0.5481],[15,-0.5253],[16,-0.4894],[17,-0.4305],[18,-0.3385],[19,-0.1894],[20,-0.0182],[21,0.0959],[22,0.147],[23,0.1762],[24,0.1765],[25,0.16],[26,0.1454],[27,0.1356],[28,0.1305],[29,0.1287],[30,0.1288],[31,0.1297],[33,0.1297],[34,0.1328],[35,0.1349],[36,0.1375],[37,0.1404],[39,0.1474],[41,0.1549],[42,0.159],[44,0.1679],[45,0.1719],[46,0.1748],[47,0.1768],[48,0.1784],[49,0.1784],[50,0.1806],[51,0.1806],[52,0.1822],[53,0.1822],[54,0.1842],[55,0.1852],[56,0.1852],[57,0.1858],[58,0.185],[59,0.1831],[60,0.1801],[61,0.1756],[62,0.1644],[63,0.1302],[64,0.0565],[65,-0.1308],[66,-0.3234],[67,-0.3948],[68,-0.389],[69,-0.3454],[70,-0.3259],[71,-0.3276],[72,-0.3303],[76,-0.3535],[78,-0.3633],[80,-0.3714],[82,-0.3772],[85,-0.3832],[86,-0.3845],[87,-0.3845],[88,-0.3864],[91,-0.3864],[92,-0.3891],[94,-0.3891],[95,-0.3921],[96,-0.3921],[97,-0.3956],[100,-0.4036],[106,-0.4153],[107,-0.4153],[108,-0.4137],[109,-0.4108],[110,-0.4066],[112,-0.3959],[114,-0.3874],[116,-0.3814],[121,-0.3725],[123,-0.3701],[124,-0.3701],[125,-0.3687],[130,-0.3687],[131,-0.3668],[134,-0.3636],[136,-0.3597],[138,-0.354],[139,-0.3505],[140,-0.3289],[141,-0.2909],[142,-0.2654],[143,-0.265],[144,-0.2792],[145,-0.3048],[146,-0.3397],[147,-0.3823],[148,-0.4352],[149,-0.5005],[150,-0.4485],[151,-0.3897],[152,-0.3678],[153,-0.3558],[154,-0.3497],[156,-0.3445],[157,-0.9267],[158,-0.6946],[159,-0.6145],[160,-0.5842],[161,-0.5662],[162,-0.5581],[163,-0.5551],[164,-0.5438],[165,-0.5401],[166,-0.5418],[167,-0.5449],[168,-0.5445],[169,-0.5383],[171,-0.5715],[172,-0.5782],[174,-0.5782],[180,-0.5695],[195,-0.568],[215,-0.5702],[219,-0.5726],[220,-0.5739],[221,-0.5766],[222,-0.5816],[223,-0.5889],[224,-0.5988],[225,-0.6162],[226,-0.6381],[227,-0.6519],[229,-0.6617],[232,-0.6713],[236,-0.6772],[253,-0.6791],[256,-0.6761],[257,-0.6737],[258,-0.6697],[259,-0.6636],[260,-0.6533],[262,-0.6266],[264,-0.6039],[265,-0.5954],[266,-0.5896],[268,-0.5822],[271,-0.5751],[281,-0.5653],[286,-0.5627],[288,-0.5591],[291,-0.5496],[293,-0.5455],[294,-0.5448],[295,-0.5462],[298,-0.5587],[299,-0.5618],[304,-0.5679],[319,-0.5719],[325,-0.5766],[330,-0.5885],[334,-0.5932],[344,-0.5958],[345,-0.6067],[346,-0.6346],[347,-0.6747],[349,-0.7692],[350,-0.8116],[351,-0.7922],[352,-0.6927],[353,-0.5677],[354,-0.4789],[355,-0.6208],[356,-0.6873],[357,-0.5869],[358,-0.5186],[359,-0.4992],[360,-0.486],[361,-0.4807],[362,-0.4858],[363,-0.4921],[364,-0.5016],[365,-0.5238],[366,-0.5516],[367,-0.5812],[368,-0.6138],[369,-0.6447],[370,-0.6728],[371,-0.6992],[372,-0.7228],[373,-0.7424],[374,-0.7579],[375,-0.7699],[376,-0.7783],[377,-0.783],[378,-0.7817],[380,-0.7683],[381,-0.765],[382,-0.7683],[384,-0.7805],[385,-0.78],[386,-0.771],[387,-0.7558],[388,-0.7367],[390,-0.6959],[392,-0.6593],[393,-0.6426],[394,-0.6277],[395,-0.6149],[397,-0.5945],[398,-0.5862],[399,-0.5794],[400,-0.5742],[401,-0.5707],[403,-0.5672],[404,-0.5672],[405,-0.5659],[409,-0.5659],[411,-0.567],[422,-0.5779],[423,-0.5779],[424,-0.579],[433,-0.5806]]
       }
       spine_02_ctrl.Rotation.Z {
        type=MovieSceneScriptingFloatChannel
        default=0
        thinned_from=434
        keys [[0,-0.0257],[10,-0.0257],[11,-0.0293],[12,-0.0379],[13,-0.0546],[14,-0.0822],[15,-0.1233],[16,-0.1806],[17,-0.2563],[18,-0.3528],[19,-0.664],[20,-1.3004],[21,-2.1397],[22,-3.0564],[23,-3.9228],[24,-4.6137],[25,-5.1114],[26,-5.4899],[27,-5.7619],[28,-5.94],[29,-6.0444],[30,-6.0979],[31,-6.1176],[33,-6.1176],[36,-6.006],[43,-5.6376],[45,-5.5571],[47,-5.5094],[48,-5.4947],[49,-5.4947],[50,-5.4752],[51,-5.4752],[52,-5.4602],[53,-5.4602],[55,-5.4275],[56,-5.4275],[60,-5.3569],[62,-5.2922],[63,-5.2291],[64,-5.091],[65,-4.8433],[66,-4.3701],[67,-3.5553],[68,-2.5771],[69,-1.7093],[70,-1.2263],[71,-1.0725],[72,-0.9936],[73,-0.9647],[74,-0.9608],[75,-0.9718],[76,-0.9966],[78,-1.0513],[80,-1.1103],[81,-1.1351],[83,-1.1751],[84,-1.1921],[85,-1.2066],[86,-1.2182],[87,-1.2182],[88,-1.2344],[91,-1.2344],[92,-1.246],[94,-1.246],[95,-1.2422],[96,-1.2422],[98,-1.2149],[99,-1.1964],[100,-1.1595],[101,-1.0956],[102,-1.0112],[103,-0.9129],[104,-0.7943],[105,-0.6545],[106,-0.5087],[107,-0.3716],[108,-0.2424],[109,-0.1148],[110,0.0044],[111,0.1082],[112,0.1946],[113,0.2674],[114,0.328],[115,0.3783],[116,0.4192],[117,0.4519],[118,0.4782],[119,0.4999],[120,0.5187],[121,0.5356],[122,0.5499],[123,0.5615],[124,0.5615],[125,0.5753],[130,0.5753],[131,0.5592],[132,0.5461],[133,0.5278],[134,0.5041],[135,0.475],[136,0.4394],[137,0.3967],[138,0.346],[139,0.2864],[140,0.2157],[141,0.1135],[142,-0.0483],[143,-0.2588],[144,-0.4981],[145,-0.7811],[146,-1.1226],[147,-1.5386],[148,-2.0495],[149,-2.6772],[150,-3.8975],[151,-3.7416],[152,-3.5979],[153,-3.4887],[154,-3.4368],[155,-3.4112],[156,-3.4044],[157,-4.6503],[158,-6.7695],[159,-7.8286],[160,-8.4456],[161,-8.783],[162,-9.007],[163,-9.1294],[164,-9.1947],[165,-9.1706],[166,-9.0845],[168,-8.8633],[169,-8.7778],[170,-8.6527],[171,-8.5545],[172,-8.5071],[174,-8.5071],[179,-8.6421],[181,-8.669],[195,-8.6822],[215,-8.6456],[219,-8.6035],[220,-8.5804],[221,-8.534],[222,-8.4488],[223,-8.3227],[224,-8.1536],[225,-7.8533],[226,-7.4774],[227,-7.2394],[229,-7.0707],[231,-6.9502],[233,-6.8694],[236,-6.804],[237,-6.804],[238,-6.785],[240,-6.785],[253,-6.7715],[256,-6.8226],[257,-6.8639],[258,-6.933],[259,-7.0376],[260,-7.2149],[262,-7.675],[264,-8.0648],[265,-8.2119],[266,-8.3112],[267,-8.3836],[270,-8.5277],[281,-8.7292],[286,-8.7749],[288,-8.8363],[291,-9.0002],[293,-9.0711],[294,-9.0836],[295,-9.0587],[298,-8.8432],[299,-8.7905],[303,-8.6966],[319,-8.6154],[325,-8.534],[330,-8.3298],[334,-8.2497],[335,-8.2497],[336,-8.2309],[338,-8.2309],[344,-8.2044],[345,-8.1733],[346,-8.0996],[347,-7.9726],[348,-7.7862],[349,-7.5343],[350,-7.2106],[351,-6.7867],[352,-6.227],[353,-5.5144],[354,-4.6327],[355,-3.0744],[356,-1.4262],[357,-0.3735],[358,0.2431],[359,0.5241],[360,0.7135],[361,0.827],[362,0.8802],[363,0.8916],[364,0.7705],[365,0.4983],[366,0.204],[367,-0.0815],[368,-0.389],[369,-0.6814],[370,-0.9551],[371,-1.2241],[372,-1.4774],[373,-1.704],[374,-1.902],[375,-2.077],[376,-2.2289],[377,-2.3579],[378,-2.4689],[379,-2.5602],[380,-2.6221],[381,-2.6448],[382,-2.62],[383,-2.5487],[384,-2.4357],[385,-2.2858],[386,-2.0917],[387,-1.8581],[388,-1.6072],[389,-1.3611],[390,-1.142],[391,-0.9469],[392,-0.7608],[393,-0.5887],[394,-0.4352],[395,-0.305],[396,-0.1942],[397,-0.0963],[398,-0.0125],[399,0.0562],[400,0.1086],[401,0.1434],[402,0.1651],[403,0.1797],[404,0.1797],[405,0.193],[409,0.193],[410,0.1795],[411,0.1795],[412,0.1569],[413,0.1437],[416,0.1031],[418,0.0753],[419,0.0616],[420,0.0484],[421,0.0359],[422,0.0243],[423,0.0243],[424,0.0053],[425,0.0053],[426,-0.0082],[428,-0.0082],[429,-0.0206],[433,-0.0206]]
       }
       spine_02_ctrl.Scale.X {type=MovieSceneScriptingFloatChannel, default=1, const [1,434]}
       spine_02_ctrl.Scale.Y {type=MovieSceneScriptingFloatChannel, default=1, const [1,434]}
       spine_02_ctrl.Scale.Z {type=MovieSceneScriptingFloatChannel, default=1, const [1,434]}
       spine_03_ctrl.Location.X {
        type=MovieSceneScriptingFloatChannel
        default=0
        thinned_from=434
        keys [[0,0.0597],[10,0.0597],[11,0.0594],[12,0.0589],[13,0.0578],[14,0.0559],[15,0.0534],[16,0.0495],[17,0.0432],[18,0.0336],[19,0.0172],[20,-0.0033],[21,-0.02],[22,-0.032],[23,-0.0417],[24,-0.0468],[25,-0.0485],[26,-0.0498],[28,-0.0516],[30,-0.0527],[32,-0.0531],[41,-0.0525],[56,-0.0535],[58,-0.0533],[60,-0.0526],[61,-0.052],[62,-0.0507],[63,-0.0473],[64,-0.0404],[65,-0.0242],[66,-0.0061],[67,0.0057],[68,0.0123],[69,0.0142],[70,0.0158],[71,0.0172],[74,0.0196],[77,0.0211],[84,0.0224],[92,0.0228],[96,0.0236],[98,0.0244],[100,0.0256],[102,0.0275],[103,0.0287],[104,0.0301],[106,0.0334],[107,0.0347],[109,0.0365],[110,0.0371],[128,0.0376],[132,0.037],[135,0.0357],[137,0.0344],[139,0.0324],[140,0.0289],[141,0.023],[142,0.0181],[143,0.0161],[144,0.0156],[145,0.0163],[146,0.0176],[147,0.0193],[148,0.0213],[149,0.024],[150,0.0189],[151,0.0158],[152,0.0148],[153,0.0143],[156,0.0138],[157,0.0874],[158,0.0423],[159,0.0194],[160,0.0067],[161,-0.0005],[162,-0.0052],[163,-0.0082],[164,-0.0085],[165,-0.0076],[166,-0.0058],[169,0.0012],[170,0.0025],[171,0.0032],[172,0.0036],[173,0.0038],[174,0.0036],[177,0.002],[181,0.0009],[192,0.0006],[216,0.0014],[219,0.002],[220,0.0024],[221,0.0032],[222,0.0046],[223,0.0068],[224,0.0096],[225,0.0148],[226,0.0212],[227,0.0253],[228,0.0268],[229,0.0281],[231,0.0302],[233,0.0316],[235,0.0324],[238,0.033],[251,0.0333],[255,0.0327],[256,0.0324],[257,0.0317],[258,0.0305],[259,0.0287],[260,0.0257],[262,0.0178],[264,0.0112],[265,0.0086],[266,0.007],[267,0.0057],[269,0.004],[271,0.0027],[277,0.0007],[286,-0.0009],[287,-0.0013],[288,-0.002],[291,-0.0048],[293,-0.006],[294,-0.0062],[295,-0.0058],[296,-0.0047],[298,-0.0021],[299,-0.0012],[302,0.0001],[307,0.001],[322,0.0024],[325,0.0032],[326,0.0037],[329,0.006],[331,0.0071],[335,0.0082],[344,0.0089],[345,0.0106],[346,0.0152],[347,0.0219],[348,0.0299],[349,0.0385],[350,0.0468],[351,0.0473],[352,0.0375],[353,0.0251],[354,0.0185],[355,0.0459],[356,0.0645],[357,0.0599],[358,0.0565],[360,0.0565],[361,0.0567],[362,0.0571],[363,0.0573],[364,0.0571],[365,0.0574],[366,0.0581],[367,0.0593],[369,0.0618],[371,0.0638],[372,0.0645],[373,0.0649],[374,0.065],[375,0.0649],[376,0.0646],[377,0.064],[378,0.0628],[380,0.0595],[381,0.0589],[382,0.0596],[383,0.0611],[384,0.0629],[385,0.0643],[386,0.0649],[387,0.0651],[388,0.065],[395,0.0615],[400,0.0601],[433,0.0597]]
       }
       spine_03_ctrl.Location.Y {
        type=MovieSceneScriptingFloatChannel
        default=0
        thinned_from=434
        keys [[0,-0.1358],[19,-0.136],[20,-0.1353],[21,-0.1334],[22,-0.1302],[23,-0.1259],[24,-0.122],[25,-0.1193],[26,-0.1169],[27,-0.1151],[28,-0.1139],[29,-0.113],[30,-0.1126],[32,-0.1123],[37,-0.1135],[44,-0.1158],[60,-0.1171],[62,-0.1177],[63,-0.1184],[64,-0.1196],[65,-0.1214],[66,-0.1243],[67,-0.1285],[68,-0.132],[69,-0.1339],[70,-0.1347],[143,-0.1355],[146,-0.1349],[148,-0.1339],[149,-0.133],[150,-0.1156],[151,-0.112],[152,-0.1114],[156,-0.1111],[157,-0.4442],[158,-0.5088],[159,-0.5182],[161,-0.5182],[164,-0.517],[223,-0.5216],[227,-0.5255],[260,-0.5256],[295,-0.518],[344,-0.5222],[345,-0.5203],[346,-0.5152],[347,-0.5077],[350,-0.4801],[351,-0.4681],[353,-0.435],[354,-0.4205],[355,-0.3362],[356,-0.2283],[357,-0.1996],[358,-0.1947],[359,-0.1971],[360,-0.2018],[361,-0.2046],[362,-0.2018],[363,-0.1932],[364,-0.1787],[365,-0.162],[366,-0.1489],[367,-0.1418],[368,-0.1383],[369,-0.1377],[373,-0.139],[378,-0.1386],[381,-0.138],[385,-0.1387],[401,-0.1354],[433,-0.1357]]
       }
       spine_03_ctrl.Location.Z {
        type=MovieSceneScriptingFloatChannel
        default=0
        thinned_from=434
        keys [[0,0.0093],[14,0.0091],[17,0.0084],[19,0.0076],[20,0.0078],[23,0.0096],[25,0.0102],[61,0.0104],[62,0.0103],[63,0.0099],[64,0.0091],[65,0.0079],[66,0.0078],[67,0.0082],[149,0.0086],[150,0.0016],[151,0.0005],[156,0.0002],[157,0.0087],[158,0.012],[159,0.0199],[160,0.0304],[162,0.0445],[163,0.0513],[164,0.0426],[165,0.0374],[167,0.0313],[168,0.0262],[169,0.0196],[170,0.025],[171,0.0338],[172,0.0378],[224,0.038],[260,0.0384],[296,0.0375],[344,0.038],[345,0.0376],[346,0.0367],[347,0.0354],[350,0.0309],[351,0.0291],[353,0.0245],[354,0.023],[355,0.0153],[356,0.0099],[357,0.0091],[358,0.0091],[361,0.0099],[362,0.0103],[363,0.0105],[366,0.0094],[433,0.0093]]
       }
       spine_03_ctrl.Rotation.X {
        type=MovieSceneScriptingFloatChannel
        default=0
        thinned_from=434
        keys [[0,-1.8404],[19,-1.8435],[20,-1.8378],[21,-1.8206],[22,-1.7886],[24,-1.705],[25,-1.6745],[26,-1.6485],[27,-1.6283],[28,-1.614],[29,-1.605],[31,-1.5978],[34,-1.6005],[45,-1.6406],[61,-1.6553],[63,-1.6631],[65,-1.6749],[66,-1.6931],[68,-1.7842],[69,-1.8122],[70,-1.8239],[142,-1.8402],[145,-1.8314],[147,-1.8176],[148,-1.8056],[149,-1.7882],[150,-1.5831],[151,-1.5465],[152,-1.5421],[155,-1.5433],[156,-1.5416],[157,-6.431],[158,-7.4522],[159,-7.5769],[160,-7.5493],[161,-7.5374],[163,-7.4879],[168,-7.6008],[169,-7.6391],[171,-7.5829],[224,-7.5821],[344,-7.5805],[345,-7.5483],[346,-7.4618],[347,-7.3363],[350,-6.8829],[351,-6.7166],[353,-6.2979],[354,-6.0998],[355,-4.8175],[356,-3.2109],[357,-2.8012],[358,-2.7382],[359,-2.7778],[360,-2.851],[361,-2.8931],[362,-2.8494],[363,-2.7167],[364,-2.497],[365,-2.2423],[366,-2.0415],[367,-1.9297],[368,-1.8727],[369,-1.8566],[373,-1.8549],[380,-1.8299],[382,-1.8296],[389,-1.8472],[433,-1.8403]]
       }
       spine_03_ctrl.Rotation.Y {
        type=MovieSceneScriptingFloatChannel
        default=0
        thinned_from=434
        keys [[0,-0.8662],[10,-0.8662],[11,-0.8631],[12,-0.8556],[13,-0.8411],[14,-0.8172],[15,-0.7842],[16,-0.7339],[17,-0.6518],[18,-0.5232],[19,-0.3192],[20,-0.0866],[21,0.0761],[22,0.1677],[23,0.2347],[24,0.2517],[25,0.2364],[26,0.2248],[27,0.2185],[28,0.2172],[29,0.2191],[31,0.225],[32,0.2263],[33,0.2268],[34,0.2279],[36,0.2312],[38,0.2356],[40,0.2411],[41,0.2447],[42,0.2493],[44,0.2608],[45,0.266],[46,0.2699],[48,0.2746],[50,0.2777],[51,0.2777],[52,0.28],[55,0.2838],[56,0.2846],[57,0.2846],[58,0.2836],[59,0.2813],[60,0.2774],[61,0.2716],[62,0.2556],[63,0.2125],[64,0.1287],[65,-0.069],[66,-0.2756],[67,-0.3732],[68,-0.3836],[69,-0.3336],[70,-0.3127],[72,-0.3237],[73,-0.3332],[76,-0.3635],[77,-0.3722],[79,-0.3864],[80,-0.3924],[81,-0.3974],[83,-0.405],[85,-0.4106],[87,-0.4141],[88,-0.4141],[89,-0.4167],[90,-0.4167],[91,-0.4189],[92,-0.4189],[93,-0.4218],[95,-0.4264],[96,-0.4298],[97,-0.4341],[98,-0.4396],[101,-0.4612],[103,-0.4776],[104,-0.4882],[105,-0.5002],[106,-0.5111],[107,-0.5186],[108,-0.5221],[109,-0.5231],[110,-0.5212],[111,-0.5162],[113,-0.5048],[115,-0.4965],[121,-0.4819],[123,-0.4784],[128,-0.4784],[129,-0.4749],[131,-0.4719],[133,-0.4664],[134,-0.4624],[135,-0.4574],[136,-0.4512],[137,-0.4437],[138,-0.4349],[139,-0.4247],[140,-0.3781],[141,-0.2972],[142,-0.2379],[143,-0.226],[144,-0.2416],[145,-0.2775],[146,-0.3286],[147,-0.391],[148,-0.4689],[149,-0.5651],[150,-0.6143],[151,-0.5584],[152,-0.5317],[153,-0.5149],[154,-0.5067],[155,-0.5029],[156,-0.5005],[157,-2.2729],[158,-2.1753],[159,-2.0716],[160,-1.9994],[161,-1.9553],[162,-1.9217],[163,-1.8948],[164,-1.9167],[167,-1.9643],[168,-1.9886],[169,-2.0235],[170,-2.0018],[171,-1.9733],[172,-1.9624],[220,-1.9592],[223,-1.9702],[224,-1.9773],[226,-2.0051],[227,-2.0146],[256,-2.0309],[259,-2.0225],[265,-1.9749],[294,-1.9373],[344,-1.9754],[345,-1.9907],[346,-2.0301],[348,-2.1416],[349,-2.1932],[350,-2.2283],[351,-2.1184],[352,-1.8194],[353,-1.4534],[354,-1.1524],[355,-1.173],[356,-1.1249],[357,-0.9188],[358,-0.7961],[359,-0.7629],[360,-0.7408],[361,-0.7296],[362,-0.729],[363,-0.7315],[364,-0.7452],[365,-0.7797],[366,-0.8215],[367,-0.8652],[368,-0.9133],[369,-0.9589],[370,-1.0003],[371,-1.0389],[372,-1.0727],[373,-1.1],[374,-1.1204],[375,-1.1351],[376,-1.1437],[377,-1.1462],[378,-1.1378],[380,-1.1029],[381,-1.095],[382,-1.103],[384,-1.1371],[385,-1.1439],[386,-1.136],[387,-1.1178],[388,-1.0927],[391,-1.009],[393,-0.9589],[394,-0.9372],[395,-0.9189],[397,-0.8893],[398,-0.8773],[399,-0.8674],[400,-0.8599],[401,-0.855],[403,-0.8499],[410,-0.8494],[424,-0.864],[433,-0.8656]]
       }
       spine_03_ctrl.Rotation.Z {
        type=MovieSceneScriptingFloatChannel
        default=0
        thinned_from=434
        keys [[0,-0.0626],[10,-0.0626],[11,-0.0673],[12,-0.0785],[13,-0.1004],[14,-0.1364],[15,-0.1903],[16,-0.2652],[17,-0.3639],[18,-0.4888],[19,-0.8982],[20,-1.7412],[21,-2.8576],[22,-4.0803],[23,-5.2383],[24,-6.1642],[25,-6.8326],[26,-7.3412],[27,-7.7067],[28,-7.9462],[29,-8.0866],[30,-8.1586],[31,-8.185],[32,-8.1888],[33,-8.1756],[35,-8.0915],[37,-7.9699],[41,-7.6676],[44,-7.4809],[46,-7.3932],[50,-7.3208],[51,-7.3208],[59,-7.1916],[62,-7.0748],[63,-6.9903],[64,-6.805],[65,-6.4731],[66,-5.8402],[67,-4.752],[68,-3.4455],[69,-2.2855],[70,-1.6404],[71,-1.4353],[72,-1.3302],[73,-1.292],[74,-1.2872],[75,-1.3022],[76,-1.3358],[78,-1.4094],[80,-1.4885],[81,-1.5218],[83,-1.5754],[84,-1.5982],[85,-1.6177],[86,-1.6333],[87,-1.6455],[88,-1.6455],[89,-1.6621],[90,-1.6621],[91,-1.6697],[94,-1.6694],[95,-1.6659],[96,-1.6589],[97,-1.6473],[98,-1.63],[99,-1.6057],[100,-1.5567],[101,-1.4719],[102,-1.3597],[103,-1.2291],[104,-1.0713],[105,-0.8855],[106,-0.6915],[107,-0.509],[108,-0.3369],[109,-0.1668],[110,-0.0078],[111,0.1309],[112,0.2464],[113,0.3436],[114,0.4247],[115,0.492],[116,0.5467],[117,0.5904],[118,0.6255],[119,0.6545],[120,0.6798],[121,0.7024],[122,0.7216],[123,0.7371],[128,0.7371],[129,0.7529],[130,0.7459],[131,0.7342],[132,0.7168],[133,0.6926],[134,0.6612],[135,0.6225],[136,0.5754],[137,0.5187],[138,0.4514],[139,0.3724],[140,0.28],[141,0.1472],[142,-0.066],[143,-0.3461],[144,-0.6658],[145,-1.0447],[146,-1.5022],[147,-2.0595],[148,-2.7442],[149,-3.5854],[150,-5.2303],[151,-5.0222],[152,-4.8296],[153,-4.683],[154,-4.6133],[155,-4.5789],[156,-4.5699],[157,-6.0871],[158,-8.8041],[159,-10.1597],[160,-10.9433],[161,-11.3703],[162,-11.6517],[163,-11.8018],[164,-11.8986],[165,-11.875],[166,-11.7676],[168,-11.4938],[169,-11.3985],[170,-11.2208],[171,-11.0767],[172,-11.0089],[173,-10.9952],[175,-11.0481],[179,-11.1848],[182,-11.2304],[218,-11.1542],[220,-11.1044],[221,-11.044],[222,-10.933],[223,-10.7688],[224,-10.5487],[225,-10.1585],[226,-9.6708],[227,-9.3623],[229,-9.1439],[231,-8.988],[233,-8.8835],[236,-8.7989],[251,-8.7518],[255,-8.7959],[256,-8.823],[257,-8.8764],[258,-8.9658],[259,-9.1011],[260,-9.3306],[262,-9.9271],[264,-10.4333],[265,-10.6247],[266,-10.7538],[267,-10.848],[270,-11.0357],[282,-11.3116],[286,-11.358],[288,-11.4381],[291,-11.6521],[293,-11.7447],[294,-11.7611],[295,-11.7286],[298,-11.4472],[299,-11.3783],[303,-11.2559],[323,-11.0888],[326,-11.0052],[330,-10.778],[335,-10.6599],[344,-10.609],[345,-10.577],[346,-10.4883],[347,-10.3336],[348,-10.1035],[349,-9.7888],[350,-9.38],[351,-8.8408],[352,-8.1236],[353,-7.2033],[354,-6.0587],[355,-4.0728],[356,-1.9318],[357,-0.5306],[358,0.2929],[359,0.6668],[360,0.9182],[361,1.0687],[362,1.1392],[363,1.1556],[364,0.9969],[365,0.6361],[366,0.2442],[367,-0.1375],[368,-0.5493],[369,-0.941],[370,-1.3077],[371,-1.6681],[372,-2.0073],[373,-2.3105],[374,-2.5753],[375,-2.8091],[376,-3.012],[377,-3.1841],[378,-3.3317],[379,-3.4526],[380,-3.5343],[381,-3.5643],[382,-3.5316],[383,-3.4373],[384,-3.2874],[385,-3.0878],[386,-2.8289],[387,-2.5167],[388,-2.1812],[389,-1.8519],[390,-1.5586],[391,-1.2972],[392,-1.048],[393,-0.8174],[394,-0.6118],[395,-0.4374],[396,-0.2889],[397,-0.1578],[398,-0.0454],[399,0.0466],[400,0.1168],[401,0.1635],[402,0.1926],[403,0.2121],[404,0.2239],[409,0.2239],[410,0.2118],[411,0.1978],[412,0.1816],[413,0.164],[416,0.1096],[418,0.0725],[419,0.0542],[420,0.0365],[421,0.0197],[422,0.0043],[423,-0.0094],[424,-0.0211],[425,-0.0211],[426,-0.0391],[427,-0.0391],[428,-0.0514],[433,-0.0514]]
       }
       spine_03_ctrl.Scale.X {type=MovieSceneScriptingFloatChannel, default=1, const [1,434]}
       spine_03_ctrl.Scale.Y {type=MovieSceneScriptingFloatChannel, default=1, const [1,434]}
       spine_03_ctrl.Scale.Z {type=MovieSceneScriptingFloatChannel, default=1, const [1,434]}
       spine_04_ctrl.Location.X {
        type=MovieSceneScriptingFloatChannel
        default=0
        thinned_from=434
        keys [[0,0.1008],[10,0.1008],[11,0.1004],[12,0.0993],[13,0.0973],[14,0.094],[15,0.0894],[16,0.0828],[17,0.0723],[18,0.0563],[19,0.0283],[20,-0.0088],[21,-0.0431],[22,-0.0727],[23,-0.0992],[24,-0.1158],[25,-0.1241],[26,-0.1307],[27,-0.1358],[28,-0.1395],[29,-0.142],[30,-0.1436],[31,-0.1445],[32,-0.1447],[33,-0.1445],[36,-0.1427],[40,-0.1388],[42,-0.1376],[57,-0.1362],[59,-0.1353],[60,-0.1345],[61,-0.1334],[62,-0.1309],[63,-0.1254],[64,-0.1155],[65,-0.0956],[66,-0.0704],[67,-0.0448],[68,-0.0236],[69,-0.0116],[70,-0.0041],[71,-0.0002],[72,0.0021],[73,0.0039],[74,0.0053],[75,0.0064],[76,0.0072],[78,0.0082],[92,0.0094],[94,0.01],[96,0.011],[97,0.0118],[98,0.0129],[99,0.0143],[100,0.0162],[101,0.0187],[102,0.0218],[103,0.0252],[104,0.0295],[106,0.0393],[107,0.0435],[108,0.047],[109,0.0501],[110,0.0525],[111,0.0542],[113,0.0563],[115,0.0577],[117,0.0585],[124,0.0594],[128,0.0593],[130,0.0588],[132,0.0577],[133,0.0569],[134,0.0558],[135,0.0544],[136,0.0527],[137,0.0506],[138,0.0482],[139,0.0454],[140,0.0371],[141,0.0233],[142,0.0112],[143,0.0046],[144,0.0013],[145,-0.0001],[149,-0.0027],[150,-0.0037],[156,-0.0032],[157,0.2377],[158,0.1939],[159,0.1536],[160,0.1252],[161,0.109],[162,0.0967],[163,0.0877],[166,0.1037],[167,0.1099],[168,0.1185],[169,0.1302],[170,0.1252],[171,0.118],[172,0.1155],[173,0.1157],[174,0.1154],[177,0.113],[180,0.1116],[183,0.111],[191,0.1108],[216,0.1121],[219,0.113],[220,0.1136],[221,0.1148],[222,0.117],[223,0.1202],[224,0.1246],[225,0.1323],[226,0.1419],[227,0.1479],[228,0.1502],[229,0.1522],[231,0.1553],[233,0.1573],[236,0.159],[251,0.1598],[255,0.159],[256,0.1585],[257,0.1574],[258,0.1557],[259,0.153],[260,0.1485],[262,0.1368],[263,0.1317],[264,0.1268],[265,0.1231],[266,0.1205],[267,0.1186],[269,0.116],[271,0.1141],[275,0.1118],[282,0.1095],[285,0.1089],[287,0.108],[288,0.107],[290,0.104],[291,0.1027],[292,0.1017],[293,0.1009],[294,0.1006],[295,0.1012],[296,0.1028],[298,0.1068],[299,0.1082],[301,0.1096],[304,0.1109],[322,0.1136],[325,0.1148],[326,0.1155],[329,0.1191],[330,0.12],[333,0.1217],[338,0.1229],[344,0.1234],[345,0.126],[346,0.133],[347,0.1429],[349,0.166],[350,0.1765],[351,0.1676],[352,0.1335],[353,0.0923],[354,0.0632],[355,0.0893],[356,0.1108],[357,0.1037],[358,0.0999],[359,0.1013],[360,0.1024],[361,0.1031],[362,0.1034],[363,0.1031],[364,0.102],[365,0.1007],[366,0.0998],[370,0.0984],[371,0.0977],[372,0.0967],[373,0.0953],[374,0.0937],[375,0.0918],[376,0.0895],[377,0.0869],[378,0.0832],[379,0.0787],[380,0.0748],[381,0.0732],[382,0.0748],[383,0.0789],[384,0.0839],[385,0.0882],[386,0.0916],[387,0.0943],[388,0.0965],[389,0.0981],[390,0.0991],[393,0.1009],[397,0.1024],[404,0.1033],[433,0.1009]]
       }
       spine_04_ctrl.Location.Y {
        type=MovieSceneScriptingFloatChannel
        default=0
        thinned_from=434
        keys [[0,-0.1644],[19,-0.1649],[20,-0.1638],[21,-0.1611],[22,-0.1568],[23,-0.1513],[24,-0.1465],[25,-0.143],[26,-0.1402],[27,-0.1379],[28,-0.1364],[29,-0.1353],[30,-0.1348],[32,-0.1345],[37,-0.1359],[44,-0.1386],[59,-0.14],[61,-0.1406],[62,-0.1411],[63,-0.1423],[64,-0.1445],[65,-0.1486],[67,-0.1588],[68,-0.1622],[69,-0.1637],[70,-0.1642],[108,-0.1647],[139,-0.164],[146,-0.1644],[149,-0.1636],[150,-0.162],[156,-0.1625],[157,-0.8927],[158,-1.074],[159,-1.0858],[160,-1.0744],[161,-1.0699],[163,-1.0574],[164,-1.067],[165,-1.0732],[167,-1.0817],[168,-1.0899],[169,-1.1038],[170,-1.0922],[171,-1.0783],[172,-1.0729],[344,-1.0725],[345,-1.0672],[346,-1.0529],[347,-1.0321],[350,-0.9562],[351,-0.9277],[352,-0.8922],[354,-0.817],[355,-0.6143],[356,-0.3707],[357,-0.3082],[358,-0.2979],[359,-0.3029],[360,-0.313],[361,-0.3184],[362,-0.3104],[363,-0.2899],[364,-0.258],[365,-0.2217],[366,-0.1932],[367,-0.1775],[368,-0.1698],[369,-0.168],[373,-0.1699],[382,-0.1692],[385,-0.1695],[398,-0.1644],[433,-0.1644]]
       }
       spine_04_ctrl.Location.Z {
        type=MovieSceneScriptingFloatChannel
        default=0
        thinned_from=434
        keys [[0,0.0032],[14,0.003],[17,0.0022],[19,0.0013],[20,0.0016],[23,0.004],[24,0.0045],[28,0.0053],[61,0.0051],[62,0.005],[63,0.0045],[64,0.0035],[65,0.0017],[66,0.0014],[68,0.0021],[149,0.0025],[150,-0.007],[151,-0.0085],[152,-0.0088],[156,-0.0089],[157,-0.0767],[158,-0.1006],[159,-0.0908],[160,-0.0748],[161,-0.0645],[162,-0.0541],[163,-0.0443],[164,-0.0574],[165,-0.0651],[167,-0.0747],[168,-0.083],[169,-0.0945],[170,-0.0849],[171,-0.0708],[172,-0.0645],[222,-0.0645],[225,-0.0639],[227,-0.0634],[259,-0.0632],[295,-0.065],[344,-0.0642],[345,-0.0638],[346,-0.0628],[347,-0.0612],[350,-0.0555],[351,-0.0535],[352,-0.051],[353,-0.0481],[354,-0.0446],[355,-0.0295],[356,-0.0123],[357,-0.0079],[358,-0.0072],[359,-0.0072],[360,-0.0077],[361,-0.0078],[362,-0.0065],[364,-0.0026],[365,-0.0004],[366,0.0014],[367,0.0023],[368,0.0028],[433,0.0032]]
       }
       spine_04_ctrl.Rotation.X {
        type=MovieSceneScriptingFloatChannel
        default=0
        thinned_from=434
        keys [[0,2.2771],[18,2.2645],[20,2.2598],[21,2.2474],[22,2.1995],[23,2.1235],[24,2.0396],[25,1.964],[26,1.8971],[27,1.8453],[28,1.8099],[29,1.7887],[30,1.7777],[31,1.7738],[33,1.7761],[36,1.8007],[43,1.8835],[46,1.9065],[61,1.9338],[62,1.9319],[63,1.9161],[64,1.8763],[65,1.7577],[66,1.695],[67,1.8134],[68,1.987],[69,2.1179],[70,2.1754],[71,2.1908],[72,2.1981],[75,2.1993],[96,2.17],[100,2.1788],[104,2.2165],[108,2.2608],[112,2.2848],[138,2.2903],[141,2.2802],[143,2.2591],[144,2.2419],[145,2.2177],[146,2.1831],[147,2.1329],[148,2.0593],[149,1.951],[150,1.1152],[151,0.9992],[152,0.9935],[153,1.0038],[155,1.0099],[156,1.0099],[157,-24.0502],[158,-31.4463],[159,-31.658],[160,-31.0499],[161,-30.84],[163,-30.2787],[164,-30.7833],[165,-31.0902],[167,-31.467],[168,-31.8693],[169,-32.5853],[170,-31.9409],[171,-31.1782],[172,-30.8793],[222,-30.8608],[225,-30.6731],[227,-30.4912],[258,-30.4049],[295,-31.0639],[344,-30.781],[345,-30.5518],[346,-29.9422],[347,-29.0649],[349,-26.9575],[351,-24.9857],[352,-23.9311],[353,-22.7507],[354,-21.4115],[355,-13.7219],[356,-4.9464],[357,-2.7723],[358,-2.4379],[359,-2.6133],[360,-2.9709],[361,-3.1518],[362,-2.8344],[363,-2.0874],[364,-0.9774],[365,0.2898],[366,1.2794],[367,1.8264],[368,2.0987],[369,2.1662],[370,2.149],[373,2.0678],[375,2.0292],[379,1.9683],[380,1.9579],[381,1.9546],[382,1.9603],[383,1.9741],[384,1.9949],[385,2.0213],[389,2.1577],[390,2.1836],[392,2.2228],[394,2.2504],[398,2.278],[433,2.2774]]
       }
       spine_04_ctrl.Rotation.Y {
        type=MovieSceneScriptingFloatChannel
        default=0
        thinned_from=434
        keys [[0,-2.7828],[10,-2.7828],[11,-2.7691],[12,-2.7366],[13,-2.6734],[14,-2.5692],[15,-2.4284],[16,-2.2299],[17,-1.9252],[18,-1.466],[19,-0.6155],[20,0.6164],[21,1.9085],[22,3.1764],[23,4.3558],[24,5.1545],[25,5.6178],[26,5.983],[27,6.2555],[28,6.4472],[29,6.5725],[30,6.6465],[31,6.6816],[32,6.6901],[33,6.6795],[35,6.6156],[37,6.5241],[40,6.3494],[42,6.2664],[45,6.1923],[58,6.0922],[60,6.042],[61,5.9994],[62,5.9188],[63,5.7604],[64,5.5009],[65,5.0785],[66,4.429],[67,3.428],[68,2.3857],[69,1.6459],[70,1.2068],[71,1.0228],[72,0.9226],[73,0.8558],[74,0.8101],[75,0.7801],[76,0.7644],[77,0.757],[78,0.7555],[79,0.7594],[85,0.7952],[87,0.803],[88,0.8051],[90,0.8048],[91,0.8017],[92,0.7965],[93,0.789],[94,0.7784],[95,0.7628],[96,0.7403],[97,0.7092],[98,0.6676],[99,0.6136],[100,0.5346],[101,0.4229],[102,0.285],[103,0.1269],[104,-0.0652],[105,-0.2895],[106,-0.5164],[107,-0.7162],[108,-0.8885],[109,-1.0467],[110,-1.1813],[111,-1.2829],[112,-1.3582],[113,-1.4213],[114,-1.4734],[115,-1.516],[116,-1.5501],[117,-1.5767],[118,-1.5973],[120,-1.6262],[122,-1.6454],[124,-1.6557],[125,-1.6557],[126,-1.6592],[127,-1.6592],[129,-1.6492],[130,-1.6368],[131,-1.6178],[132,-1.5906],[133,-1.5539],[134,-1.5064],[135,-1.4472],[136,-1.3748],[137,-1.2878],[138,-1.1848],[139,-1.0643],[140,-0.7575],[141,-0.2505],[142,0.2255],[143,0.5355],[144,0.7493],[145,0.9182],[146,1.084],[147,1.2831],[148,1.5174],[149,1.7953],[150,1.7063],[151,1.5015],[152,1.4219],[153,1.3771],[154,1.3544],[155,1.3445],[156,1.3445],[157,-11.7415],[158,-13.7566],[159,-13.0766],[160,-12.1246],[161,-11.5998],[162,-11.1021],[163,-10.6854],[164,-11.1988],[165,-11.5511],[167,-12.0628],[168,-12.5281],[169,-13.2775],[170,-12.6753],[171,-11.9614],[172,-11.6895],[221,-11.6838],[224,-11.7672],[226,-11.9255],[227,-11.9842],[256,-12.0904],[259,-12.0351],[264,-11.7873],[294,-11.5711],[345,-11.7565],[347,-11.714],[348,-11.651],[349,-11.5469],[350,-11.3922],[351,-10.393],[352,-8.2937],[353,-5.8917],[354,-3.9827],[355,-3.1817],[356,-3.0412],[357,-2.9178],[358,-2.9475],[359,-3.069],[360,-3.1537],[361,-3.199],[362,-3.2],[363,-3.1823],[364,-3.1187],[365,-2.9901],[367,-2.7173],[368,-2.5738],[370,-2.3075],[377,-1.3417],[378,-1.1724],[379,-0.9827],[380,-0.8281],[381,-0.7642],[382,-0.8306],[383,-0.9953],[384,-1.2068],[385,-1.4137],[387,-1.7909],[388,-1.9675],[389,-2.1235],[390,-2.2492],[391,-2.3543],[392,-2.4537],[393,-2.5452],[394,-2.6268],[395,-2.6961],[396,-2.7552],[397,-2.8071],[398,-2.8514],[399,-2.8876],[400,-2.9152],[401,-2.934],[403,-2.9538],[409,-2.9586],[424,-2.8098],[425,-2.8035],[426,-2.8035],[427,-2.7937],[428,-2.7937],[429,-2.7873],[433,-2.7873]]
       }
       spine_04_ctrl.Rotation.Z {
        type=MovieSceneScriptingFloatChannel
        default=0
        thinned_from=434
        keys [[0,0.0662],[10,0.0662],[11,0.0596],[12,0.044],[13,0.0134],[14,-0.0369],[15,-0.1127],[16,-0.2207],[17,-0.3668],[18,-0.5574],[19,-1.1363],[20,-2.2823],[21,-3.7541],[22,-5.3307],[23,-6.8146],[24,-7.9997],[25,-8.8545],[26,-9.5035],[27,-9.9701],[28,-10.2757],[29,-10.4548],[30,-10.5465],[31,-10.5802],[32,-10.585],[33,-10.5684],[35,-10.4619],[37,-10.3079],[41,-9.9251],[44,-9.688],[46,-9.5766],[59,-9.3179],[61,-9.2234],[62,-9.161],[63,-9.0389],[64,-8.7649],[65,-8.2358],[66,-7.3106],[67,-5.8724],[68,-4.1958],[69,-2.7203],[70,-1.8993],[71,-1.6369],[72,-1.5016],[73,-1.452],[74,-1.4454],[75,-1.4642],[76,-1.5068],[78,-1.6002],[80,-1.7006],[81,-1.7428],[83,-1.8108],[84,-1.8398],[85,-1.8644],[86,-1.8842],[88,-1.9117],[89,-1.9117],[90,-1.9269],[93,-1.9317],[95,-1.926],[96,-1.9174],[97,-1.903],[98,-1.8815],[99,-1.8512],[100,-1.7899],[101,-1.6833],[102,-1.5424],[103,-1.3782],[104,-1.18],[105,-0.9467],[106,-0.7029],[107,-0.4737],[108,-0.2573],[109,-0.0431],[110,0.1571],[111,0.3319],[112,0.4777],[113,0.6004],[114,0.7027],[115,0.7875],[116,0.8566],[117,0.9118],[118,0.9561],[119,0.9928],[120,1.0247],[121,1.0533],[122,1.0776],[123,1.0971],[124,1.123],[125,1.123],[126,1.1351],[128,1.1332],[129,1.1173],[130,1.1085],[131,1.094],[132,1.0723],[133,1.0419],[134,1.0027],[135,0.9544],[136,0.8955],[137,0.8247],[138,0.7405],[139,0.6417],[140,0.5285],[141,0.3664],[142,0.1016],[143,-0.2505],[144,-0.6545],[145,-1.1342],[146,-1.7142],[147,-2.4208],[148,-3.2892],[149,-4.3568],[150,-6.5747],[152,-6.08],[153,-5.8953],[154,-5.8076],[155,-5.7642],[156,-5.7642],[157,-8.1779],[158,-11.3113],[159,-13.4206],[160,-14.7521],[161,-15.4609],[162,-15.9612],[163,-16.2747],[164,-16.2235],[165,-16.0806],[166,-15.8723],[167,-15.6117],[168,-15.2932],[169,-14.9008],[170,-14.9158],[171,-15.0061],[172,-15.0235],[173,-15.0067],[175,-15.0715],[180,-15.263],[218,-15.2012],[220,-15.1404],[221,-15.0664],[222,-14.9306],[223,-14.7296],[224,-14.4598],[225,-13.9807],[226,-13.3806],[227,-13.0004],[229,-12.7308],[231,-12.5382],[233,-12.409],[236,-12.3043],[252,-12.2573],[255,-12.3005],[256,-12.3341],[257,-12.4001],[258,-12.5107],[259,-12.6779],[260,-12.9612],[262,-13.6962],[264,-14.3181],[265,-14.5529],[266,-14.7111],[268,-14.9144],[271,-15.1094],[287,-15.4855],[292,-15.8727],[294,-15.9422],[295,-15.9025],[298,-15.5592],[299,-15.4752],[304,-15.3056],[324,-15.0965],[326,-15.019],[330,-14.7408],[338,-14.5657],[344,-14.5336],[345,-14.4889],[346,-14.3649],[347,-14.1518],[348,-13.8398],[349,-13.4191],[350,-12.8799],[351,-12.129],[352,-11.0882],[353,-9.7684],[354,-8.1821],[355,-5.4438],[356,-2.4833],[357,-0.6336],[358,0.4313],[359,0.9115],[360,1.2322],[361,1.4283],[362,1.5366],[363,1.5775],[364,1.3855],[365,0.9349],[366,0.4448],[367,-0.0336],[368,-0.5515],[369,-1.0456],[370,-1.5089],[371,-1.964],[372,-2.3923],[373,-2.7751],[374,-3.109],[375,-3.4037],[376,-3.6593],[377,-3.8759],[378,-4.0613],[379,-4.2127],[380,-4.3149],[381,-4.3524],[382,-4.3115],[383,-4.1934],[384,-4.0054],[385,-3.7544],[386,-3.428],[387,-3.0344],[388,-2.6109],[389,-2.1952],[390,-1.8248],[391,-1.4946],[392,-1.1798],[393,-0.8885],[394,-0.6286],[395,-0.4083],[396,-0.2207],[397,-0.0549],[398,0.0871],[399,0.2035],[400,0.2921],[401,0.3512],[402,0.3879],[403,0.4126],[404,0.4275],[409,0.4275],[410,0.4123],[411,0.3946],[412,0.3742],[413,0.3519],[416,0.2833],[419,0.2134],[420,0.1911],[421,0.17],[422,0.1506],[423,0.1333],[424,0.1185],[425,0.1062],[426,0.1062],[427,0.0872],[428,0.0872],[429,0.0749],[433,0.0749]]
       }
       spine_04_ctrl.Scale.X {type=MovieSceneScriptingFloatChannel, default=1, const [1,434]}
       spine_04_ctrl.Scale.Y {type=MovieSceneScriptingFloatChannel, default=1, const [1,434]}
       spine_04_ctrl.Scale.Z {type=MovieSceneScriptingFloatChannel, default=1, const [1,434]}
       head_ctrl.Location.X {type=MovieSceneScriptingFloatChannel, default=0, thinned_from=434, keys [[0,0],[433,0]]}
       head_ctrl.Location.Y {type=MovieSceneScriptingFloatChannel, default=0, thinned_from=434, keys [[0,0],[433,0]]}
       head_ctrl.Location.Z {type=MovieSceneScriptingFloatChannel, default=0, thinned_from=434, keys [[0,0],[433,0]]}
       head_ctrl.Rotation.X {
        type=MovieSceneScriptingFloatChannel
        default=0
        thinned_from=434
        keys [[0,1.1142],[2,1.1142],[4,1.1016],[5,1.091],[6,1.0758],[7,1.055],[8,1.0279],[9,0.9937],[10,0.9488],[11,0.8886],[12,0.8089],[13,0.7065],[14,0.5782],[15,0.427],[16,0.2744],[17,0.1456],[18,0.059],[19,0.0116],[20,0.0073],[21,0.0434],[22,0.1003],[23,0.195],[24,0.326],[25,0.4704],[26,0.6057],[27,0.7164],[28,0.7907],[29,0.8371],[30,0.8724],[31,0.8981],[32,0.9162],[33,0.9308],[35,0.942],[36,0.9445],[37,0.9423],[38,0.9334],[39,0.9186],[40,0.8991],[41,0.8758],[44,0.7948],[45,0.7698],[46,0.7488],[48,0.7123],[50,0.6785],[52,0.6485],[54,0.6247],[55,0.6354],[56,0.6858],[57,0.7719],[58,0.8939],[59,1.0494],[60,1.2369],[62,1.6722],[63,1.8365],[64,1.9221],[65,1.8514],[66,1.7088],[67,1.5527],[68,1.4293],[69,1.3772],[70,1.3553],[71,1.3475],[72,1.3458],[81,1.3584],[86,1.3522],[87,1.3457],[89,1.3457],[90,1.3325],[91,1.3318],[92,1.3232],[93,1.336],[94,1.3764],[95,1.4328],[96,1.4962],[97,1.5535],[98,1.5959],[99,1.612],[102,1.615],[104,1.5937],[105,1.586],[106,1.5599],[107,1.4898],[108,1.3425],[109,1.137],[110,0.9235],[111,0.7521],[112,0.6369],[113,0.5559],[114,0.5045],[115,0.478],[116,0.4584],[117,0.4328],[118,0.4048],[119,0.3777],[120,0.3551],[121,0.3361],[122,0.3178],[123,0.3009],[124,0.2853],[125,0.2853],[126,0.2658],[127,0.2641],[128,0.2699],[129,0.2806],[130,0.292],[131,0.3026],[132,0.3086],[133,0.3076],[134,0.2714],[135,0.1364],[136,-0.1539],[137,-0.657],[138,-1.4306],[139,-2.5324],[140,-4.6716],[141,-7.9404],[142,-11.4563],[143,-15.4824],[144,-19.201],[145,-21.2784],[146,-22.2982],[147,-22.7354],[148,-22.5854],[149,-21.4962],[150,-8.8919],[151,-2.5539],[152,-2.8205],[153,-4.5409],[154,-6.382],[155,-8.4771],[156,-10.8203],[157,-37.3852],[158,-61.9973],[159,-69.961],[160,-73.4405],[161,-73.003],[162,-70.1098],[163,-66.4545],[164,-64.8411],[166,-66.0584],[167,-66.5279],[168,-66.739],[169,-67.1518],[170,-66.8512],[171,-67.4709],[173,-70.4556],[174,-71.2606],[175,-71.661],[177,-71.6898],[182,-70.6445],[187,-68.8331],[193,-67.9278],[219,-66.6995],[224,-65.9274],[228,-65.9778],[256,-67.9946],[257,-67.7542],[258,-67.0563],[259,-65.8548],[260,-63.857],[262,-57.85],[263,-55.3079],[264,-53.1929],[265,-51.7404],[266,-50.8326],[267,-50.1876],[268,-49.7647],[269,-49.5152],[275,-48.8393],[282,-48.8025],[284,-49.0223],[285,-49.2941],[286,-49.7291],[287,-50.3677],[288,-51.5617],[289,-53.404],[291,-57.6426],[292,-59.9338],[293,-62.029],[294,-63.4298],[295,-64.3977],[297,-65.8781],[298,-66.4405],[304,-68.5839],[311,-69.6788],[343,-68.9834],[344,-68.8691],[345,-68.5294],[346,-67.7894],[347,-66.6721],[348,-65.1991],[349,-63.3925],[350,-61.2755],[351,-58.2648],[352,-54.2239],[353,-49.8456],[354,-45.8137],[355,-33.8299],[356,-18.4622],[357,-9.8988],[358,-4.7629],[359,-3.475],[360,-4.5734],[361,-5.8758],[362,-6.3076],[363,-6.015],[364,-4.3718],[365,-1.8553],[366,0.164],[367,1.2412],[368,1.8616],[369,2.0445],[370,1.7124],[371,0.946],[372,0.0032],[373,-0.8581],[374,-1.6657],[375,-2.5202],[376,-3.2729],[377,-3.7906],[378,-4.0066],[379,-3.9952],[380,-3.8506],[381,-3.6667],[382,-3.4488],[383,-3.1542],[384,-2.8046],[385,-2.4221],[386,-2.0231],[387,-1.6165],[388,-1.2153],[389,-0.833],[390,-0.4832],[391,-0.1658],[392,0.1273],[393,0.3945],[394,0.6339],[395,0.8438],[396,1.0276],[397,1.1905],[398,1.3336],[399,1.458],[400,1.5648],[401,1.6551],[402,1.7267],[403,1.7784],[404,1.8131],[405,1.8345],[406,1.8449],[407,1.8475],[408,1.8413],[409,1.8248],[410,1.8],[411,1.769],[416,1.582],[418,1.4947],[421,1.3576],[422,1.3159],[423,1.2787],[424,1.2473],[425,1.2204],[426,1.1961],[427,1.1753],[428,1.1569],[429,1.142],[430,1.1298],[431,1.121],[433,1.121]]
       }
       head_ctrl.Rotation.Y {
        type=MovieSceneScriptingFloatChannel
        default=0
        thinned_from=434
        keys [[0,2.4337],[6,2.4256],[9,2.4064],[10,2.3935],[11,2.3983],[12,2.4366],[13,2.5341],[14,2.714],[15,2.9995],[16,3.3787],[17,3.8244],[18,4.3108],[19,5.0043],[20,5.9179],[21,6.7831],[22,7.591],[23,8.306],[24,8.6078],[25,8.5737],[27,8.4469],[28,8.4116],[36,8.2805],[38,8.2862],[41,8.3597],[52,8.6992],[55,8.7403],[56,8.7111],[57,8.6245],[58,8.4529],[59,8.1718],[60,7.7564],[61,7.1818],[62,6.4034],[63,5.4456],[64,4.386],[66,2.3575],[67,1.2308],[68,0.1143],[69,-1.0116],[70,-2.0952],[71,-2.9617],[72,-3.5598],[73,-3.9628],[74,-4.2182],[75,-4.3451],[76,-4.3764],[78,-4.3873],[81,-4.3798],[82,-4.3895],[85,-4.3823],[86,-4.3857],[87,-4.3785],[89,-4.3855],[90,-4.3779],[91,-4.3858],[92,-4.3819],[94,-4.4006],[96,-4.4598],[98,-4.5569],[100,-4.6847],[101,-4.7601],[102,-4.853],[103,-4.9718],[104,-5.1255],[105,-5.3142],[107,-5.7428],[108,-6.002],[109,-6.2937],[110,-6.5222],[111,-6.5926],[113,-6.5252],[115,-6.4978],[125,-6.5575],[132,-6.6632],[133,-6.6694],[134,-6.6508],[135,-6.5796],[136,-6.4472],[137,-6.2494],[138,-5.9863],[139,-5.6616],[140,-5.1178],[141,-4.4006],[142,-3.7581],[143,-3.0799],[144,-2.7971],[145,-3.2286],[146,-3.7146],[147,-3.9433],[148,-4.1072],[149,-4.2053],[150,-5.1594],[151,-5.5106],[152,-5.6234],[153,-5.6726],[154,-5.7032],[155,-5.7164],[156,-5.7123],[157,-4.2498],[158,-3.0604],[159,-2.3345],[160,-1.7842],[161,-1.2285],[162,-0.4697],[163,0.5048],[164,1.2142],[165,1.6742],[166,2.1744],[167,2.6346],[168,2.9404],[169,2.8888],[170,2.7957],[171,2.467],[172,1.9412],[173,1.5],[174,1.2961],[175,1.2274],[176,1.2509],[177,1.3115],[178,1.4006],[179,1.5364],[181,1.8554],[182,1.9922],[183,2.1232],[185,2.417],[186,2.5546],[187,2.6667],[188,2.7386],[189,2.7816],[190,2.8167],[191,2.8445],[192,2.8659],[193,2.8816],[196,2.9011],[206,2.9016],[207,2.909],[208,2.9596],[209,3.0918],[210,3.345],[211,3.7575],[212,4.3683],[213,5.2162],[214,6.3419],[215,8.0457],[216,10.4243],[217,13.2057],[218,16.1137],[219,18.8728],[220,21.2077],[221,22.7133],[222,23.3746],[223,23.5483],[224,23.3961],[225,22.778],[226,21.7913],[227,20.9588],[228,20.4775],[229,20.1641],[230,19.9744],[231,19.8638],[233,19.7639],[249,19.6446],[250,19.5925],[251,19.5],[252,19.3543],[253,19.1459],[254,18.8624],[255,18.4932],[256,16.9115],[257,13.544],[258,9.2242],[259,4.7891],[260,-0.0825],[261,-5.2209],[262,-9.2919],[263,-12.0661],[264,-14.0372],[265,-15.2653],[266,-16.0927],[267,-16.7969],[268,-17.3309],[269,-17.6398],[273,-18.202],[275,-18.365],[278,-18.4489],[283,-18.443],[285,-18.3331],[286,-18.2113],[287,-18.0261],[288,-17.4598],[289,-16.2292],[290,-14.3814],[291,-12.0289],[292,-8.4027],[293,-3.7296],[294,-0.0059],[295,2.2161],[296,3.9084],[297,5.1497],[298,5.9935],[299,6.5006],[300,6.8457],[301,7.1606],[302,7.4443],[303,7.696],[304,7.9151],[305,8.1011],[306,8.2538],[307,8.3735],[308,8.4626],[309,8.5255],[310,8.5629],[311,8.5794],[313,8.5572],[315,8.4819],[317,8.3647],[318,8.2703],[319,8.1535],[320,8.014],[321,7.8557],[323,7.4898],[325,7.0755],[326,6.8241],[327,6.5245],[328,6.2099],[329,5.9135],[330,5.6685],[331,5.4685],[332,5.2915],[333,5.1469],[334,5.044],[335,4.9921],[336,4.9921],[337,5.0184],[338,5.0745],[339,5.1501],[342,5.4149],[343,5.4901],[344,5.5478],[345,5.5482],[346,5.4787],[347,5.3677],[348,5.2442],[349,5.1382],[350,5.0814],[351,5.3824],[352,6.0292],[353,6.54],[354,6.3867],[355,4.5239],[356,2.5631],[357,2.262],[358,2.0205],[359,1.9449],[360,1.969],[361,2.1361],[362,2.5512],[363,3.0333],[364,3.4578],[365,3.8491],[366,4.0759],[367,4.099],[368,3.9925],[369,3.8246],[370,3.6319],[372,3.2091],[373,3.0148],[374,2.8111],[375,2.6156],[376,2.5129],[377,2.4981],[378,2.5171],[379,2.5664],[380,2.6225],[381,2.6632],[382,2.6747],[383,2.6638],[386,2.5876],[387,2.5701],[389,2.55],[395,2.5098],[400,2.4622],[410,2.4138],[418,2.4089],[425,2.4272],[426,2.4233],[427,2.43],[428,2.4285],[433,2.432]]
       }
       head_ctrl.Rotation.Z {
        type=MovieSceneScriptingFloatChannel
        default=0
        thinned_from=434
        keys [[0,4.4422],[2,4.4422],[3,4.4028],[4,4.3488],[5,4.2598],[6,4.127],[7,3.9416],[8,3.695],[9,3.3782],[10,2.8161],[11,1.8126],[12,0.3519],[13,-1.5925],[14,-4.0443],[15,-7.547],[16,-12.064],[17,-16.7484],[18,-20.753],[19,-23.6178],[20,-25.6415],[21,-27.1209],[22,-28.0421],[23,-28.4284],[25,-28.8271],[31,-29.0374],[34,-29.1255],[41,-29.9578],[55,-30.2201],[56,-30.2015],[57,-30.0894],[58,-29.8771],[59,-29.5184],[60,-28.9767],[61,-28.2151],[62,-26.3218],[63,-23.0416],[64,-19.322],[65,-14.9345],[66,-10.3751],[67,-6.5288],[68,-3.4792],[69,-1.3676],[70,0.2713],[71,1.5014],[72,2.3022],[73,2.6933],[74,2.7995],[75,2.746],[76,2.6165],[78,2.3334],[80,2.0311],[81,1.8979],[82,1.7885],[83,1.6633],[84,1.5575],[85,1.4662],[86,1.4049],[87,1.3606],[88,1.3127],[89,1.2963],[91,1.2492],[92,1.2362],[93,1.2382],[94,1.2561],[95,1.3116],[96,1.4033],[97,1.5856],[98,1.8391],[99,2.1924],[100,3.4421],[101,5.9001],[102,8.8563],[103,11.6008],[104,14.1685],[105,16.8625],[106,19.3756],[107,21.4017],[108,22.8936],[109,24.0035],[110,24.725],[111,25.0518],[115,25.8304],[118,26.1938],[121,26.3581],[130,26.2952],[132,26.1104],[133,25.9711],[134,25.6199],[135,24.9367],[136,23.9954],[137,22.8689],[139,20.3494],[140,18.8354],[141,16.903],[142,14.6418],[143,11.6391],[144,8.3832],[145,5.4829],[146,3.3801],[147,2.2741],[148,1.4829],[149,0.9343],[150,0.7678],[151,0.3724],[152,0.1182],[153,-0.0421],[154,-0.1369],[155,-0.1864],[156,-0.2139],[157,-2.7134],[158,-5.2523],[159,-6.4918],[160,-7.3665],[161,-8.0052],[162,-8.5133],[163,-8.9523],[164,-9.3372],[165,-9.7531],[166,-10.1954],[167,-10.5499],[168,-10.728],[169,-10.7589],[172,-10.5316],[173,-10.5073],[174,-10.5652],[175,-10.746],[178,-11.4851],[181,-12.2997],[183,-12.7033],[185,-13.0522],[186,-13.1827],[187,-13.2701],[188,-13.2992],[199,-13.3277],[200,-13.3131],[208,-13.3358],[210,-13.2486],[211,-13.144],[212,-12.9771],[213,-12.7348],[214,-12.3861],[215,-11.8655],[216,-11.1172],[218,-9.3295],[219,-8.4626],[220,-7.7479],[221,-7.3272],[222,-7.2122],[223,-7.2642],[224,-7.4132],[225,-7.6988],[226,-8.0736],[227,-8.3527],[228,-8.4934],[229,-8.5775],[230,-8.6214],[232,-8.6479],[236,-8.6195],[247,-8.66],[249,-8.6465],[250,-8.6552],[251,-8.6968],[252,-8.8066],[253,-8.9977],[254,-9.283],[255,-9.6944],[256,-10.8898],[257,-13.1397],[258,-15.8699],[259,-18.5619],[260,-21.362],[261,-24.5752],[262,-27.7215],[263,-30.5215],[264,-33.0039],[265,-34.8781],[266,-36.1633],[267,-37.1042],[268,-37.7601],[269,-38.1903],[271,-38.6963],[273,-39.0177],[278,-39.2846],[283,-39.253],[284,-39.1759],[285,-39.0042],[286,-38.7257],[287,-38.2826],[288,-37.3248],[289,-35.696],[290,-33.678],[291,-31.5301],[292,-28.6892],[293,-25.5853],[294,-23.6282],[295,-22.7683],[296,-22.1399],[297,-21.7352],[300,-20.894],[302,-20.4369],[309,-19.2716],[315,-18.3315],[318,-17.9333],[321,-17.6341],[324,-17.2479],[327,-16.7576],[333,-15.9625],[334,-15.8679],[335,-15.8334],[336,-15.8334],[337,-15.8688],[341,-16.1281],[343,-16.1695],[344,-16.1167],[345,-15.9881],[346,-15.7534],[347,-15.4355],[348,-15.0611],[350,-14.2426],[351,-13.8702],[352,-13.4359],[353,-12.7813],[354,-11.8318],[355,-8.9794],[356,-7.0491],[357,-6.1819],[358,-4.9626],[359,-3.6638],[360,-2.2208],[361,-0.8402],[362,0.3839],[363,1.6048],[364,3.0622],[365,4.6606],[366,6.0339],[367,7.048],[368,7.8324],[369,8.4138],[370,8.8423],[371,9.1554],[372,9.362],[373,9.4714],[374,9.486],[375,9.4223],[376,9.3071],[377,9.1621],[378,8.9788],[379,8.7479],[381,8.2211],[382,7.924],[383,7.5796],[385,6.8446],[386,6.4652],[388,5.6616],[389,5.2896],[390,4.9722],[391,4.7126],[392,4.4934],[393,4.3097],[394,4.1564],[395,4.0287],[396,3.9225],[397,3.8368],[398,3.771],[399,3.7248],[400,3.6977],[401,3.6893],[402,3.696],[403,3.7126],[404,3.717],[405,3.7512],[407,3.8013],[408,3.8412],[409,3.9017],[410,3.9208],[411,3.9657],[412,3.9898],[421,4.2517],[422,4.2758],[423,4.3085],[426,4.3768],[427,4.3892],[428,4.4162],[430,4.4373],[431,4.4558],[433,4.4558]]
       }
       head_ctrl.Scale.X {type=MovieSceneScriptingFloatChannel, default=1, const [1,434]}
       head_ctrl.Scale.Y {type=MovieSceneScriptingFloatChannel, default=1, const [1,434]}
       head_ctrl.Scale.Z {type=MovieSceneScriptingFloatChannel, default=1, const [1,434]}
       clavicle_l_ctrl.Location.X {type=MovieSceneScriptingFloatChannel, default=0, thinned_from=434, keys [[0,0],[433,0]]}
       clavicle_l_ctrl.Location.Y {type=MovieSceneScriptingFloatChannel, default=0, thinned_from=434, keys [[0,0],[433,0]]}
       clavicle_l_ctrl.Location.Z {type=MovieSceneScriptingFloatChannel, default=0, thinned_from=434, keys [[0,0],[433,0]]}
       clavicle_l_ctrl.Rotation.X {
        type=MovieSceneScriptingFloatChannel
        default=0
        thinned_from=434
        keys [[0,0],[144,0.0001],[145,-0.0001],[176,0],[365,0],[366,-0.0001],[367,0.0001],[370,0.0001],[371,-0.0001],[372,0.0001],[433,0]]
       }
       clavicle_l_ctrl.Rotation.Y {
        type=MovieSceneScriptingFloatChannel
        default=0
        thinned_from=434
        keys [[0,0],[14,0],[15,-0.0063],[18,-0.0083],[20,-0.009],[21,-0.0091],[22,-0.0088],[23,-0.008],[24,-0.0067],[25,-0.005],[27,-0.0008],[28,0.0008],[29,0.002],[30,0.0028],[31,0.0035],[32,0.0037],[33,0.0033],[35,0.0012],[40,-0.0052],[42,-0.0071],[43,-0.0077],[45,-0.0077],[46,-0.009],[47,-0.009],[48,0],[62,0],[63,-0.0085],[64,-0.0071],[65,-0.0047],[66,-0.0016],[68,0.0054],[70,0.0116],[71,0.0137],[73,0.0162],[74,0.0173],[75,0.018],[76,0.0185],[77,0.0187],[78,0.0184],[79,0.0178],[81,0.0178],[82,0],[103,-0.0001],[104,0.0001],[105,0.0005],[106,0.0013],[107,0.0025],[108,0.0042],[109,0.0063],[110,0.0088],[111,0.0088],[112,0.015],[114,0.0228],[115,0.0265],[120,0.0428],[121,0.0455],[122,0.0473],[125,0.0508],[129,0.0508],[130,0.069],[131,0.069],[132,0.0878],[133,0.0878],[134,0.1218],[135,0.1486],[136,0.1844],[137,0.231],[138,0.29],[139,0.3634],[140,0.4846],[141,0.6865],[142,0.9691],[143,1.4828],[144,1.9785],[145,2.266],[146,2.5367],[147,2.7636],[148,2.9195],[149,2.9775],[150,1.9862],[151,-1.2424],[152,-2.1065],[153,-2.2624],[154,-2.2403],[155,-2.0858],[156,-1.6665],[157,5.2006],[158,6.1423],[159,5.5743],[160,4.2706],[161,2.329],[162,0.256],[163,-1.4752],[164,-2.6682],[165,-3.1992],[166,-3.5111],[167,-3.6584],[168,-3.6955],[169,-3.3915],[170,-2.7465],[171,-1.9389],[172,-1.0391],[173,-0.2857],[174,0.3244],[175,0.886],[176,1.2975],[177,1.4573],[178,1.406],[179,1.2755],[180,1.101],[181,0.9174],[182,0.7601],[183,0.6223],[184,0.4789],[185,0.3375],[186,0.2055],[187,0.0902],[188,0],[189,-0.0722],[190,-0.1342],[191,-0.1869],[192,-0.23],[193,-0.2636],[194,-0.2876],[195,-0.2876],[196,-0.3069],[214,-0.3069],[215,-0.3468],[216,-0.4524],[217,-0.6013],[219,-0.9404],[220,-1.0863],[221,-1.2344],[222,-1.3951],[223,-1.5239],[224,-1.5768],[225,-1.4472],[226,-1.1607],[227,-0.8702],[228,-0.6027],[229,-0.3164],[230,-0.0613],[231,0.1123],[232,0.2139],[233,0.2891],[234,0.3419],[235,0.3761],[236,0.3957],[256,0.3957],[257,0.3693],[258,0.3052],[259,0.1792],[260,-0.1166],[261,-0.5368],[262,-0.8729],[263,-1.0489],[264,-1.1408],[265,-1.1673],[266,-1.1312],[267,-1.0421],[268,-0.9293],[269,-0.8217],[270,-0.7182],[272,-0.493],[273,-0.3909],[274,-0.309],[275,-0.2435],[276,-0.1844],[277,-0.1321],[278,-0.087],[279,-0.0497],[280,-0.0205],[281,0],[282,0],[283,0.0247],[285,0.0247],[286,0.0406],[296,0.0406],[297,0.0205],[298,0],[301,0],[302,-0.0171],[303,-0.0171],[304,-0.0393],[305,-0.0393],[306,-0.0606],[307,-0.0606],[308,-0.0808],[309,-0.0808],[310,-0.1018],[311,-0.1018],[312,-0.1221],[313,-0.1221],[314,-0.1404],[316,-0.1404],[317,-0.1613],[320,-0.1613],[321,-0.1771],[330,-0.1771],[331,-0.1553],[332,-0.1553],[333,-0.135],[334,-0.135],[335,-0.1101],[336,-0.1101],[337,-0.0871],[338,-0.0871],[339,-0.0682],[340,-0.0682],[341,-0.0505],[342,-0.0505],[343,-0.0337],[344,-0.0337],[345,-0.0172],[346,0],[358,0],[359,0.0641],[360,0.2193],[361,0.5492],[362,1.0384],[363,1.5837],[364,2.1224],[365,2.5782],[366,2.9413],[367,3.1942],[368,3.3839],[369,3.4587],[370,3.385],[371,3.1945],[372,2.9319],[373,2.5298],[375,1.5515],[376,1.0568],[377,0.6479],[378,0.3551],[379,0.1482],[380,0.0271],[381,-0.0428],[382,-0.0978],[383,-0.1395],[384,-0.1697],[385,-0.1902],[386,-0.1902],[387,-0.2087],[389,-0.2087],[390,-0.1923],[391,-0.1748],[392,-0.1561],[393,-0.1355],[394,-0.1116],[395,-0.0871],[396,-0.0648],[397,-0.0473],[398,-0.0473],[399,-0.0224],[400,0],[433,0]]
       }
       clavicle_l_ctrl.Rotation.Z {
        type=MovieSceneScriptingFloatChannel
        default=0
        thinned_from=434
        keys [[0,0],[14,0],[15,0.045],[16,0.2304],[17,0.6312],[18,1.3227],[19,2.7114],[20,4.5724],[21,6.0336],[22,6.7539],[23,7.0745],[24,7.1475],[25,7.0097],[26,6.6543],[27,6.1686],[28,5.6398],[29,5.0441],[30,4.3508],[31,3.6363],[32,2.9769],[33,2.3539],[34,1.7395],[35,1.197],[36,0.7894],[37,0.5218],[38,0.3421],[39,0.2278],[40,0.1566],[41,0.1061],[42,0.0684],[43,0.0473],[45,0.0473],[46,0.0265],[47,0.0265],[48,0],[62,0],[63,-0.0508],[64,-0.1],[65,-0.1599],[66,-0.2364],[67,-0.3357],[68,-0.4259],[69,-0.4868],[70,-0.5335],[71,-0.5523],[72,-0.5274],[73,-0.4659],[74,-0.3875],[75,-0.2818],[76,-0.1591],[77,-0.0712],[78,-0.03],[79,-0.0089],[81,-0.0089],[82,0],[99,0],[100,-0.0969],[101,-0.3377],[102,-0.6474],[103,-0.9509],[104,-1.2657],[105,-1.6194],[106,-1.9518],[107,-2.2027],[108,-2.3517],[109,-2.4322],[110,-2.4653],[111,-2.4653],[112,-2.3612],[113,-2.0955],[114,-1.773],[115,-1.492],[116,-1.2571],[117,-1.0177],[118,-0.789],[119,-0.5863],[120,-0.4249],[121,-0.2992],[122,-0.1941],[123,-0.1107],[124,-0.0499],[125,-0.0126],[129,-0.0126],[130,0],[149,0],[160,0.0001],[161,-0.0001],[358,0],[359,0.0509],[360,0.1725],[361,0.4567],[362,0.8174],[363,1.0942],[364,1.3336],[365,1.4714],[366,1.5627],[367,1.6131],[368,1.6407],[369,1.6491],[370,1.6172],[371,1.5316],[372,1.4065],[373,1.2012],[375,0.6969],[376,0.4423],[377,0.2407],[378,0.1117],[379,0.0356],[380,0.0074],[381,0.0032],[382,0.0009],[383,0.0001],[433,0]]
       }
       clavicle_l_ctrl.Scale.X {type=MovieSceneScriptingFloatChannel, default=1, const [1,434]}
       clavicle_l_ctrl.Scale.Y {type=MovieSceneScriptingFloatChannel, default=1, const [1,434]}
       clavicle_l_ctrl.Scale.Z {type=MovieSceneScriptingFloatChannel, default=1, const [1,434]}
       clavicle_r_ctrl.Location.X {type=MovieSceneScriptingFloatChannel, default=0, thinned_from=434, keys [[0,0],[433,0]]}
       clavicle_r_ctrl.Location.Y {type=MovieSceneScriptingFloatChannel, default=0, thinned_from=434, keys [[0,0],[433,0]]}
       clavicle_r_ctrl.Location.Z {type=MovieSceneScriptingFloatChannel, default=0, thinned_from=434, keys [[0,-0.0001],[433,-0.0001]]}
       clavicle_r_ctrl.Rotation.X {
        type=MovieSceneScriptingFloatChannel
        default=0
        thinned_from=434
        keys [[0,0],[146,0],[147,0.0002],[149,-0.0001],[152,0.0001],[153,-0.0001],[352,0],[353,0.0001],[354,-0.0002],[433,0]]
       }
       clavicle_r_ctrl.Rotation.Y {
        type=MovieSceneScriptingFloatChannel
        default=0
        thinned_from=434
        keys [[0,-6.48],[11,-6.48],[12,-6.4411],[13,-6.3878],[14,-6.3],[15,-6.1249],[16,-5.8164],[17,-5.3759],[18,-4.8046],[19,-3.9077],[20,-2.7777],[21,-1.8462],[22,-1.2636],[23,-0.8795],[24,-0.675],[25,-0.5614],[26,-0.462],[27,-0.3757],[28,-0.3016],[29,-0.2387],[30,-0.1861],[31,-0.1428],[32,-0.1078],[33,-0.0802],[34,-0.0589],[35,-0.043],[36,-0.0317],[37,-0.0239],[38,-0.0192],[39,-0.0168],[49,-0.0168],[50,-0.0329],[52,-0.0329],[53,-0.0564],[54,-0.0564],[55,-0.0791],[56,-0.1064],[57,-0.1569],[58,-0.2399],[59,-0.3651],[60,-0.5415],[61,-0.779],[62,-1.1494],[63,-1.6841],[64,-2.331],[65,-3.0378],[66,-3.7525],[67,-4.7496],[68,-5.6871],[69,-6.1744],[70,-6.4903],[71,-6.6617],[72,-6.7527],[73,-6.8137],[74,-6.8507],[76,-6.8767],[78,-6.8767],[82,-6.755],[86,-6.6055],[90,-6.4999],[91,-6.4999],[92,-6.4798],[129,-6.48],[132,-6.4256],[133,-6.3936],[134,-6.3185],[135,-6.1683],[136,-5.9434],[137,-5.645],[138,-5.2735],[139,-4.8301],[140,-4.1859],[141,-3.2502],[142,-2.0801],[143,-0.2441],[144,1.5106],[145,2.2021],[146,2.341],[147,2.3175],[148,2.1529],[149,1.7062],[150,-6.3368],[151,-2.5227],[152,1.9104],[153,6.2168],[154,10.0241],[155,13.6365],[156,17.132],[157,20.4908],[158,23.5333],[159,26.702],[160,28.4553],[161,25.5068],[162,20.9863],[163,17.8376],[164,15.879],[165,15.3413],[166,15.1433],[167,15.115],[168,15.6315],[169,16.8632],[171,19.7903],[172,21.1366],[173,22.1229],[174,22.7287],[175,23.1212],[176,23.3329],[177,23.3963],[178,23.2966],[179,23.0364],[183,21.4763],[186,20.0953],[187,19.7008],[188,19.4012],[190,18.9963],[192,18.7278],[194,18.5758],[196,18.5201],[212,18.4497],[215,18.3421],[217,18.1531],[218,17.9953],[219,17.7829],[220,17.5066],[221,17.058],[223,15.7925],[224,15.2791],[225,14.9579],[226,14.7782],[227,14.7222],[228,14.7569],[231,15.0096],[238,15.3241],[241,15.3881],[256,15.5657],[257,15.6089],[258,15.7217],[259,15.9543],[260,16.5503],[261,17.4475],[262,18.2385],[263,18.8241],[264,19.3033],[265,19.5662],[266,19.6616],[267,19.7142],[269,19.7022],[271,19.5536],[273,19.2918],[277,18.649],[280,18.2803],[282,18.1039],[289,17.9017],[295,17.8855],[319,18.8562],[338,18.9125],[341,18.8148],[343,18.6706],[345,18.4545],[346,18.287],[347,17.9528],[348,17.3423],[349,16.346],[350,14.8542],[351,12.6781],[352,9.8676],[353,6.6712],[354,3.3374],[355,0.1148],[356,-2.748],[357,-4.8006],[358,-5.9464],[359,-6.4052],[360,-6.4708],[361,-6.1763],[362,-5.6294],[363,-5.3349],[364,-5.7269],[365,-6.5813],[366,-7.4151],[367,-8.2011],[368,-8.9664],[369,-9.3106],[370,-9.1941],[371,-8.9023],[374,-7.7508],[376,-6.8807],[377,-6.4777],[378,-6.1032],[379,-5.7417],[380,-5.4182],[381,-5.1576],[382,-4.9634],[383,-4.8234],[384,-4.7386],[385,-4.7102],[386,-4.7102],[388,-4.7558],[389,-4.7995],[390,-4.8625],[391,-4.95],[392,-5.06],[395,-5.4476],[396,-5.5859],[399,-6.0399],[400,-6.1704],[401,-6.2735],[402,-6.3486],[403,-6.4035],[404,-6.4413],[405,-6.4647],[406,-6.4647],[407,-6.48],[433,-6.48]]
       }
       clavicle_r_ctrl.Rotation.Z {
        type=MovieSceneScriptingFloatChannel
        default=0
        thinned_from=434
        keys [[0,0],[14,0],[15,-0.0147],[16,-0.1174],[17,-0.3963],[18,-0.9393],[19,-3.0092],[20,-6.1768],[21,-7.9928],[22,-8.2205],[23,-8.3044],[24,-8.3164],[25,-7.8457],[26,-6.7169],[27,-5.3547],[28,-4.184],[29,-3.1898],[30,-2.188],[31,-1.3509],[32,-0.8501],[33,-0.6044],[34,-0.4129],[35,-0.269],[36,-0.166],[37,-0.0968],[38,-0.055],[39,-0.0334],[49,-0.0334],[50,-0.0244],[59,-0.0242],[61,-0.0238],[62,-0.0081],[63,0.1015],[64,0.399],[65,1.3933],[66,2.6717],[67,3.7652],[68,4.3631],[69,3.9416],[70,3.032],[71,2.1669],[72,1.4855],[73,0.8484],[74,0.3699],[75,0.1135],[76,0.0156],[78,0.0156],[79,0],[99,0],[100,0.0755],[101,0.2639],[102,0.5081],[103,0.751],[104,1.0382],[105,1.3798],[106,1.6656],[107,1.7854],[108,1.7485],[109,1.6498],[110,1.5072],[111,1.3384],[112,1.1173],[113,0.8447],[114,0.5789],[115,0.3783],[116,0.2438],[117,0.138],[118,0.0618],[119,0.0155],[129,0.0155],[130,0],[144,0],[145,0.0371],[146,0.1682],[147,0.4226],[148,0.8361],[149,1.6292],[150,3.6777],[151,5.991],[152,6.6334],[153,6.7616],[214,6.7616],[215,6.8085],[216,6.9476],[217,7.1768],[218,7.4936],[219,7.896],[220,8.3816],[221,9.1253],[222,10.1419],[223,11.1897],[224,12.0266],[225,12.516],[226,12.7739],[227,12.9379],[228,13.0604],[229,13.121],[251,13.2395],[256,13.0676],[257,12.9969],[258,12.8124],[259,12.4316],[260,11.301],[261,9.5629],[262,8.1791],[263,7.5393],[264,7.2538],[265,7.116],[266,7.02],[267,6.9782],[269,6.9388],[273,6.8237],[280,6.7623],[344,6.7616],[345,6.7286],[346,6.634],[347,6.4847],[348,6.2873],[349,6.0487],[350,5.7756],[351,5.4544],[352,5.0492],[353,4.5336],[354,3.8805],[355,2.7596],[356,1.393],[357,0.0859],[358,-1.0055],[359,-1.6767],[360,-2.1661],[361,-2.5567],[362,-2.7956],[363,-2.8501],[364,-2.7896],[365,-2.6265],[366,-2.389],[367,-2.0534],[368,-1.6432],[369,-1.257],[370,-0.8756],[371,-0.4638],[372,-0.1342],[373,0.0006],[433,0]]
       }
       clavicle_r_ctrl.Scale.X {type=MovieSceneScriptingFloatChannel, default=1, const [1,434]}
       clavicle_r_ctrl.Scale.Y {type=MovieSceneScriptingFloatChannel, default=1, const [1,434]}
       clavicle_r_ctrl.Scale.Z {type=MovieSceneScriptingFloatChannel, default=1, const [1,434]}
       neck_01_ctrl.Location.X {type=MovieSceneScriptingFloatChannel, default=0, const [0,434]}
       neck_01_ctrl.Location.Y {type=MovieSceneScriptingFloatChannel, default=0, const [0,434]}
       neck_01_ctrl.Location.Z {type=MovieSceneScriptingFloatChannel, default=0, const [0,434]}
       neck_01_ctrl.Rotation.X {type=MovieSceneScriptingFloatChannel, default=0, const [0,434]}
       neck_01_ctrl.Rotation.Y {type=MovieSceneScriptingFloatChannel, default=0, const [0,434]}
       neck_01_ctrl.Rotation.Z {type=MovieSceneScriptingFloatChannel, default=0, const [0,434]}
       neck_01_ctrl.Scale.X {type=MovieSceneScriptingFloatChannel, default=1, const [1,434]}
       neck_01_ctrl.Scale.Y {type=MovieSceneScriptingFloatChannel, default=1, const [1,434]}
       neck_01_ctrl.Scale.Z {type=MovieSceneScriptingFloatChannel, default=1, const [1,434]}
       neck_02_ctrl.Location.X {type=MovieSceneScriptingFloatChannel, default=0, const [0,434]}
       neck_02_ctrl.Location.Y {type=MovieSceneScriptingFloatChannel, default=0, const [0,434]}
       neck_02_ctrl.Location.Z {type=MovieSceneScriptingFloatChannel, default=0, const [0,434]}
       neck_02_ctrl.Rotation.X {type=MovieSceneScriptingFloatChannel, default=0, const [0,434]}
       neck_02_ctrl.Rotation.Y {type=MovieSceneScriptingFloatChannel, default=0, const [0,434]}
       neck_02_ctrl.Rotation.Z {type=MovieSceneScriptingFloatChannel, default=0, const [0,434]}
       neck_02_ctrl.Scale.X {type=MovieSceneScriptingFloatChannel, default=1, const [1,434]}
       neck_02_ctrl.Scale.Y {type=MovieSceneScriptingFloatChannel, default=1, const [1,434]}
       neck_02_ctrl.Scale.Z {type=MovieSceneScriptingFloatChannel, default=1, const [1,434]}
       head_ik_ctrl.Location.X {type=MovieSceneScriptingFloatChannel, default=0, thinned_from=434, keys [[0,0],[433,0]]}
       head_ik_ctrl.Location.Y {type=MovieSceneScriptingFloatChannel, default=0, thinned_from=434, keys [[0,0],[433,0]]}
       head_ik_ctrl.Location.Z {type=MovieSceneScriptingFloatChannel, default=0, thinned_from=434, keys [[0,0],[433,0]]}
       head_ik_ctrl.Rotation.X {
        type=MovieSceneScriptingFloatChannel
        default=0
        thinned_from=434
        keys [[0,1.1142],[2,1.1142],[4,1.1016],[5,1.091],[6,1.0758],[7,1.055],[8,1.0279],[9,0.9937],[10,0.9488],[11,0.8886],[12,0.8089],[13,0.7065],[14,0.5782],[15,0.427],[16,0.2744],[17,0.1456],[18,0.059],[19,0.0116],[20,0.0073],[21,0.0434],[22,0.1003],[23,0.195],[24,0.326],[25,0.4704],[26,0.6057],[27,0.7164],[28,0.7907],[29,0.8371],[30,0.8724],[31,0.8981],[32,0.9162],[33,0.9308],[35,0.942],[36,0.9445],[37,0.9423],[38,0.9334],[39,0.9186],[40,0.8991],[41,0.8758],[44,0.7948],[45,0.7698],[46,0.7488],[48,0.7123],[50,0.6785],[52,0.6485],[54,0.6247],[55,0.6354],[56,0.6858],[57,0.7719],[58,0.8939],[59,1.0494],[60,1.2369],[62,1.6722],[63,1.8365],[64,1.9221],[65,1.8514],[66,1.7088],[67,1.5527],[68,1.4293],[69,1.3772],[70,1.3553],[71,1.3475],[72,1.3458],[81,1.3584],[86,1.3522],[87,1.3457],[89,1.3457],[90,1.3325],[91,1.3318],[92,1.3232],[93,1.336],[94,1.3764],[95,1.4328],[96,1.4962],[97,1.5535],[98,1.5959],[99,1.612],[102,1.615],[104,1.5937],[105,1.586],[106,1.5599],[107,1.4898],[108,1.3425],[109,1.137],[110,0.9235],[111,0.7521],[112,0.6369],[113,0.5559],[114,0.5045],[115,0.478],[116,0.4584],[117,0.4328],[118,0.4048],[119,0.3777],[120,0.3551],[121,0.3361],[122,0.3178],[123,0.3009],[124,0.2853],[125,0.2853],[126,0.2658],[127,0.2641],[128,0.2699],[129,0.2806],[130,0.292],[131,0.3026],[132,0.3086],[133,0.3076],[134,0.2714],[135,0.1364],[136,-0.1539],[137,-0.657],[138,-1.4306],[139,-2.5324],[140,-4.6716],[141,-7.9404],[142,-11.4563],[143,-15.4824],[144,-19.201],[145,-21.2784],[146,-22.2982],[147,-22.7354],[148,-22.5854],[149,-21.4962],[150,-8.8919],[151,-2.5539],[152,-2.8205],[153,-4.5409],[154,-6.382],[155,-8.4771],[156,-10.8203],[157,-37.3852],[158,-61.9973],[159,-69.961],[160,-73.4405],[161,-73.003],[162,-70.1098],[163,-66.4545],[164,-64.8411],[166,-66.0584],[167,-66.5279],[168,-66.739],[169,-67.1518],[170,-66.8512],[171,-67.4709],[173,-70.4556],[174,-71.2606],[175,-71.661],[177,-71.6898],[182,-70.6445],[187,-68.8331],[193,-67.9278],[219,-66.6995],[224,-65.9274],[228,-65.9778],[256,-67.9946],[257,-67.7542],[258,-67.0563],[259,-65.8548],[260,-63.857],[262,-57.85],[263,-55.3079],[264,-53.1929],[265,-51.7404],[266,-50.8326],[267,-50.1876],[268,-49.7647],[269,-49.5152],[275,-48.8393],[282,-48.8025],[284,-49.0223],[285,-49.2941],[286,-49.7291],[287,-50.3677],[288,-51.5617],[289,-53.404],[291,-57.6426],[292,-59.9338],[293,-62.029],[294,-63.4298],[295,-64.3977],[297,-65.8781],[298,-66.4405],[304,-68.5839],[311,-69.6788],[343,-68.9834],[344,-68.8691],[345,-68.5294],[346,-67.7894],[347,-66.6721],[348,-65.1991],[349,-63.3925],[350,-61.2755],[351,-58.2648],[352,-54.2239],[353,-49.8456],[354,-45.8137],[355,-33.8299],[356,-18.4622],[357,-9.8988],[358,-4.7629],[359,-3.475],[360,-4.5734],[361,-5.8758],[362,-6.3076],[363,-6.015],[364,-4.3718],[365,-1.8553],[366,0.164],[367,1.2412],[368,1.8616],[369,2.0445],[370,1.7124],[371,0.946],[372,0.0032],[373,-0.8581],[374,-1.6657],[375,-2.5202],[376,-3.2729],[377,-3.7906],[378,-4.0066],[379,-3.9952],[380,-3.8506],[381,-3.6667],[382,-3.4488],[383,-3.1542],[384,-2.8046],[385,-2.4221],[386,-2.0231],[387,-1.6165],[388,-1.2153],[389,-0.833],[390,-0.4832],[391,-0.1658],[392,0.1273],[393,0.3945],[394,0.6339],[395,0.8438],[396,1.0276],[397,1.1905],[398,1.3336],[399,1.458],[400,1.5648],[401,1.6551],[402,1.7267],[403,1.7784],[404,1.8131],[405,1.8345],[406,1.8449],[407,1.8475],[408,1.8413],[409,1.8248],[410,1.8],[411,1.769],[416,1.582],[418,1.4947],[421,1.3576],[422,1.3159],[423,1.2787],[424,1.2473],[425,1.2204],[426,1.1961],[427,1.1753],[428,1.1569],[429,1.142],[430,1.1298],[431,1.121],[433,1.121]]
       }
       head_ik_ctrl.Rotation.Y {
        type=MovieSceneScriptingFloatChannel
        default=0
        thinned_from=434
        keys [[0,2.4337],[6,2.4256],[9,2.4064],[10,2.3935],[11,2.3983],[12,2.4366],[13,2.5341],[14,2.714],[15,2.9995],[16,3.3787],[17,3.8244],[18,4.3108],[19,5.0043],[20,5.9179],[21,6.7831],[22,7.591],[23,8.306],[24,8.6078],[25,8.5737],[27,8.4469],[28,8.4116],[36,8.2805],[38,8.2862],[41,8.3597],[52,8.6992],[55,8.7403],[56,8.7111],[57,8.6245],[58,8.4529],[59,8.1718],[60,7.7564],[61,7.1818],[62,6.4034],[63,5.4456],[64,4.386],[66,2.3575],[67,1.2308],[68,0.1143],[69,-1.0116],[70,-2.0952],[71,-2.9617],[72,-3.5598],[73,-3.9628],[74,-4.2182],[75,-4.3451],[76,-4.3764],[78,-4.3873],[81,-4.3798],[82,-4.3895],[85,-4.3823],[86,-4.3857],[87,-4.3785],[89,-4.3855],[90,-4.3779],[91,-4.3858],[92,-4.3819],[94,-4.4006],[96,-4.4598],[98,-4.5569],[100,-4.6847],[101,-4.7601],[102,-4.853],[103,-4.9718],[104,-5.1255],[105,-5.3142],[107,-5.7428],[108,-6.002],[109,-6.2937],[110,-6.5222],[111,-6.5926],[113,-6.5252],[115,-6.4978],[125,-6.5575],[132,-6.6632],[133,-6.6694],[134,-6.6508],[135,-6.5796],[136,-6.4472],[137,-6.2494],[138,-5.9863],[139,-5.6616],[140,-5.1178],[141,-4.4006],[142,-3.7581],[143,-3.0799],[144,-2.7971],[145,-3.2286],[146,-3.7146],[147,-3.9433],[148,-4.1072],[149,-4.2053],[150,-5.1594],[151,-5.5106],[152,-5.6234],[153,-5.6726],[154,-5.7032],[155,-5.7164],[156,-5.7123],[157,-4.2498],[158,-3.0604],[159,-2.3345],[160,-1.7842],[161,-1.2285],[162,-0.4697],[163,0.5048],[164,1.2142],[165,1.6742],[166,2.1744],[167,2.6346],[168,2.9404],[169,2.8888],[170,2.7957],[171,2.467],[172,1.9412],[173,1.5],[174,1.2961],[175,1.2274],[176,1.2509],[177,1.3115],[178,1.4006],[179,1.5364],[181,1.8554],[182,1.9922],[183,2.1232],[185,2.417],[186,2.5546],[187,2.6667],[188,2.7386],[189,2.7816],[190,2.8167],[191,2.8445],[192,2.8659],[193,2.8816],[196,2.9011],[206,2.9016],[207,2.909],[208,2.9596],[209,3.0918],[210,3.345],[211,3.7575],[212,4.3683],[213,5.2162],[214,6.3419],[215,8.0457],[216,10.4243],[217,13.2057],[218,16.1137],[219,18.8728],[220,21.2077],[221,22.7133],[222,23.3746],[223,23.5483],[224,23.3961],[225,22.778],[226,21.7913],[227,20.9588],[228,20.4775],[229,20.1641],[230,19.9744],[231,19.8638],[233,19.7639],[249,19.6446],[250,19.5925],[251,19.5],[252,19.3543],[253,19.1459],[254,18.8624],[255,18.4932],[256,16.9115],[257,13.544],[258,9.2242],[259,4.7891],[260,-0.0825],[261,-5.2209],[262,-9.2919],[263,-12.0661],[264,-14.0372],[265,-15.2653],[266,-16.0927],[267,-16.7969],[268,-17.3309],[269,-17.6398],[273,-18.202],[275,-18.365],[278,-18.4489],[283,-18.443],[285,-18.3331],[286,-18.2113],[287,-18.0261],[288,-17.4598],[289,-16.2292],[290,-14.3814],[291,-12.0289],[292,-8.4027],[293,-3.7296],[294,-0.0059],[295,2.2161],[296,3.9084],[297,5.1497],[298,5.9935],[299,6.5006],[300,6.8457],[301,7.1606],[302,7.4443],[303,7.696],[304,7.9151],[305,8.1011],[306,8.2538],[307,8.3735],[308,8.4626],[309,8.5255],[310,8.5629],[311,8.5794],[313,8.5572],[315,8.4819],[317,8.3647],[318,8.2703],[319,8.1535],[320,8.014],[321,7.8557],[323,7.4898],[325,7.0755],[326,6.8241],[327,6.5245],[328,6.2099],[329,5.9135],[330,5.6685],[331,5.4685],[332,5.2915],[333,5.1469],[334,5.044],[335,4.9921],[336,4.9921],[337,5.0184],[338,5.0745],[339,5.1501],[342,5.4149],[343,5.4901],[344,5.5478],[345,5.5482],[346,5.4787],[347,5.3677],[348,5.2442],[349,5.1382],[350,5.0814],[351,5.3824],[352,6.0292],[353,6.54],[354,6.3867],[355,4.5239],[356,2.5631],[357,2.262],[358,2.0205],[359,1.9449],[360,1.969],[361,2.1361],[362,2.5512],[363,3.0333],[364,3.4578],[365,3.8491],[366,4.0759],[367,4.099],[368,3.9925],[369,3.8246],[370,3.6319],[372,3.2091],[373,3.0148],[374,2.8111],[375,2.6156],[376,2.5129],[377,2.4981],[378,2.5171],[379,2.5664],[380,2.6225],[381,2.6632],[382,2.6747],[383,2.6638],[386,2.5876],[387,2.5701],[389,2.55],[395,2.5098],[400,2.4622],[410,2.4138],[418,2.4089],[425,2.4272],[426,2.4233],[427,2.43],[428,2.4285],[433,2.432]]
       }
       head_ik_ctrl.Rotation.Z {
        type=MovieSceneScriptingFloatChannel
        default=0
        thinned_from=434
        keys [[0,4.4422],[2,4.4422],[3,4.4028],[4,4.3488],[5,4.2598],[6,4.127],[7,3.9416],[8,3.695],[9,3.3782],[10,2.8161],[11,1.8126],[12,0.3519],[13,-1.5925],[14,-4.0443],[15,-7.547],[16,-12.064],[17,-16.7484],[18,-20.753],[19,-23.6178],[20,-25.6415],[21,-27.1209],[22,-28.0421],[23,-28.4284],[25,-28.8271],[31,-29.0374],[34,-29.1255],[41,-29.9578],[55,-30.2201],[56,-30.2015],[57,-30.0894],[58,-29.8771],[59,-29.5184],[60,-28.9767],[61,-28.2151],[62,-26.3218],[63,-23.0416],[64,-19.322],[65,-14.9345],[66,-10.3751],[67,-6.5288],[68,-3.4792],[69,-1.3676],[70,0.2713],[71,1.5014],[72,2.3022],[73,2.6933],[74,2.7995],[75,2.746],[76,2.6165],[78,2.3334],[80,2.0311],[81,1.8979],[82,1.7885],[83,1.6633],[84,1.5575],[85,1.4662],[86,1.4049],[87,1.3606],[88,1.3127],[89,1.2963],[91,1.2492],[92,1.2362],[93,1.2382],[94,1.2561],[95,1.3116],[96,1.4033],[97,1.5856],[98,1.8391],[99,2.1924],[100,3.4421],[101,5.9001],[102,8.8563],[103,11.6008],[104,14.1685],[105,16.8625],[106,19.3756],[107,21.4017],[108,22.8936],[109,24.0035],[110,24.725],[111,25.0518],[115,25.8304],[118,26.1938],[121,26.3581],[130,26.2952],[132,26.1104],[133,25.9711],[134,25.6199],[135,24.9367],[136,23.9954],[137,22.8689],[139,20.3494],[140,18.8354],[141,16.903],[142,14.6418],[143,11.6391],[144,8.3832],[145,5.4829],[146,3.3801],[147,2.2741],[148,1.4829],[149,0.9343],[150,0.7678],[151,0.3724],[152,0.1182],[153,-0.0421],[154,-0.1369],[155,-0.1864],[156,-0.2139],[157,-2.7134],[158,-5.2523],[159,-6.4918],[160,-7.3665],[161,-8.0052],[162,-8.5133],[163,-8.9523],[164,-9.3372],[165,-9.7531],[166,-10.1954],[167,-10.5499],[168,-10.728],[169,-10.7589],[172,-10.5316],[173,-10.5073],[174,-10.5652],[175,-10.746],[178,-11.4851],[181,-12.2997],[183,-12.7033],[185,-13.0522],[186,-13.1827],[187,-13.2701],[188,-13.2992],[199,-13.3277],[200,-13.3131],[208,-13.3358],[210,-13.2486],[211,-13.144],[212,-12.9771],[213,-12.7348],[214,-12.3861],[215,-11.8655],[216,-11.1172],[218,-9.3295],[219,-8.4626],[220,-7.7479],[221,-7.3272],[222,-7.2122],[223,-7.2642],[224,-7.4132],[225,-7.6988],[226,-8.0736],[227,-8.3527],[228,-8.4934],[229,-8.5775],[230,-8.6214],[232,-8.6479],[236,-8.6195],[247,-8.66],[249,-8.6465],[250,-8.6552],[251,-8.6968],[252,-8.8066],[253,-8.9977],[254,-9.283],[255,-9.6944],[256,-10.8898],[257,-13.1397],[258,-15.8699],[259,-18.5619],[260,-21.362],[261,-24.5752],[262,-27.7215],[263,-30.5215],[264,-33.0039],[265,-34.8781],[266,-36.1633],[267,-37.1042],[268,-37.7601],[269,-38.1903],[271,-38.6963],[273,-39.0177],[278,-39.2846],[283,-39.253],[284,-39.1759],[285,-39.0042],[286,-38.7257],[287,-38.2826],[288,-37.3248],[289,-35.696],[290,-33.678],[291,-31.5301],[292,-28.6892],[293,-25.5853],[294,-23.6282],[295,-22.7683],[296,-22.1399],[297,-21.7352],[300,-20.894],[302,-20.4369],[309,-19.2716],[315,-18.3315],[318,-17.9333],[321,-17.6341],[324,-17.2479],[327,-16.7576],[333,-15.9625],[334,-15.8679],[335,-15.8334],[336,-15.8334],[337,-15.8688],[341,-16.1281],[343,-16.1695],[344,-16.1167],[345,-15.9881],[346,-15.7534],[347,-15.4355],[348,-15.0611],[350,-14.2426],[351,-13.8702],[352,-13.4359],[353,-12.7813],[354,-11.8318],[355,-8.9794],[356,-7.0491],[357,-6.1819],[358,-4.9626],[359,-3.6638],[360,-2.2208],[361,-0.8402],[362,0.3839],[363,1.6048],[364,3.0622],[365,4.6606],[366,6.0339],[367,7.048],[368,7.8324],[369,8.4138],[370,8.8423],[371,9.1554],[372,9.362],[373,9.4714],[374,9.486],[375,9.4223],[376,9.3071],[377,9.1621],[378,8.9788],[379,8.7479],[381,8.2211],[382,7.924],[383,7.5796],[385,6.8446],[386,6.4652],[388,5.6616],[389,5.2896],[390,4.9722],[391,4.7126],[392,4.4934],[393,4.3097],[394,4.1564],[395,4.0287],[396,3.9225],[397,3.8368],[398,3.771],[399,3.7248],[400,3.6977],[401,3.6893],[402,3.696],[403,3.7126],[404,3.717],[405,3.7512],[407,3.8013],[408,3.8412],[409,3.9017],[410,3.9208],[411,3.9657],[412,3.9898],[421,4.2517],[422,4.2758],[423,4.3085],[426,4.3768],[427,4.3892],[428,4.4162],[430,4.4373],[431,4.4558],[433,4.4558]]
       }
       head_ik_ctrl.Scale.X {type=MovieSceneScriptingFloatChannel, default=1, const [1,434]}
       head_ik_ctrl.Scale.Y {type=MovieSceneScriptingFloatChannel, default=1, const [1,434]}
       head_ik_ctrl.Scale.Z {type=MovieSceneScriptingFloatChannel, default=1, const [1,434]}
       chest_ctrl.Location.X {
        type=MovieSceneScriptingFloatChannel
        default=0
        thinned_from=434
        keys [[0,0.3621],[10,0.3621],[11,0.3605],[12,0.3566],[13,0.349],[14,0.3365],[15,0.3183],[16,0.29],[17,0.2451],[18,0.1769],[19,0.0517],[20,-0.1171],[21,-0.2644],[22,-0.373],[23,-0.4603],[24,-0.515],[25,-0.544],[26,-0.5637],[27,-0.5778],[28,-0.5875],[29,-0.5937],[30,-0.5974],[31,-0.5991],[34,-0.5988],[42,-0.5869],[56,-0.5828],[58,-0.5803],[59,-0.5777],[60,-0.5739],[61,-0.5684],[62,-0.5585],[63,-0.5305],[64,-0.4674],[65,-0.3118],[66,-0.1341],[67,-0.0185],[68,0.0584],[69,0.104],[70,0.1327],[71,0.1468],[72,0.1552],[73,0.1609],[74,0.1646],[75,0.167],[76,0.1683],[77,0.1689],[78,0.1693],[86,0.1674],[87,0.1675],[88,0.1671],[94,0.1678],[95,0.1695],[96,0.1698],[97,0.1725],[98,0.1747],[99,0.1776],[100,0.1819],[101,0.1879],[102,0.1953],[103,0.2038],[104,0.2142],[106,0.2384],[107,0.2491],[108,0.2583],[109,0.2666],[110,0.2737],[111,0.279],[112,0.2829],[114,0.2888],[116,0.2927],[119,0.2959],[123,0.2979],[130,0.298],[133,0.2926],[134,0.29],[135,0.2869],[136,0.283],[137,0.2783],[138,0.2728],[139,0.2664],[140,0.2496],[141,0.2219],[142,0.1961],[143,0.1795],[144,0.1684],[146,0.1516],[147,0.1416],[148,0.1299],[149,0.1158],[150,-0.0209],[151,-0.045],[152,-0.0474],[153,-0.0464],[155,-0.0458],[156,-0.0468],[157,-0.0554],[158,-0.4427],[159,-0.6227],[160,-0.7208],[161,-0.7749],[162,-0.8098],[163,-0.8304],[164,-0.8353],[165,-0.8295],[166,-0.8163],[169,-0.7642],[170,-0.7522],[171,-0.7446],[172,-0.7403],[173,-0.7393],[174,-0.7405],[177,-0.7532],[181,-0.7624],[194,-0.7643],[218,-0.7563],[220,-0.7504],[221,-0.744],[222,-0.7324],[223,-0.715],[224,-0.6916],[225,-0.6496],[226,-0.5963],[227,-0.5622],[228,-0.5491],[229,-0.5379],[230,-0.5283],[231,-0.5204],[232,-0.5139],[233,-0.5087],[235,-0.5014],[236,-0.4991],[239,-0.4959],[252,-0.4942],[254,-0.4963],[256,-0.5018],[257,-0.5079],[258,-0.5179],[259,-0.5331],[260,-0.5587],[261,-0.5925],[262,-0.6244],[263,-0.6527],[264,-0.6792],[265,-0.6997],[266,-0.7134],[267,-0.7234],[269,-0.7375],[271,-0.7477],[276,-0.7621],[285,-0.7742],[287,-0.7798],[288,-0.7851],[291,-0.8071],[293,-0.8166],[294,-0.8182],[295,-0.8149],[296,-0.8067],[298,-0.786],[299,-0.7789],[302,-0.7684],[309,-0.76],[319,-0.7546],[320,-0.7529],[321,-0.7528],[325,-0.744],[326,-0.74],[329,-0.7211],[331,-0.7123],[334,-0.7049],[337,-0.7017],[343,-0.6994],[344,-0.6982],[345,-0.6883],[346,-0.6623],[347,-0.6223],[348,-0.5707],[349,-0.5104],[350,-0.4444],[351,-0.3947],[352,-0.3658],[353,-0.3328],[354,-0.2653],[355,0.0303],[356,0.2752],[357,0.3242],[358,0.3428],[359,0.3577],[360,0.3677],[361,0.376],[362,0.386],[363,0.3923],[364,0.3863],[365,0.3742],[366,0.3645],[370,0.3373],[377,0.2847],[378,0.2754],[379,0.2651],[380,0.2567],[381,0.2533],[382,0.2569],[383,0.2659],[385,0.2889],[387,0.3098],[388,0.3196],[389,0.3281],[390,0.3349],[392,0.3457],[394,0.3546],[396,0.3611],[398,0.3659],[400,0.369],[403,0.3709],[409,0.3713],[427,0.3628],[433,0.3623]]
       }
       chest_ctrl.Location.Y {
        type=MovieSceneScriptingFloatChannel
        default=0
        thinned_from=434
        keys [[0,-1.1228],[19,-1.1256],[20,-1.1192],[21,-1.1028],[22,-1.0767],[23,-1.0438],[24,-1.0151],[25,-0.995],[26,-0.9785],[27,-0.9658],[28,-0.9568],[29,-0.9511],[31,-0.9464],[33,-0.9464],[45,-0.9708],[60,-0.979],[62,-0.9841],[63,-0.9908],[64,-1.0033],[65,-1.0263],[67,-1.0852],[68,-1.1066],[69,-1.1166],[70,-1.1203],[146,-1.1211],[149,-1.1151],[150,-0.8795],[151,-0.8236],[152,-0.8132],[154,-0.8097],[155,-0.8095],[156,-0.807],[157,-1.9053],[158,-1.9333],[159,-1.9985],[160,-2.0463],[163,-2.11],[164,-2.0654],[165,-2.0411],[167,-2.0196],[168,-1.9921],[169,-1.9392],[170,-1.9973],[171,-2.0632],[172,-2.0893],[220,-2.0843],[222,-2.0933],[224,-2.1126],[226,-2.1527],[227,-2.1654],[256,-2.1858],[259,-2.1756],[265,-2.1089],[288,-2.0662],[293,-2.0489],[295,-2.0498],[299,-2.0695],[326,-2.0895],[344,-2.1096],[346,-2.1021],[350,-2.0687],[351,-2.0482],[352,-2.0118],[353,-1.9798],[354,-1.9698],[355,-1.7771],[356,-1.4409],[357,-1.3364],[358,-1.3153],[359,-1.3248],[360,-1.3421],[361,-1.3547],[362,-1.3541],[363,-1.3303],[364,-1.2786],[365,-1.2179],[366,-1.1702],[367,-1.1453],[368,-1.135],[369,-1.1353],[373,-1.1498],[378,-1.1524],[382,-1.1498],[385,-1.1522],[399,-1.1214],[433,-1.1228]]
       }
       chest_ctrl.Location.Z {
        type=MovieSceneScriptingFloatChannel
        default=0
        thinned_from=434
        keys [[0,-0.089],[13,-0.0891],[16,-0.0896],[19,-0.0905],[20,-0.0897],[21,-0.0875],[22,-0.0846],[23,-0.0811],[24,-0.078],[25,-0.0759],[26,-0.0742],[27,-0.0729],[28,-0.0719],[29,-0.0713],[31,-0.0708],[36,-0.0714],[45,-0.0732],[59,-0.0739],[61,-0.0744],[62,-0.0748],[63,-0.0758],[64,-0.0776],[65,-0.0805],[66,-0.0823],[68,-0.0866],[69,-0.0877],[70,-0.0881],[145,-0.0881],[148,-0.0875],[149,-0.0871],[150,-0.0747],[151,-0.0707],[152,-0.0699],[156,-0.0694],[157,-0.1549],[158,-0.1399],[159,-0.1351],[160,-0.1212],[161,-0.109],[162,-0.0963],[163,-0.084],[164,-0.0981],[165,-0.1063],[167,-0.1165],[168,-0.1233],[169,-0.1291],[170,-0.1258],[171,-0.1158],[172,-0.1104],[174,-0.1104],[181,-0.1094],[218,-0.1097],[221,-0.1102],[222,-0.1107],[223,-0.1115],[224,-0.1124],[225,-0.1141],[226,-0.116],[227,-0.1171],[232,-0.1185],[256,-0.1189],[258,-0.1184],[259,-0.118],[260,-0.1172],[264,-0.1129],[265,-0.1121],[267,-0.1111],[287,-0.1086],[293,-0.1069],[294,-0.1068],[295,-0.107],[299,-0.1086],[325,-0.1102],[334,-0.1119],[344,-0.1122],[345,-0.1128],[346,-0.1146],[347,-0.1171],[350,-0.1257],[352,-0.131],[353,-0.1341],[354,-0.1386],[355,-0.1397],[356,-0.119],[357,-0.1102],[358,-0.1083],[359,-0.109],[360,-0.1104],[361,-0.1113],[362,-0.1108],[363,-0.1082],[364,-0.1035],[365,-0.0979],[366,-0.0936],[367,-0.0912],[368,-0.0902],[369,-0.0902],[373,-0.0914],[377,-0.0916],[381,-0.0911],[385,-0.0915],[399,-0.0888],[433,-0.089]]
       }
       chest_ctrl.Rotation.X {
        type=MovieSceneScriptingFloatChannel
        default=0
        thinned_from=434
        keys [[0,-5.6245],[19,-5.6253],[20,-5.6189],[21,-5.5908],[22,-5.5227],[24,-5.3237],[26,-5.1737],[27,-5.119],[28,-5.0807],[29,-5.0568],[31,-5.0385],[33,-5.0385],[46,-5.1685],[63,-5.2003],[64,-5.1855],[65,-5.1186],[66,-5.0999],[67,-5.2266],[68,-5.3851],[69,-5.4957],[70,-5.5439],[101,-5.5606],[141,-5.6279],[144,-5.5973],[146,-5.5506],[147,-5.512],[148,-5.4564],[149,-5.3759],[150,-4.3529],[151,-4.1635],[152,-4.1379],[155,-4.1404],[156,-4.1317],[157,-12.2109],[158,-13.3186],[159,-13.7074],[160,-13.8597],[161,-13.9301],[163,-14.0208],[164,-13.9211],[165,-13.8691],[167,-13.8306],[168,-13.7687],[169,-13.637],[171,-13.9566],[172,-14.0208],[221,-14.0152],[224,-14.0936],[226,-14.2198],[258,-14.3088],[266,-14.0618],[295,-13.8981],[344,-14.0841],[345,-14.0387],[346,-13.9161],[347,-13.7396],[349,-13.3213],[351,-12.9315],[353,-12.4733],[354,-12.3149],[355,-10.4512],[356,-7.8845],[357,-7.2321],[358,-7.1453],[359,-7.2338],[360,-7.3738],[361,-7.464],[362,-7.426],[363,-7.2225],[364,-6.8314],[365,-6.3659],[366,-5.9933],[367,-5.7812],[368,-5.6674],[369,-5.6277],[380,-5.4623],[382,-5.4618],[390,-5.5929],[433,-5.6247]]
       }
       chest_ctrl.Rotation.Y {
        type=MovieSceneScriptingFloatChannel
        default=0
        thinned_from=434
        keys [[0,-2.0782],[10,-2.0782],[11,-2.0711],[12,-2.0543],[13,-2.0216],[14,-1.9677],[15,-1.8911],[16,-1.7703],[17,-1.5698],[18,-1.2542],[19,-0.7615],[20,-0.2288],[21,0.0799],[22,0.1624],[23,0.1794],[24,0.1156],[25,0.0119],[26,-0.0714],[27,-0.1272],[28,-0.1585],[29,-0.1717],[30,-0.1747],[31,-0.1725],[33,-0.1725],[34,-0.1595],[35,-0.1495],[36,-0.1375],[37,-0.1237],[38,-0.1081],[39,-0.0912],[40,-0.0737],[41,-0.0566],[42,-0.0382],[43,-0.0184],[44,0.0011],[45,0.0183],[46,0.031],[47,0.0398],[48,0.0466],[49,0.0504],[50,0.0552],[51,0.057],[52,0.0626],[53,0.0662],[54,0.0722],[55,0.0769],[56,0.079],[57,0.0807],[58,0.0788],[59,0.0738],[60,0.065],[61,0.0521],[62,0.0152],[63,-0.0993],[64,-0.3431],[66,-1.5927],[67,-1.7683],[68,-1.6508],[69,-1.4021],[70,-1.2788],[71,-1.2679],[72,-1.2686],[73,-1.2849],[75,-1.3282],[77,-1.3746],[79,-1.413],[81,-1.4443],[84,-1.4758],[86,-1.4894],[89,-1.4998],[90,-1.4998],[94,-1.5122],[95,-1.5198],[96,-1.5227],[99,-1.5472],[100,-1.5532],[101,-1.5539],[104,-1.542],[105,-1.5353],[106,-1.5254],[107,-1.5107],[108,-1.4906],[109,-1.4659],[112,-1.3777],[113,-1.353],[114,-1.3321],[115,-1.3146],[116,-1.3001],[118,-1.2785],[122,-1.2485],[123,-1.2432],[124,-1.2432],[125,-1.2379],[128,-1.2379],[135,-1.2225],[139,-1.2017],[140,-1.1305],[141,-1.0029],[142,-0.9274],[143,-0.9489],[144,-1.027],[145,-1.1514],[146,-1.3154],[147,-1.5149],[148,-1.7613],[149,-2.0632],[150,-1.95],[151,-1.714],[152,-1.6217],[153,-1.5696],[154,-1.5432],[156,-1.5217],[157,-4.6389],[158,-4.5104],[159,-4.5636],[160,-4.644],[162,-4.7182],[163,-4.7461],[165,-4.6865],[167,-4.6449],[168,-4.6072],[169,-4.5457],[171,-4.6417],[172,-4.66],[222,-4.6574],[224,-4.6438],[227,-4.594],[257,-4.5705],[260,-4.5925],[265,-4.6465],[344,-4.6459],[345,-4.6706],[346,-4.7338],[348,-4.9061],[349,-4.978],[350,-5.0147],[351,-4.7907],[352,-4.2276],[353,-3.5369],[354,-2.9501],[355,-2.9273],[356,-2.709],[357,-2.1613],[358,-1.8207],[359,-1.708],[360,-1.63],[361,-1.5917],[362,-1.6005],[363,-1.6237],[364,-1.6815],[365,-1.8046],[367,-2.0859],[368,-2.2388],[369,-2.3833],[370,-2.5161],[371,-2.6423],[372,-2.7562],[373,-2.8524],[374,-2.9303],[375,-2.9926],[376,-3.0391],[377,-3.0696],[378,-3.0764],[380,-3.0426],[381,-3.0326],[382,-3.0423],[383,-3.06],[384,-3.0686],[385,-3.0511],[386,-2.9978],[387,-2.9173],[390,-2.6208],[392,-2.4453],[393,-2.3653],[394,-2.2938],[395,-2.233],[396,-2.1812],[397,-2.1353],[398,-2.0958],[399,-2.0634],[400,-2.0387],[401,-2.0223],[403,-2.0055],[409,-2.0003],[422,-2.0614],[433,-2.0761]]
       }
       chest_ctrl.Rotation.Z {
        type=MovieSceneScriptingFloatChannel
        default=0
        thinned_from=434
        keys [[0,-0.0126],[10,-0.0126],[11,-0.0258],[12,-0.0569],[13,-0.1175],[14,-0.2175],[15,-0.3666],[16,-0.5747],[17,-0.8514],[18,-1.2063],[19,-2.3306],[20,-4.6123],[21,-7.6068],[22,-10.8682],[23,-13.9496],[24,-16.4048],[25,-18.1714],[26,-19.515],[27,-20.4806],[28,-21.1134],[29,-21.4845],[30,-21.6751],[31,-21.7452],[33,-21.7452],[36,-21.3495],[43,-20.0429],[45,-19.7576],[47,-19.5888],[58,-19.1849],[60,-19.0472],[62,-18.8148],[63,-18.5857],[64,-18.0846],[65,-17.1802],[66,-15.4744],[67,-12.5702],[68,-9.0959],[69,-6.0203],[70,-4.3084],[71,-3.7621],[72,-3.4816],[73,-3.3783],[74,-3.3637],[75,-3.4019],[76,-3.4893],[78,-3.6823],[80,-3.8904],[81,-3.9781],[83,-4.1193],[84,-4.1794],[85,-4.2307],[86,-4.2719],[87,-4.2839],[88,-4.3195],[89,-4.336],[90,-4.336],[91,-4.3567],[92,-4.3682],[94,-4.3676],[96,-4.3488],[98,-4.2581],[99,-4.1923],[100,-4.0608],[101,-3.8337],[102,-3.5341],[103,-3.1849],[104,-2.7635],[105,-2.2671],[106,-1.7491],[107,-1.2627],[108,-0.8044],[109,-0.3518],[110,0.0707],[111,0.4384],[112,0.7444],[113,1.0019],[114,1.2168],[115,1.3947],[116,1.5395],[117,1.6552],[118,1.7482],[119,1.8249],[120,1.8917],[121,1.9514],[122,2.0021],[123,2.043],[124,2.043],[125,2.0848],[128,2.0848],[129,2.089],[130,2.0819],[132,1.9876],[133,1.9226],[134,1.8385],[135,1.7347],[136,1.6083],[137,1.4564],[138,1.2759],[139,1.0638],[140,0.81],[141,0.4422],[142,-0.1354],[143,-0.8823],[144,-1.7297],[145,-2.7307],[146,-3.9383],[147,-5.4093],[148,-7.2164],[149,-9.4371],[150,-13.7936],[151,-13.2484],[152,-12.7408],[153,-12.354],[154,-12.1703],[155,-12.0795],[156,-12.0558],[157,-16.1008],[158,-23.4685],[159,-27.157],[160,-29.313],[161,-30.4936],[162,-31.2785],[163,-31.7064],[164,-31.9402],[165,-31.8579],[166,-31.5563],[168,-30.783],[169,-30.4895],[170,-30.0425],[171,-29.6894],[172,-29.52],[174,-29.5236],[179,-29.9952],[194,-30.1413],[218,-29.9232],[220,-29.7781],[221,-29.6148],[222,-29.3149],[223,-28.8711],[224,-28.2759],[225,-27.2203],[226,-25.8997],[227,-25.064],[229,-24.4721],[231,-24.0495],[233,-23.7662],[236,-23.5366],[252,-23.4239],[255,-23.5284],[256,-23.602],[257,-23.7467],[258,-23.9892],[259,-24.356],[260,-24.9781],[262,-26.5938],[264,-27.9637],[265,-28.4813],[266,-28.8304],[267,-29.0851],[270,-29.5924],[281,-30.3021],[286,-30.463],[288,-30.6793],[291,-31.2569],[293,-31.5069],[294,-31.5511],[295,-31.4632],[298,-30.7037],[299,-30.5179],[303,-30.1872],[321,-29.8334],[325,-29.6147],[330,-28.8958],[334,-28.614],[344,-28.4434],[345,-28.3475],[346,-28.0945],[347,-27.6575],[348,-27.014],[349,-26.1417],[350,-25.0178],[351,-23.5578],[352,-21.6407],[353,-19.1895],[354,-16.132],[355,-10.7074],[356,-4.9381],[357,-1.2351],[358,0.9381],[359,1.931],[360,2.6001],[361,3.0017],[362,3.1913],[363,3.2332],[364,2.8052],[365,1.8421],[366,0.8005],[367,-0.2098],[368,-1.2976],[369,-2.3314],[370,-3.2993],[371,-4.251],[372,-5.1473],[373,-5.9493],[374,-6.6504],[375,-7.2701],[376,-7.8086],[377,-8.2663],[378,-8.6607],[379,-8.9857],[380,-9.2064],[381,-9.2878],[382,-9.1992],[383,-8.9452],[384,-8.5431],[385,-8.0105],[386,-7.3224],[387,-6.4949],[388,-5.6064],[389,-4.7353],[390,-3.9602],[391,-3.2697],[392,-2.6117],[393,-2.0028],[394,-1.4598],[395,-0.9996],[396,-0.6077],[397,-0.2615],[398,0.0351],[399,0.2779],[400,0.4631],[401,0.5864],[402,0.6631],[403,0.7145],[404,0.7262],[405,0.7557],[409,0.7557],[410,0.7137],[411,0.6871],[412,0.6339],[413,0.5873],[416,0.4434],[418,0.3452],[419,0.2967],[420,0.2498],[421,0.2054],[422,0.1646],[423,0.1385],[424,0.1079],[425,0.0884],[426,0.057],[427,0.0433],[429,0.0187],[430,0.0069],[433,0.0069]]
       }
       chest_ctrl.Scale.X {type=MovieSceneScriptingFloatChannel, default=1, const [1,434]}
       chest_ctrl.Scale.Y {type=MovieSceneScriptingFloatChannel, default=1, const [1,434]}
       chest_ctrl.Scale.Z {type=MovieSceneScriptingFloatChannel, default=1, const [1,434]}
       chest_tan_ctrl.Location.X {
        type=MovieSceneScriptingFloatChannel
        default=0
        thinned_from=434
        keys [[0,0.0428],[10,0.0428],[11,0.0426],[12,0.0421],[13,0.0412],[14,0.0398],[15,0.0378],[16,0.0347],[17,0.0299],[18,0.0226],[19,0.0096],[20,-0.008],[21,-0.0239],[22,-0.0371],[23,-0.0487],[24,-0.0562],[25,-0.0604],[26,-0.0636],[27,-0.066],[28,-0.0677],[29,-0.0688],[30,-0.0695],[31,-0.0699],[33,-0.0699],[37,-0.0687],[41,-0.067],[49,-0.066],[55,-0.0659],[58,-0.0655],[60,-0.0648],[61,-0.0642],[62,-0.0631],[63,-0.0603],[64,-0.0547],[65,-0.0421],[66,-0.0269],[67,-0.0142],[68,-0.0044],[69,0.0012],[70,0.0047],[71,0.0064],[72,0.0074],[74,0.0087],[76,0.0093],[92,0.0097],[95,0.0101],[98,0.011],[99,0.0115],[100,0.0122],[101,0.0132],[103,0.0156],[104,0.0172],[106,0.021],[107,0.0226],[109,0.0252],[110,0.0262],[111,0.0269],[114,0.0281],[124,0.0292],[129,0.0291],[130,0.0288],[131,0.0289],[134,0.0279],[136,0.0267],[138,0.0251],[139,0.0241],[140,0.0211],[141,0.0162],[142,0.0119],[143,0.0094],[144,0.0079],[145,0.0071],[147,0.0057],[148,0.0047],[149,0.0036],[150,-0.0061],[151,-0.0075],[155,-0.0072],[156,-0.0072],[157,0.0705],[158,0.0456],[159,0.03],[160,0.0202],[161,0.0147],[162,0.0108],[163,0.0083],[164,0.0085],[165,0.0095],[166,0.011],[167,0.013],[168,0.0152],[169,0.0176],[170,0.018],[172,0.018],[173,0.0182],[179,0.0163],[182,0.0157],[183,0.0159],[194,0.0157],[195,0.016],[197,0.0158],[203,0.0162],[204,0.0161],[213,0.0163],[215,0.0164],[216,0.0163],[219,0.0167],[220,0.017],[221,0.0176],[222,0.0187],[223,0.0203],[224,0.0224],[225,0.0261],[226,0.0307],[227,0.0336],[229,0.0356],[231,0.0371],[234,0.0384],[237,0.0391],[239,0.0391],[240,0.0393],[242,0.0391],[251,0.0393],[253,0.039],[255,0.0389],[256,0.0386],[257,0.0381],[258,0.0373],[259,0.036],[260,0.0339],[262,0.0283],[264,0.0235],[265,0.0217],[266,0.0204],[267,0.0195],[270,0.0177],[277,0.0158],[282,0.0149],[283,0.0149],[286,0.0146],[287,0.0143],[288,0.0138],[291,0.0117],[293,0.0108],[294,0.0106],[295,0.0109],[296,0.0117],[298,0.0137],[299,0.0144],[303,0.0156],[305,0.016],[307,0.0161],[308,0.016],[309,0.0162],[312,0.0163],[313,0.0165],[314,0.0163],[318,0.0166],[319,0.0168],[320,0.0168],[325,0.0176],[327,0.0185],[330,0.0202],[335,0.0215],[341,0.0216],[342,0.0218],[343,0.0217],[344,0.0219],[345,0.0228],[346,0.0255],[347,0.0294],[348,0.0341],[350,0.0439],[351,0.0425],[352,0.0329],[353,0.0214],[354,0.0148],[355,0.03],[356,0.0425],[358,0.0416],[359,0.0426],[362,0.0442],[363,0.0444],[364,0.044],[366,0.0427],[371,0.0402],[374,0.0379],[377,0.035],[378,0.0335],[379,0.0318],[380,0.0304],[381,0.0298],[382,0.0304],[383,0.0319],[385,0.0355],[387,0.0382],[388,0.0393],[389,0.0402],[393,0.0421],[399,0.0436],[404,0.044],[422,0.043],[427,0.0429],[428,0.0427],[433,0.0428]]
       }
       chest_tan_ctrl.Location.Y {
        type=MovieSceneScriptingFloatChannel
        default=-0.0303
        thinned_from=434
        keys [[0,-0.1227],[19,-0.123],[20,-0.1226],[21,-0.1214],[22,-0.1194],[24,-0.1141],[25,-0.1121],[26,-0.1105],[27,-0.1092],[28,-0.1083],[29,-0.1078],[31,-0.1073],[33,-0.1073],[46,-0.1101],[61,-0.111],[63,-0.1115],[65,-0.1124],[66,-0.1136],[68,-0.1193],[69,-0.1211],[70,-0.1218],[143,-0.1227],[146,-0.1219],[147,-0.1214],[148,-0.1206],[149,-0.1194],[150,-0.1086],[151,-0.1068],[152,-0.1067],[156,-0.1068],[157,-0.3465],[158,-0.3973],[159,-0.4019],[160,-0.3987],[163,-0.3929],[165,-0.3967],[167,-0.3993],[169,-0.4043],[170,-0.4026],[171,-0.4],[172,-0.399],[223,-0.3995],[259,-0.4028],[296,-0.3974],[344,-0.3999],[345,-0.3983],[346,-0.3939],[347,-0.3875],[351,-0.3573],[353,-0.3379],[354,-0.3288],[355,-0.2667],[356,-0.1892],[357,-0.1695],[358,-0.1665],[359,-0.1684],[360,-0.1719],[361,-0.174],[362,-0.1718],[363,-0.1654],[364,-0.1547],[365,-0.1423],[366,-0.1325],[367,-0.127],[368,-0.1242],[369,-0.1234],[381,-0.1213],[390,-0.1228],[433,-0.1227]]
       }
       chest_tan_ctrl.Location.Z {
        type=MovieSceneScriptingFloatChannel
        default=0.265
        thinned_from=434
        keys [[0,0.2374],[15,0.2379],[19,0.2396],[20,0.2395],[22,0.2382],[62,0.2381],[64,0.2394],[65,0.2409],[66,0.241],[69,0.2391],[149,0.2391],[150,0.2473],[151,0.2484],[156,0.2486],[157,0.1508],[158,0.1225],[160,0.1],[161,0.092],[162,0.0844],[163,0.0771],[164,0.0862],[165,0.0917],[167,0.0979],[168,0.1029],[169,0.109],[170,0.104],[171,0.0953],[172,0.0912],[289,0.0913],[296,0.0915],[344,0.0911],[345,0.0923],[346,0.0953],[347,0.0997],[348,0.1047],[350,0.1153],[351,0.1215],[353,0.1376],[354,0.1446],[355,0.1798],[356,0.2146],[357,0.2225],[358,0.2236],[359,0.2227],[360,0.2213],[361,0.2203],[362,0.2207],[363,0.2227],[364,0.2265],[365,0.2309],[366,0.2343],[367,0.2361],[368,0.237],[433,0.2374]]
       }
       chest_tan_ctrl.Rotation.X {
        type=MovieSceneScriptingFloatChannel
        default=0
        thinned_from=434
        keys [[0,0],[1,1.8415],[18,1.8407],[19,1.8326],[20,1.8346],[21,1.8531],[22,1.8641],[23,1.8456],[24,1.7873],[25,1.7222],[26,1.6798],[27,1.6408],[28,1.6083],[29,1.5835],[30,1.5666],[31,1.5557],[33,1.5557],[34,1.5447],[39,1.5596],[47,1.5917],[62,1.6049],[63,1.6108],[64,1.6243],[65,1.6643],[66,1.679],[67,1.6754],[68,1.731],[69,1.7926],[70,1.8221],[71,1.8322],[110,1.8473],[141,1.8343],[144,1.8379],[148,1.8276],[149,1.8226],[150,2.5272],[151,1.8135],[152,1.6265],[153,1.5904],[155,1.58],[156,1.5865],[157,0],[158,6.1771],[159,7.2704],[160,7.5466],[161,7.6577],[162,7.6677],[163,7.6497],[164,7.8473],[166,7.8251],[167,7.8477],[168,7.9233],[169,8.0377],[170,7.7142],[171,7.6589],[172,7.7429],[173,7.8036],[225,7.8049],[344,7.8137],[345,7.8285],[346,7.8245],[347,7.7655],[348,7.6648],[349,7.534],[350,7.3849],[351,7.2665],[352,7.1071],[353,6.8338],[354,6.5048],[355,6.8894],[356,6.0247],[357,3.6211],[358,2.8976],[359,2.7111],[360,2.7186],[362,2.8785],[363,2.9069],[364,2.8628],[365,2.6701],[366,2.3771],[367,2.1088],[368,1.9544],[369,1.87],[370,1.8481],[374,1.8829],[378,1.8964],[382,1.8939],[386,1.9127],[387,1.9149],[389,1.908],[396,1.8603],[402,1.8381],[433,1.8407]]
       }
       chest_tan_ctrl.Rotation.Y {
        type=MovieSceneScriptingFloatChannel
        default=0
        thinned_from=434
        keys [[0,0],[1,0.8637],[10,0.8637],[13,0.876],[14,0.8759],[15,0.8685],[16,0.8656],[17,0.8664],[18,0.8561],[19,0.8907],[20,0.7843],[21,0.4321],[22,0.0998],[23,0],[67,0],[68,0.0771],[69,0.219],[70,0.2187],[71,0.2305],[72,0.2532],[73,0.2655],[74,0.2798],[75,0.2932],[76,0.3049],[77,0.3154],[78,0.3242],[79,0.3314],[80,0.3376],[81,0.3428],[83,0.3506],[85,0.3561],[86,0.3581],[87,0.3581],[88,0.3604],[90,0.3604],[91,0.3628],[92,0.3634],[93,0.3654],[94,0.3682],[95,0.3643],[96,0.3729],[97,0.3686],[98,0.3753],[99,0.3797],[100,0.3849],[101,0.3911],[102,0.3989],[104,0.4174],[105,0.4297],[106,0.4479],[107,0.4689],[108,0.4862],[109,0.4977],[110,0.5074],[111,0.5147],[112,0.5168],[124,0.5018],[125,0.5021],[126,0.5009],[130,0.5009],[131,0.5027],[132,0.4984],[134,0.4939],[135,0.4904],[136,0.4858],[137,0.4798],[138,0.4725],[139,0.4636],[140,0.4826],[141,0.4648],[142,0.3674],[143,0.2664],[144,0.2251],[145,0.2176],[146,0.2343],[147,0.2684],[148,0.309],[149,0.3607],[150,0.9014],[151,0.5956],[152,0.4514],[153,0.4103],[154,0.3933],[155,0.3827],[156,0.3834],[157,0],[158,2.8926],[159,1.5877],[160,1.0007],[161,0.6922],[162,0.5239],[163,0.403],[164,0.3843],[165,0.3218],[166,0.2929],[167,0.3022],[168,0.3514],[169,0.4194],[170,0.3458],[171,0.3623],[172,0.4203],[173,0.4644],[174,0.4742],[175,0.4851],[176,0.4822],[178,0.4611],[180,0.4486],[181,0.4432],[182,0.4338],[183,0.44],[184,0.4321],[195,0.4321],[196,0.4298],[197,0.4298],[198,0.4331],[199,0.4309],[203,0.4321],[205,0.4375],[208,0.4375],[209,0.436],[213,0.436],[214,0.4392],[215,0.4392],[216,0.4372],[217,0.437],[218,0.439],[219,0.4325],[220,0.4339],[221,0.4248],[222,0.4115],[223,0.405],[224,0.4058],[225,0.3631],[226,0.3808],[227,0.5426],[228,0.6811],[229,0.7082],[230,0.7319],[231,0.7523],[232,0.7698],[233,0.7846],[234,0.7969],[235,0.8069],[237,0.8239],[239,0.8271],[240,0.8317],[242,0.8304],[243,0.8338],[252,0.8338],[253,0.8379],[254,0.8373],[255,0.8407],[256,0.8408],[259,0.8632],[260,0.8847],[261,0.8822],[262,0.8256],[263,0.7637],[264,0.7145],[265,0.6492],[266,0.5895],[267,0.5525],[268,0.5269],[269,0.5106],[271,0.4834],[272,0.4727],[273,0.4644],[276,0.4474],[278,0.439],[281,0.431],[282,0.4245],[285,0.4201],[286,0.4256],[287,0.4246],[288,0.4302],[289,0.4315],[290,0.4219],[291,0.404],[292,0.3877],[293,0.3741],[294,0.3566],[295,0.3317],[296,0.3152],[297,0.3181],[298,0.3364],[299,0.3657],[300,0.3894],[301,0.3996],[302,0.4081],[303,0.415],[304,0.4205],[305,0.4205],[306,0.4281],[307,0.4295],[308,0.4325],[310,0.432],[311,0.4343],[312,0.4355],[313,0.4399],[315,0.4376],[316,0.4376],[317,0.44],[318,0.4407],[319,0.4451],[320,0.439],[321,0.4485],[322,0.4394],[323,0.4449],[325,0.447],[326,0.4445],[327,0.443],[328,0.4487],[329,0.4601],[330,0.4756],[331,0.4893],[332,0.4979],[333,0.5052],[334,0.5111],[335,0.5187],[337,0.5214],[338,0.5214],[340,0.5255],[341,0.5255],[342,0.5293],[344,0.5265],[346,0.4775],[347,0.4979],[348,0.5535],[349,0.6333],[350,0.7255],[351,0.8849],[352,0.9321],[353,0.7269],[354,0.3709],[355,0],[356,0.0725],[357,0.9132],[358,0.8728],[359,0.7818],[360,0.7767],[361,0.7675],[362,0.7521],[363,0.7609],[364,0.7905],[365,0.8007],[366,0.8071],[367,0.824],[368,0.8511],[370,0.9203],[371,0.9519],[372,0.9812],[373,1.0067],[374,1.0266],[375,1.0409],[376,1.0506],[378,1.0598],[379,1.0524],[380,1.028],[381,0.9962],[382,0.9695],[383,0.966],[384,0.9822],[385,1.0084],[386,1.0292],[387,1.0344],[388,1.0305],[389,1.0199],[390,1.005],[394,0.9349],[395,0.9195],[396,0.9065],[398,0.8856],[400,0.8696],[401,0.864],[403,0.858],[412,0.8563],[425,0.8635],[433,0.8639]]
       }
       chest_tan_ctrl.Rotation.Z {
        type=MovieSceneScriptingFloatChannel
        default=0
        thinned_from=434
        keys [[0,0],[1,0.0904],[10,0.0904],[11,0.0826],[12,0.0766],[13,0.0702],[14,0.0684],[15,0.0745],[16,0.0935],[17,0.129],[18,0.1832],[19,0],[21,0],[22,0.8349],[23,2.1626],[24,3.7031],[25,5.0536],[26,5.9857],[27,6.7305],[28,7.3038],[29,7.7065],[30,7.9593],[31,8.1061],[33,8.1061],[34,8.2488],[38,8.0843],[40,7.9337],[42,7.7645],[46,7.4826],[47,7.4262],[49,7.3497],[53,7.3099],[54,7.3127],[56,7.2577],[57,7.2643],[61,7.2175],[62,7.1843],[63,7.1992],[64,7.2756],[65,7.317],[66,7.489],[67,7.6366],[68,6.9243],[69,5.3791],[70,3.3648],[71,1.9894],[72,1.6189],[73,1.4031],[74,1.31],[75,1.2727],[76,1.2573],[77,1.2856],[78,1.3231],[79,1.3544],[80,1.3966],[81,1.4451],[82,1.4874],[84,1.5498],[86,1.6041],[87,1.6041],[88,1.622],[90,1.622],[91,1.6612],[93,1.6823],[94,1.6832],[95,1.6895],[96,1.6786],[97,1.7029],[98,1.6894],[99,1.6838],[100,1.7008],[101,1.7117],[102,1.6724],[103,1.5915],[104,1.5062],[105,1.3956],[106,1.2239],[107,1.0109],[108,0.8115],[109,0.6363],[110,0.4476],[111,0.2547],[112,0.0774],[113,0],[144,0],[145,0.0449],[146,0.2944],[147,0.588],[148,0.936],[149,1.3633],[150,0.9275],[151,5.5984],[152,5.3554],[154,4.8104],[155,4.6823],[156,4.6062],[157,1.7627],[158,1.8254],[159,6.7452],[160,9.0009],[161,10.3752],[162,11.0331],[163,11.5258],[165,12.1108],[166,12.2171],[167,12.1734],[169,11.8652],[170,11.8075],[172,11.3296],[173,11.1829],[175,11.11],[176,11.1357],[178,11.2451],[182,11.3787],[183,11.3609],[184,11.3891],[197,11.3955],[218,11.3384],[219,11.3563],[220,11.3437],[222,11.3866],[223,11.3643],[224,11.2941],[225,11.3559],[226,11.132],[227,10.3575],[228,9.7401],[230,9.4711],[232,9.2728],[234,9.1343],[237,9.0039],[240,8.9666],[252,8.9537],[253,8.9397],[256,8.9432],[258,8.9203],[259,8.933],[260,8.9118],[261,9.0139],[263,9.6771],[264,9.9575],[265,10.2897],[266,10.5805],[267,10.7649],[268,10.8946],[271,11.1262],[276,11.3201],[285,11.4714],[286,11.4573],[289,11.476],[290,11.5399],[292,11.7201],[294,11.8696],[295,11.9655],[296,12.0115],[297,11.9681],[298,11.8608],[299,11.7151],[300,11.6021],[303,11.4694],[304,11.4409],[305,11.4409],[306,11.4024],[319,11.3098],[320,11.3265],[321,11.2888],[322,11.3171],[327,11.2582],[328,11.215],[331,11],[335,10.8525],[344,10.8022],[345,10.8334],[346,10.8932],[347,10.9195],[348,10.8974],[349,10.8162],[351,10.5018],[352,10.273],[353,9.8616],[354,9.2568],[355,9.48],[356,7.7438],[357,4.3273],[358,1.9452],[359,0.3688],[360,0],[369,0],[370,0.3638],[371,0.7422],[372,1.1386],[373,1.5382],[374,1.9059],[375,2.2225],[376,2.5078],[377,2.7618],[378,2.9742],[380,3.3507],[381,3.5181],[382,3.6526],[383,3.7228],[384,3.7216],[385,3.6547],[386,3.5537],[387,3.3826],[388,3.1087],[389,2.7619],[390,2.3719],[391,2.0248],[393,1.4616],[394,1.1889],[395,0.9307],[396,0.7128],[397,0.5353],[398,0.3725],[399,0.2261],[400,0.0975],[401,0],[422,0],[423,0.0109],[424,0.0181],[425,0.0293],[426,0.0353],[427,0.0532],[428,0.0532],[429,0.0668],[430,0.0675],[433,0.0675]]
       }
       chest_tan_ctrl.Scale.X {type=MovieSceneScriptingFloatChannel, default=1, const [1,434]}
       chest_tan_ctrl.Scale.Y {type=MovieSceneScriptingFloatChannel, default=1, const [1,434]}
       chest_tan_ctrl.Scale.Z {type=MovieSceneScriptingFloatChannel, default=1, const [1,434]}
      ]
     }]
   }
   {
    type=MovieSceneControlRigParameterTrack
    display_name=Fortnite_Legacy_Face_CtrlRig
    sections [{
      type=MovieSceneControlRigParameterSection
      channels [
       Flipbook {type=MovieSceneScriptingBoolChannel, default=false}
       Face_Ctrl_Master.Location.X {type=MovieSceneScriptingFloatChannel, default=-24.0277}
       Face_Ctrl_Master.Location.Y {type=MovieSceneScriptingFloatChannel, default=0}
       Face_Ctrl_Master.Location.Z {type=MovieSceneScriptingFloatChannel, default=0}
       Face_Ctrl_Master.Rotation.X {type=MovieSceneScriptingFloatChannel, default=0}
       Face_Ctrl_Master.Rotation.Y {type=MovieSceneScriptingFloatChannel, default=0}
       Face_Ctrl_Master.Rotation.Z {type=MovieSceneScriptingFloatChannel, default=0}
       Face_Ctrl_Master.Scale.X {type=MovieSceneScriptingFloatChannel, default=1}
       Face_Ctrl_Master.Scale.Y {type=MovieSceneScriptingFloatChannel, default=1}
       Face_Ctrl_Master.Scale.Z {type=MovieSceneScriptingFloatChannel, default=1}
       gui_follow_head {type=MovieSceneScriptingFloatChannel, default=1}
      ]
     }]
   }
  ]
 }]
display_rate {denominator=1, numerator=30}
tick_resolution {denominator=1, numerator=24000}
tracks []
